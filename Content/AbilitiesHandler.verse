using { /Fortnite.com/Devices }
using { /Fortnite.com/Characters }
using { /Fortnite.com/UI }
using { /Verse.org/Simulation }
using { /Verse.org/Simulation/Tags }
using { /Verse.org/Assets }
using { /Verse.org/Random }
using { /Verse.org/Colors }
using { /UnrealEngine.com/Temporary/Diagnostics }
using { /UnrealEngine.com/Temporary/UI }
using { /UnrealEngine.com/Temporary/SpatialMath }
using { /Verse.org/SceneGraph }
using { /Verse.org/SpatialMath }

var AllUIs : weak_map(session, []ShopUI_V2) = map{}

UI := module:
    Parts<public> := module:

SceneGraphObjects := module:
    Tasks<public> := module:

Icons := module:
    AbilitiesBlocks<public> := module:

AllIcons := module:
    None<public>: texture = Icons.IconNone
    Addition<public>: texture = Icons.IconAddition
    Arrow<public>: texture = Icons.IconArrow
    Behind<public>: texture = Icons.IconBehind
    Bracket<public>: texture = Icons.IconBracket
    Broken<public>: texture = Icons.IconBroken
    Butterfly<public>: texture = Icons.IconButterfly
    Cell<public>: texture = Icons.IconCell
    Checkmark<public>: texture = Icons.IconCheckmark
    Coin<public>: texture = Icons.IconCoin
    Crown<public>: texture = Icons.IconCrown
    Door<public>: texture = Icons.IconDoor
    Double<public>: texture = Icons.IconDouble
    Electric<public>: texture = Icons.IconElectric
    Energy<public>: texture = Icons.IconEnergy
    Eye<public>: texture = Icons.IconEye
    Heart<public>: texture = Icons.IconHeart
    Lock<public>: texture = Icons.IconLock
    Place<public>: texture = Icons.IconPlace
    Reboot<public>: texture = Icons.IconReboot
    Rift<public>: texture = Icons.IconRift
    Travel<public>: texture = Icons.IconTravel
    Umbrella<public>: texture = Icons.IconUmbrella
    Up<public>: texture = Icons.IconUp

ShopCon := class:
    MainClass : ShopUI
    AbilityID : int

    Connect(WM : widget_message):void=
        # Print("Button triggered with ID {AbilityID}")
        MainClass.ButtonTriggered(AbilityID, WM.Source)

FakeTaskCon := class:
    Prop : creative_prop
    Button : button_device
    Marker : beacon_device
    TrackMarker : beacon_device

    MainClass : AbilitiesHandler

    Connect(Agent : agent):void=
        Prop.Dispose()
        Button.Disable()
        Marker.Disable()
        MainClass.Life.InBasementNotification.SetText(MessagesBank{}.FakeTaskFound)
        if (Monika := MainClass.RolesInfo.Monster?):
            MainClass.Life.InBasementNotification.Hide(Monika)
            MainClass.Life.InBasementNotification.Show(Monika)
        MainClass.Life.InBasementNotification.SetText(MessagesBank{}.FakeTaskTriggered)
        MainClass.Life.InBasementNotification.Hide(Agent)
        MainClass.Life.InBasementNotification.Show(Agent)
        TrackMarker.Enable()
        spawn{ConnectSus()}
    
    ConnectSus()<suspends>:void=
        Sleep(5.0)
        TrackMarker.Disable()

ShopUI := class:

    Agent : agent
    Abilities : []Ability

    MainClass : AbilitiesHandler
    
    var Canvas<private> : canvas = canvas{}

    var Buttons : []button_regular = array{}

    var PrestigeButton : button_quiet = button_quiet:
        DefaultText := MessagesBank{}.Prestige
        # OnClick().Subscribe(Prestige)

    var DailyButton : button_quiet = button_quiet:
        DefaultText := MessagesBank{}.DailyReward
        # OnClick().Subscribe(DailyReward)

    var CurrentGold : text_block = text_block{
        DefaultTextColor := color{R:=1.0,G:=1.0,B:=1.0}
    }

    var TimerTextBlock : text_block = text_block{
        DefaultText := MessagesBank{}.Empty
        DefaultTextColor := color{R:=1.0,G:=1.0,B:=1.0}
    }

    var SelectedID : int = -1
    var Ended : logic = false

    Start()<suspends>:void=
        PrestigeButton.SetEnabled(false)
        DailyButton.SetEnabled(false)
        # Print(MessagesBank{}.Connect(MessagesBank{}.ShopUiHeader, Agent, MessagesBank{}.Empty))
        loop:
            if (MainClass.Achievements.CheckExistingProfile[Agent]):
                # Print("Profile exists")
                break
            else:
                # Print("Profile does not exist")
            Sleep(0.02)
        spawn{DailyLoop()}
        var AllShopItems : []stack_box_slot = array{}
        CurrentGold.SetText(MessagesBank{}.ToMsg(MainClass.Achievements.GetAchievement(Agent, PossibleAchievements.Coins)))
        # Print("Current gold: {MainClass.Achievements.GetAchievement(Agent, PossibleAchievements.Coins)}")
        if (MainClass.Achievements.GetBoughtItems(Agent).Length = 12 or MainClass.DEBUG_PrestigeAlwaysAvailable = true):
            PrestigeButton.SetEnabled(true)
        PrestigeButton.OnClick().Subscribe(Prestige)
        DailyButton.OnClick().Subscribe(DailyReward)
        for (i : Abilities, i.IsEnabled = true):
            var Icon : texture = AllIcons.Cell
            var Price : message = MessagesBank{}.ToMsg(i.Price)
            if (MainClass.DEBUG_AllFree = true):
                set Price = MessagesBank{}.ToMsg(0)
            case (i.Icon):
                IconsList.Empty =>
                    set Icon = AllIcons.Cell
                IconsList.Addition =>
                    set Icon = AllIcons.Addition
                IconsList.Arrow =>
                    set Icon = AllIcons.Arrow
                IconsList.Behind =>
                    set Icon = AllIcons.Behind
                IconsList.Bracket =>
                    set Icon = AllIcons.Bracket
                IconsList.Broken =>
                    set Icon = AllIcons.Broken
                IconsList.Butterfly =>
                    set Icon = AllIcons.Butterfly
                IconsList.Cell =>
                    set Icon = AllIcons.Cell
                IconsList.Checkmark =>
                    set Icon = AllIcons.Checkmark
                IconsList.Coin =>
                    set Icon = AllIcons.Coin
                IconsList.Crown =>
                    set Icon = AllIcons.Crown
                IconsList.Door =>
                    set Icon = AllIcons.Door
                IconsList.Double =>
                    set Icon = AllIcons.Double
                IconsList.Electric =>
                    set Icon = AllIcons.Electric
                IconsList.Energy =>
                    set Icon = AllIcons.Energy
                IconsList.Eye =>
                    set Icon = AllIcons.Eye
                IconsList.Heart =>
                    set Icon = AllIcons.Heart
                IconsList.Lock =>
                    set Icon = AllIcons.Lock
                IconsList.Place =>
                    set Icon = AllIcons.Place
                IconsList.Reboot =>
                    set Icon = AllIcons.Reboot
                IconsList.Rift =>
                    set Icon = AllIcons.Rift
                IconsList.Travel =>
                    set Icon = AllIcons.Travel
                IconsList.Umbrella =>
                    set Icon = AllIcons.Umbrella
                IconsList.Up =>
                    set Icon = AllIcons.Up
            
            BTN := button_regular:
                # DefaultText := block:
                #     var found : logic = false
                # Print("Amount of bought items: {MainClass.Achievements.GetBoughtItems(Agent).Length}")
                #     for (j : MainClass.Achievements.GetBoughtItems(Agent)):
                #         # Print("Checking {j} == {i.ID}")
                #         if (j = i.ID):
                #             set found = true
                #         else:
                #             # Print("{j} != {i.ID}")
                #     if (found = false):
                #         # Print("Not found bought item {i.ID}")
                #         MessagesBank{}.BuyItem
                #     else:
                #         # Print("Found bought item {i.ID}")
                #         MessagesBank{}.EquipItem
                DefaultText := MessagesBank{}.BuyItem
            for (j : MainClass.Achievements.GetBoughtItems(Agent)):
                if (j = i.ID):
                    BTN.SetText(MessagesBank{}.EquipItem)
            res := BTN.GetText()
            if (i.MinimumSurvivors > MainClass.RolesInfo.Survivors.Length):
                if (res = Localize(MessagesBank{}.EquipItem)):
                    BTN.SetEnabled(false)
                    BTN.SetText(MessagesBank{}.AbilityNotAvailable)
                else:
                    BTN.SetText(MessagesBank{}.BuyNotAvailable)
            set Buttons += array{BTN}
            BTN.OnClick().Subscribe(ShopCon{MainClass := Self, AbilityID := i.ID}.Connect)
            # if (Button := Buttons[Buttons.Length-1]):
            #     Button.OnClick().Subscribe(ShopCon{MainClass := Self, AbilityID := i.ID}.Connect)
            set AllShopItems += array:
                stack_box_slot:
                    HorizontalAlignment := horizontal_alignment.Fill
                    VerticalAlignment := vertical_alignment.Fill
                    Widget := stack_box:
                        Orientation := orientation.Horizontal
                        Slots := array:
                            stack_box_slot:
                                Widget := texture_block:
                                    DefaultImage := Icon
                                    DefaultTint := i.Tint
                                    DefaultDesiredSize := vector2{X:=128.0,Y:=128.0}
                            stack_box_slot:
                                Padding := margin{
                                    Top := 0.0,
                                    Bottom := 0.0,
                                    Left := 35.0,
                                    Right := 0.0
                                }
                                HorizontalAlignment := horizontal_alignment.Fill
                                VerticalAlignment := vertical_alignment.Center
                                Widget := text_block:
                                    DefaultText := block:
                                        case (i.ID):
                                            0 =>
                                                MessagesBank{}.SurvivorsItem1Name
                                            1 =>
                                                MessagesBank{}.SurvivorsItem2Name
                                            2 =>
                                                MessagesBank{}.SurvivorsItem3Name
                                            3 =>
                                                MessagesBank{}.SurvivorsItem4Name
                                            4 =>
                                                MessagesBank{}.SurvivorsItem5Name
                                            5 =>
                                                MessagesBank{}.SurvivorsItem6Name
                                            6 =>
                                                MessagesBank{}.MonsterItem1Name
                                            7 =>
                                                MessagesBank{}.MonsterItem2Name
                                            8 =>
                                                MessagesBank{}.MonsterItem3Name
                                            9 =>
                                                MessagesBank{}.MonsterItem4Name
                                            10 =>
                                                MessagesBank{}.MonsterItem5Name
                                            11 =>
                                                MessagesBank{}.MonsterItem6Name
                                            _ =>
                                                MessagesBank{}.Empty
                                    DefaultTextColor := color{R:=1.0,G:=1.0,B:=1.0}
                            # stack_box_slot:
                            #     Padding := margin{
                            #         Top := 0.0,
                            #         Bottom := 0.0,
                            #         Left := 0.0,
                            #         Right := 30.0
                            #     }
                            #     HorizontalAlignment := horizontal_alignment.Right
                            #     VerticalAlignment := vertical_alignment.Center
                            #     Distribution := option{1.0}
                            #     Widget := texture_block:
                            #         DefaultImage := Icons.IconCheckmark
                            #         DefaultTint := color{R:=0.0,G:=1.0,B:=0.0}
                            #         DefaultDesiredSize := block:
                            #             for (j : MainClass.Achievements.GetBoughtItems(Agent)):
                            #                 if (j = i.ID):
                            #                     vector2{X:=100.0,Y:=100.0}
                            #             vector2{X:=0.0,Y:=0.0}
                            stack_box_slot:
                                Padding := margin{
                                    Top := 0.0,
                                    Bottom := 0.0,
                                    Left := 0.0,
                                    Right := 5.0
                                }
                                HorizontalAlignment := horizontal_alignment.Right
                                VerticalAlignment := vertical_alignment.Center
                                Distribution := option{1.0}
                                Widget := texture_block:
                                    DefaultImage := Icons.IconCoin
                                    DefaultTint := color{R:=1.0,G:=0.75,B:=0.0}
                                    DefaultDesiredSize := vector2{X:=100.0,Y:=100.0}
                            stack_box_slot:
                                Padding := margin{
                                    Top := 0.0,
                                    Bottom := 0.0,
                                    Left := 0.0,
                                    Right := 30.0
                                }
                                HorizontalAlignment := horizontal_alignment.Right
                                VerticalAlignment := vertical_alignment.Center
                                Widget := text_block:
                                    DefaultText := Price
                                    DefaultTextColor := color{R:=1.0,G:=1.0,B:=1.0}
                            stack_box_slot:
                                HorizontalAlignment := horizontal_alignment.Right
                                VerticalAlignment := vertical_alignment.Center
                                Widget := block:
                                    # if (Button := Buttons[Buttons.Length-1]):
                                    #     Button
                                    BTN
                                    # button_regular:
                                    #     DefaultText := block:
                                    #         for (j : MainClass.Achievements.GetBoughtItems(Agent)):
                                    #             if (j = i.ID):
                                    #                 MessagesBank{}.EquipItem
                                    #         MessagesBank{}.BuyItem
        
        ReadyButton := button_loud:
            DefaultText := MessagesBank{}.Ready
        ReadyButton.OnClick().Subscribe(Ready)
        set Canvas = canvas:
            Slots := array:
                # canvas_slot: # Prestige button
                #     Anchors := anchors{
                #         Minimum := vector2{X:=0.5,Y:=0.0}
                #         Maximum := vector2{X:=0.5,Y:=0.0}
                #     }
                #     Offsets := margin{
                #         Left := 675.0,
                #         Top := 30.0,
                #         Right := 281.0,
                #         Bottom := 56.0
                #     }
                #     Alignment := vector2{X:=1.0,Y:=0.0}
                #     SizeToContent := true
                #     Widget := PrestigeButton
                canvas_slot: # Prestige & Daily
                    Anchors := anchors{
                        Minimum := vector2{X:=1.0,Y:=0.0},
                        Maximum := vector2{X:=1.0,Y:=0.0}
                    }
                    Offsets := margin{
                        Left := -40.0
                        Top := 30.0
                        Right := 100.0
                        Bottom := 30.0
                    }
                    Alignment := vector2{X:=1.0,Y:=0.0}
                    SizeToContent := true
                    Widget := overlay:
                        Slots := array:
                            overlay_slot:
                                Padding := margin{
                                    Left := -30.0
                                    Top := -20.0
                                    Right := -30.0
                                    Bottom := -20.0
                                }
                                HorizontalAlignment := horizontal_alignment.Fill
                                VerticalAlignment := vertical_alignment.Fill
                                Widget := UI.WB_BoxBackground{}
                            overlay_slot:
                                Padding := margin{
                                    Left := 0.0
                                    Top := 0.0
                                    Right := 0.0
                                    Bottom := 0.0
                                }
                                HorizontalAlignment := horizontal_alignment.Fill
                                VerticalAlignment := vertical_alignment.Fill
                                Widget := stack_box:
                                    Orientation := orientation.Horizontal
                                    Slots := array:
                                        stack_box_slot:
                                            HorizontalAlignment := horizontal_alignment.Fill
                                            VerticalAlignment := vertical_alignment.Center
                                            Padding := margin{
                                                Right := 10.0
                                            }
                                            Widget := PrestigeButton
                                        stack_box_slot:
                                            HorizontalAlignment := horizontal_alignment.Right
                                            VerticalAlignment := vertical_alignment.Center
                                            Widget := DailyButton
                canvas_slot: # Coins
                    Anchors := anchors{
                        Minimum := vector2{X:=0.0,Y:=0.0},
                        Maximum := vector2{X:=0.0,Y:=0.0}
                    }
                    Offsets := margin{
                        Left := 12.0 # Position X
                        Top := 12.0 # Position Y
                        Right := 260.0 # Size X
                        Bottom := 100.0 # Size Y
                    }
                    Alignment := vector2{X:=0.0,Y:=0.0}
                    SizeToContent := true
                    Widget := overlay: # Coins
                        Slots := array:
                            overlay_slot:
                                Padding := margin{
                                    Top := 0.0,
                                    Bottom := 0.0,
                                    Left := 0.0,
                                    Right := 0.0
                                }
                                HorizontalAlignment := horizontal_alignment.Fill
                                VerticalAlignment := vertical_alignment.Center
                                Widget := stack_box:
                                    Orientation := orientation.Horizontal
                                    Slots := array:
                                        stack_box_slot:
                                            HorizontalAlignment := horizontal_alignment.Left
                                            VerticalAlignment := vertical_alignment.Center
                                            Widget := texture_block:
                                                DefaultImage := Icons.IconCoin
                                                DefaultTint := color{R:=1.0,G:=0.75,B:=0.0}
                                                DefaultDesiredSize := vector2{X:=100.0,Y:=100.0}
                                        stack_box_slot:
                                            Padding := margin{
                                                Top := 0.0,
                                                Bottom := 0.0,
                                                Left := 20.0,
                                                Right := 0.0
                                            }
                                            HorizontalAlignment := horizontal_alignment.Right
                                            VerticalAlignment := vertical_alignment.Center
                                            Distribution := option{1.0}
                                            Widget := CurrentGold
                canvas_slot:
                    Anchors := anchors{
                        Minimum := vector2{X:=0.5,Y:=0.5},
                        Maximum := vector2{X:=0.5,Y:=0.5}
                    }
                    Offsets := margin{
                        Left := 0.0
                        Top := -10.0
                        Right := 1500.0
                        Bottom := 800.0
                    }
                    Alignment := vector2{X:=0.5,Y:=0.5}
                    Widget := overlay: # Shop
                        Slots := array:
                            overlay_slot:
                                Padding := margin{
                                    Top := -10.0,
                                    Bottom := 800.0,
                                    Left := -10.0,
                                    Right := 1500.0
                                }
                                Widget := color_block:
                                    DefaultColor := color{R:=0.0,G:=0.0,B:=0.0}
                                    DefaultOpacity := 0.0
                            overlay_slot:
                                Padding := margin{
                                    Top := 0.0,
                                    Bottom := 0.0,
                                    Left := 40.0,
                                    Right := 40.0
                                }
                                HorizontalAlignment := horizontal_alignment.Fill
                                VerticalAlignment := vertical_alignment.Center
                                Widget := stack_box:
                                    Orientation := orientation.Vertical
                                    Slots := AllShopItems
                canvas_slot: # Ready
                    Anchors := anchors{
                        Minimum := vector2{X:=1.0,Y:=1.0},
                        Maximum := vector2{X:=1.0,Y:=1.0}
                    }
                    Offsets := margin{
                        Left := -275.0
                        Top := -30.0
                        Right := 100.0
                        Bottom := 30.0
                    }
                    Alignment := vector2{X:=1.0,Y:=1.0}
                    SizeToContent := true
                    Widget := overlay:
                        Slots := array:
                            overlay_slot:
                                HorizontalAlignment := horizontal_alignment.Fill
                                VerticalAlignment := vertical_alignment.Fill
                                Padding := margin:
                                    Top := -20.0
                                    Bottom := -20.0
                                Widget := UI.WB_BoxBackground{}
                            overlay_slot:
                                HorizontalAlignment := horizontal_alignment.Fill
                                VerticalAlignment := vertical_alignment.Fill
                                Padding := margin:
                                    Left := 40.0
                                    Right := 40.0
                                Widget := stack_box:
                                    Orientation := orientation.Horizontal
                                    Slots := array:
                                        stack_box_slot:
                                            HorizontalAlignment := horizontal_alignment.Fill
                                            VerticalAlignment := vertical_alignment.Center
                                            Padding := margin:
                                                Right := 20.0
                                            Widget := TimerTextBlock
                                        stack_box_slot:
                                            HorizontalAlignment := horizontal_alignment.Right
                                            VerticalAlignment := vertical_alignment.Center
                                            Widget := ReadyButton
                # canvas_slot:
                #     Anchors := anchors{Minimum := vector2{X := 0.5, Y := 0.0}, Maximum := vector2{X := 0.5, Y := 0.0}}
                #     Offsets := margin{Top := 0.0, Left := 0.0, Right := 0.0, Bottom := 0.0}
                #     Alignment := vector2{X := 0.0, Y := 0.275}
                #     SizeToContent := true
                #     Widget := texture_block{
                #         DefaultImage := UI.T_MapCode
                #         DefaultDesiredSize := vector2{X := 450.0, Y := 450.0}
                #     }
        if (PlUi := GetPlayerUI[player[Agent]]):
            PlUi.AddWidget(Canvas, player_ui_slot{InputMode := ui_input_mode.All})
            # Print("Added shop to player")
        spawn{CountDown()}

    CountDown()<suspends>:void=
        for (i := 0..20):
            TimerTextBlock.SetText(MessagesBank{}.ToMsg(20 - i))
            Sleep(1.0)

    ButtonTriggered(Button:int, Widget:widget):void=
        for (i : MainClass.Achievements.GetBoughtItems(Agent)):
            if (i = Button, BTN := button_regular[Widget]):
                # Print("{i} equipped")
                set SelectedID = Button
                # for (g := 0..Buttons.Length-1, btn := Buttons[g]):
                #     if (btn = BTN):
                #         btn.SetText(MessagesBank{}.ItemEquipped)
                #         btn.SetEnabled(false)
                #     else:
                #         for (item : MainClass.Achievements.GetBoughtItems(Agent)):
                #             if (item = g):
                #                 btn.SetText(MessagesBank{}.EquipItem)
                #         btn.SetEnabled(true)
                for (g := 0..Abilities.Length-1, btn := Buttons[g], ab := Abilities[g]):
                    if (btn = BTN):
                        btn.SetText(MessagesBank{}.ItemEquipped)
                        btn.SetEnabled(false)
                    else:
                        # for (item : MainClass.Achievements.GetBoughtItems(Agent)):
                        #     if (item = ab.ID):
                        #         btn.SetText(MessagesBank{}.EquipItem)
                        txt := btn.GetText()
                        if (txt <> Localize(MessagesBank{}.BuyItem) and txt <> Localize(MessagesBank{}.AbilityNotAvailable) and txt <> Localize(MessagesBank{}.BuyNotAvailable)):
                            btn.SetText(MessagesBank{}.EquipItem)
                            btn.SetEnabled(true)
                return
        var Price : int = 999
        for (i : Abilities):
            if (i.ID = Button):
                set Price = i.Price
            if (MainClass.DEBUG_AllFree = true):
                set Price = 0
        # if (price := MainClass.SurvivorsAbilities[Button].Price):
        #     set Price = price
        if (MainClass.Achievements.GetAchievement(Agent, PossibleAchievements.Coins) >= Price, BTN := button_regular[Widget]):
            MainClass.Achievements.RemoveAchievement(Agent, PossibleAchievements.Coins, Price)
            CurrentGold.SetText(MessagesBank{}.ToMsg(MainClass.Achievements.GetAchievement(Agent, PossibleAchievements.Coins)))
            for (g := 0..Abilities.Length-1, btn := Buttons[g], ab := Abilities[g]):
                if (btn = BTN):
                    MainClass.Achievements.BuyItem(Agent, ab.ID)
                    btn.SetText(MessagesBank{}.ItemEquipped)
                    btn.SetEnabled(false)
                else:
                    # for (item : MainClass.Achievements.GetBoughtItems(Agent)):
                        # if (item = ab.ID):
                        #     btn.SetText(MessagesBank{}.EquipItem)
                    txt := btn.GetText()
                    if (txt = Localize(MessagesBank{}.ItemEquipped)):
                        # Print("Enabled a button")
                        btn.SetText(MessagesBank{}.EquipItem)
                    else:
                        # Print("{txt} <> {Localize(MessagesBank{}.BuyItem)}")
                    if (txt <> Localize(MessagesBank{}.AbilityNotAvailable)):
                        btn.SetEnabled(true)
            for (i : Abilities):
                if (i.ID = Button):
                    if (i.MinimumSurvivors > MainClass.RolesInfo.Survivors.Length):
                        BTN.SetEnabled(false)
                        BTN.SetText(MessagesBank{}.AbilityNotAvailable)
                    else:
                        set SelectedID = Button
            if (MainClass.Achievements.GetBoughtItems(Agent).Length = 12 or MainClass.DEBUG_PrestigeAlwaysAvailable = true):
                PrestigeButton.SetEnabled(true)
            else:
                PrestigeButton.SetEnabled(false)
            # Print("{Button} bought")
        else:
            # Print("Not enough coins: {MainClass.Achievements.GetAchievement(Agent, PossibleAchievements.Coins)} < {Price}")
            return
    
    Ready(widget_info : widget_message):void=
        if (btn := button_loud[widget_info.Source]):
            txt := btn.GetText()
            if (txt = Localize(MessagesBank{}.Ready)):
                btn.SetText(MessagesBank{}.Unready)
                # set MainClass.PlayersReady += 1
                set MainClass.PlayersReady += array{widget_info.Player}
                if (MainClass.PlayersReady = MainClass.RolesInfo.Survivors.Length + 1):
                    widget_info.Source.SetEnabled(false)
            else:
                btn.SetText(MessagesBank{}.Ready)
                # set MainClass.PlayersReady -= 1
                set MainClass.PlayersReady = MainClass.PlayersReady.RemoveAllElements(widget_info.Player)

    Prestige(widget_info : widget_message):void=
        MainClass.Achievements.SetAchievement(Agent, PossibleAchievements.Coins, 0)
        MainClass.Achievements.ResetItems(Agent)
        CurrentGold.SetText(MessagesBank{}.ToMsg(0))
        # for (i := 0..Abilities.Length-1, btn := Buttons[i]):
        #     btn.SetText(MessagesBank{}.BuyItem)
        #     btn.SetEnabled(true)
        # Enable all buttons, set N/A if not enough players
        for (i : Buttons):
            BTN_TXT := i.GetText()
            if (BTN_TXT = Localize(MessagesBank{}.AbilityNotAvailable)):
                i.SetText(MessagesBank{}.BuyNotAvailable)
            else:
                i.SetText(MessagesBank{}.BuyItem)
            i.SetEnabled(true)
        # for (i : Buttons):
        #     i.SetText(MessagesBank{}.BuyItem)
        #     i.SetEnabled(true)
        set SelectedID = -1
        if (MainClass.DEBUG_PrestigeAlwaysAvailable = false):
            PrestigeButton.SetEnabled(false)
        before := MainClass.Achievements.GetAchievement(Agent, PossibleAchievements.Prestige)
        MainClass.Achievements.SaveAchievement(Agent, PossibleAchievements.Prestige, 1)
        MainClass.Feedback_Prestige.Submit(Agent)
        # Print("Prestige {before} -> {MainClass.Achievements.GetAchievement(Agent, PossibleAchievements.Prestige)}")
        set MainClass.Prestiged += array{Agent}

    DailyLoop()<suspends>:void=
        if:
            player[Agent].IsActive[]
            LatestDaily := PlayersStats[Agent].LatestDailyRecieved
        then:
            loop:
                if:
                    player[Agent].IsActive[]
                    Time := 86400.0 - Abs(GetSecondsSinceEpoch() - LatestDaily)
                    Time > 0.0
                    # Print("{86400.0 - Abs(GetSecondsSinceEpoch() - LatestDaily)}")
                then:
                    # Use MessagesBank{}.Daytime() with H, M, and S, each only if it's not 0
                    if:
                        H := Floor[Time / 3600.0]
                        M := Mod[Floor[Time / 60.0], 60]
                        S := Mod[Floor[Time], 60]
                    then:
                        var StringH : string = block:
                            if (H > 9):
                                "{H}"
                            else:
                                "0{H}"
                        var StringM : string = block:
                            if (M > 9):
                                "{M}"
                            else:
                                "0{M}"
                        var StringS : string = block:
                            if (S > 9):
                                "{S}"
                            else:
                                "0{S}"
                        if (H > 0):
                            DailyButton.SetText(MessagesBank{}.Daytime(StringH, StringM, StringS))
                        else if (M > 0):
                            DailyButton.SetText(MessagesBank{}.Daytime(StringM, StringS))
                        else:
                            DailyButton.SetText(MessagesBank{}.Daytime(StringS))
                    Sleep(1.0)
                    if (Ended = true):
                        break
                else:
                    break
            DailyButton.SetText(MessagesBank{}.DailyReward)
            DailyButton.SetEnabled(true)
        else:
            Print("Error with player!", ?Color := NamedColors.Red)

    DailyReward(widget_info : widget_message):void=
        Res := MainClass.Achievements.GetDailyReward(Agent)
        if (Res <> -1):
            DailyButton.SetEnabled(false)
            CurrentGold.SetText(MessagesBank{}.ToMsg(Res))
            set MainClass.DailyRewarded += array{Agent}
            spawn{DailyLoop()}
        else:
            Print("Issues getting daily reward", ?Color := NamedColors.Red)

    Ending():int=
        # if (PlUi := GetPlayerUI[player[Agent]]):
        #     PlUi.RemoveWidget(Canvas)
        #     Print("Ended shop for player")
        for (i := 0..3):
            if (PlUi := GetPlayerUI[player[Agent]], Agent <> MainClass.RolesInfo.Monster?):
                PlUi.RemoveWidget(Canvas)
                # Print("Ended shop for player")
            else:
                spawn{MonsterWait()}
        set Ended = true
        return SelectedID

    MonsterWait()<suspends>:void=
        if (PlUi := GetPlayerUI[player[Agent]]):
            PlUi.RemoveWidget(Canvas)
            txt := text_block:
                DefaultText := MessagesBank{}.YouWillBeReleasedIn
                DefaultTextColor := color{R:=1.0,G:=1.0,B:=1.0}
            set Canvas = canvas:
                Slots := array:
                    canvas_slot:
                        Widget := txt
                        Anchors := anchors{
                            Minimum := vector2{X:=0.5,Y:=0.5},
                            Maximum := vector2{X:=0.5,Y:=0.5}
                        }
                        Offsets := margin{
                            Left := 0.0,
                            Top := 0.0,
                            Right := 0.0,
                            Bottom := 0.0
                        }
                        Alignment := vector2{X:=0.5,Y:=0.5}
            PlUi.AddWidget(Canvas, player_ui_slot{InputMode := ui_input_mode.All})
            for (i := 0..MainClass.RolesInfo.MonsterReleaseTime):
                txt.SetText(MessagesBank{}.Connect(MessagesBank{}.YouWillBeReleasedIn, "{MainClass.RolesInfo.MonsterReleaseTime + 1 - i}...", MessagesBank{}.Empty))
                Sleep(1.0)
            PlUi.RemoveWidget(Canvas)
            # Print("Ended shop for player")
            Sleep(0.01)
            spawn{MainClass.RolesInfo.WaitForMove(Agent)}

ShopUI_V2 := class:
    Player : player
    PlUI : player_ui
    Abilities : []Ability
    IntroSeq : cinematic_sequence_device
    RolePart : widget

    PlayersToShow : []?player

    MainClass : AbilitiesHandler

    var NpUI<private> : ?NameplateUI = false
    var Canvas<private> : canvas = canvas{}
    var AbilitiesButtons<private> : []button_quiet = array{}
    var AllUiButtons<private> : []stack_box_slot = array{}
    var AllPlayersChoices<private> : [player]texture_block = map{}
    var AllPlayersReady<private> : [player]text_block = map{}

    var PriceImage<private> : texture_block = texture_block:
        DefaultImage := Icons.IconCoin
        DefaultTint := color{R:=1.0,G:=0.75,B:=0.0}
        DefaultDesiredSize := vector2{X:=100.0,Y:=100.0}

    var AbilityNameText<private> : text_block = text_block:
        DefaultText := MessagesBank{}.Empty
        DefaultTextColor := color{R:=1.0,G:=1.0,B:=1.0}

    var TimerText<private> : text_block = text_block:
        DefaultText := MessagesBank{}.Empty
        DefaultTextColor := color{R:=1.0,G:=1.0,B:=1.0}

    var PlayerGoldText<private> : text_block = text_block:
        DefaultText := MessagesBank{}.Empty
        DefaultTextColor := color{R:=1.0,G:=1.0,B:=1.0}

    var PriceText<private> : text_block = text_block:
        DefaultText := MessagesBank{}.Empty
        DefaultTextColor := color{R:=1.0,G:=1.0,B:=1.0}

    var ApplyButton<private> : button_regular = button_regular: # Buy / Buy (N/A) / Equip / N/A
        DefaultText := MessagesBank{}.Empty

    var PrestigeButton<private> : button_quiet = button_quiet:
        DefaultText := MessagesBank{}.Prestige

    var DailyButton<private> : button_quiet = button_quiet:
        DefaultText := MessagesBank{}.DailyReward

    var NameplateButton<private> : button_quiet = button_quiet:
        DefaultText := MessagesBank{}.Nameplate

    var ReadyButton<private> : button_loud = button_loud:
        DefaultText := MessagesBank{}.Ready

    var Ended<private> : logic = false
    var FocusedID<private> : int = -1
    var SelectedID<private> : int = -1

    CreateCanvas():void=
        var PlayersList : []stack_box_slot = array{}
        for (i : PlayersToShow):
            NewTexture := texture_block:
                DefaultImage := AllIcons.None
                DefaultTint := color{R:=1.0,G:=1.0,B:=1.0}
                DefaultDesiredSize := vector2{X:=40.0,Y:=40.0}
            NewName := text_block:
                DefaultText := block:
                    if (Pl := i?):
                        MessagesBank{}.ToMsg(Pl)
                    else:
                        MessagesBank{}.Empty
                DefaultTextColor := color{R:=1.0,G:=1.0,B:=1.0}
            NewReady := text_block:
                DefaultText := block:
                    if (i?.IsActive[]):
                        MessagesBank{}.NotReady
                    else:
                        MessagesBank{}.Disconnected
                DefaultTextColor := block:
                    if (i?.IsActive[]):
                        color{R:=1.0}
                    else:
                        NamedColors.Gray
            if:
                set AllPlayersChoices[i?] = NewTexture
                set AllPlayersReady[i?] = NewReady
            set PlayersList += array:
                stack_box_slot:
                    HorizontalAlignment := horizontal_alignment.Fill
                    VerticalAlignment := vertical_alignment.Fill
                    Widget := overlay:
                        Slots := array:
                            overlay_slot:
                                HorizontalAlignment := horizontal_alignment.Fill
                                VerticalAlignment := vertical_alignment.Fill
                                Widget := color_block:
                                    DefaultOpacity := 0.0
                                    DefaultDesiredSize := vector2:
                                        X := 412.5
                                        Y := 70.0
                            overlay_slot:
                                HorizontalAlignment := horizontal_alignment.Fill
                                VerticalAlignment := vertical_alignment.Fill
                                Widget := UI.Parts.WB_Shop_NickBackground{}
                            overlay_slot:
                                Padding := margin:
                                    Left := 15.0
                                    Top := 15.0
                                    Right := 15.0
                                    Bottom := 15.0
                                HorizontalAlignment := horizontal_alignment.Left
                                VerticalAlignment := vertical_alignment.Fill
                                Widget := stack_box:
                                    Orientation := orientation.Horizontal
                                    Slots := array:
                                        stack_box_slot:
                                            Padding := margin:
                                                Right := 10.0
                                            HorizontalAlignment := horizontal_alignment.Left
                                            VerticalAlignment := vertical_alignment.Top
                                            Widget := NewTexture
                                        stack_box_slot:
                                            Padding := margin:
                                                Top := -10.0
                                                Bottom := -5.0
                                            HorizontalAlignment := horizontal_alignment.Fill
                                            VerticalAlignment := vertical_alignment.Fill
                                            Widget := stack_box:
                                                Orientation := orientation.Vertical
                                                Slots := array:
                                                    stack_box_slot:
                                                        HorizontalAlignment := horizontal_alignment.Fill
                                                        VerticalAlignment := vertical_alignment.Fill
                                                        Widget := NewName
                                                    stack_box_slot:
                                                        HorizontalAlignment := horizontal_alignment.Fill
                                                        VerticalAlignment := vertical_alignment.Fill
                                                        Widget := NewReady
        set Canvas = canvas:
            Slots := array:
                canvas_slot: # Top bar
                    Anchors := anchors:
                        Minimum := vector2:
                            X := 0.5
                            Y := 0.0
                        Maximum := vector2:
                            X := 0.5
                            Y := 0.0
                    Offsets := margin:
                        Left := 0.0
                        Top := 25.0
                    Alignment := vector2:
                        X := 0.5
                        Y := 0.0
                    SizeToContent := true
                    Widget := UI.Parts.WB_Shop_TopBar{}
                canvas_slot: # Like and Favorite
                    Anchors := anchors:
                        Minimum := vector2:
                            X := 1.0
                            Y := 0.5
                        Maximum := vector2:
                            X := 1.0
                            Y := 0.5
                    Offsets := margin:
                        Left := -20.0
                        Top := 0.0
                    Alignment := vector2:
                        X := 1.0
                        Y := 0.5
                    SizeToContent := true
                    Widget := UI.Parts.WB_Shop_LikeFavorite{}
                canvas_slot: # Role
                    Anchors := anchors:
                        Minimum := vector2:
                            X := 0.4
                            Y := 1.0
                        Maximum := vector2:
                            X := 0.4
                            Y := 1.0
                    Offsets := margin:
                        Left := 0.0
                        Top := -30.0
                    Alignment := vector2:
                        X := 0.0
                        Y := 1.0
                    SizeToContent := true
                    Widget := RolePart
                canvas_slot: # Coins
                    Anchors := anchors:
                        Minimum := vector2:
                            X:=0.0
                            Y:=0.0
                        Maximum := vector2:
                            X:=0.0
                            Y:=0.0
                    Offsets := margin:
                        Left := 12.0 # Position X
                        Top := 12.0 # Position Y
                    Alignment := vector2:
                        X:=0.0
                        Y:=0.0
                    SizeToContent := true
                    Widget := overlay: # Coins
                        Slots := array:
                            overlay_slot:
                                HorizontalAlignment := horizontal_alignment.Fill
                                VerticalAlignment := vertical_alignment.Fill
                                Widget := UI.Parts.WB_Shop_ShopBackground{}
                            overlay_slot:
                                Padding := margin:
                                    Left := 20.0
                                    Right := 20.0
                                HorizontalAlignment := horizontal_alignment.Fill
                                VerticalAlignment := vertical_alignment.Center
                                Widget := stack_box:
                                    Orientation := orientation.Horizontal
                                    Slots := array:
                                        stack_box_slot:
                                            HorizontalAlignment := horizontal_alignment.Left
                                            VerticalAlignment := vertical_alignment.Center
                                            Widget := texture_block:
                                                DefaultImage := Icons.IconCoin
                                                DefaultTint := color{R:=1.0,G:=0.75,B:=0.0}
                                                DefaultDesiredSize := vector2{X:=100.0,Y:=100.0}
                                        stack_box_slot:
                                            Padding := margin:
                                                Top := 0.0
                                                Bottom := 0.0
                                                Left := 20.0
                                                Right := 0.0
                                            HorizontalAlignment := horizontal_alignment.Right
                                            VerticalAlignment := vertical_alignment.Center
                                            Distribution := option{1.0}
                                            Widget := PlayerGoldText
                canvas_slot: # Rewards
                    Anchors := anchors:
                        Minimum := vector2:
                            X:=0.0
                            Y:=0.5
                        Maximum := vector2:
                            X:=0.0
                            Y:=0.5
                    Offsets := margin:
                        Left := 25.0
                        Top := 0.0
                    Alignment := vector2:
                        X:=0.0
                        Y:=0.0
                    SizeToContent := true
                    Widget := overlay:
                        Slots := array:
                            overlay_slot:
                                Padding := margin:
                                    Left := -30.0
                                    Top := -20.0
                                    Right := -30.0
                                    Bottom := -20.0
                                HorizontalAlignment := horizontal_alignment.Fill
                                VerticalAlignment := vertical_alignment.Fill
                                Widget := UI.Parts.WB_Shop_ShopBackground{}
                            overlay_slot:
                                Padding := margin:
                                    Left := 0.0
                                    Top := 0.0
                                    Right := 0.0
                                    Bottom := 0.0
                                HorizontalAlignment := horizontal_alignment.Fill
                                VerticalAlignment := vertical_alignment.Fill
                                Widget := stack_box:
                                    Orientation := orientation.Vertical
                                    Slots := array:
                                        stack_box_slot:
                                            HorizontalAlignment := horizontal_alignment.Fill
                                            VerticalAlignment := vertical_alignment.Center
                                            Widget := PrestigeButton
                                        stack_box_slot:
                                            HorizontalAlignment := horizontal_alignment.Fill
                                            VerticalAlignment := vertical_alignment.Center
                                            Padding := margin:
                                                Top := 10.0
                                            Widget := DailyButton
                                        stack_box_slot:
                                            HorizontalAlignment := horizontal_alignment.Fill
                                            VerticalAlignment := vertical_alignment.Center
                                            Padding := margin:
                                                Top := 10.0
                                            Widget := NameplateButton
                canvas_slot: # Shop
                    Anchors := anchors:
                        Minimum := vector2:
                            X:=0.0
                            Y:=1.0
                        Maximum := vector2:
                            X:=0.0
                            Y:=1.0
                    Offsets := margin:
                        Left := 25.0
                        Top := -30.0
                    Alignment := vector2:
                        X:=0.0
                        Y:=1.0
                    SizeToContent := true
                    Widget := overlay:
                        Slots := array:
                            overlay_slot:
                                Padding := margin:
                                    Left := -10.0
                                    Top := -10.0
                                    Right := -10.0
                                    Bottom := -10.0
                                HorizontalAlignment := horizontal_alignment.Fill
                                VerticalAlignment := vertical_alignment.Fill
                                Widget := UI.Parts.WB_Shop_ShopBackground{}
                            overlay_slot:
                                Padding := margin:
                                    Left := 15.0
                                    Top := 15.0
                                    Right := 15.0
                                    Bottom := 15.0
                                HorizontalAlignment := horizontal_alignment.Fill
                                VerticalAlignment := vertical_alignment.Fill
                                Widget := stack_box:
                                    Orientation := orientation.Vertical
                                    Slots := array:
                                        stack_box_slot:
                                            HorizontalAlignment := horizontal_alignment.Left
                                            Widget := AbilityNameText
                                        stack_box_slot: # Price and equip button
                                            HorizontalAlignment := horizontal_alignment.Fill
                                            VerticalAlignment := vertical_alignment.Fill
                                            Padding := margin:
                                                Top := 10.0
                                                Bottom := 20.0
                                            Widget := stack_box:
                                                Orientation := orientation.Horizontal
                                                Slots := array:
                                                    stack_box_slot:
                                                        Padding := margin:
                                                            Right := 5.0
                                                        HorizontalAlignment := horizontal_alignment.Left
                                                        VerticalAlignment := vertical_alignment.Center
                                                        Widget := PriceImage
                                                    stack_box_slot:
                                                        Distribution := option{1.0}
                                                        HorizontalAlignment := horizontal_alignment.Left
                                                        VerticalAlignment := vertical_alignment.Center
                                                        Widget := PriceText
                                                    stack_box_slot:
                                                        Padding := margin:
                                                            Right := 20.0
                                                        HorizontalAlignment := horizontal_alignment.Fill
                                                        VerticalAlignment := vertical_alignment.Center
                                                        Widget := ApplyButton
                                        stack_box_slot:
                                            HorizontalAlignment := horizontal_alignment.Fill
                                            VerticalAlignment := vertical_alignment.Fill
                                            Widget := stack_box:
                                                Orientation := orientation.Horizontal
                                                Slots := AllUiButtons
                canvas_slot: # Ready
                    Anchors := anchors:
                        Minimum := vector2:
                            X:=1.0
                            Y:=1.0
                        Maximum := vector2:
                            X:=1.0
                            Y:=1.0
                    Offsets := margin:
                        Left := -20.0
                        Top := -40.0
                    Alignment := vector2{X:=1.0,Y:=1.0}
                    SizeToContent := true
                    Widget := overlay:
                        Slots := array:
                            overlay_slot:
                                HorizontalAlignment := horizontal_alignment.Fill
                                VerticalAlignment := vertical_alignment.Fill
                                Padding := margin:
                                    Top := -20.0
                                    Bottom := -20.0
                                Widget := UI.Parts.WB_Shop_ShopBackground{}
                            overlay_slot:
                                HorizontalAlignment := horizontal_alignment.Fill
                                VerticalAlignment := vertical_alignment.Fill
                                Padding := margin:
                                    Left := 40.0
                                    Right := 40.0
                                Widget := stack_box:
                                    Orientation := orientation.Horizontal
                                    Slots := array:
                                        stack_box_slot:
                                            HorizontalAlignment := horizontal_alignment.Fill
                                            VerticalAlignment := vertical_alignment.Center
                                            Padding := margin:
                                                Right := 20.0
                                            Widget := TimerText
                                        stack_box_slot:
                                            HorizontalAlignment := horizontal_alignment.Right
                                            VerticalAlignment := vertical_alignment.Center
                                            Widget := ReadyButton
                # TODO: Add players ready - Check
                canvas_slot:
                    Anchors := anchors:
                        Minimum := vector2:
                            X:=0.5
                            Y:=0.5
                        Maximum := vector2:
                            X:=0.5
                            Y:=0.5
                    Offsets := margin:
                        Left := 70.0
                        Top := 0.0
                    Alignment := vector2{X:=0.5,Y:=4.0}
                    SizeToContent := true
                    Widget := stack_box:
                        Orientation := orientation.Horizontal
                        Slots := PlayersList

    DailyLoop()<suspends>:void=
        if:
            Player.IsActive[]
            LatestDaily := PlayersStats[Player].LatestDailyRecieved
        then:
            loop:
                if:
                    Player.IsActive[]
                    Time := 86400.0 - Abs(GetSecondsSinceEpoch() - LatestDaily)
                    Time > 0.0
                then:
                    if:
                        H := Floor[Time / 3600.0]
                        M := Mod[Floor[Time / 60.0], 60]
                        S := Mod[Floor[Time], 60]
                    then:
                        var StringH : string = block:
                            if (H > 9):
                                "{H}"
                            else:
                                "0{H}"
                        var StringM : string = block:
                            if (M > 9):
                                "{M}"
                            else:
                                "0{M}"
                        var StringS : string = block:
                            if (S > 9):
                                "{S}"
                            else:
                                "0{S}"
                        if (H > 0):
                            DailyButton.SetText(MessagesBank{}.Daytime(StringH, StringM, StringS))
                        else if (M > 0):
                            DailyButton.SetText(MessagesBank{}.Daytime(StringM, StringS))
                        else:
                            DailyButton.SetText(MessagesBank{}.Daytime(StringS))
                    Sleep(1.0)
                    if (Ended = true):
                        break
                else:
                    break
            DailyButton.SetText(MessagesBank{}.DailyReward)
            DailyButton.SetEnabled(true)
        else:
            Print("Error with player!", ?Color := NamedColors.Red)

    Start()<suspends>:void=
        ApplyButton.SetEnabled(false)
        PrestigeButton.SetEnabled(false)
        DailyButton.SetEnabled(false)
        PriceImage.SetVisibility(widget_visibility.Collapsed)
        AbilityNameText.SetVisibility(widget_visibility.Collapsed)
        PriceText.SetVisibility(widget_visibility.Collapsed)
        ApplyButton.SetVisibility(widget_visibility.Collapsed)
        loop:
            if (MainClass.Achievements.CheckExistingProfile[Player]):
                # Print("Profile exists")
                break
            else:
                # Print("Profile does not exist")
            Sleep(0.2)
        spawn{DailyLoop()}
        IntroSeq.Play(Player)
        PlayerGoldText.SetText(MessagesBank{}.ToMsg(MainClass.Achievements.GetAchievement(Player, PossibleAchievements.Coins)))
        CheckPrestigeAvailability()
        PrestigeButton.OnClick().Subscribe(Prestige)
        DailyButton.OnClick().Subscribe(DailyReward)
        NameplateButton.OnClick().Subscribe(Nameplate)
        ApplyButton.OnClick().Subscribe(Apply)
        ReadyButton.OnClick().Subscribe(Ready)
        for (i : Abilities):
            NewButton := button_quiet:
                DefaultText := MessagesBank{}.Empty
            NewButton.SetEnabled(i.IsEnabled)
            set AbilitiesButtons += array{NewButton}
            set AllUiButtons += array:
                stack_box_slot:
                    Padding := margin:
                        Right := 15.0
                    HorizontalAlignment := horizontal_alignment.Fill
                    VerticalAlignment := vertical_alignment.Fill
                    Widget := overlay:
                        Slots := array:
                            overlay_slot:
                                Padding := margin:
                                    Left := 10.0
                                    Top := 10.0
                                    Right := 10.0
                                    Bottom := 10.0
                                HorizontalAlignment := horizontal_alignment.Center
                                VerticalAlignment := vertical_alignment.Center
                                Widget := texture_block:
                                    DefaultDesiredSize := vector2:
                                        X := 80.0
                                        Y := 80.0
                                    DefaultImage := block:
                                        case (i.Icon):
                                            IconsList.Empty =>
                                                AllIcons.Cell
                                            IconsList.Addition =>
                                                AllIcons.Addition
                                            IconsList.Arrow =>
                                                AllIcons.Arrow
                                            IconsList.Behind =>
                                                AllIcons.Behind
                                            IconsList.Bracket =>
                                                AllIcons.Bracket
                                            IconsList.Broken =>
                                                AllIcons.Broken
                                            IconsList.Butterfly =>
                                                AllIcons.Butterfly
                                            IconsList.Cell =>
                                                AllIcons.Cell
                                            IconsList.Checkmark =>
                                                AllIcons.Checkmark
                                            IconsList.Coin =>
                                                AllIcons.Coin
                                            IconsList.Crown =>
                                                AllIcons.Crown
                                            IconsList.Door =>
                                                AllIcons.Door
                                            IconsList.Double =>
                                                AllIcons.Double
                                            IconsList.Electric =>
                                                AllIcons.Electric
                                            IconsList.Energy =>
                                                AllIcons.Energy
                                            IconsList.Eye =>
                                                AllIcons.Eye
                                            IconsList.Heart =>
                                                AllIcons.Heart
                                            IconsList.Lock =>
                                                AllIcons.Lock
                                            IconsList.Place =>
                                                AllIcons.Place
                                            IconsList.Reboot =>
                                                AllIcons.Reboot
                                            IconsList.Rift =>
                                                AllIcons.Rift
                                            IconsList.Travel =>
                                                AllIcons.Travel
                                            IconsList.Umbrella =>
                                                AllIcons.Umbrella
                                            IconsList.Up =>
                                                AllIcons.Up
                                    DefaultTint := block:
                                        # if (i.MinimumSurvivors > MainClass.RolesInfo.Survivors.Length):
                                        #     NamedColors.Red
                                        # else if (MainClass.Achievements.GetBoughtItems(Player).Find[i.ID]):
                                        #     NamedColors.White
                                        # else:
                                        NamedColors.Gray
                                        # Check if bought

                            overlay_slot:
                                HorizontalAlignment := horizontal_alignment.Fill
                                VerticalAlignment := vertical_alignment.Fill
                                Widget := NewButton
            NewButton.OnClick().Subscribe(ChangeSelection)
        CreateCanvas()
        UpdateColors()
        PlUI.AddWidget(Canvas, player_ui_slot{InputMode := ui_input_mode.All})
        spawn{CountDown()}

    Final():int=
        IntroSeq.Stop(Player)
        if (NP := NpUI?):
            NP.Remove()
        if (Player <> MainClass.RolesInfo.Monster?):
            PlUI.RemoveWidget(Canvas)
        else:
            spawn{MonsterWait()}
        set Ended = true
        return SelectedID

    ChangeSelection(Info : widget_message):void=
        if (Selected := Abilities[AbilitiesButtons.Find[button_quiet[Info.Source]]]):
            ApplyButton.SetEnabled(true)
            AbilityNameText.SetText(
                block:
                    case (Selected.ID):
                        0 =>
                            MessagesBank{}.SurvivorsItem1Name
                        1 =>
                            MessagesBank{}.SurvivorsItem2Name
                        2 =>
                            MessagesBank{}.SurvivorsItem3Name
                        3 =>
                            MessagesBank{}.SurvivorsItem4Name
                        4 =>
                            MessagesBank{}.SurvivorsItem5Name
                        5 =>
                            MessagesBank{}.SurvivorsItem6Name
                        6 =>
                            MessagesBank{}.MonsterItem1Name
                        7 =>
                            MessagesBank{}.MonsterItem2Name
                        8 =>
                            MessagesBank{}.MonsterItem3Name
                        9 =>
                            MessagesBank{}.MonsterItem4Name
                        10 =>
                            MessagesBank{}.MonsterItem5Name
                        11 =>
                            MessagesBank{}.MonsterItem6Name
                        _ =>
                            MessagesBank{}.Empty
            )
            if (MainClass.DEBUG_AllFree?):
                PriceText.SetText(MessagesBank{}.ToMsg(0))
            else:
                PriceText.SetText(MessagesBank{}.ToMsg(Selected.Price))
            var IsOwned : logic = false
            if (MainClass.Achievements.GetBoughtItems(Player).Find[Selected.ID], set IsOwned = true) {}
            if (Selected.MinimumSurvivors > MainClass.RolesInfo.Survivors.Length):
                if (IsOwned?):
                    ApplyButton.SetEnabled(false)
                    ApplyButton.SetText(MessagesBank{}.AbilityNotAvailable)
                else:
                    ApplyButton.SetText(MessagesBank{}.BuyNotAvailable)
            else:
                if (IsOwned?):
                    ApplyButton.SetText(MessagesBank{}.EquipItem)
                else:
                    ApplyButton.SetText(MessagesBank{}.BuyItem)
            if (SelectedID = Selected.ID):
                ApplyButton.SetEnabled(false)
                ApplyButton.SetText(MessagesBank{}.ItemEquipped)
            if (not MainClass.DEBUG_AllFree? and MainClass.Achievements.GetAchievement(Player, PossibleAchievements.Coins) < Selected.Price):
                ApplyButton.SetEnabled(false)
            if (FocusedID = Selected.ID):
                set FocusedID = -1
                PriceImage.SetVisibility(widget_visibility.Collapsed)
                AbilityNameText.SetVisibility(widget_visibility.Collapsed)
                PriceText.SetVisibility(widget_visibility.Collapsed)
                ApplyButton.SetVisibility(widget_visibility.Collapsed)
            else:
                set FocusedID = Selected.ID
                PriceImage.SetVisibility(widget_visibility.Visible)
                AbilityNameText.SetVisibility(widget_visibility.Visible)
                PriceText.SetVisibility(widget_visibility.Visible)
                ApplyButton.SetVisibility(widget_visibility.Visible)
        # for (i := 0..AllUiButtons.Length - 1, Texture := texture_block[overlay[AllUiButtons[i].Widget].Slots[0].Widget], TheAb := Abilities[i]):
        #     Texture.SetTint(
        #         block:
        #             if (TheAb.MinimumSurvivors > MainClass.RolesInfo.Survivors.Length):
        #                 NamedColors.Red
        #             else if (SelectedID = TheAb.ID):
        #                 NamedColors.Green
        #             else if (FocusedID = TheAb.ID):
        #                 NamedColors.Yellow
        #             else if (MainClass.Achievements.GetBoughtItems(Player).Find[TheAb.ID]):
        #                 NamedColors.White
        #             else:
        #                 NamedColors.Gray
        #     )
        UpdateColors()
            
    UpdateUiApply(Agent : player, Selection : texture):void=
        if (TheTexture := AllPlayersChoices[Agent]):
            TheTexture.SetImage(Selection)

    UpdateUiReady(Agent : player, State : logic):void=
        if (TheReady := AllPlayersReady[Agent]):
            if (Agent.IsActive[]):
                case (State):
                    true =>
                        TheReady.SetText(MessagesBank{}.Ready)
                        TheReady.SetTextColor(color{G := 1.0})
                    false =>
                        TheReady.SetText(MessagesBank{}.NotReady)
                        TheReady.SetTextColor(color{R := 1.0})
                    _ =>
            else:
                TheReady.SetText(MessagesBank{}.Disconnected)
                TheReady.SetTextColor(NamedColors.Gray)

    Apply(Info : widget_message):void=
        if (FocusedID = -1):
            Print("No ability is focused", ?Color := NamedColors.Red)
            return
        var Selected : Ability = Ability{}
        var IsOwned : logic = false
        var IsAvailable : logic = true
        for (i : Abilities):
            if (i.ID = FocusedID):
                set Selected = i
                if (temp := MainClass.Achievements.GetBoughtItems(Player).Find[i.ID], set IsOwned = true) {}
                if (i.MinimumSurvivors > MainClass.RolesInfo.Survivors.Length, set IsAvailable = false) {}
        if (not IsOwned?):
            if (MainClass.DEBUG_AllFree?) {}
            else if (MainClass.Achievements.GetAchievement(Player, PossibleAchievements.Coins) >= Selected.Price):
                MainClass.Achievements.RemoveAchievement(Player, PossibleAchievements.Coins, Selected.Price)
                PlayerGoldText.SetText(MessagesBank{}.ToMsg(MainClass.Achievements.GetAchievement(Player, PossibleAchievements.Coins)))
            else:
                if (IsAvailable?):
                    ApplyButton.SetText(MessagesBank{}.NotEnough)
                else:
                    ApplyButton.SetText(MessagesBank{}.NotEnoughNotAvailable)
                return
            MainClass.Achievements.BuyItem(Player, Selected.ID)
            set IsOwned = true
        if (IsAvailable?):
            set SelectedID = FocusedID
            ApplyButton.SetText(MessagesBank{}.ItemEquipped)
        else:
            ApplyButton.SetText(MessagesBank{}.AbilityNotAvailable)
        ApplyButton.SetEnabled(false)
        CheckPrestigeAvailability()
        NewIcon := block:
            case (Selected.Icon):
                IconsList.Empty =>
                    AllIcons.Cell
                IconsList.Addition =>
                    AllIcons.Addition
                IconsList.Arrow =>
                    AllIcons.Arrow
                IconsList.Behind =>
                    AllIcons.Behind
                IconsList.Bracket =>
                    AllIcons.Bracket
                IconsList.Broken =>
                    AllIcons.Broken
                IconsList.Butterfly =>
                    AllIcons.Butterfly
                IconsList.Cell =>
                    AllIcons.Cell
                IconsList.Checkmark =>
                    AllIcons.Checkmark
                IconsList.Coin =>
                    AllIcons.Coin
                IconsList.Crown =>
                    AllIcons.Crown
                IconsList.Door =>
                    AllIcons.Door
                IconsList.Double =>
                    AllIcons.Double
                IconsList.Electric =>
                    AllIcons.Electric
                IconsList.Energy =>
                    AllIcons.Energy
                IconsList.Eye =>
                    AllIcons.Eye
                IconsList.Heart =>
                    AllIcons.Heart
                IconsList.Lock =>
                    AllIcons.Lock
                IconsList.Place =>
                    AllIcons.Place
                IconsList.Reboot =>
                    AllIcons.Reboot
                IconsList.Rift =>
                    AllIcons.Rift
                IconsList.Travel =>
                    AllIcons.Travel
                IconsList.Umbrella =>
                    AllIcons.Umbrella
                IconsList.Up =>
                    AllIcons.Up
        UpdateColors()
        for (i : AllUIs[GetSession()], IsOwned? and IsAvailable?):
            i.UpdateUiApply(Player, NewIcon)

    Ready(widget_info : widget_message):void=
        if (btn := button_loud[widget_info.Source]):
            txt := btn.GetText()
            if (txt = Localize(MessagesBank{}.Ready)):
                if (Player <> MainClass.RolesInfo.Monster?):
                    for (i : MainClass.RolesInfo.Survivors):
                        MainClass.ReadyAudio.Play(i)
                else:
                    MainClass.ReadyAudio.Play(Player)
                btn.SetText(MessagesBank{}.Unready)
                # set MainClass.PlayersReady += 1
                set MainClass.PlayersReady += array{Player}
                # if (MainClass.PlayersReady = MainClass.RolesInfo.Survivors.Length + 1):
                #     widget_info.Source.SetEnabled(false)
                for (i : AllUIs[GetSession()]):
                    i.UpdateUiReady(Player, true)
            else:
                if (Player <> MainClass.RolesInfo.Monster?):
                    for (i : MainClass.RolesInfo.Survivors):
                        MainClass.NotReadyAudio.Play(i)
                else:
                    MainClass.NotReadyAudio.Play(Player)
                btn.SetText(MessagesBank{}.Ready)
                # set MainClass.PlayersReady -= 1
                set MainClass.PlayersReady = MainClass.PlayersReady.RemoveAllElements(Player)
                for (i : AllUIs[GetSession()]):
                    i.UpdateUiReady(Player, false)
            spawn{CooldownButton(btn, 0.5)}

    CooldownButton(Button : text_button_base, Duration : float)<suspends>:void=
        Button.SetEnabled(false)
        Sleep(Duration)
        Button.SetEnabled(true)

    Prestige(widget_info : widget_message):void=
        MainClass.Achievements.SetAchievement(Player, PossibleAchievements.Coins, 0)
        MainClass.Achievements.ResetItems(Player)
        PlayerGoldText.SetText(MessagesBank{}.ToMsg(0))
        PriceImage.SetVisibility(widget_visibility.Collapsed)
        AbilityNameText.SetVisibility(widget_visibility.Collapsed)
        PriceText.SetVisibility(widget_visibility.Collapsed)
        ApplyButton.SetVisibility(widget_visibility.Collapsed)
        set FocusedID = -1
        set SelectedID = -1
        if (MainClass.DEBUG_PrestigeAlwaysAvailable = false):
            PrestigeButton.SetEnabled(false)
        MainClass.Achievements.SaveAchievement(Player, PossibleAchievements.Prestige, 1)
        MainClass.Feedback_Prestige.Submit(Player)
        # set MainClass.Prestiged += array{Player}
        MainClass.Achievements.Award(Player, 7)
        for (i : AllUIs[GetSession()]):
            i.UpdateUiApply(Player, AllIcons.None)
        UpdateColors()

    DailyReward(Info : widget_message):void=
        Res := MainClass.Achievements.GetDailyReward(Player)
        if (Res <> -1):
            DailyButton.SetEnabled(false)
            PlayerGoldText.SetText(MessagesBank{}.ToMsg(Res))
            # set MainClass.DailyRewarded += array{Player}
            MainClass.Achievements.Award(Player, 8)
            # Check if the currently focused ability can now be bought
            if (FocusedID <> -1):
                for (i : Abilities):
                    if (i.ID = FocusedID, not MainClass.Achievements.GetBoughtItems(Player).Find[i.ID]):
                        if (MainClass.Achievements.GetAchievement(Player, PossibleAchievements.Coins) >= i.Price and i.MinimumSurvivors <= MainClass.RolesInfo.Survivors.Length):
                            ApplyButton.SetEnabled(true)
                        else:
                            ApplyButton.SetEnabled(false)
            else:
                ApplyButton.SetEnabled(false) 
            spawn{DailyLoop()}
        else:
            Print("Issues getting daily reward", ?Color := NamedColors.Red)

    Nameplate(Info : widget_message):void=
        PlUI.RemoveWidget(Canvas)
        if (NP := NpUI?):
            NP.ReturnHere()
        else:
            NP := NameplateUI{Player := Player, PlUI := PlUI, GoBackTo := ReturnHere, Achievements := MainClass.Achievements}
            set NpUI = option{NP}
            NP.Start()

    ReturnHere():void=
        PlUI.AddWidget(Canvas, player_ui_slot{InputMode := ui_input_mode.All})

    CountDown()<suspends>:void=
        for (i := 0..19):
            TimerText.SetText(MessagesBank{}.ToMsg(20 - i))
            Sleep(1.0)

    MonsterWait()<suspends>:void=
        PlUI.RemoveWidget(Canvas)
        txt := text_block:
            DefaultText := MessagesBank{}.YouWillBeReleasedIn
            DefaultTextColor := color{R:=1.0,G:=1.0,B:=1.0}
        set Canvas = canvas:
            Slots := array:
                canvas_slot:
                    Widget := txt
                    Anchors := anchors{
                        Minimum := vector2{X:=0.5,Y:=0.5},
                        Maximum := vector2{X:=0.5,Y:=0.5}
                    }
                    Offsets := margin{
                        Left := 0.0,
                        Top := 0.0,
                        Right := 0.0,
                        Bottom := 0.0
                    }
                    Alignment := vector2{X:=0.5,Y:=0.5}
        PlUI.AddWidget(Canvas, player_ui_slot{InputMode := ui_input_mode.All})
        for (i := 0..MainClass.RolesInfo.MonsterReleaseTime):
            txt.SetText(MessagesBank{}.Connect(MessagesBank{}.YouWillBeReleasedIn, "{MainClass.RolesInfo.MonsterReleaseTime + 1 - i}...", MessagesBank{}.Empty))
            Sleep(1.0)
        PlUI.RemoveWidget(Canvas)
        # Print("Ended shop for player")
        Sleep(0.01)
        spawn{MainClass.RolesInfo.WaitForMove(Player)}

    CheckPrestigeAvailability():void=
        if (MainClass.Achievements.GetBoughtItems(Player).Length = 12 or MainClass.DEBUG_PrestigeAlwaysAvailable = true):
            PrestigeButton.SetEnabled(true)

    UpdateColors():void=
        for (i := 0..Abilities.Length - 1, CurAbility := Abilities[i], TheTexture := texture_block[overlay[AllUiButtons[i].Widget].Slots[0].Widget]):
            TheTexture.SetTint(
                block:
                    if (CurAbility.MinimumSurvivors > MainClass.RolesInfo.Survivors.Length):
                        NamedColors.Red
                    else if (SelectedID = CurAbility.ID):
                        NamedColors.Green
                    else if (FocusedID = CurAbility.ID):
                        NamedColors.Yellow
                    else if (MainClass.Achievements.GetBoughtItems(Player).Find[CurAbility.ID]):
                        NamedColors.White
                    else:
                        NamedColors.Gray
            )


roomchoice := enum:
    Kitchen
    Vault
    Jungle
    Library
    DiningRoom
    SodaStorage
    Church
    AlchemistsRoom
    Chaos
    Warehouse
    SurvivorsSpawn

TeleportMonsterCon := class:
    Room : roomchoice
    MainClass : AbilitiesHandler
    MiddleClass : TeleportMonsterUI

    Connect(WM:widget_message):void=
        MiddleClass.Close(WM)
        MainClass.Monster_TeleportToSurvivor_Trigger.Unregister(MiddleClass.Agent)
        case (Room):
            roomchoice.Kitchen =>
                MainClass.KitchenTeleport.Teleport(MiddleClass.Agent)
            roomchoice.Vault =>
                MainClass.VaultTeleport.Teleport(MiddleClass.Agent)
            roomchoice.Jungle =>
                MainClass.JungleTeleport.Teleport(MiddleClass.Agent)
            roomchoice.Library =>
                MainClass.LibraryTeleport.Teleport(MiddleClass.Agent)
            roomchoice.DiningRoom =>
                MainClass.DiningRoomTeleport.Teleport(MiddleClass.Agent)
            roomchoice.SodaStorage =>
                MainClass.SodaStorageTeleport.Teleport(MiddleClass.Agent)
            roomchoice.Church =>
                MainClass.ChurchTeleport.Teleport(MiddleClass.Agent)
            roomchoice.AlchemistsRoom =>
                MainClass.AlchemistsRoomTeleport.Teleport(MiddleClass.Agent)
            roomchoice.Chaos =>
                MainClass.ChaosTeleport.Teleport(MiddleClass.Agent)
            roomchoice.Warehouse =>
                MainClass.WarehouseTeleport.Teleport(MiddleClass.Agent)
            roomchoice.SurvivorsSpawn =>
                MainClass.SurvivorsSpawnTeleport.Teleport(MiddleClass.Agent)
        if (set MainClass.RolesInfo.PlayersParams[player[MiddleClass.Agent]].TeleportToSurvivor.Value -= 1) {}
        if (MainClass.RolesInfo.PlayersParams[player[MiddleClass.Agent]].TeleportToSurvivor.Value <= 0, set MainClass.RolesInfo.PlayersParams[player[MiddleClass.Agent]].TeleportToSurvivor.GrayedOut = true) {}
        if (PlStatHud := StatsUiCon[GetSession()][MiddleClass.Agent]):
            PlStatHud.FastUpdate()

TeleportMonsterUI := class:
    Agent : agent
    PlUI : player_ui
    MainClass : AbilitiesHandler

    KitchenButton : button_quiet = button_quiet:
        DefaultText := MessagesBank{}.Kitchen
    VaultButton : button_quiet = button_quiet:
        DefaultText := MessagesBank{}.Vault
    JungleButton : button_quiet = button_quiet:
        DefaultText := MessagesBank{}.Jungle
    LibraryButton : button_quiet = button_quiet:
        DefaultText := MessagesBank{}.Library
    DiningRoomButton : button_quiet = button_quiet:
        DefaultText := MessagesBank{}.DiningRoom
    SodaStorageButton : button_quiet = button_quiet:
        DefaultText := MessagesBank{}.SodaStorage
    ChurchButton : button_quiet = button_quiet:
        DefaultText := MessagesBank{}.Church
    AlchemistsRoomButton : button_quiet = button_quiet:
        DefaultText := MessagesBank{}.AlchemistsRoom
    ChaosButton : button_quiet = button_quiet:
        DefaultText := MessagesBank{}.Chaos
    WarehouseButton : button_quiet = button_quiet:
        DefaultText := MessagesBank{}.Warehouse
    SurvivorsSpawnButton : button_quiet = button_quiet:
        DefaultText := MessagesBank{}.SurvivorsSpawn
    
    CancelButton : button_loud = button_loud:
        DefaultText := MessagesBank{}.Unready

    var Canvas<private> : canvas = canvas{}

    Start():void=
        KitchenButton.OnClick().Subscribe(TeleportMonsterCon{Room := roomchoice.Kitchen, MainClass := MainClass, MiddleClass := Self}.Connect)
        VaultButton.OnClick().Subscribe(TeleportMonsterCon{Room := roomchoice.Vault, MainClass := MainClass, MiddleClass := Self}.Connect)
        JungleButton.OnClick().Subscribe(TeleportMonsterCon{Room := roomchoice.Jungle, MainClass := MainClass, MiddleClass := Self}.Connect)
        LibraryButton.OnClick().Subscribe(TeleportMonsterCon{Room := roomchoice.Library, MainClass := MainClass, MiddleClass := Self}.Connect)
        DiningRoomButton.OnClick().Subscribe(TeleportMonsterCon{Room := roomchoice.DiningRoom, MainClass := MainClass, MiddleClass := Self}.Connect)
        SodaStorageButton.OnClick().Subscribe(TeleportMonsterCon{Room := roomchoice.SodaStorage, MainClass := MainClass, MiddleClass := Self}.Connect)
        ChurchButton.OnClick().Subscribe(TeleportMonsterCon{Room := roomchoice.Church, MainClass := MainClass, MiddleClass := Self}.Connect)
        AlchemistsRoomButton.OnClick().Subscribe(TeleportMonsterCon{Room := roomchoice.AlchemistsRoom, MainClass := MainClass, MiddleClass := Self}.Connect)
        ChaosButton.OnClick().Subscribe(TeleportMonsterCon{Room := roomchoice.Chaos, MainClass := MainClass, MiddleClass := Self}.Connect)
        WarehouseButton.OnClick().Subscribe(TeleportMonsterCon{Room := roomchoice.Warehouse, MainClass := MainClass, MiddleClass := Self}.Connect)
        SurvivorsSpawnButton.OnClick().Subscribe(TeleportMonsterCon{Room := roomchoice.SurvivorsSpawn, MainClass := MainClass, MiddleClass := Self}.Connect)
        CancelButton.OnClick().Subscribe(Close)
        set Canvas = canvas:
            Slots := array:
                canvas_slot:
                    ZOrder := 0
                    Anchors := anchors:
                        Minimum := vector2:
                            X := 0.0
                            Y := 0.0
                        Maximum := vector2:
                            X := 1.0
                            Y := 1.0
                    Widget := UI.WB_MonsterTeleport{}
                canvas_slot:
                    ZOrder := 1
                    Anchors := anchors{
                        Minimum := vector2{X:=0.5,Y:=0.5},
                        Maximum := vector2{X:=0.5,Y:=0.5}
                    }
                    Offsets := margin{
                        # Left := 0.0,
                        # Top := 0.0,
                        Right := 1500.0
                        Bottom := 800.0
                    }
                    Alignment := vector2{X:=0.5,Y:=0.5}
                    Widget := overlay:
                        Slots := array:
                            overlay_slot:
                                Padding := margin:
                                    Top := 250.0
                                HorizontalAlignment := horizontal_alignment.Fill
                                VerticalAlignment := vertical_alignment.Fill
                                Widget := stack_box:
                                    Orientation := orientation.Vertical
                                    Slots := array:
                                        AddLine(KitchenButton, DiningRoomButton, AlchemistsRoomButton)
                                        AddLine(VaultButton, SodaStorageButton, ChaosButton)
                                        AddLine(JungleButton, ChurchButton, WarehouseButton)
                                        AddLine(LibraryButton, CancelButton, SurvivorsSpawnButton)
        PlUI.AddWidget(Canvas, player_ui_slot{InputMode := ui_input_mode.All})
    
    AddLine(Button1 : widget, Button2 : widget, Button3 : widget):stack_box_slot=
        return stack_box_slot:
            Padding := margin:
                Bottom := 40.0
            HorizontalAlignment := horizontal_alignment.Fill
            VerticalAlignment := vertical_alignment.Fill
            Widget := stack_box:
                Orientation := orientation.Horizontal
                Slots := array:
                    AddButton(Button1)
                    AddButton(Button2)
                    AddButton(Button3)

    AddButton(Button : widget):stack_box_slot=
        return stack_box_slot:
            Distribution := option{1.0}
            HorizontalAlignment := horizontal_alignment.Center
            VerticalAlignment := vertical_alignment.Center
            Widget := Button

    Close(WM:widget_message):void=
        PlUI.RemoveWidget(Canvas)


IconsList := enum:
    Empty
    Addition
    Arrow
    Behind
    Bracket
    Broken
    Butterfly
    Cell
    Checkmark
    Coin
    Crown
    Door
    Double
    Electric
    Energy
    Eye
    Heart
    Lock
    Place
    Reboot
    Rift
    Travel
    Umbrella
    Up

Ability := struct<concrete>:
    @editable
    IsEnabled: logic = false
    @editable
    ID: int = 0
    @editable
    Price: int = 0
    @editable
    Icon: IconsList = IconsList.Empty
    @editable
    Tint: color = color{R := 1.0, G := 1.0, B := 1.0}
    @editable
    Feedback : analytics_device = analytics_device{}
    @editable_slider(int):
        MinValue := option{1}
        MaxValue := option{4}
    MinimumSurvivors: int = 1

FakeTransmitter := class(tag){}
AbilitiesHandler := class(creative_device):

    @editable
    VersionToUse : Versions = Versions.V1

    @editable
    DEBUG_AllFree : logic = false
    @editable
    DEBUG_PrestigeAlwaysAvailable : logic = false

    @editable
    Feedback_Prestige : analytics_device = analytics_device{}

    @editable
    ReadyAudio : audio_player_device = audio_player_device{}
    @editable
    NotReadyAudio : audio_player_device = audio_player_device{}

    @editable
    SurvivorsAbilities: []Ability = array{}
    @editable
    MonsterAbilities: []Ability = array{}

    @editable
    RandomLocations: []teleporter_device = array{}

    @editable
    KitchenTeleport: teleporter_device = teleporter_device{}
    @editable
    VaultTeleport: teleporter_device = teleporter_device{}
    @editable
    JungleTeleport: teleporter_device = teleporter_device{}
    @editable
    LibraryTeleport: teleporter_device = teleporter_device{}
    @editable
    DiningRoomTeleport: teleporter_device = teleporter_device{}
    @editable
    SodaStorageTeleport: teleporter_device = teleporter_device{}
    @editable
    ChurchTeleport: teleporter_device = teleporter_device{}
    @editable
    AlchemistsRoomTeleport: teleporter_device = teleporter_device{}
    @editable
    ChaosTeleport: teleporter_device = teleporter_device{}
    @editable
    WarehouseTeleport: teleporter_device = teleporter_device{}
    @editable
    SurvivorsSpawnTeleport: teleporter_device = teleporter_device{}


    @editable
    YouAreMonsterShop: hud_message_device = hud_message_device{}
    @editable
    YouAreSurvivorShop: hud_message_device = hud_message_device{}

    @editable
    SurvivorShopIntros : []cinematic_sequence_device = array{}
    @editable
    MonsterShopIntro : cinematic_sequence_device = cinematic_sequence_device{}

    @editable
    SurvivorsReferences : []player_reference_device = array{}

    @editable
    StoreMusic : radio_device = radio_device{}

    @editable
    Monster_Placeholder: input_trigger_device = input_trigger_device{}

    @editable
    Survivor_Gates_Trigger: input_trigger_device = input_trigger_device{}
    @editable
    Survivor_EmergencyTeleport_Trigger: input_trigger_device = input_trigger_device{}
    @editable
    Survivor_MonsterTeleport_Trigger: input_trigger_device = input_trigger_device{}

    @editable
    Monster_Scream_Trigger: input_trigger_device = input_trigger_device{}
    @editable
    Monster_Jump_MutatorZone: mutator_zone_device = mutator_zone_device{}
    @editable
    Monster_NoJump_MutatorZone: mutator_zone_device = mutator_zone_device{}

    @editable
    Monster_FakeTask_Trigger: input_trigger_device = input_trigger_device{}
    @editable
    Monster_FakeTask_Prop : creative_prop_asset := DefaultCreativePropAsset
    @editable
    Monster_FakeTask_TrackMarkers: []beacon_device = array{}
    @editable
    Monster_FakeTask_FakeMarkers: []beacon_device = array{}
    @editable
    Monster_FakeTask_FakeMarkers_Pumpkin: []beacon_device = array{}
    @editable
    Monster_FakeTask_Buttons: []button_device = array{}
    @editable
    Monster_FakeTask_TeleportDepth: float = 20.0

    @editable
    Monster_TeleportToSurvivor_Trigger: input_trigger_device = input_trigger_device{}

    @editable
    Monster_Scream_Markers: []beacon_device = array{}
    @editable
    Monster_Scream_Audio: audio_player_device = audio_player_device{}
    @editable
    Monster_Scream_Notification: hud_message_device = hud_message_device{}

    @editable
    Survivor_Gates_Cooldown: int = 60
    @editable
    Monster_Scream_Duration: int = 7
    @editable
    Monster_Scream_Cooldown: int = 60

    @editable
    RolesInfo : RolesHandler = RolesHandler{}
    @editable
    Achievements : AchievmentsHandler = AchievmentsHandler{}
    @editable
    Life : LifeHandler = LifeHandler{}
    # @editable
    # Gates : GatesHandler = GatesHandler{}
    @editable
    Tasks : TasksHandler = TasksHandler{}
    @editable
    Vents : VentHandler = VentHandler{}

    var Prestiged : []agent = array{}
    var DailyRewarded : []agent = array{}
    var PlayersToAbilities: [agent]Ability = map{}
    var FakeTasks: []creative_prop = array{}
    var FakeTasksV2: []entity = array{}
    var FakeTasksButtons: []button_device = array{}
    var FakeTasksMarkers: []beacon_device = array{}
    var FakeTasksSet: int = 0
    var Jumped : int = 0

    # var PlayersReady : int = 0
    var PlayersReady : []player = array{}

    ProbablyTimeRunsOut()<suspends>:void=
        for (i := 0..20):
            YouAreSurvivorShop.SetText(MessagesBank{}.ToMsg(20 - i))
            YouAreMonsterShop.SetText(MessagesBank{}.ToMsg(20 - i))
            for (pl : RolesInfo.Survivors):
                if (pl.IsActive[]):
                    YouAreSurvivorShop.Hide(pl)
                    YouAreSurvivorShop.Show(pl)
            if (pl := RolesInfo.Monster?):
                YouAreMonsterShop.Hide(pl)
                YouAreMonsterShop.Show(pl)
            Sleep(1.0)

    ProbablyAllAreReady()<suspends>:void=
        loop:
            # if (PlayersReady.Length = RolesInfo.Survivors.Length + 1):
            #     break
            var AllPlaying : []player = array{}
            for (i : RolesInfo.Survivors, i.IsActive[]):
                set AllPlaying += array{i}
            if (Monika := RolesInfo.Monster?, Monika.IsActive[]):
                set AllPlaying += array{Monika}
            for (i : PlayersReady, not i.IsActive[]):
                set PlayersReady = PlayersReady.RemoveAllElements(i)
                for (j : AllUIs[GetSession()]):
                    j.UpdateUiReady(i, false)
            if (PlayersReady.Length = AllPlaying.Length):
                break
            Sleep(0.1)
    
    OnBegin<override>()<suspends>:void=
        # ToEnd := GetPlayspace().PlayerRemovedEvent().Subscribe(StopTheShop)
        var allplayers : []player = GetPlayspace().GetPlayers()
        # if (allplayers.Length < 2):
        #     # Print("Not enough players")
        #     return
        # spawn{SyncMarkers()}
        Survivor_Gates_Trigger.UnregisterAll()
        Survivor_EmergencyTeleport_Trigger.UnregisterAll()
        Survivor_MonsterTeleport_Trigger.UnregisterAll()
        Monster_Scream_Trigger.UnregisterAll()
        Monster_FakeTask_Trigger.UnregisterAll()
        Monster_TeleportToSurvivor_Trigger.UnregisterAll()
        StoreMusic.Play()
        loop:
            if (WaitingForPlayersStatus[GetSession()] = false):
                break
            Sleep(0.1)
        Sleep(0.1)
        if (VersionToUse = Versions.V1):
            # Print("Version 2 of the AbilitiesHandler is started!")
            var AllUis : []ShopUI = array{}
            for (i : SurvivorsAbilities):
                var FoundSelf : logic = false
                for (g : SurvivorsAbilities):
                    if (i.ID = g.ID):
                        if (FoundSelf = false):
                            set FoundSelf = true
                        else:
                            # Print("Duplicate ability ID: {i.ID}")
                for (g : MonsterAbilities):
                    if (i.ID = g.ID):
                        if (FoundSelf = false):
                            set FoundSelf = true
                        else:
                            # Print("Duplicate ability ID: {i.ID}")
            for (i : MonsterAbilities):
                var FoundSelf : logic = false
                for (g : MonsterAbilities):
                    if (i.ID = g.ID):
                        if (FoundSelf = false):
                            set FoundSelf = true
                        else:
                            # Print("Duplicate ability ID: {i.ID}")
            for (i : RolesInfo.Survivors, i <> RolesInfo.Monster?):
                YouAreSurvivorShop.Show(i)
                ui := ShopUI:
                    Agent := i
                    Abilities := SurvivorsAbilities
                    MainClass := Self
                set AllUis += array{ui}
                if (Ui := AllUis[AllUis.Length-1]):
                    spawn{ui.Start()}
            if (ag := RolesInfo.Monster?):
                YouAreMonsterShop.Show(ag)
                ui := ShopUI:
                    Agent := ag
                    Abilities := MonsterAbilities
                    MainClass := Self
                set AllUis += array{ui}
                if (Ui := AllUis[AllUis.Length-1]):
                    spawn{ui.Start()}
            # WaitForChoices()
            race:
                ProbablyTimeRunsOut()
                ProbablyAllAreReady()
            for (i : AllUis):
                Selection := i.Ending()
                # if (Selection < SurvivorsAbilities.Length, Selected := SurvivorsAbilities[Selection], set PlayersToAbilities[i.Agent] = Selected) {}
                for (g : SurvivorsAbilities):
                    if (Selection = g.ID, set PlayersToAbilities[i.Agent] = g):
                        g.Feedback.Submit(i.Agent)
                for (g : MonsterAbilities):
                    if (Selection = g.ID, set PlayersToAbilities[i.Agent] = g):
                        g.Feedback.Submit(i.Agent)
                # Print("Player has chosen {Selection}")
                case (Selection):
                    0 =>
                        if (set RolesInfo.PlayersParams[player[i.Agent]] = player_params{ImprovedFlashlightBattery := player_param_info{Value := 0}, PublicIcon := Icons.IconEnergy}) {}
                    1 =>
                        if (set RolesInfo.PlayersParams[player[i.Agent]] = player_params{GatesRemoteControl := player_param_info{Value := 0}, PublicIcon := Icons.IconLock}) {}
                        Survivor_Gates_Trigger.Register(i.Agent)
                    2 =>
                        if (set RolesInfo.PlayersParams[player[i.Agent]] = player_params{OneTimeVentAccess := player_param_info{Value := 1}, PublicIcon := Icons.IconBracket}) {}
                        for (g : RolesInfo.VentsMarkers):
                            g.AddToShowList(i.Agent)
                    3 =>
                        if (set RolesInfo.PlayersParams[player[i.Agent]] = player_params{OneTimeInstantRevive := player_param_info{Value := 1}, PublicIcon := Icons.IconReboot}) {}
                    4 =>
                        if (set RolesInfo.PlayersParams[player[i.Agent]] = player_params{OneTimeEmergencyTeleport := player_param_info{Value := 1}, PublicIcon := Icons.IconRift}) {}
                        Survivor_EmergencyTeleport_Trigger.Register(i.Agent)
                    5 =>
                        if (set RolesInfo.PlayersParams[player[i.Agent]] = player_params{OneTimeMonsterTeleport := player_param_info{Value := 1}, PublicIcon := Icons.IconPlace}) {}
                        Survivor_MonsterTeleport_Trigger.Register(i.Agent)
                    6 =>
                        if (set RolesInfo.PlayersParams[player[i.Agent]] = player_params{OneAndHalfDamage := player_param_info{Value := 0}, PublicIcon := Icons.IconAddition}) {}
                    7 =>
                        if (set RolesInfo.PlayersParams[player[i.Agent]] = player_params{MonsterJumps := player_param_info{Value := 3}, PublicIcon := Icons.IconUp}) {}
                        if (chr := i.Agent.GetFortCharacter[]):
                            chr.JumpedEvent().Subscribe(JumpedEvent)
                        Monster_Jump_MutatorZone.Enable()
                        Monster_NoJump_MutatorZone.Disable()
                    8 =>
                        if (set RolesInfo.PlayersParams[player[i.Agent]] = player_params{GatesControl := player_param_info{Value := 0}, PublicIcon := Icons.IconDoor}) {}
                        if (Root := GetSimulationEntity[]):
                            for (Entity : Root.FindDescendantEntitiesWithComponent(SG_Gate), Info := Entity.GetComponent[SG_Gate]):
                                Info.SetPublic()
                    9 =>
                        if (set RolesInfo.PlayersParams[player[i.Agent]] = player_params{FakeTasks := player_param_info{Value := 3}, PublicIcon := Icons.IconBroken}) {}
                        Monster_FakeTask_Trigger.Register(i.Agent)
                    10 =>
                        if (set RolesInfo.PlayersParams[player[i.Agent]] = player_params{TeleportToSurvivor := player_param_info{Value := 1, GrayedOut := true}, PublicIcon := Icons.IconBehind}) {}
                        Monster_TeleportToSurvivor_Trigger.Register(i.Agent)
                    11 =>
                        if (set RolesInfo.PlayersParams[player[i.Agent]] = player_params{OneTimeTwoHealthDamage := player_param_info{Value := 1}, PublicIcon := Icons.IconDouble}) {}
                    _ =>
                        if (set RolesInfo.PlayersParams[player[i.Agent]] = player_params{}) {}
                        # Print("Ability not implemented")
                if (GetRandomInt(0, 729674) = 729674, set RolesInfo.PlayersParams[player[i.Agent]].GhostSave = player_param_info{Value := 1}) {}
            for (i : RolesInfo.Survivors):
                YouAreSurvivorShop.Hide(i)
            if (pl := RolesInfo.Monster?):
                if (set RolesInfo.PlayersParams[pl].MonsterScream = player_param_info{Value := 0}) {}
                YouAreMonsterShop.Hide(pl)
        else:
            Print("Version 2 of the AbilitiesHandler is started!")
            # var AllUis : []ShopUI_V2 = array{}
            if (set AllUIs[GetSession()] = array{}) {}
            var AllSurvivors : []?player = array:
                option{RolesInfo.Survivors[2]}
                option{RolesInfo.Survivors[0]}
                option{RolesInfo.Survivors[1]}
                option{RolesInfo.Survivors[3]}
            for (i := 0..RolesInfo.Survivors.Length, Player := RolesInfo.Survivors[i], Ref := SurvivorsReferences[i], Player <> RolesInfo.Monster?, PlUI := GetPlayerUI[Player], IntroSeq := SurvivorShopIntros[i]):
                ui := ShopUI_V2:
                    Player := Player
                    PlUI := PlUI
                    Abilities := SurvivorsAbilities
                    IntroSeq := IntroSeq
                    RolePart := UI.Parts.WB_Shop_YouAreSurvivor{}
                    PlayersToShow := AllSurvivors
                    MainClass := Self
                if (set AllUIs[GetSession()] += array{ui}) {}
                Ref.Register(Player)
                spawn{ui.Start()}
                if (chr := Player.GetFortCharacter[]):
                    chr.Hide()
            if (Monika := RolesInfo.Monster?, PlUI := GetPlayerUI[Monika]):
                ui := ShopUI_V2:
                    Player := Monika
                    PlUI := PlUI
                    Abilities := MonsterAbilities
                    IntroSeq := MonsterShopIntro
                    RolePart := UI.Parts.WB_Shop_YouAreMonster{}
                    PlayersToShow := array{RolesInfo.Monster}
                    MainClass := Self
                if (set AllUIs[GetSession()] += array{ui}) {}
                spawn{ui.Start()}
                if (chr := Monika.GetFortCharacter[]):
                    chr.Hide()
            # WaitForChoices()
            race:
                Sleep(20.0)
                ProbablyAllAreReady()
            for (i : AllUIs[GetSession()]):
                Selection := i.Final()
                # if (Selection < SurvivorsAbilities.Length, Selected := SurvivorsAbilities[Selection], set PlayersToAbilities[i.Agent] = Selected) {}
                for (g : SurvivorsAbilities):
                    if (Selection = g.ID, set PlayersToAbilities[i.Player] = g):
                        g.Feedback.Submit(i.Player)
                for (g : MonsterAbilities):
                    if (Selection = g.ID, set PlayersToAbilities[i.Player] = g):
                        g.Feedback.Submit(i.Player)
                Print("Player has chosen {Selection}")
                case (Selection):
                    0 =>
                        if (set RolesInfo.PlayersParams[i.Player] = player_params{ImprovedFlashlightBattery := player_param_info{Value := 0}, PublicIcon := Icons.IconEnergy}) {}
                    1 =>
                        if (set RolesInfo.PlayersParams[i.Player] = player_params{GatesRemoteControl := player_param_info{Value := 0}, PublicIcon := Icons.IconLock}) {}
                        Survivor_Gates_Trigger.Register(i.Player)
                    2 =>
                        if (set RolesInfo.PlayersParams[i.Player] = player_params{OneTimeVentAccess := player_param_info{Value := 1}, PublicIcon := Icons.IconBracket}) {}
                        for (g : RolesInfo.VentsMarkers):
                            g.AddToShowList(i.Player)
                    3 =>
                        if (set RolesInfo.PlayersParams[i.Player] = player_params{OneTimeInstantRevive := player_param_info{Value := 1}, PublicIcon := Icons.IconReboot}) {}
                    4 =>
                        if (set RolesInfo.PlayersParams[i.Player] = player_params{OneTimeEmergencyTeleport := player_param_info{Value := 1}, PublicIcon := Icons.IconRift}) {}
                        Survivor_EmergencyTeleport_Trigger.Register(i.Player)
                    5 =>
                        if (set RolesInfo.PlayersParams[i.Player] = player_params{OneTimeMonsterTeleport := player_param_info{Value := 1}, PublicIcon := Icons.IconPlace}) {}
                        Survivor_MonsterTeleport_Trigger.Register(i.Player)
                    6 =>
                        if (set RolesInfo.PlayersParams[i.Player] = player_params{OneAndHalfDamage := player_param_info{Value := 0}, PublicIcon := Icons.IconAddition}) {}
                    7 =>
                        if (set RolesInfo.PlayersParams[i.Player] = player_params{MonsterJumps := player_param_info{Value := 3}, PublicIcon := Icons.IconUp}) {}
                        if (chr := i.Player.GetFortCharacter[]):
                            chr.JumpedEvent().Subscribe(JumpedEvent)
                        Monster_Jump_MutatorZone.Enable()
                        Monster_NoJump_MutatorZone.Disable()
                    8 =>
                        if (set RolesInfo.PlayersParams[i.Player] = player_params{GatesControl := player_param_info{Value := 0}, PublicIcon := Icons.IconDoor}) {}
                        if (Root := GetSimulationEntity[]):
                            for (Entity : Root.FindDescendantEntitiesWithComponent(SG_Gate), Info := Entity.GetComponent[SG_Gate]):
                                Info.SetPublic()
                    9 =>
                        if (set RolesInfo.PlayersParams[i.Player] = player_params{FakeTasks := player_param_info{Value := 3}, PublicIcon := Icons.IconBroken}) {}
                        Monster_FakeTask_Trigger.Register(i.Player)
                    10 =>
                        if (set RolesInfo.PlayersParams[i.Player] = player_params{TeleportToSurvivor := player_param_info{Value := 45, GrayedOut := true}, PublicIcon := Icons.IconBehind}) {}
                        # Monster_TeleportToSurvivor_Trigger.Register(i.Player)
                        spawn{WaitForTeleport(i.Player)}
                    11 =>
                        if (set RolesInfo.PlayersParams[i.Player] = player_params{OneTimeTwoHealthDamage := player_param_info{Value := 1}, PublicIcon := Icons.IconDouble}) {}
                    _ =>
                        if (set RolesInfo.PlayersParams[i.Player] = player_params{}) {}
                        Print("Ability not implemented", ?Color := NamedColors.Red)
                if (GetRandomInt(0, 729674) = 729674, set RolesInfo.PlayersParams[i.Player].GhostSave = player_param_info{Value := 1}) {}
            if (set RolesInfo.PlayersParams[RolesInfo.Monster?].MonsterScream = player_param_info{Value := 0}) {}
        StoreMusic.Stop()
        if (Mon := RolesInfo.Monster?):
            Monster_Scream_Trigger.Register(Mon)
        # Print("Starting")
        set RolesInfo.Start = true
        Survivor_Gates_Trigger.PressedEvent.Subscribe(ClosingGate)
        Survivor_EmergencyTeleport_Trigger.PressedEvent.Subscribe(EmergencyTeleport)
        Survivor_MonsterTeleport_Trigger.PressedEvent.Subscribe(MonsterTeleport)
        Monster_Scream_Trigger.PressedEvent.Subscribe(Screaming)
        Monster_FakeTask_Trigger.PressedEvent.Subscribe(FakeTask)
        Monster_TeleportToSurvivor_Trigger.PressedEvent.Subscribe(TeleportToSelectedPlace)
        for (i : RolesInfo.Survivors, chr := i.GetFortCharacter[]):
            chr.Show()
        for (i : SurvivorsReferences):
            i.Clear()
        loop:
            if (Agent := RolesInfo.Monster?):
                Monster_Scream_Audio.Register(Agent)
                break
            Sleep(0.1)
        Sleep(5.0)
        for (i : DailyRewarded):
            if (RolesInfo.Monster? <> i):
                Achievements.Award(i, 8)
        for (i : Prestiged):
            if (RolesInfo.Monster? <> i):
                Achievements.Award(i, 7)
        if (not player[RolesInfo.Monster?].IsActive[]):
            for (ag : RolesInfo.Survivors, ag.IsActive[]):
                Winning{
                    Reason := MessagesBank{}.MonsterLeft
                    MainClass := RolesInfo
                    Feedback := Life.Feedback_S_MonsterLeft
                }.SurvivorsWin(ag)
        if (Monika := RolesInfo.Monster?, Prestiged.Find[Monika] or DailyRewarded.Find[Monika]):
            Sleep(15.0)
            if (DailyRewarded.Find[Monika]):
                Achievements.Award(Monika, 8)
            if (Prestiged.Find[Monika]):
                Achievements.Award(Monika, 7)
        for (i : RolesInfo.Survivors):
            if (i.IsActive[]):
                return
        if (i := RolesInfo.Monster?, player[i].IsActive[]):
            Winning{
                Reason := MessagesBank{}.AllSurvivorsLeft
                MainClass := RolesInfo
                Feedback := Life.Feedback_M_AllSurvivorsLeft
            }.MonsterWin(i)
        # ToEnd.Cancel()

    # SyncMarkers()<suspends>:void=
    #     AllPls := RolesInfo.Survivors
    #     loop:
    #         for (i := 0..AllPls.Length-1, Pl := AllPls[i], PlChar := Pl.GetFortCharacter[], Mark := Monster_Scream_Markers[i], Pl.IsActive[], PlChar.IsActive[], PlChar.IsVulnerable[], Pl <> RolesInfo.Monster?):
    #             spawn{Mark.MoveTo(PlChar.GetTransform(), 0.1)}
    #         Sleep(0.1)

    StopTheShop(Agent: agent):void=
        # Check if no survivors left or monster left
        if (Agent = RolesInfo.Monster? and not player[RolesInfo.Monster?].IsActive[]):
            # Survivors win
            for (ag : RolesInfo.Survivors):
                if (ag.IsActive[]):
                    Winning{
                        Reason := MessagesBank{}.MonsterLeft
                        MainClass := RolesInfo
                        Feedback := Life.Feedback_S_MonsterLeft
                    }.SurvivorsWin(ag)
        for (i : RolesInfo.Survivors):
            if (i.IsActive[]):
                return
        if (i := RolesInfo.Monster?, player[i].IsActive[]):
            Winning{
                Reason := MessagesBank{}.AllSurvivorsLeft
                MainClass := RolesInfo
                Feedback := Life.Feedback_M_AllSurvivorsLeft
            }.MonsterWin(i)


    Screaming(Agent: agent):void=
        spawn{ScreamingSus(Agent)}
    
    ScreamingSus(Agent:agent)<suspends>:void=
        Monster_Scream_Audio.Play(Agent)
        Monster_Scream_Trigger.Disable()
        Monster_Placeholder.Enable()
        AllPls := RolesInfo.Survivors
        for (i := 0..AllPls.Length-1, Pl := AllPls[i], PlChar := Pl.GetFortCharacter[], Mark := Monster_Scream_Markers[i], Pl.IsActive[], PlChar.IsActive[], PlChar.IsVulnerable[], Pl <> RolesInfo.Monster?):
            Monster_Scream_Audio.Play(Pl)
            if:
                Mark.TeleportTo[PlChar.GetTransform()]
            # Sleep(0.01)
            Mark.Enable()
            # Monster_Scream_Notification.SetText(Msg("You\'ve been scanned! Run!"))
            Monster_Scream_Notification.SetText(MessagesBank{}.YouveBeenScanned)
            Monster_Scream_Notification.Hide(Pl)
            Monster_Scream_Notification.Show(Pl)
        # Sleep(Monster_Scream_Duration)
        if:
            PlStatHud := StatsUiCon[GetSession()][Agent]
            set RolesInfo.PlayersParams[player[Agent]].MonsterScream.Value = Monster_Scream_Cooldown
            set RolesInfo.PlayersParams[player[Agent]].MonsterScream.GrayedOut = true
        then:
            PlStatHud.FastUpdate()
            for (i := 1..Monster_Scream_Duration):
                Sleep(1.0)
                if (set RolesInfo.PlayersParams[player[Agent]].MonsterScream.Value -= 1) {}
                PlStatHud.FastUpdate()
            for (Mark : Monster_Scream_Markers):
                Mark.Disable()
            # Sleep(Monster_Scream_Cooldown - Monster_Scream_Duration)
            for (i := 1..Monster_Scream_Cooldown - Monster_Scream_Duration):
                Sleep(1.0)
                if (set RolesInfo.PlayersParams[player[Agent]].MonsterScream.Value -= 1) {}
                PlStatHud.FastUpdate()
            if (set RolesInfo.PlayersParams[player[Agent]].MonsterScream.GrayedOut = false) {}
            PlStatHud.FastUpdate()
        Monster_Placeholder.Disable()
        Monster_Scream_Trigger.Enable()
        # Monster_Scream_Notification.SetText(Msg("Scan is available!"))
        if (RolesInfo.DEBUG_InfoHudVersion = Versions.V2):
            Monster_Scream_Notification.SetText(MessagesBank{}.ScanAvailable)
            Monster_Scream_Notification.Hide(Agent)
            Monster_Scream_Notification.Show(Agent)

    ClosingGate(Agent: agent):void=
        if (Tasks.SceneGraphActive?):
            spawn:
                ClosingGateSusV2(Agent)
        else:
            Print("ClosingGate: Support for legacy gates is deprecated!", ?Color := NamedColors.Red)
            # spawn:
            #     ClosingGateSus(Agent)

    # ClosingGateSus(Agent: agent)<suspends>:void=
    #     # Open the nearest gate to the player
    #     var MinimumDistance : float = 1000000.0
    #     var NearestGate : Gate = Gate{}
    #     for (i : Gates.Gates):
    #         if (Dis := Distance(Agent.GetFortCharacter[].GetTransform().Translation, i.GateProp.GetTransform().Translation), Dis < MinimumDistance and i.IsClosed = false):
    #             set MinimumDistance = Dis
    #             set NearestGate = i
    #     NearestGate.Close(Agent)
    #     Survivor_Gates_Trigger.Unregister(Agent)
    #     if (PlStatHud := StatsUiCon[GetSession()][Agent]):
    #         if:
    #             set RolesInfo.PlayersParams[player[Agent]].GatesRemoteControl.Value = Survivor_Gates_Cooldown
    #             set RolesInfo.PlayersParams[player[Agent]].GatesRemoteControl.GrayedOut = true
    #         then:
    #             PlStatHud.FastUpdate()
    #         for (i := 1..Survivor_Gates_Cooldown):
    #             Sleep(1.0)
    #             if (set RolesInfo.PlayersParams[player[Agent]].GatesRemoteControl.Value -= 1) {}
    #             PlStatHud.FastUpdate()
    #         if (set RolesInfo.PlayersParams[player[Agent]].GatesRemoteControl.GrayedOut = false) {}
    #         PlStatHud.FastUpdate()
    #     if (RolesInfo.DEBUG_InfoHudVersion = Versions.V2):
    #         Monster_Scream_Notification.SetText(MessagesBank{}.GateControlAvailable)
    #         Monster_Scream_Notification.Hide(Agent)
    #         Monster_Scream_Notification.Show(Agent)
    #     Survivor_Gates_Trigger.Register(Agent)
    
    ClosingGateSusV2(Agent: agent)<suspends>:void=
        # Open the nearest gate to the player
        var MinimumDistance : float = 1000000.0
        var NearestGate : entity = entity{}
        if (Root := GetSimulationEntity[]):
            for (i : Root.FindDescendantEntitiesWithComponent(SG_Gate), GateInfo := i.GetComponent[SG_Gate], GateInfo.IsClosed = false):
                if (Dis := Distance(FromTransform(Agent.GetFortCharacter[].GetTransform()).Translation, i.GetGlobalTransform().Translation), Dis < MinimumDistance):
                    set MinimumDistance = Dis
                    set NearestGate = i
        if (GateInfo := NearestGate.GetComponent[SG_Gate]):
            GateInfo.Close(Agent)
            Survivor_Gates_Trigger.Unregister(Agent)
            if (PlStatHud := StatsUiCon[GetSession()][Agent]):
                if:
                    set RolesInfo.PlayersParams[player[Agent]].GatesRemoteControl.Value = Survivor_Gates_Cooldown
                    set RolesInfo.PlayersParams[player[Agent]].GatesRemoteControl.GrayedOut = true
                then:
                    PlStatHud.FastUpdate()
                for (i := 1..Survivor_Gates_Cooldown):
                    Sleep(1.0)
                    if (set RolesInfo.PlayersParams[player[Agent]].GatesRemoteControl.Value -= 1) {}
                    PlStatHud.FastUpdate()
                if (set RolesInfo.PlayersParams[player[Agent]].GatesRemoteControl.GrayedOut = false) {}
                PlStatHud.FastUpdate()
            if (RolesInfo.DEBUG_InfoHudVersion = Versions.V1):
                Monster_Scream_Notification.SetText(MessagesBank{}.GateControlAvailable)
                Monster_Scream_Notification.Hide(Agent)
                Monster_Scream_Notification.Show(Agent)
            Survivor_Gates_Trigger.Register(Agent)

    EmergencyTeleport(Agent: agent):void=
        if (Life.IsInBasement[Agent] = true or not Agent.GetFortCharacter[].IsVulnerable[]):
            Life.InBasementNotification.SetText(MessagesBank{}.CantTeleport)
            Life.InBasementNotification.Hide(Agent)
            Life.InBasementNotification.Show(Agent)
            return
        # if (TP := RandomLocations[GetRandomInt(0, RandomLocations.Length-1)]):
        #     TP.Teleport(Agent)
        var MaxDistance : float = 0.0
        var MaxTeleporter : teleporter_device = teleporter_device{}
        for (i : RandomLocations):
            if (NewDis := Distance(i.GetTransform().Translation, Agent.GetFortCharacter[].GetTransform().Translation), NewDis > MaxDistance):
                set MaxDistance = NewDis
                set MaxTeleporter = i
        MaxTeleporter.Teleport(Agent)
        Survivor_EmergencyTeleport_Trigger.Unregister(Agent)
        if (set RolesInfo.PlayersParams[player[Agent]].OneTimeEmergencyTeleport.Value -= 1) {}
        if (RolesInfo.PlayersParams[player[Agent]].OneTimeEmergencyTeleport.Value <= 0, set RolesInfo.PlayersParams[player[Agent]].OneTimeEmergencyTeleport.GrayedOut = true) {}
        if (PlStatHud := StatsUiCon[GetSession()][Agent]):
            PlStatHud.FastUpdate()

    MonsterTeleport(Agent: agent):void=
        if (RolesInfo.MonsterAFK = true or Life.IsInBasement[Agent] = true or not Agent.GetFortCharacter[].IsVulnerable[]):
            Monster_Scream_Notification.SetText(MessagesBank{}.CantTeleportMonster)
            Monster_Scream_Notification.Hide(Agent)
            Monster_Scream_Notification.Show(Agent)
            return
        if (Pl := RolesInfo.Monster?):
            if (TP := RandomLocations[GetRandomInt(0, RandomLocations.Length-1)]):
                TP.Teleport(Pl)
                Monster_Scream_Notification.SetText(MessagesBank{}.YouveBeenTeleported)
                Monster_Scream_Notification.Hide(Pl)
                Monster_Scream_Notification.Show(Pl)
        Survivor_MonsterTeleport_Trigger.Unregister(Agent)
        if (set RolesInfo.PlayersParams[player[Agent]].OneTimeMonsterTeleport.Value -= 1) {}
        if (RolesInfo.PlayersParams[player[Agent]].OneTimeMonsterTeleport.Value <= 0, set RolesInfo.PlayersParams[player[Agent]].OneTimeMonsterTeleport.GrayedOut = true) {}
        if (PlStatHud := StatsUiCon[GetSession()][Agent]):
            PlStatHud.FastUpdate()

    JumpedEvent(Agent: fort_character):void=
        set Jumped += 1
        if (set RolesInfo.PlayersParams[player[Agent.GetAgent[]]].MonsterJumps.Value -= 1) {}
        if (RolesInfo.PlayersParams[player[Agent.GetAgent[]]].MonsterJumps.Value <= 0, set RolesInfo.PlayersParams[player[Agent.GetAgent[]]].MonsterJumps.GrayedOut = true) {}
        Monster_Scream_Notification.SetText(
            block:
                if (Jumped = 1):
                    MessagesBank{}.TwoJumpsLeft
                else if (Jumped = 2):
                    MessagesBank{}.OneJumpLeft
                else if (Jumped = 3):
                    MessagesBank{}.NoJumpsLeft
                MessagesBank{}.Error
        )
        if (Pl := agent[Agent], RolesInfo.DEBUG_InfoHudVersion = Versions.V2):
            Monster_Scream_Notification.Hide(Pl)
            Monster_Scream_Notification.Show(Pl)
        if (Jumped >= 3):
            Monster_Jump_MutatorZone.Disable()
            Monster_NoJump_MutatorZone.Enable()
        if (PlStatHud := StatsUiCon[GetSession()][Agent.GetAgent[]]):
            PlStatHud.FastUpdate()

    FakeTask(Agent: agent):void=
        if (FakeTasksSet >= 3 or Vents.PlayerInVent[Agent]):
            Print("FakeTask: The monster is in the vent or all fake tasks are placed", ?Color := NamedColors.Red)
            Life.InBasementNotification.SetText(MessagesBank{}.CantPlaceTask)
            Life.InBasementNotification.Hide(Agent)
            Life.InBasementNotification.Show(Agent)
            return
        if (Tasks.SceneGraphActive?):
            Print("FakeTask: Scene graph is active")
            var ToPlace : (/Verse.org/SpatialMath:)vector3 = (/Verse.org/SpatialMath:)vector3{}
            if (info := Agent.GetFortCharacter[].GetTransform().Translation, set ToPlace = FromVector3(info)):
                set ToPlace.Up -= 77.0
            if (Root := GetSimulationEntity[]):
                for (Entity : Root.FindDescendantEntitiesWithComponent(SG_GeneralTask)):
                    if (Distance(Entity.GetGlobalTransform().Translation, ToPlace) < 1000.0):
                        Print("FakeTask: A task is too close", ?Color := NamedColors.Red)
                        Life.InBasementNotification.SetText(MessagesBank{}.CantPlaceTask)
                        Life.InBasementNotification.Hide(Agent)
                        Life.InBasementNotification.Show(Agent)
                        return
                for (Entity : Root.FindDescendantEntitiesWithComponent(SG_PortalTask)):
                    if (Distance(Entity.GetGlobalTransform().Translation, ToPlace) < 1000.0):
                        Print("FakeTask: A portal task is too close", ?Color := NamedColors.Red)
                        Life.InBasementNotification.SetText(MessagesBank{}.CantPlaceTask)
                        Life.InBasementNotification.Hide(Agent)
                        Life.InBasementNotification.Show(Agent)
                        return
                var AllPossible : []entity = array{} 
                if (Tasks.TransmittersObjective.GetValue() <> Tasks.TransmittersObjective.GetTarget()):
                    set AllPossible += array:
                        SceneGraphObjects.Tasks.EPD_Transmitter{}
                if (Tasks.PumpkinsObjective.GetValue() <> Tasks.PumpkinsObjective.GetTarget()):
                    set AllPossible += array:
                        SceneGraphObjects.Tasks.EPD_Pumpkin{}
                if (AllPossible.Length <> 0, NewTask := AllPossible[GetRandomInt(0, AllPossible.Length-1)]):
                    if (TaskInfo := NewTask.GetComponent[SG_GeneralTask]):
                        set TaskInfo.IsFake = true
                        set TaskInfo.AbilitiesManager = option{Self}
                        set TaskInfo.TaskManager = Tasks
                        var DefaultInfo : (/Verse.org/SpatialMath:)transform = (/Verse.org/SpatialMath:)transform:
                            Translation := ToPlace
                            Scale := (/Verse.org/SpatialMath:)vector3{Left := 1.0, Up := 1.0, Forward := 1.0}
                        set DefaultInfo.Translation = ToPlace
                        case (TaskInfo.TaskType):
                            task_type.Transmitter =>
                                if:
                                    NewMarker := Monster_FakeTask_FakeMarkers[FakeTasksV2.Length]
                                    MonsterMarker := Monster_FakeTask_TrackMarkers[FakeTasksV2.Length]
                                    set TaskInfo.Marker = NewMarker
                                    set TaskInfo.FailMarker = MonsterMarker
                                else:
                                    Print("Error retreiving markers", ?Color := NamedColors.Red)
                            task_type.Pumpkin =>
                                if:
                                    NewMarker := Monster_FakeTask_FakeMarkers_Pumpkin[FakeTasksV2.Length]
                                    MonsterMarker := Monster_FakeTask_TrackMarkers[FakeTasksV2.Length]
                                    set TaskInfo.Marker = NewMarker
                                    set TaskInfo.FailMarker = MonsterMarker
                                    # Turn ToFormRot 90 degrees counterclockwise
                                    RotBy : (/UnrealEngine.com/Temporary/SpatialMath:)rotation = (/UnrealEngine.com/Temporary/SpatialMath:)MakeRotationFromYawPitchRollDegrees(90.0, 0.0, 0.0)
                                    var ToFormRot : (/Verse.org/SpatialMath:)rotation = FromRotation(Agent.GetFortCharacter[].GetTransform().Rotation.RotateBy(RotBy))
                                    set DefaultInfo.Rotation = ToFormRot
                                else:
                                    Print("Error retreiving markers", ?Color := NamedColors.Red)
                            task_type.Portal =>
                                Print("FakeTask: Trying to place portal, reverting", ?Color := NamedColors.Red)
                                return
                        # NewTask.SetGlobalTransform(DefaultInfo)
                        # KFM := KeyframedMovement.keyframed_movement_component:
                        #     Entity := NewTask
                        # KFM.SetKeyframes(array{KeyframedMovement.keyframed_movement_delta{Transform := DefaultInfo, Duration := 0.01}}, KeyframedMovement.oneshot_keyframed_movement_playback_mode{})
                        # NewTask.AddComponents(array{KFM})
                        # Root.AddEntities(array{NewTask})
                        # KFM.Play()
                        TaskHolder: entity = entity{}
                        Root.AddEntities(array{TaskHolder})
                        TaskHolder.AddComponents(array:
                            transform_component:
                                Entity := TaskHolder
                                GlobalTransform := DefaultInfo
                        )
                        TaskHolder.AddEntities(array{NewTask})
                        set TaskInfo.Start = true
                        # Teleport the monster Monster_FakeTask_TeleportDepth units forward
                        if (chr := Agent.GetFortCharacter[], var Transform : (/UnrealEngine.com/Temporary/SpatialMath:)transform = chr.GetTransform()):
                            # Calculate forward vector depending on the rotation
                            # Forward := Transform.Rotation.GetLocalForward()
                            # PositionInFront := Transform.Translation + (Forward * Monster_FakeTask_TeleportDepth)
                            set Transform.Translation.Z += 75.0
                            if (chr.TeleportTo[Transform.Translation, Transform.Rotation]) {}
                            # if (chr.TeleportTo[PositionInFront, Transform.Rotation]) {}
                            else:
                                Print("FakeTask: Couldn\'t teleport the monster", ?Color := NamedColors.Red)
                    set FakeTasksV2 += array{NewTask}

                else:
                    Print("FakeTask: All regular tasks are done, ignoring", ?Color := NamedColors.Red)
                    Life.InBasementNotification.SetText(MessagesBank{}.CantPlaceTask)
                    Life.InBasementNotification.Hide(Agent)
                    Life.InBasementNotification.Show(Agent)
                    return
        else:
            var ToPlace : (/UnrealEngine.com/Temporary/SpatialMath:)vector3 = (/UnrealEngine.com/Temporary/SpatialMath:)vector3{}
            if (info := Agent.GetFortCharacter[].GetTransform().Translation, set ToPlace = info, set ToPlace.Z -= 75.0) {}
            for (i : FindCreativeObjectsWithTag(TransmitterTask{})):
                if (Distance(i.GetTransform().Translation, ToPlace) < 1000.0):
                    Life.InBasementNotification.SetText(MessagesBank{}.CantPlaceTask)
                    Life.InBasementNotification.Hide(Agent)
                    Life.InBasementNotification.Show(Agent)
                    return
            for (i : FindCreativeObjectsWithTag(PlacedTransmitter{})):
                if (Distance(i.GetTransform().Translation, ToPlace) < 1000.0):
                    Life.InBasementNotification.SetText(MessagesBank{}.CantPlaceTask)
                    Life.InBasementNotification.Hide(Agent)
                    Life.InBasementNotification.Show(Agent)
                    return
            for (i : FindCreativeObjectsWithTag(PumpkinTask{})):
                if (Distance(i.GetTransform().Translation, ToPlace) < 1000.0):
                    Life.InBasementNotification.SetText(MessagesBank{}.CantPlaceTask)
                    Life.InBasementNotification.Hide(Agent)
                    Life.InBasementNotification.Show(Agent)
                    return
            for (i : FindCreativeObjectsWithTag(PortalTask{})):
                if (Distance(i.GetTransform().Translation, ToPlace) < 1000.0):
                    Life.InBasementNotification.SetText(MessagesBank{}.CantPlaceTask)
                    Life.InBasementNotification.Hide(Agent)
                    Life.InBasementNotification.Show(Agent)
                    return
            for (i : FindCreativeObjectsWithTag(FakeTransmitter{})):
                if (Distance(i.GetTransform().Translation, ToPlace) < 1000.0):
                    Life.InBasementNotification.SetText(MessagesBank{}.CantPlaceTask)
                    Life.InBasementNotification.Hide(Agent)
                    Life.InBasementNotification.Show(Agent)
                    return

            RandomInt := GetRandomInt(0, 3)
            RandomAngleDegrees := 90.0 * RandomInt
            RandomAngleRadians := RandomAngleDegrees * (PiFloat / 180.0)

            ToSetUp := (/UnrealEngine.com/Temporary/SpatialMath:)transform{
                Translation := ToPlace
                Rotation := MakeRotation((/UnrealEngine.com/Temporary/SpatialMath:)vector3{X := 0.0, Y := 0.0, Z := 1.0}, RandomAngleRadians)
            }
            MarkerToPlace := (/UnrealEngine.com/Temporary/SpatialMath:)transform{
                Translation := (/UnrealEngine.com/Temporary/SpatialMath:)vector3{X := ToPlace.X, Y := ToPlace.Y, Z := ToPlace.Z - 175.0}
            }
            trans := SpawnProp(Monster_FakeTask_Prop, ToSetUp)
            if:
                prop := trans(0)?
                btn := Monster_FakeTask_Buttons[FakeTasksSet],
                markr := Monster_FakeTask_FakeMarkers[FakeTasksSet]
                track := Monster_FakeTask_TrackMarkers[FakeTasksSet]
                btn.TeleportTo[prop.GetTransform()]
                markr.TeleportTo[MarkerToPlace]
                track.TeleportTo[MarkerToPlace]
            then:
                btn.SetInteractionText(MessagesBank{}.SetUpTransmitter)
                btn.InteractedWithEvent.Subscribe(
                    FakeTaskCon{
                        Prop := prop
                        Button := btn
                        Marker := markr
                        TrackMarker := track
                        MainClass := Self
                    }.Connect
                )
                btn.Enable()
                markr.Enable()
            if (set FakeTasks += array{trans(0)?}) {}
        set FakeTasksSet += 1
        if (set RolesInfo.PlayersParams[player[Agent]].FakeTasks.Value -= 1) {}
        if (RolesInfo.PlayersParams[player[Agent]].FakeTasks.Value <= 0, set RolesInfo.PlayersParams[player[Agent]].FakeTasks.GrayedOut = true) {}
        if (FakeTasksSet >= 3):
            Monster_FakeTask_Trigger.Unregister(Agent)
        if (PlStatHud := StatsUiCon[GetSession()][Agent]):
            PlStatHud.FastUpdate()
    
    # TeleportToSurvivor(Agent:agent):void=
    #     var min_distance : float = 1000000.0
    #     var nearest : teleporter_device = teleporter_device{}
    #     for (i : RolesInfo.Survivors):
    #         if (dis := Distance(Agent.GetFortCharacter[].GetTransform().Translation, i.GetFortCharacter[].GetTransform().Translation), dis < min_distance):
    #             var min_distance2 : float = 1000000.0
    #             for (g : RandomLocations):
    #                 if (dis2 := Distance(i.GetFortCharacter[].GetTransform().Translation, g.GetTransform().Translation), dis2 < min_distance2):
    #                     set min_distance2 = dis2
    #                     set nearest = g
    #             set min_distance = dis
    #     nearest.Teleport(Agent)
    #     Monster_TeleportToSurvivor_Trigger.Unregister(Agent)
    #     if (set RolesInfo.PlayersParams[player[Agent]].TeleportToSurvivor.Value -= 1) {}
    #     if (RolesInfo.PlayersParams[player[Agent]].TeleportToSurvivor.Value <= 0, set RolesInfo.PlayersParams[player[Agent]].TeleportToSurvivor.GrayedOut = true) {}
    #     if (PlStatHud := StatsUiCon[GetSession()][Agent]):
    #         PlStatHud.FastUpdate()

    WaitForTeleport(Agent:player)<suspends>:void=
        for (i := 1..45):
            if:
                PlStatHud := StatsUiCon[GetSession()][Agent]
            then:
                PlStatHud.FastUpdate()
            Sleep(1.0)
            if (set RolesInfo.PlayersParams[Agent].TeleportToSurvivor.Value -= 1) {}
        if (set RolesInfo.PlayersParams[Agent].TeleportToSurvivor.GrayedOut = false) {}
        if:
            PlStatHud := StatsUiCon[GetSession()][Agent]
        then:
            PlStatHud.FastUpdate()
        Monster_TeleportToSurvivor_Trigger.Register(Agent)

    FakeV2Triggered(Agent:agent):void=
        if (Monika := RolesInfo.Monster?):
            Life.InBasementNotification.SetText(MessagesBank{}.FakeTaskFound)
            Life.InBasementNotification.Hide(Monika)
            Life.InBasementNotification.Show(Monika)
        Life.InBasementNotification.SetText(MessagesBank{}.FakeTaskTriggered)
        Life.InBasementNotification.Hide(Agent)
        Life.InBasementNotification.Show(Agent)

    TeleportToSelectedPlace(Agent:agent):void=
        if (PlUI := GetPlayerUI[player[Agent]]):
            TeleportMonsterUI{
                MainClass := Self
                Agent := Agent
                PlUI := PlUI
            }.Start()

    # WaitForChoices()<suspends>:void=
    #     loop:
    #         if (RolesInfo.Survivors.Length > 0):
    #             break
    #         Sleep(0.1)
