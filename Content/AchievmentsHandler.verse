using { /Fortnite.com/Devices }
using { /Verse.org/Simulation }
using { /Verse.org/Random }
using { /UnrealEngine.com/Temporary/Diagnostics }
using { /UnrealEngine.com/Temporary/SpatialMath }
using { /Verse.org/Colors }
using { /Verse.org/Assets }
using { /UnrealEngine.com/Temporary/UI }
using { /Fortnite.com/UI }

var PlayersActivations : weak_map(session, [player]int) = map{}
var PlayersCodePanels : weak_map(session, [player]CodePanel) = map{}

Mats := module:
    Backgrounds<public> := module:
        DEV<public> := module:
        YT<public> := module:
        Secret<public> := module:
        Limited<public> := module:

nameplate := struct:
    ID : string
    Name : string
    Material : material

AllNameplates := class:
    NameplatesInfo : [string]nameplate = map:
        "DEV_GameExpF" => nameplate:
            ID := "DEV_GameExpF"
            Name := "GameExpF"
            Material := Mats.Backgrounds.DEV.MI_PlayerNameBackground_DEV_GameExpF{}
        "DEV_KilluaZold1k" => nameplate:
            ID := "DEV_KilluaZold1k"
            Name := "Killua Zold1k"
            Material := Mats.Backgrounds.DEV.MI_PlayerNameBackground_DEV_KilluaZold1k{}
        "YT_Olilz" => nameplate:
            ID := "YT_Olilz"
            Name := "Olilz"
            Material := Mats.Backgrounds.YT.MI_PlayerNameBackground_YT_Olilz{}
        "Secret_Her" => nameplate:
            ID := "Secret_Her"
            Name := "Her"
            Material := Mats.Backgrounds.Secret.MI_PlayerNameBackground_Secret_Her{}
        "Limited_Picks2025" => nameplate:
            ID := "Limited_Picks2025"
            Name := "Epic\'s Picks 2025"
            Material := Mats.Backgrounds.Limited.MI_PlayerNameBackground_Limited_Picks{}

    DefaultNameplate : nameplate = nameplate:
        ID := ""
        Name := "Default"
        Material := Mats.Backgrounds.M_PlayerNameBackground{}

code_activation_info := class<final><persistable>:

    ActivationTime : float
    Code : string
    IsActivated : logic

CodePanel := class:
    Player : player
    PlUI : player_ui

    Achievements : AchievmentsHandler
    var Canvas<private> : canvas = canvas{}

    var CodeTextBox<private> : text_block = text_block:
        DefaultText := MessagesBank{}.Empty
        DefaultShadowOpacity := 1.0

    var ActivationInfo<private> : text_block = text_block:
        DefaultText := MessagesBank{}.AwaitingActivation
        DefaultTextColor := color{R := 0.3, G := 0.3, B := 0.3}

    var CloseButton<private> : button_quiet = button_quiet:
        DefaultText := MessagesBank{}.Close

    var NameplateButton<private> : button_quiet = button_quiet:
        DefaultText := MessagesBank{}.Nameplate

    var ClearButton<private> : button_quiet = button_quiet:
        DefaultText := MessagesBank{}.Clear

    var ActivateButton<private> : button_loud = button_loud:
        DefaultText := MessagesBank{}.Activate

    var FullCode<private> : string = ""

    Start():void=
        CloseButton.OnClick().Subscribe(Close)
        NameplateButton.OnClick().Subscribe(Nameplate)
        ActivateButton.OnClick().Subscribe(Activate)
        ClearButton.OnClick().Subscribe(Clear)

        var Keyboard : []stack_box_slot = array{}
        var Numpad : []stack_box_slot = array{}

        Keys := array{
            array{"Q", "W", "E", "R", "T", "Y", "U", "I", "O", "P"}
            array{"A", "S", "D", "F", "G", "H", "J", "K", "L"}
            array{"Z", "X", "C", "V", "B", "N", "M", "←"}
        }

        for (i : Keys):
            var ThisLine : []stack_box_slot = array{}
            for (j : i):
                NewButton := button_quiet:
                    DefaultText := MessagesBank{}.ToMsg(j)
                NewButton.OnClick().Subscribe(AddToText)
                set ThisLine += array:
                    stack_box_slot:
                        HorizontalAlignment := horizontal_alignment.Fill
                        VerticalAlignment := vertical_alignment.Fill
                        Widget := NewButton
            set Keyboard += array:
                stack_box_slot:
                    HorizontalAlignment := block:
                        if (Keyboard.Length < 2):
                            horizontal_alignment.Center
                        else:
                            horizontal_alignment.Left
                    VerticalAlignment := vertical_alignment.Center
                    Padding := block:
                        if (Keyboard.Length < 2):
                            margin{}
                        else:
                            margin:
                                Left := 90.0
                    Widget := stack_box:
                        Orientation := orientation.Horizontal
                        Slots := ThisLine

        Nums := array{
            array{"7", "8", "9"}
            array{"4", "5", "6"}
            array{"1", "2", "3"}
            array{"0"}
        }

        for (i : Nums):
            var ThisLine : []stack_box_slot = array{}
            for (j : i):
                NewButton := button_quiet:
                    DefaultText := MessagesBank{}.ToMsg(j)
                NewButton.OnClick().Subscribe(AddToText)
                set ThisLine += array:
                    stack_box_slot:
                        HorizontalAlignment := horizontal_alignment.Fill
                        VerticalAlignment := vertical_alignment.Fill
                        Distribution := option{1.0}
                        Widget := NewButton
            set Numpad += array:
                stack_box_slot:
                    HorizontalAlignment := horizontal_alignment.Fill
                    VerticalAlignment := vertical_alignment.Center
                    Widget := stack_box:
                        Orientation := orientation.Horizontal
                        Slots := ThisLine

        set Canvas = canvas:
            Slots := array:
                canvas_slot:
                    Anchors := anchors:
                        Minimum := vector2:
                            X := 0.5
                            Y := 0.5
                        Maximum := vector2:
                            X := 0.5
                            Y := 0.5
                    Offsets := margin:
                        Right := 1500.0
                        Bottom := 800.0
                    Alignment := vector2:
                        X := 0.5
                        Y := 0.5
                    SizeToContent := false
                    Widget := overlay:
                        Slots := array:
                            overlay_slot:
                                HorizontalAlignment := horizontal_alignment.Fill
                                VerticalAlignment := vertical_alignment.Fill
                                Widget := UI.Parts.WB_Shop_NickBackground{}
                            overlay_slot:
                                HorizontalAlignment := horizontal_alignment.Left
                                VerticalAlignment := vertical_alignment.Top
                                Padding := margin:
                                    Left := 30.0
                                    Top := 30.0
                                Widget := CloseButton
                            # overlay_slot:
                            #     HorizontalAlignment := horizontal_alignment.Right
                            #     VerticalAlignment := vertical_alignment.Top
                            #     Padding := margin:
                            #         Top := 30.0
                            #         Right := 30.0
                            #     Widget := NameplateButton
                            overlay_slot:
                                HorizontalAlignment := horizontal_alignment.Left
                                VerticalAlignment := vertical_alignment.Bottom
                                Padding := margin:
                                    Left := 30.0
                                    Bottom := 30.0
                                Widget := ClearButton
                            overlay_slot:
                                HorizontalAlignment := horizontal_alignment.Center
                                VerticalAlignment := vertical_alignment.Bottom
                                Padding := margin:
                                    Bottom := 40.0
                                Widget := ActivateButton
                            overlay_slot:
                                HorizontalAlignment := horizontal_alignment.Center
                                VerticalAlignment := vertical_alignment.Top
                                Padding := margin:
                                    Top := 250.0
                                Widget := stack_box:
                                    Orientation := orientation.Vertical
                                    Slots := array:
                                        stack_box_slot:
                                            HorizontalAlignment := horizontal_alignment.Center
                                            VerticalAlignment := vertical_alignment.Top
                                            Widget := overlay:
                                                Slots := array:
                                                    overlay_slot:
                                                        HorizontalAlignment := horizontal_alignment.Fill
                                                        VerticalAlignment := vertical_alignment.Fill
                                                        Widget := color_block:
                                                            DefaultColor := color{R := 0.3, G := 0.3, B := 0.3}
                                                            DefaultDesiredSize := vector2{X := 500.0, Y := 50.0}
                                                    overlay_slot:
                                                        HorizontalAlignment := horizontal_alignment.Center
                                                        VerticalAlignment := vertical_alignment.Center
                                                        Widget := CodeTextBox
                                        stack_box_slot:
                                            HorizontalAlignment := horizontal_alignment.Center
                                            VerticalAlignment := vertical_alignment.Center
                                            Widget := ActivationInfo
                            overlay_slot:
                                HorizontalAlignment := horizontal_alignment.Center
                                VerticalAlignment := vertical_alignment.Bottom
                                Padding := margin:
                                    Bottom := 100.0
                                Widget := stack_box:
                                    Orientation := orientation.Horizontal
                                    Slots := array:
                                        stack_box_slot:
                                            VerticalAlignment := vertical_alignment.Top
                                            Widget := stack_box:
                                                Orientation := orientation.Vertical
                                                Slots := Keyboard
                                        stack_box_slot:
                                            Padding := margin:
                                                Left := 30.0
                                            HorizontalAlignment := horizontal_alignment.Fill
                                            VerticalAlignment := vertical_alignment.Fill
                                            Widget := stack_box:
                                                Orientation := orientation.Vertical
                                                Slots := Numpad
        PlUI.AddWidget(Canvas, player_ui_slot{InputMode := ui_input_mode.All})
        Achievements.StartupAudio.Play(Player)

    AddToText(Info : widget_message):void=
        Achievements.TypingAudio.Play(Player)
        ActInfo := ActivationInfo.GetText()
        if (ActInfo <> Localize(MessagesBank{}.AwaitingActivation)):
            ActivationInfo.SetText(MessagesBank{}.AwaitingActivation)
            ActivationInfo.SetTextColor(color{R := 0.2, G := 0.2, B := 0.2})
        if (BTN := button_quiet[Info.Source]):
            NewLetter := BTN.GetText()
            if:
                NewLetter = "←"
                if (FullCode.Length > 0):
                    NewCode := FullCode.Slice[0, FullCode.Length-1]
                    set FullCode = NewCode
            else if (FullCode.Length < 20):
                set FullCode += NewLetter
            CodeTextBox.SetText(MessagesBank{}.ToMsg(FullCode))

    Close(Info : widget_message):void=
        PlUI.RemoveWidget(Canvas)

    Remove():void=
        PlUI.RemoveWidget(Canvas)

    Nameplate(Info : widget_message):void=
        PlUI.RemoveWidget(Canvas)
        NameplateUI{Player := Player, PlUI := PlUI, GoBackTo := ReturnHere, Achievements := Achievements}.Start()

    ReturnHere():void=
        PlUI.AddWidget(Canvas, player_ui_slot{InputMode := ui_input_mode.All})

    Activate(Info : widget_message):void=
        CodeRes := Achievements.ActivateCode(Player, FullCode)
        if (CodeRes = true):
            Achievements.SuccessAudio.Play(Player)
            ActivationInfo.SetText(MessagesBank{}.ActivatedSuccessfully)
            ActivationInfo.SetTextColor(color{G := 1.0})
        else:
            # TODO: Add a falure sound effect - Check
            Achievements.FailAudio.Play(Player)
            if (PlayersActivations[GetSession()][Player] > 3):
                ActivationInfo.SetText(MessagesBank{}.TooManyRequests)
            else:
                ActivationInfo.SetText(MessagesBank{}.ActivationFailed)
            ActivationInfo.SetTextColor(color{R := 1.0})

    Clear(Info : widget_message):void=
        set FullCode = ""
        CodeTextBox.SetText(MessagesBank{}.Empty)
        ActInfo := ActivationInfo.GetText()
        if (ActInfo <> Localize(MessagesBank{}.AwaitingActivation)):
            ActivationInfo.SetText(MessagesBank{}.AwaitingActivation)
            ActivationInfo.SetTextColor(color{R := 0.2, G := 0.2, B := 0.2})

PossibleAchievements := enum:
    Level
    Prestige
    XP
    Coins
    GamesPlayed
    CompletedQuests
    PlayersSaved
    PlayersKnocked
    PlayersInPortal
    PlayersEliminated
    Wins

player_profile := class<final><persistable>:
    Version: int = 2
    FirstLogin: float = -1.0
    Prestige: int = 0
    Level: int = 0
    XP: int = 0
    Coins: int = 0
    GamesPlayed: int = 0
    CompletedQuests: int = 0
    PlayersSaved: int = 0
    PlayersKnocked: int = 0
    PlayersInPortal: int = 0
    PlayersEliminated: int = 0
    Wins: int = 0
    BoughtUpgrades: []int = array{}
    OwnedNameplates: []string = array{}
    ActivatedCodes: [string]code_activation_info = map{}
    SelectedNameplate: string = ""
    LatestDailyRecieved: float = 0.0
    WasMonsterLastTime: logic = false

UIsCon := class<concrete>:
    @editable
    CompletedQuests: score_manager_device = score_manager_device{}
    @editable
    PlayersSaved: score_manager_device = score_manager_device{}
    @editable
    PlayersKnocked: score_manager_device = score_manager_device{}
    @editable
    PlayersInPortal: score_manager_device = score_manager_device{}
    @editable
    PlayersEliminated: score_manager_device = score_manager_device{}

MakePlayerProfile<constructor>(Src:player_profile)<transacts> := player_profile:
    Version:= Src.Version
    FirstLogin:= Src.FirstLogin
    Prestige:= Clamp(Src.Prestige, 0, 999)
    Level:= Src.Level
    XP:= Src.XP
    Coins:= Clamp(Src.Coins, 0, 999)
    GamesPlayed:= Src.GamesPlayed
    CompletedQuests:= Src.CompletedQuests
    PlayersSaved:= Src.PlayersSaved
    PlayersKnocked:= Src.PlayersKnocked
    PlayersInPortal:= Src.PlayersInPortal
    PlayersEliminated:= Src.PlayersEliminated
    Wins:= Src.Wins
    BoughtUpgrades:= Src.BoughtUpgrades
    OwnedNameplates:= Src.OwnedNameplates
    ActivatedCodes:= Src.ActivatedCodes
    SelectedNameplate:= Src.SelectedNameplate
    LatestDailyRecieved:= Src.LatestDailyRecieved
    WasMonsterLastTime:= Src.WasMonsterLastTime

var PlayersStats : weak_map(player, player_profile) = map{}

codes_analytics := class<concrete>:
    @editable
    DEV_GameExpF_Analytics : analytics_device = analytics_device{}
    @editable
    DEV_KilluaZold1k_Analytics : analytics_device = analytics_device{}
    @editable
    YT_Olilz_Analytics : analytics_device = analytics_device{}
    @editable
    Secret_Her_Analytics : analytics_device = analytics_device{}
    @editable
    Limited_Picks_Analytics : analytics_device = analytics_device{}

AchievmentsHandler := class(creative_device):

    @editable
    CodeLaptops : []button_device = array{}

    @editable
    CodesAnalytics : codes_analytics = codes_analytics{}

    @editable
    StartupAudio : audio_player_device = audio_player_device{}
    @editable
    TypingAudio : audio_player_device = audio_player_device{}
    @editable
    SuccessAudio : audio_player_device = audio_player_device{}
    @editable
    FailAudio : audio_player_device = audio_player_device{}

    @editable
    Achievements : []accolades_device = array{}

    @editable
    UIs : UIsCon = UIsCon{}

    @editable
    CheckDate : creative_prop = creative_prop{}
    
    OnBegin<override>()<suspends>:void=
        # GetPlayspace().PlayerAddedEvent().Subscribe(PlayerAdded)
        loop:
            if (WaitingForPlayersStatus[GetSession()] = false):
                break
            Sleep(0.1)
        if (set PlayersActivations[GetSession()] = map{}) {}
        if (set PlayersCodePanels[GetSession()] = map{}) {}
        for (i : CodeLaptops):
            i.InteractedWithEvent.Subscribe(ConnectToLaptop)
        Players := GetPlayspace().GetPlayers()
        for (Player : Players):
            PlayerAdded(Player)

    PlayerAdded(Player : player):void=
        if:
            Player.IsActive[]
            not PlayersStats[Player]
            set PlayersStats[Player] = player_profile:
                FirstLogin := GetSecondsSinceEpoch()
            # Print("New player profile created")
        else if (Player.IsActive[], PlayersStats[Player]):
            # Print("Player profile already exists")
        else:
            # Print("Player profile creation failed")
        if:
            Player.IsActive[]
            PS := PlayersStats[Player]
            case (PS.Version):
                0 =>
                    set PlayersStats[Player] = player_profile:
                        MakePlayerProfile<constructor>(PS)
                        Version := 2
                        Prestige := 1
                        WasMonsterLastTime := false
                    # Print("Player profile updated to version 1, prestige added")
                1 =>
                    set PlayersStats[Player] = player_profile:
                        MakePlayerProfile<constructor>(PS)
                        Version := 2
                        WasMonsterLastTime := false
                2 =>
                    set PlayersStats[Player] = player_profile:
                        MakePlayerProfile<constructor>(PS)
                _ =>
                    # Print("Unknown player profile version: {PS.Version}")
        # SaveAchievement(Player, PossibleAchievements.GamesPlayed, 1)
        # Check if it's not December 18th 2024
        # if:
        #     Day := Int[CheckDate.GetTransform().Scale.Z]
        #     Month := Int[CheckDate.GetTransform().Scale.Y]
        #     Year := Int[CheckDate.GetTransform().Scale.X]
        # then:
        #     if (Year > 2024 or (Month = 12 and Day >= 18)):
        #         # Print("It\'s past December 18th 2024")
        #     else if (GetAchievement(Player, PossibleAchievements.Prestige) = 0):
        #         # Print("It\'s before December 18th 2024 and player has no prestige")
        #         SetAchievement(Player, PossibleAchievements.Prestige, 1)
        #     else:
        #         # Print("It\'s before December 18th 2024 and player has prestige")

    ConnectToLaptop(Agent:agent):void=
        if:
            Player := player[Agent]
            PlUI := GetPlayerUI[Player]
        then:
            if (CP := PlayersCodePanels[GetSession()][Player]):
                CP.ReturnHere()
            else if:
                CP := CodePanel{Player := Player, PlUI := PlUI, Achievements := Self}
                set PlayersCodePanels[GetSession()][Player] = CP
            then:
                CP.Start()
            else:
                Print("Failed to show the code panel", ?Color := NamedColors.Red)
    
    Award(Agent : agent, ID : int):void=
        if (ToGrant := Achievements[ID]):
            ToGrant.Award(Agent)

    Award(Team : []agent, ID : int):void=
        if (ToGrant := Achievements[ID]):
            for (Agent : Team):
                ToGrant.Award(Agent)

    SaveAchievement(Agent : agent, Achievment : PossibleAchievements, Amount : int):void=
        if:
            Player := player[Agent]
            Player.IsActive[]
            Ach := PlayersStats[Player]
            case (Achievment):
                PossibleAchievements.Level =>
                    set PlayersStats[Player] = player_profile:
                        MakePlayerProfile<constructor>(Ach)
                        XP := Ach.Level + Amount
                PossibleAchievements.Prestige =>
                    set PlayersStats[Player] = player_profile:
                        MakePlayerProfile<constructor>(Ach)
                        Prestige := Ach.Prestige + Amount
                    if (Pl := PlayersStats[Player]):
                        # Print("Prestige saved: {Pl.Prestige} +{Amount}\nResult: {Pl.Prestige}")
                PossibleAchievements.XP =>
                    set PlayersStats[Player] = player_profile:
                        MakePlayerProfile<constructor>(Ach)
                        XP := Ach.XP + Amount
                PossibleAchievements.Coins =>
                    set PlayersStats[Player] = player_profile:
                        MakePlayerProfile<constructor>(Ach)
                        Coins := Ach.Coins + Amount
                PossibleAchievements.GamesPlayed =>
                    set PlayersStats[Player] = player_profile:
                        MakePlayerProfile<constructor>(Ach)
                        GamesPlayed := Ach.GamesPlayed + Amount
                PossibleAchievements.CompletedQuests =>
                    set PlayersStats[Player] = player_profile:
                        MakePlayerProfile<constructor>(Ach)
                        CompletedQuests := Ach.CompletedQuests + Amount
                PossibleAchievements.PlayersSaved =>
                    set PlayersStats[Player] = player_profile:
                        MakePlayerProfile<constructor>(Ach)
                        PlayersSaved := Ach.PlayersSaved + Amount
                PossibleAchievements.PlayersKnocked =>
                    set PlayersStats[Player] = player_profile:
                        MakePlayerProfile<constructor>(Ach)
                        PlayersKnocked := Ach.PlayersKnocked + Amount
                PossibleAchievements.PlayersInPortal =>
                    set PlayersStats[Player] = player_profile:
                        MakePlayerProfile<constructor>(Ach)
                        PlayersInPortal := Ach.PlayersInPortal + Amount
                PossibleAchievements.PlayersEliminated =>
                    set PlayersStats[Player] = player_profile:
                        MakePlayerProfile<constructor>(Ach)
                        PlayersEliminated := Ach.PlayersEliminated + Amount
                PossibleAchievements.Wins =>
                    set PlayersStats[Player] = player_profile:
                        MakePlayerProfile<constructor>(Ach)
                        Wins := Ach.Wins + Amount
            # Print("Achievement saved: +{Amount}")
        case (Achievment):
            PossibleAchievements.GamesPlayed =>
                SaveAchievement(Agent, PossibleAchievements.Coins, 5 * Amount)
            PossibleAchievements.CompletedQuests =>
                UIs.CompletedQuests.Activate(Agent)
                SaveAchievement(Agent, PossibleAchievements.Coins, UIs.CompletedQuests.GetScoreAward() * Amount)
            PossibleAchievements.PlayersSaved =>
                UIs.PlayersSaved.Activate(Agent)
                SaveAchievement(Agent, PossibleAchievements.Coins, UIs.PlayersSaved.GetScoreAward() * Amount)
            PossibleAchievements.PlayersKnocked =>
                UIs.PlayersKnocked.Activate(Agent)
                SaveAchievement(Agent, PossibleAchievements.Coins, UIs.PlayersKnocked.GetScoreAward() * Amount)
            PossibleAchievements.PlayersInPortal =>
                UIs.PlayersInPortal.Activate(Agent)
                SaveAchievement(Agent, PossibleAchievements.Coins, UIs.PlayersInPortal.GetScoreAward() * Amount)
            PossibleAchievements.PlayersEliminated =>
                UIs.PlayersEliminated.Activate(Agent)
                SaveAchievement(Agent, PossibleAchievements.Coins, UIs.PlayersEliminated.GetScoreAward() * Amount)
            _ =>

    RemoveAchievement(Agent : agent, Achievment : PossibleAchievements, Amount : int):void=
        if:
            Player := player[Agent]
            Player.IsActive[]
            Ach := PlayersStats[Player]
            case (Achievment):
                PossibleAchievements.Level =>
                    set PlayersStats[Player] = player_profile:
                        MakePlayerProfile<constructor>(Ach)
                        XP := Ach.Level - Amount
                PossibleAchievements.Prestige =>
                    set PlayersStats[Player] = player_profile:
                        MakePlayerProfile<constructor>(Ach)
                        Prestige := Ach.Prestige - Amount
                PossibleAchievements.XP =>
                    set PlayersStats[Player] = player_profile:
                        MakePlayerProfile<constructor>(Ach)
                        XP := Ach.XP - Amount
                PossibleAchievements.Coins =>
                    set PlayersStats[Player] = player_profile:
                        MakePlayerProfile<constructor>(Ach)
                        Coins := Ach.Coins - Amount
                PossibleAchievements.GamesPlayed =>
                    set PlayersStats[Player] = player_profile:
                        MakePlayerProfile<constructor>(Ach)
                        GamesPlayed := Ach.GamesPlayed - Amount
                PossibleAchievements.CompletedQuests =>
                    set PlayersStats[Player] = player_profile:
                        MakePlayerProfile<constructor>(Ach)
                        CompletedQuests := Ach.CompletedQuests - Amount
                PossibleAchievements.PlayersSaved =>
                    set PlayersStats[Player] = player_profile:
                        MakePlayerProfile<constructor>(Ach)
                        PlayersSaved := Ach.PlayersSaved - Amount
                PossibleAchievements.PlayersKnocked =>
                    set PlayersStats[Player] = player_profile:
                        MakePlayerProfile<constructor>(Ach)
                        PlayersKnocked := Ach.PlayersKnocked - Amount
                PossibleAchievements.PlayersInPortal =>
                    set PlayersStats[Player] = player_profile:
                        MakePlayerProfile<constructor>(Ach)
                        PlayersInPortal := Ach.PlayersInPortal - Amount
                PossibleAchievements.PlayersEliminated =>
                    set PlayersStats[Player] = player_profile:
                        MakePlayerProfile<constructor>(Ach)
                        PlayersEliminated := Ach.PlayersEliminated - Amount
                PossibleAchievements.Wins =>
                    set PlayersStats[Player] = player_profile:
                        MakePlayerProfile<constructor>(Ach)
                        Wins := Ach.Wins - Amount
            # Print("Achievement removed: -{Amount}")

    SetAchievement(Agent : agent, Achievment : PossibleAchievements, Amount : int):void=
        if:
            Player := player[Agent]
            Player.IsActive[]
            Ach := PlayersStats[Player]
            case (Achievment):
                PossibleAchievements.Level =>
                    set PlayersStats[Player] = player_profile:
                        MakePlayerProfile<constructor>(Ach)
                        XP := Amount
                PossibleAchievements.Prestige =>
                    set PlayersStats[Player] = player_profile:
                        MakePlayerProfile<constructor>(Ach)
                        Prestige := Amount
                PossibleAchievements.XP =>
                    set PlayersStats[Player] = player_profile:
                        MakePlayerProfile<constructor>(Ach)
                        XP := Amount
                PossibleAchievements.Coins =>
                    set PlayersStats[Player] = player_profile:
                        MakePlayerProfile<constructor>(Ach)
                        Coins := Amount
                PossibleAchievements.GamesPlayed =>
                    set PlayersStats[Player] = player_profile:
                        MakePlayerProfile<constructor>(Ach)
                        GamesPlayed := Amount
                PossibleAchievements.CompletedQuests =>
                    set PlayersStats[Player] = player_profile:
                        MakePlayerProfile<constructor>(Ach)
                        CompletedQuests := Amount
                PossibleAchievements.PlayersSaved =>
                    set PlayersStats[Player] = player_profile:
                        MakePlayerProfile<constructor>(Ach)
                        PlayersSaved := Amount
                PossibleAchievements.PlayersKnocked =>
                    set PlayersStats[Player] = player_profile:
                        MakePlayerProfile<constructor>(Ach)
                        PlayersKnocked := Amount
                PossibleAchievements.PlayersInPortal =>
                    set PlayersStats[Player] = player_profile:
                        MakePlayerProfile<constructor>(Ach)
                        PlayersInPortal := Amount
                PossibleAchievements.PlayersEliminated =>
                    set PlayersStats[Player] = player_profile:
                        MakePlayerProfile<constructor>(Ach)
                        PlayersEliminated := Amount
                PossibleAchievements.Wins =>
                    set PlayersStats[Player] = player_profile:
                        MakePlayerProfile<constructor>(Ach)
                        Wins := Amount
            # Print("Achievement set: {Amount}")

    SaveAchievement(Agents:[]player, Achievment:PossibleAchievements, Amount:int):void=
        for (Agent : Agents):
            SaveAchievement(Agent, Achievment, Amount)

    RemoveAchievement(Agents:[]player, Achievment:PossibleAchievements, Amount:int):void=
        for (Agent : Agents):
            RemoveAchievement(Agent, Achievment, Amount)

    SetAchievement(Agents:[]player, Achievment:PossibleAchievements, Amount:int):void=
        for (Agent : Agents):
            SetAchievement(Agent, Achievment, Amount)

    GetAchievement(Agent:agent, Achievment:PossibleAchievements)<transacts>:int=
        if:
            Player := player[Agent]
            Player.IsActive[]
            Ach := PlayersStats[Player]
        then:
            case (Achievment):
                PossibleAchievements.Level =>
                    return Ach.Level
                PossibleAchievements.Prestige =>
                    return Ach.Prestige
                PossibleAchievements.XP =>
                    return Ach.XP
                PossibleAchievements.Coins =>
                    return Ach.Coins
                PossibleAchievements.GamesPlayed =>
                    return Ach.GamesPlayed
                PossibleAchievements.CompletedQuests =>
                    return Ach.CompletedQuests
                PossibleAchievements.PlayersSaved =>
                    return Ach.PlayersSaved
                PossibleAchievements.PlayersKnocked =>
                    return Ach.PlayersKnocked
                PossibleAchievements.PlayersInPortal =>
                    return Ach.PlayersInPortal
                PossibleAchievements.PlayersEliminated =>
                    return Ach.PlayersEliminated
                PossibleAchievements.Wins =>
                    return Ach.Wins
        return 0

    GetBoughtItems(Agent:agent)<transacts>:[]int=
        if:
            Player := player[Agent]
            Player.IsActive[]
            Ach := PlayersStats[Player]
        then:
            return Ach.BoughtUpgrades
        return array{}

    BuyItem(Agent:agent, Item:int):void=
        if:
            Player := player[Agent]
            Player.IsActive[]
            Ach := PlayersStats[Player]
            set PlayersStats[Player] = player_profile:
                MakePlayerProfile<constructor>(Ach)
                BoughtUpgrades := Ach.BoughtUpgrades + array{Item}
            # Print("Item bought {Item}")

    ResetItems(Agent:agent):void=
        if:
            Player := player[Agent]
            Player.IsActive[]
            Ach := PlayersStats[Player]
            set PlayersStats[Player] = player_profile:
                MakePlayerProfile<constructor>(Ach)
                BoughtUpgrades := array{}
            # Print("Items reset")

    GetDailyReward(Agent:agent):int=
        # Check if the latest daily reward was received less than 24 hours ago
        CurrentTime := GetSecondsSinceEpoch()
        if:
            Player := player[Agent]
            Player.IsActive[]
            Ach := PlayersStats[Player]
            86400.0 - Abs(GetSecondsSinceEpoch() - Ach.LatestDailyRecieved) < 0.0
            set PlayersStats[Player] = player_profile:
                MakePlayerProfile<constructor>(Ach)
                LatestDailyRecieved := CurrentTime
                Coins := Ach.Coins + 30
        then:
            Print("You have received your daily reward", ?Color := NamedColors.Green)
            return Ach.Coins + 30
        else:
            Print("You have already received your daily reward today", ?Color := NamedColors.Red)
            return -1

    CheckExistingProfile(Agent:agent)<transacts><decides>:void=
        Player := player[Agent]
        if (Player.IsActive[]):
            PlayersStats[Player]

    # BlockMonster(Agent:agent):void=
    #     for (i : GetPlayspace().GetPlayers()):
    #         if:
    #             Player := player[Agent]
    #             Player.IsActive[]
    #             Ach := PlayersStats[Player]
    #         then:
    #             if (i <> Agent):
    #                 if:
    #                     set PlayersStats[Player] = player_profile:
    #                         MakePlayerProfile<constructor>(Ach)
    #                         WasMonsterLastTime := false
    #                 then:
    #                     Print(MessagesBank{}.ToMsg(i, " is allowed to be the monster next time"))
    #                 else:
    #                     Print(MessagesBank{}.ToMsg(i, " encountered an error with allowing to be the monster next time"))
    #             else if (i = Agent):
    #                 if:
    #                     set PlayersStats[Player] = player_profile:
    #                         MakePlayerProfile<constructor>(Ach)
    #                         WasMonsterLastTime := true
    #                 then:
    #                     Print(MessagesBank{}.ToMsg(i, " is not allowed to be the monster next time"))
    #                 else:
    #                     Print(MessagesBank{}.ToMsg(i, " encountered an error with not allowing to be the monster next time"))

    GetRandomMonster():?player=
        for (i : Shuffle(GetPlayspace().GetPlayers())):
            if:
                Player := player[i]
                Player.IsActive[]
            then:
                if (Player.IsActive[], PlayersStats[Player].WasMonsterLastTime = false):
                    # Print(MessagesBank{}.ToMsg(i, " was not a monster last time, choosing"))
                    # BlockMonster(i)
                    return option{i}
                else if (Player.IsActive[], PlayersStats[Player].WasMonsterLastTime = true):
                    # Print(MessagesBank{}.ToMsg(i, " was a monster last time"))
                else:
                    # Print(MessagesBank{}.ToMsg(i, " encountered an error with checking if was a monster last time"))
        if (Another := Shuffle(GetPlayspace().GetPlayers())[0]):
            # Print(MessagesBank{}.ToMsg(Another, " was chosen randomly"))
            return option{Another}
        return false

    GetNameplate<private>(ID : string)<transacts>:nameplate=
        if (NP := AllNameplates{}.NameplatesInfo[ID]):
            NP
        else:
            AllNameplates{}.DefaultNameplate

    GetPlayerNameplate(Player:player)<transacts>:nameplate=
        if (Player.IsActive[], Info := PlayersStats[Player]):
            GetNameplate(Info.SelectedNameplate)
        else:
            AllNameplates{}.DefaultNameplate

    GetPlayerNameplates(Player:player)<transacts>:[]nameplate=
        var NameplatesToReturn : []nameplate = array{AllNameplates{}.DefaultNameplate}
        if (Player.IsActive[], Info := PlayersStats[Player]):
            for (i : Info.OwnedNameplates, NP := AllNameplates{}.NameplatesInfo[i]):
                set NameplatesToReturn += array{NP}
        return NameplatesToReturn

    SetNewNameplate(Player:player, ID : string):void=
        if:
            Player.IsActive[]
            Ach := PlayersStats[Player]
            set PlayersStats[Player] = player_profile:
                MakePlayerProfile<constructor>(Ach)
                SelectedNameplate := ID

    ActivateCode(Player:player, Code:string):logic=
        if (PlayersActivations[GetSession()][Player] > 3):
            return false
        if:
            Player.IsActive[]
            Ach := PlayersStats[Player]
        then:
            if:
                not Ach.ActivatedCodes[Code].IsActivated?
                var NewAC : [string]code_activation_info = Ach.ActivatedCodes
                set NewAC[Code] = code_activation_info:
                    ActivationTime := GetSecondsSinceEpoch()
                    Code := Code
                    IsActivated := true
            then:
                case (Code):
                    "EUMIRCT9075I4V621GP2" =>
                    # "GAMEEXPF2021" =>
                        if:
                            set PlayersStats[Player] = player_profile:
                                MakePlayerProfile<constructor>(Ach)
                                OwnedNameplates := Ach.OwnedNameplates + array{"DEV_GameExpF"}
                                ActivatedCodes := NewAC
                        CodesAnalytics.DEV_GameExpF_Analytics.Submit(Player)
                    "UL0GLVBZP9WWB2AES3BO" =>
                    # "BATMANZOLD1K" =>
                        if:
                            set PlayersStats[Player] = player_profile:
                                MakePlayerProfile<constructor>(Ach)
                                OwnedNameplates := Ach.OwnedNameplates + array{"DEV_KilluaZold1k"}
                                ActivatedCodes := NewAC
                        CodesAnalytics.DEV_KilluaZold1k_Analytics.Submit(Player)
                    "XW5M3TBBRSLY3N92QXO7" =>
                    # "OLILZ" =>
                        if:
                            set PlayersStats[Player] = player_profile:
                                MakePlayerProfile<constructor>(Ach)
                                OwnedNameplates := Ach.OwnedNameplates + array{"YT_Olilz"}
                                ActivatedCodes := NewAC
                        CodesAnalytics.YT_Olilz_Analytics.Submit(Player)
                    "729674" =>
                        if:
                            set PlayersStats[Player] = player_profile:
                                MakePlayerProfile<constructor>(Ach)
                                OwnedNameplates := Ach.OwnedNameplates + array{"Secret_Her"}
                                ActivatedCodes := NewAC
                        CodesAnalytics.Secret_Her_Analytics.Submit(Player)
                    "PICKS2025" =>
                        if:
                            set PlayersStats[Player] = player_profile:
                                MakePlayerProfile<constructor>(Ach)
                                OwnedNameplates := Ach.OwnedNameplates + array{"Limited_Picks2025"}
                                ActivatedCodes := NewAC
                        CodesAnalytics.Limited_Picks_Analytics.Submit(Player)
                    _ =>
                        if:
                            not PlayersActivations[GetSession()][Player]
                            set PlayersActivations[GetSession()][Player] = 1
                        else if:
                            set PlayersActivations[GetSession()][Player] += 1
                        return false
                return true
            else:
                if (Ach.ActivatedCodes[Code].IsActivated?):
                    if:
                        not PlayersActivations[GetSession()][Player]
                        set PlayersActivations[GetSession()][Player] = 1
                    else if:
                        set PlayersActivations[GetSession()][Player] += 1
                return false
        else:
            return false