using { /Fortnite.com/Devices }
using { /Fortnite.com/Characters }
using { /Verse.org/Simulation }
using { /UnrealEngine.com/Temporary/Diagnostics }
using { /UnrealEngine.com/Temporary/SpatialMath }
using { /Verse.org/SceneGraph }
using { /Verse.org/SpatialMath }
using { /Fortnite.com/Devices/CreativeAnimation }
using { /Verse.org/SceneGraph/KeyframedMovement }

GateStates := enum:
    Interactable
    Open
    Lock
    LockOnProximity

SG_Gate<public> := class<final_super>(component):

    @editable
    var Switches : []switch_device = array{}
    @editable
    PublicSwitches : []switch_device = array{}

    @editable
    StartState : GateStates = GateStates.Interactable
    @editable
    EndingState : GateStates = GateStates.Interactable

    var IsFinale : logic = false
    var IsClosed : logic = true

    var IsInProcess : logic = false

    OnSimulate<override>()<suspends>:void=
        if:
            StartState <> GateStates.Lock
            set IsClosed = false
            GateInfo := Entity.GetComponent[transform_component]
            var GatePos : (/Verse.org/SpatialMath:)transform = GateInfo.GlobalTransform
            set GatePos.Translation.Up += 219.0
            Entity.SetGlobalTransform(GatePos)
        case (StartState):
            GateStates.Interactable =>
                for (sw : Switches):
                    sw.TurnedOnEvent.Subscribe(Close)
                for (sw : PublicSwitches):
                    sw.TurnedOnEvent.Subscribe(Close)
            GateStates.Open =>
                for (sw : Switches):
                    sw.Disable()
                for (sw : PublicSwitches):
                    sw.Disable()
            GateStates.Lock =>
                # TODO: Add agent reference
                for (sw : Switches):
                    # sw.TurnOn(Agent)
                    sw.Disable()
                for (sw : PublicSwitches):
                    # sw.TurnOn(Agent)
                    sw.Disable()
            GateStates.LockOnProximity =>

    Close(Agent:agent):void=
        spawn{CloseSus(Agent)}

    Open(Agent:agent):void=
        spawn{OpenSus(Agent)}
    
    CloseSus(Agent:agent)<suspends>:void=
        var TurnedTheOtherSwitch : logic = false
        set IsClosed = true
        for (sw : Switches):
            if (not sw.GetCurrentState[]):
                sw.TurnOn(Agent)
                set TurnedTheOtherSwitch = true
            sw.Disable()
        for (sw : PublicSwitches):
            if (not sw.GetCurrentState[]):
                sw.TurnOn(Agent)
                set TurnedTheOtherSwitch = true
            sw.Disable()
        if (TurnedTheOtherSwitch = false and Switches.Length >= 2):
            return
        # GatePos := GateProp.GetTransform()
        # CloseAudio.Play()
        # GateProp.MoveTo(vector3{X:=GatePos.Translation.X,Y:=GatePos.Translation.Y,Z:=GatePos.Translation.Z - 219.0}, GatePos.Rotation, 1.5)
        CloseGate(1.5)
        Sleep(5.0)
        # if (IsFinale = true and CloseAtFinale = true):
        if ((IsFinale = true and EndingState = GateStates.Lock) or IsClosed = false):
            return
        OpenSus(Agent)
    
    OpenSus(Agent:agent)<suspends>:void=
        if (IsClosed = false):
            return
        set IsClosed = false
        for (sw : Switches):
            sw.Enable()
            sw.TurnOff(Agent)
            sw.Disable()
        for (sw : PublicSwitches):
            sw.Enable()
            sw.TurnOff(Agent)
            sw.Disable()
        # GatePos := GateProp.GetTransform()
        # GateProp.MoveTo(vector3{X:=GatePos.Translation.X,Y:=GatePos.Translation.Y,Z:=GatePos.Translation.Z + 219.0}, GatePos.Rotation, 2.3)
        OpenGate(2.3)
        if (IsFinale = true and EndingState = GateStates.Open):
            return
        Sleep(10.0)
        for (sw : Switches):
            sw.Enable()

    Finale(Agent : agent):void=
        set IsFinale = true
        # if (CloseAtFinale = true):
        #     Close(Agent)
        case (EndingState):
            GateStates.Interactable =>
                if (IsClosed = true):
                    Open(Agent)
                else:
                    for (sw : Switches):
                        sw.Enable()
                        sw.TurnOff(Agent)
                    for (sw : PublicSwitches):
                        sw.Enable()
                        sw.TurnOff(Agent)
                if (StartState <> GateStates.Interactable):
                    for (sw : Switches):
                        sw.TurnedOnEvent.Subscribe(Close)
                    for (sw : PublicSwitches):
                        sw.TurnedOnEvent.Subscribe(Close)
            GateStates.Open =>
                Open(Agent)
            GateStates.Lock =>
                Close(Agent)
            GateStates.LockOnProximity =>

    SetPublic():void=
        if (Switches.Length <> PublicSwitches.Length):
            return
        # Change Switches and PublicSwitches places
        var ToPlacePublicToPositions : [](/UnrealEngine.com/Temporary/SpatialMath:)transform = array{}
        var ToPlaceRegularToPositions : [](/UnrealEngine.com/Temporary/SpatialMath:)transform = array{}
        for (i : Switches):
            set ToPlacePublicToPositions += array:
                i.GetTransform()
        for (i : PublicSwitches):
            set ToPlaceRegularToPositions += array:
                i.GetTransform()
        for (
            i := 0..Switches.Length - 1
            Switches[i].TeleportTo[ToPlaceRegularToPositions[i]]
            PublicSwitches[i].TeleportTo[ToPlacePublicToPositions[i]]
        ) {}

    CloseGate<private>(Time:float)<suspends>:void=
        loop:
            if (IsInProcess = true):
                return
            break
        set IsInProcess = true
        if:
            KFM : keyframed_movement_component = Entity.GetComponent[keyframed_movement_component]
            # PosInfo := Entity.GetComponent[transform_component].GlobalTransform.Translation
        then:
            for (i : Entity.FindDescendantEntitiesWithComponent(SG_Placed), SoundPlayer := i.GetComponent[sound_component]):
                SoundPlayer.Play()
            Keyframes := array:
                keyframed_movement_delta:
                    Transform := (/Verse.org/SpatialMath:)transform:
                        Translation := (/Verse.org/SpatialMath:)vector3:
                            Up := -219.0
                        Scale := (/Verse.org/SpatialMath:)vector3:
                            Left := 0.0
                            Forward := 0.0
                            Up := 0.0
                    Easing := ease_in_cubic_bezier_easing_function{}
                    Duration := Time
            KFM.SetKeyframes(Keyframes, oneshot_keyframed_movement_playback_mode{})
            KFM.Play()
        Sleep(Time)
        set IsInProcess = false

    OpenGate<private>(Time:float)<suspends>:void=
        loop:
            if (IsInProcess = true):
                return
            break
        set IsInProcess = true
        if:
            KFM : keyframed_movement_component = Entity.GetComponent[keyframed_movement_component]
            # PosInfo := Entity.GetComponent[transform_component].GlobalTransform.Translation
        then:
            for (i : Entity.FindDescendantEntitiesWithComponent(SG_NotPlaced), SoundPlayer := i.GetComponent[sound_component]):
                SoundPlayer.Play()
            Keyframes := array:
                keyframed_movement_delta:
                    Transform := (/Verse.org/SpatialMath:)transform:
                        Translation := (/Verse.org/SpatialMath:)vector3:
                            Up := 219.0
                        Scale := (/Verse.org/SpatialMath:)vector3:
                            Left := 0.0
                            Forward := 0.0
                            Up := 0.0
                    Easing := linear_easing_function{}
                    Duration := Time
            KFM.SetKeyframes(Keyframes,oneshot_keyframed_movement_playback_mode{})
            KFM.Play()
        Sleep(Time)
        set IsInProcess = false

Gate := class<concrete>:

    #! DEPRECATED

    @editable
    GateProp : creative_prop = creative_prop{}
    @editable
    Switches : []switch_device = array{}
    @editable
    PublicSwitches : []switch_device = array{}
    @editable
    OpenAudio : audio_player_device = audio_player_device{}
    @editable
    CloseAudio : audio_player_device = audio_player_device{}

    # @editable
    # CloseAtFinale : logic = false #! DEPRECATED
    @editable
    StartState : GateStates = GateStates.Interactable
    @editable
    EndingState : GateStates = GateStates.Interactable

    var IsFinale : logic = false
    var IsClosed : logic = false
    var IsClosedOnProximity : logic = false
    
    Start(Agent:agent):void=
        case (StartState):
            GateStates.Interactable =>
                for (sw : Switches):
                    sw.TurnedOnEvent.Subscribe(Close)
                for (sw : PublicSwitches):
                    sw.TurnedOnEvent.Subscribe(Close)
            GateStates.Open =>
                for (sw : Switches):
                    sw.Disable()
                for (sw : PublicSwitches):
                    sw.Disable()
            GateStates.Lock =>
                set IsClosed = true
                GatePos := GateProp.GetTransform()
                if (GateProp.TeleportTo[(/UnrealEngine.com/Temporary/SpatialMath:)vector3{X:=GatePos.Translation.X,Y:=GatePos.Translation.Y,Z:=GatePos.Translation.Z - 219.0}, GatePos.Rotation]) {}
                for (sw : Switches):
                    sw.TurnOn(Agent)
                    sw.Disable()
                for (sw : PublicSwitches):
                    sw.TurnOn(Agent)
                    sw.Disable()
            GateStates.LockOnProximity =>
                spawn{LockOnProximity(Agent)}

    Close(Agent:agent):void=
        spawn{CloseSus(Agent)}

    Open(Agent:agent):void=
        spawn{OpenSus(Agent)}

    LockOnProximity(Agent:agent)<suspends>:void=
        set IsClosedOnProximity = true
        loop:
            # if (Agent.GetDistanceTo(GateProp) < 1000.0
            if (Distance(Agent.GetFortCharacter[].GetTransform().Translation, GateProp.GetTransform().Translation) < 100.0):
                CloseSus(Agent)
                return
            if (IsClosedOnProximity = false):
                return
            Sleep(0.01)
    
    CloseSus(Agent:agent)<suspends>:void=
        var TurnedTheOtherSwitch : logic = false
        set IsClosed = true
        for (sw : Switches):
            if (not sw.GetCurrentState[]):
                sw.TurnOn(Agent)
                set TurnedTheOtherSwitch = true
            sw.Disable()
        for (sw : PublicSwitches):
            if (not sw.GetCurrentState[]):
                sw.TurnOn(Agent)
                set TurnedTheOtherSwitch = true
            sw.Disable()
        if (TurnedTheOtherSwitch = false and Switches.Length >= 2):
            return
        GatePos := GateProp.GetTransform()
        CloseAudio.Play()
        GateProp.MoveTo((/UnrealEngine.com/Temporary/SpatialMath:)vector3{X:=GatePos.Translation.X,Y:=GatePos.Translation.Y,Z:=GatePos.Translation.Z - 219.0}, GatePos.Rotation, 1.5)
        Sleep(5.0)
        # if (IsFinale = true and CloseAtFinale = true):
        if ((IsFinale = true and EndingState = GateStates.Lock) or IsClosed = false):
            return
        OpenSus(Agent)
    
    OpenSus(Agent:agent)<suspends>:void=
        if (IsClosed = false):
            return
        set IsClosed = false
        for (sw : Switches):
            sw.Enable()
            sw.TurnOff(Agent)
            sw.Disable()
        for (sw : PublicSwitches):
            sw.Enable()
            sw.TurnOff(Agent)
            sw.Disable()
        GatePos := GateProp.GetTransform()
        OpenAudio.Play()
        GateProp.MoveTo((/UnrealEngine.com/Temporary/SpatialMath:)vector3{X:=GatePos.Translation.X,Y:=GatePos.Translation.Y,Z:=GatePos.Translation.Z + 219.0}, GatePos.Rotation, 2.3)
        if (IsFinale = true and EndingState = GateStates.Open):
            return
        Sleep(10.0)
        for (sw : Switches):
            sw.Enable()

    Finale(Agent : agent):void=
        set IsFinale = true
        # if (CloseAtFinale = true):
        #     Close(Agent)
        set IsClosedOnProximity = false
        case (EndingState):
            GateStates.Interactable =>
                if (IsClosed = true):
                    Open(Agent)
                else:
                    for (sw : Switches):
                        sw.Enable()
                        sw.TurnOff(Agent)
                    for (sw : PublicSwitches):
                        sw.Enable()
                        sw.TurnOff(Agent)
                if (StartState <> GateStates.Interactable):
                    for (sw : Switches):
                        sw.TurnedOnEvent.Subscribe(Close)
                    for (sw : PublicSwitches):
                        sw.TurnedOnEvent.Subscribe(Close)
            GateStates.Open =>
                Open(Agent)
            GateStates.Lock =>
                Close(Agent)
            GateStates.LockOnProximity =>
                spawn{LockOnProximity(Agent)}

    SetPublic():void=
        for (i := 0..Switches.Length, ToSet := Switches[i].GetTransform(), NewSwitch := PublicSwitches[i], NewSwitch.TeleportTo[ToSet], Switches[i].TeleportTo[
            (/UnrealEngine.com/Temporary/SpatialMath:)vector3{X:=ToSet.Translation.X,Y:=ToSet.Translation.Y,Z:=ToSet.Translation.Z - 219.0}, ToSet.Rotation
        ]) {}



GatesHandler := class(creative_device):

    #! DEPRECATED

    @editable
    Gates : []Gate = array{}

    @editable
    RolesInfo : RolesHandler = RolesHandler{}
    
    OnBegin<override>()<suspends>:void=
        loop:
            Sleep(0.1)
            if (WaitingForPlayersStatus[GetSession()] = false and RolesInfo.Survivors[0]):
                break
        if (Agent := RolesInfo.Survivors[0]):
            # Print("Enabling gates")
            for (gate : Gates):
                gate.Start(Agent)
        else:
            # Print("Gates: No survivors found")

    Finale(Agent:agent):void=
        # Print("Finale triggered")
        for (gate : Gates):
            gate.Finale(Agent)