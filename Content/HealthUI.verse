using { /Fortnite.com/Characters }
using { /Fortnite.com/Devices }
using { /Fortnite.com/UI }
using { /UnrealEngine.com/Temporary/Diagnostics }
using { /UnrealEngine.com/Temporary/SpatialMath }
using { /UnrealEngine.com/Temporary/UI }
using { /Verse.org/Assets }
using { /Verse.org/Colors }
using { /Verse.org/Colors/NamedColors }
using { /Verse.org/Simulation }

var PublicHealthHud : weak_map(session, []PlayersStatusesUI) = map{}
var PublicNameplateUI : weak_map(session, [agent]NameplateUI) = map{}

(Map:[keytype]valuetype where keytype:subtype(comparable), valuetype:type).RemoveKey(ElementToRemove:keytype)<transacts>:[keytype]valuetype=
    var NewMap:[keytype]valuetype = map{}
    for (Key -> Value : Map, Key <> ElementToRemove):
        set NewMap = ConcatenateMaps(NewMap, map{Key => Value})
    return NewMap

PlayerStates := enum:
    Alive
    Knocked
    InPortal
    Dead

Versions := enum:
    V1
    V2

NameplateUiButton := struct:
    Button : button_quiet
    ID : string

NameplateUI := class:
    Player : player
    PlUI : player_ui
    GoBackTo : type{_():void}

    Achievements : AchievmentsHandler

    var Canvas<private> : canvas = canvas{}
    var NameplatesButtons<private> : []NameplateUiButton = array{}

    var BackButton<private> : button_quiet = button_quiet:
        DefaultText := MessagesBank{}.Back

    UpdateList<private>():void=
        var AllButtons : []stack_box_slot = array{}
        SelectedNameplate := Achievements.GetPlayerNameplate(Player).ID
        for (i : Achievements.GetPlayerNameplates(Player)):
            NewButton := NameplateUiButton:
                Button := button_quiet:
                    DefaultText := MessagesBank{}.ToMsg(i.Name)
                ID := i.ID
            NewButton.Button.OnClick().Subscribe(Select)
            if (i.ID = SelectedNameplate):
                NewButton.Button.SetEnabled(false)
            set NameplatesButtons += array{NewButton}
            set AllButtons += array:
                stack_box_slot:
                    HorizontalAlignment := horizontal_alignment.Fill
                    VerticalAlignment := vertical_alignment.Fill
                    Padding := margin:
                        Bottom := 25.0
                    Widget := overlay:
                        Slots := array:
                            overlay_slot:
                                HorizontalAlignment := horizontal_alignment.Fill
                                VerticalAlignment := vertical_alignment.Fill
                                Padding := margin:
                                    # Left := 5.0
                                    # Top := 1.0
                                    # Right := 5.0
                                    # Bottom := 1.0
                                Widget := material_block:
                                    DefaultImage := i.Material
                                    DefaultDesiredSize := vector2:
                                        X := 500.0
                                        Y := 50.0
                                    DefaultTint := color{R := 0.3, G := 0.3, B := 0.3}
                            overlay_slot:
                                HorizontalAlignment := horizontal_alignment.Fill
                                VerticalAlignment := vertical_alignment.Fill
                                Widget := NewButton.Button
        set Canvas = canvas:
            Slots := array:
                canvas_slot:
                    Anchors := anchors:
                        Minimum := vector2:
                            X := 0.5
                            Y := 0.5
                        Maximum := vector2:
                            X := 0.5
                            Y := 0.5
                    Offsets := margin:
                        Right := 1500.0
                        Bottom := 800.0
                    Alignment := vector2:
                        X := 0.5
                        Y := 0.5
                    SizeToContent := false
                    Widget := overlay:
                        Slots := array:
                            overlay_slot:
                                HorizontalAlignment := horizontal_alignment.Fill
                                VerticalAlignment := vertical_alignment.Fill
                                Widget := UI.Parts.WB_Shop_NickBackground{}
                            overlay_slot:
                                HorizontalAlignment := horizontal_alignment.Left
                                VerticalAlignment := vertical_alignment.Top
                                Padding := margin:
                                    Left := 30.0
                                    Top := 30.0
                                Widget := BackButton
                            overlay_slot:
                                HorizontalAlignment := horizontal_alignment.Center
                                VerticalAlignment := vertical_alignment.Center
                                Widget := stack_box:
                                    Orientation := orientation.Vertical
                                    Slots := AllButtons


    Start():void=
        if (set PublicNameplateUI[GetSession()][Player] = Self) {}
        BackButton.OnClick().Subscribe(GoBack)
        ReturnHere()

    ReturnHere():void=
        AllButtons := UpdateList()
        PlUI.AddWidget(Canvas, player_ui_slot{InputMode := ui_input_mode.All})

    Select(Info : widget_message):void=
        if (BTN := button_quiet[Info.Source]):
            for (i : NameplatesButtons):
                if (i.Button = BTN):
                    Achievements.SetNewNameplate(Player, i.ID)
                    i.Button.SetEnabled(false)
                else:
                    i.Button.SetEnabled(true)
        for (i : PublicHealthHud[GetSession()]):
            spawn{i.Update(?UpdateNameplates := true)}

    GoBack(Info : widget_message):void=
        PlUI.RemoveWidget(Canvas)
        GoBackTo()

    Remove():void=
        if (AllNPs := PublicNameplateUI[GetSession()], set PublicNameplateUI[GetSession()] = AllNPs.RemoveKey(Player)) {}
        PlUI.RemoveWidget(Canvas)

NameplateTools := module:
    UpdateStatusColor<public>(Color : color, NameplateMat : material):void=
        if (newMat := Mats.Backgrounds.M_PlayerNameBackground[NameplateMat]):
            set newMat.StatusColor = Color
        else if (newMat := Mats.Backgrounds.DEV.MI_PlayerNameBackground_DEV_GameExpF[NameplateMat]):
            set newMat.StatusColor = Color
        else if (newMat := Mats.Backgrounds.DEV.MI_PlayerNameBackground_DEV_KilluaZold1k[NameplateMat]):
            # set newMat.StatusColor = Color
        else if (newMat := Mats.Backgrounds.YT.MI_PlayerNameBackground_YT_Olilz[NameplateMat]):
            set newMat.StatusColor = Color
        else if (newMat := Mats.Backgrounds.Secret.MI_PlayerNameBackground_Secret_Her[NameplateMat]):
            set newMat.StatusColor = Color
        else if (newMat := Mats.Backgrounds.Limited.MI_PlayerNameBackground_Limited_Picks[NameplateMat]):
            set newMat.StatusColor = Color
        else:
            Print("Failed to update status color, nameplate type not found", ?Color := NamedColors.Red)

    GetNameplateColor<public>(NameplateMat : material):color=
        if (newMat := Mats.Backgrounds.M_PlayerNameBackground[NameplateMat]):
            return newMat.StatusColor
        else if (newMat := Mats.Backgrounds.DEV.MI_PlayerNameBackground_DEV_GameExpF[NameplateMat]):
            return newMat.StatusColor
        else if (newMat := Mats.Backgrounds.DEV.MI_PlayerNameBackground_DEV_KilluaZold1k[NameplateMat]):
            return newMat.StatusColor
        else if (newMat := Mats.Backgrounds.YT.MI_PlayerNameBackground_YT_Olilz[NameplateMat]):
            return newMat.StatusColor
        else if (newMat := Mats.Backgrounds.Secret.MI_PlayerNameBackground_Secret_Her[NameplateMat]):
            return newMat.StatusColor
        else if (newMat := Mats.Backgrounds.Limited.MI_PlayerNameBackground_Limited_Picks[NameplateMat]):
            return newMat.StatusColor
        else:
            Print("Failed to get status color, nameplate type not found", ?Color := NamedColors.Red)
            return color{}

    CheckIfNameplateWorks<public>(NameplateMat : material)<decides><transacts>:void=
        if:
            not Mats.Backgrounds.M_PlayerNameBackground[NameplateMat] and
            not Mats.Backgrounds.DEV.MI_PlayerNameBackground_DEV_GameExpF[NameplateMat] and
            not Mats.Backgrounds.DEV.MI_PlayerNameBackground_DEV_KilluaZold1k[NameplateMat] and
            not Mats.Backgrounds.YT.MI_PlayerNameBackground_YT_Olilz[NameplateMat] and
            not Mats.Backgrounds.Secret.MI_PlayerNameBackground_Secret_Her[NameplateMat] and
            not Mats.Backgrounds.Limited.MI_PlayerNameBackground_Limited_Picks[NameplateMat]
        then:
            false
        else:
            true

GetPrestigeColor(Prestige:int):color=
    case (Prestige):
        0 => return color{R := 0.75, G := 0.75, B := 0.75}  # Light Gray
        # Levels 1-10: Silver to Light Blue
        1 => return color{R := 0.8, G := 0.8, B := 0.85}
        2 => return color{R := 0.78, G := 0.8, B := 0.88}
        3 => return color{R := 0.76, G := 0.8, B := 0.9}
        4 => return color{R := 0.74, G := 0.8, B := 0.92}
        5 => return color{R := 0.72, G := 0.8, B := 0.94}
        6 => return color{R := 0.7, G := 0.8, B := 0.96}
        7 => return color{R := 0.68, G := 0.8, B := 0.98}
        8 => return color{R := 0.66, G := 0.8, B := 1.0}
        9 => return color{R := 0.64, G := 0.8, B := 1.0}
        10 => return color{R := 0.62, G := 0.8, B := 1.0}  # Sky Blue
        
        # Levels 11-20: Light Blue to Sky Blue
        11 => return color{R := 0.5, G := 0.75, B := 1.0}
        12 => return color{R := 0.45, G := 0.75, B := 1.0}
        13 => return color{R := 0.4, G := 0.7, B := 1.0}
        14 => return color{R := 0.35, G := 0.7, B := 1.0}
        15 => return color{R := 0.3, G := 0.65, B := 1.0}
        16 => return color{R := 0.25, G := 0.65, B := 1.0}
        17 => return color{R := 0.2, G := 0.6, B := 1.0}
        18 => return color{R := 0.15, G := 0.6, B := 1.0}
        19 => return color{R := 0.1, G := 0.55, B := 1.0}
        20 => return color{R := 0.05, G := 0.55, B := 1.0}  # Light Blue

        # Levels 21-30: Green to Lime Green
        21 => return color{R := 0.0, G := 1.0, B := 0.0}
        22 => return color{R := 0.1, G := 0.9, B := 0.0}
        23 => return color{R := 0.15, G := 0.85, B := 0.0}
        24 => return color{R := 0.2, G := 0.8, B := 0.0}
        25 => return color{R := 0.25, G := 0.75, B := 0.0}
        26 => return color{R := 0.3, G := 0.7, B := 0.0}
        27 => return color{R := 0.35, G := 0.65, B := 0.0}
        28 => return color{R := 0.4, G := 0.6, B := 0.0}
        29 => return color{R := 0.45, G := 0.55, B := 0.0}
        30 => return color{R := 0.5, G := 0.5, B := 0.0}  # Yellow-Green
        
        # Levels 31-40: Yellow-Green to Yellow
        31 => return color{R := 0.6, G := 0.6, B := 0.0}
        32 => return color{R := 0.65, G := 0.65, B := 0.0}
        33 => return color{R := 0.7, G := 0.7, B := 0.0}
        34 => return color{R := 0.75, G := 0.75, B := 0.0}
        35 => return color{R := 0.8, G := 0.8, B := 0.0}
        36 => return color{R := 0.85, G := 0.85, B := 0.0}
        37 => return color{R := 0.9, G := 0.9, B := 0.0}
        38 => return color{R := 0.95, G := 0.95, B := 0.0}
        39 => return color{R := 1.0, G := 1.0, B := 0.0}
        40 => return color{R := 1.0, G := 0.95, B := 0.2}  # Yellow

        # Levels 41-50: Orange to Deep Orange
        41 => return color{R := 1.0, G := 0.8, B := 0.0}
        42 => return color{R := 1.0, G := 0.7, B := 0.0}
        43 => return color{R := 1.0, G := 0.6, B := 0.0}
        44 => return color{R := 1.0, G := 0.5, B := 0.0}
        45 => return color{R := 1.0, G := 0.4, B := 0.0}
        46 => return color{R := 1.0, G := 0.3, B := 0.0}
        47 => return color{R := 1.0, G := 0.2, B := 0.0}
        48 => return color{R := 1.0, G := 0.1, B := 0.0}
        49 => return color{R := 1.0, G := 0.05, B := 0.0}
        50 => return color{R := 1.0, G := 0.0, B := 0.0}  # Red

        # Levels 51-60: Red to Deep Red
        51 => return color{R := 1.0, G := 0.0, B := 0.0}
        52 => return color{R := 0.9, G := 0.0, B := 0.1}
        53 => return color{R := 0.8, G := 0.0, B := 0.2}
        54 => return color{R := 0.7, G := 0.0, B := 0.3}
        55 => return color{R := 0.6, G := 0.0, B := 0.4}
        56 => return color{R := 0.5, G := 0.0, B := 0.5}
        57 => return color{R := 0.6, G := 0.0, B := 0.4}
        58 => return color{R := 0.7, G := 0.0, B := 0.3}
        59 => return color{R := 0.8, G := 0.0, B := 0.2}
        60 => return color{R := 0.9, G := 0.0, B := 0.1}

        # Levels 61-70: Purple to Violet
        61 => return color{R := 0.6, G := 0.0, B := 0.8}
        62 => return color{R := 0.7, G := 0.0, B := 0.85}
        63 => return color{R := 0.75, G := 0.1, B := 0.9}
        64 => return color{R := 0.7, G := 0.1, B := 0.85}
        65 => return color{R := 0.6, G := 0.1, B := 0.8}
        66 => return color{R := 0.5, G := 0.1, B := 0.75}
        67 => return color{R := 0.4, G := 0.1, B := 0.7}
        68 => return color{R := 0.3, G := 0.1, B := 0.65}
        69 => return color{R := 0.2, G := 0.1, B := 0.6}
        70 => return color{R := 0.1, G := 0.1, B := 0.55}

        # Levels 71-80: Dark Blue to Midnight Blue
        71 => return color{R := 0.1, G := 0.1, B := 0.5}
        72 => return color{R := 0.1, G := 0.15, B := 0.55}
        73 => return color{R := 0.1, G := 0.2, B := 0.6}
        74 => return color{R := 0.1, G := 0.25, B := 0.65}
        75 => return color{R := 0.1, G := 0.3, B := 0.7}
        76 => return color{R := 0.1, G := 0.35, B := 0.75}
        77 => return color{R := 0.1, G := 0.4, B := 0.8}
        78 => return color{R := 0.1, G := 0.45, B := 0.85}
        79 => return color{R := 0.1, G := 0.5, B := 0.9}
        80 => return color{R := 0.1, G := 0.55, B := 0.95}

        # Levels 81-90: Dark Gray to Black
        81 => return color{R := 0.3, G := 0.3, B := 0.3}
        82 => return color{R := 0.28, G := 0.28, B := 0.28}
        83 => return color{R := 0.26, G := 0.26, B := 0.26}
        84 => return color{R := 0.24, G := 0.24, B := 0.24}
        85 => return color{R := 0.22, G := 0.22, B := 0.22}
        86 => return color{R := 0.2, G := 0.2, B := 0.2}
        87 => return color{R := 0.18, G := 0.18, B := 0.18}
        88 => return color{R := 0.16, G := 0.16, B := 0.16}
        89 => return color{R := 0.14, G := 0.14, B := 0.14}
        90 => return color{R := 0.12, G := 0.12, B := 0.12}

        # Levels 91-99: Darker Gray to Very Dark Gray
        91 => return color{R := 0.1, G := 0.1, B := 0.1}
        92 => return color{R := 0.09, G := 0.09, B := 0.09}
        93 => return color{R := 0.08, G := 0.08, B := 0.08}
        94 => return color{R := 0.07, G := 0.07, B := 0.07}
        95 => return color{R := 0.06, G := 0.06, B := 0.06}
        96 => return color{R := 0.05, G := 0.05, B := 0.05}
        97 => return color{R := 0.04, G := 0.04, B := 0.04}
        98 => return color{R := 0.03, G := 0.03, B := 0.03}
        99 => return color{R := 0.02, G := 0.02, B := 0.02}  # Near Black

        _ => return color{R := 1.0, G := 0.84, B := 0.0}  # Gold

PlayersStatusesUI := class:
    # Player : player
    PlUI : player_ui
    MainClass : HealthUI
    # AliveMaterial : texture = UI.T_Green
    # KnockedMaterial : texture
    # InPortalMaterial : texture
    # DeadMaterial : texture
    # IsMonster : logic = false

    var Canvas<private> : canvas = canvas{}
    var PlayersStates<private> : [player]PlayerStates = map{}
    var Pl1 : ?player = false
    var Pl2 : ?player = false
    var Pl3 : ?player = false
    var Pl4 : ?player = false
    var Pl5 : ?player = false
    NamePl1<private> : text_block = text_block{DefaultTextColor := color{R := 1.0, G := 1.0, B := 1.0}}
    NamePl2<private> : text_block = text_block{DefaultTextColor := color{R := 1.0, G := 1.0, B := 1.0}}
    NamePl3<private> : text_block = text_block{DefaultTextColor := color{R := 1.0, G := 1.0, B := 1.0}}
    NamePl4<private> : text_block = text_block{DefaultTextColor := color{R := 1.0, G := 1.0, B := 1.0}}
    NamePl5<private> : text_block = text_block{DefaultTextColor := color{R := 1.0, G := 1.0, B := 1.0}}
    PresPl1<private> : text_block = text_block{DefaultTextColor := color{R := 1.0, G := 1.0, B := 1.0}}
    PresPl2<private> : text_block = text_block{DefaultTextColor := color{R := 1.0, G := 1.0, B := 1.0}}
    PresPl3<private> : text_block = text_block{DefaultTextColor := color{R := 1.0, G := 1.0, B := 1.0}}
    PresPl4<private> : text_block = text_block{DefaultTextColor := color{R := 1.0, G := 1.0, B := 1.0}}
    PresPl5<private> : text_block = text_block{DefaultTextColor := color{R := 1.0, G := 1.0, B := 1.0}}
    StatusPl1<private> : texture_block = texture_block{DefaultImage := UI.T_Green, DefaultDesiredSize := vector2{X := 50.0, Y := 50.0}}
    StatusPl2<private> : texture_block = texture_block{DefaultImage := UI.T_Green, DefaultDesiredSize := vector2{X := 50.0, Y := 50.0}}
    StatusPl3<private> : texture_block = texture_block{DefaultImage := UI.T_Green, DefaultDesiredSize := vector2{X := 50.0, Y := 50.0}}
    StatusPl4<private> : texture_block = texture_block{DefaultImage := UI.T_Green, DefaultDesiredSize := vector2{X := 50.0, Y := 50.0}}
    StatusPl5<private> : texture_block = texture_block{DefaultImage := UI.T_Green, DefaultDesiredSize := vector2{X := 50.0, Y := 50.0}}
    # BackgroundPl1<private> : texture_block = texture_block{DefaultImage := UI.PlayerLineBG, DefaultDesiredSize := vector2{X := 500.0, Y := 50.0}, DefaultTint := color{R := 0.0, G := 0.2, B := 0.0}}
    # BackgroundPl2<private> : texture_block = texture_block{DefaultImage := UI.PlayerLineBG, DefaultDesiredSize := vector2{X := 500.0, Y := 50.0}, DefaultTint := color{R := 0.0, G := 0.2, B := 0.0}}
    # BackgroundPl3<private> : texture_block = texture_block{DefaultImage := UI.PlayerLineBG, DefaultDesiredSize := vector2{X := 500.0, Y := 50.0}, DefaultTint := color{R := 0.0, G := 0.2, B := 0.0}}
    # BackgroundPl4<private> : texture_block = texture_block{DefaultImage := UI.PlayerLineBG, DefaultDesiredSize := vector2{X := 500.0, Y := 50.0}, DefaultTint := color{R := 0.0, G := 0.2, B := 0.0}}
    # BackgroundPl5<private> : texture_block = texture_block{DefaultImage := UI.PlayerLineBG, DefaultDesiredSize := vector2{X := 500.0, Y := 50.0}, DefaultTint := color{R := 0.0, G := 0.2, B := 0.0}}
    BackgroundPl1<private> : material_block = material_block{
        DefaultImage := AllNameplates{}.DefaultNameplate.Material
        DefaultDesiredSize := vector2{X := 500.0, Y := 50.0}
        # DefaultTint := color{R := 0.0, G := 0.2, B := 0.0}
    }
    BackgroundPl2<private> : material_block = material_block{
        DefaultImage := AllNameplates{}.DefaultNameplate.Material
        DefaultDesiredSize := vector2{X := 500.0, Y := 50.0}
        # DefaultTint := color{R := 0.0, G := 0.2, B := 0.0}
    }
    BackgroundPl3<private> : material_block = material_block{
        DefaultImage := AllNameplates{}.DefaultNameplate.Material
        DefaultDesiredSize := vector2{X := 500.0, Y := 50.0}
        # DefaultTint := color{R := 0.0, G := 0.2, B := 0.0}
    }
    BackgroundPl4<private> : material_block = material_block{
        DefaultImage := AllNameplates{}.DefaultNameplate.Material
        DefaultDesiredSize := vector2{X := 500.0, Y := 50.0}
        # DefaultTint := color{R := 0.0, G := 0.2, B := 0.0}
    }
    BackgroundPl5<private> : material_block = material_block{
        DefaultImage := AllNameplates{}.DefaultNameplate.Material
        DefaultDesiredSize := vector2{X := 500.0, Y := 50.0}
        # DefaultTint := color{R := 0.0, G := 0.2, B := 0.0}
    }
    AbilityPl1<private> : texture_block = texture_block{DefaultImage := Icons.IconNone, DefaultDesiredSize := vector2{X := 40.0, Y := 40.0}}
    AbilityPl2<private> : texture_block = texture_block{DefaultImage := Icons.IconNone, DefaultDesiredSize := vector2{X := 40.0, Y := 40.0}}
    AbilityPl3<private> : texture_block = texture_block{DefaultImage := Icons.IconNone, DefaultDesiredSize := vector2{X := 40.0, Y := 40.0}}
    AbilityPl4<private> : texture_block = texture_block{DefaultImage := Icons.IconNone, DefaultDesiredSize := vector2{X := 40.0, Y := 40.0}}
    AbilityPl5<private> : texture_block = texture_block{DefaultImage := Icons.IconNone, DefaultDesiredSize := vector2{X := 40.0, Y := 40.0}}
    TimerPl1<private> : text_block = text_block{DefaultTextColor := color{R := 1.0, G := 1.0, B := 1.0}}
    TimerPl2<private> : text_block = text_block{DefaultTextColor := color{R := 1.0, G := 1.0, B := 1.0}}
    TimerPl3<private> : text_block = text_block{DefaultTextColor := color{R := 1.0, G := 1.0, B := 1.0}}
    TimerPl4<private> : text_block = text_block{DefaultTextColor := color{R := 1.0, G := 1.0, B := 1.0}}
    TimerPl5<private> : text_block = text_block{DefaultTextColor := color{R := 1.0, G := 1.0, B := 1.0}}
    # var Names<private> : []text_block = array{}
    # var Statuses<private> : []texture_block = array{}

    Start() : void =
        # MainClass.GetPlayspace().PlayerAddedEvent().Subscribe(AddTheWidget)
        case (MainClass.PlayerListVersion):
            Versions.V1 =>
                StatusPl1.SetDesiredSize(vector2{X := 50.0, Y := 50.0})
                StatusPl2.SetDesiredSize(vector2{X := 50.0, Y := 50.0})
                StatusPl3.SetDesiredSize(vector2{X := 50.0, Y := 50.0})
                StatusPl4.SetDesiredSize(vector2{X := 50.0, Y := 50.0})
            Versions.V2 =>
                StatusPl1.SetDesiredSize(vector2{X := 60.0, Y := 60.0})
                StatusPl2.SetDesiredSize(vector2{X := 60.0, Y := 60.0})
                StatusPl3.SetDesiredSize(vector2{X := 60.0, Y := 60.0})
                StatusPl4.SetDesiredSize(vector2{X := 60.0, Y := 60.0})
                StatusPl5.SetDesiredSize(vector2{X := 60.0, Y := 60.0})
                # Print("Increased size of statuses")
        # var Falloff : int = 0
        for (ind := 0..MainClass.RolesInfo.Survivors.Length, pl := MainClass.RolesInfo.Survivors[ind]):
            # if (pl = Player and MainClass.PlayerListVersion = Versions.V1):
            #     set Falloff += 1
            if (AbilityIcon := MainClass.RolesInfo.PlayersParams[pl].PublicIcon):
                # ind := i - Falloff
                if (set PlayersStates[pl] = MainClass.DefineState(pl)) {}
                Prestige := MainClass.RolesInfo.Achievements.GetAchievement(pl, PossibleAchievements.Prestige)
                case (ind):
                    0 =>
                        NamePl1.SetText(MessagesBank{}.ToMsg(pl))
                        PresPl1.SetText(MessagesBank{}.ToMsg(Prestige))
                        PresPl1.SetTextColor(GetPrestigeColor(Prestige))
                        StatusPl1.SetImage(MainClass.AliveMaterial)
                        AbilityPl1.SetImage(AbilityIcon)
                        set Pl1 = option{pl}
                    1 =>
                        NamePl2.SetText(MessagesBank{}.ToMsg(pl))
                        PresPl2.SetText(MessagesBank{}.ToMsg(Prestige))
                        PresPl2.SetTextColor(GetPrestigeColor(Prestige))
                        StatusPl2.SetImage(MainClass.AliveMaterial)
                        AbilityPl2.SetImage(AbilityIcon)
                        set Pl2 = option{pl}
                    2 =>
                        NamePl3.SetText(MessagesBank{}.ToMsg(pl))
                        PresPl3.SetText(MessagesBank{}.ToMsg(Prestige))
                        PresPl3.SetTextColor(GetPrestigeColor(Prestige))
                        StatusPl3.SetImage(MainClass.AliveMaterial)
                        AbilityPl3.SetImage(AbilityIcon)
                        set Pl3 = option{pl}
                    3 =>
                        NamePl4.SetText(MessagesBank{}.ToMsg(pl))
                        PresPl4.SetText(MessagesBank{}.ToMsg(Prestige))
                        PresPl4.SetTextColor(GetPrestigeColor(Prestige))
                        StatusPl4.SetImage(MainClass.AliveMaterial)
                        AbilityPl4.SetImage(AbilityIcon)
                        set Pl4 = option{pl}
                    4 =>
                        NamePl5.SetText(MessagesBank{}.ToMsg(pl))
                        PresPl5.SetText(MessagesBank{}.ToMsg(Prestige))
                        PresPl5.SetTextColor(GetPrestigeColor(Prestige))
                        StatusPl5.SetImage(MainClass.AliveMaterial)
                        AbilityPl5.SetImage(AbilityIcon)
                        set Pl5 = option{pl}
                    _ =>

        set Canvas = canvas:
            Slots := array:
                canvas_slot:
                    Anchors := anchors{Minimum := vector2{X := 0.5, Y := 0.0}, Maximum := vector2{X := 0.5, Y := 0.0}}
                    Offsets := margin{Top := 0.0, Left := 0.0, Right := 0.0, Bottom := 0.0}
                    Alignment := vector2{X := 0.5, Y := 0.275}
                    SizeToContent := true
                    Widget := texture_block{
                        DefaultImage := UI.T_MapCode_Old
                        DefaultDesiredSize := vector2{X := 450.0, Y := 450.0}
                    }
                canvas_slot:
                    Anchors := anchors{Minimum := vector2{X := 0.0, Y := 0.96}, Maximum := vector2{X := 0.0, Y := 0.96}}
                    Offsets := margin{Left := 20.0, Top := 0.0, Right := 0.0, Bottom := 0.0}
                    Alignment := vector2{X := 0.0, Y := 1.0}
                    SizeToContent := true
                    Widget := stack_box:
                        Orientation := orientation.Vertical
                        Slots := GetPlayerList()
                canvas_slot:
                    Anchors := anchors{Minimum := vector2{X := 0.5, Y := 0.5}, Maximum := vector2{X := 0.5, Y := 0.5}}
                    Offsets := margin{Top := 0.0, Left := 0.0, Right := 0.0, Bottom := 0.0}
                    Alignment := vector2{X := 0.5, Y := 0.5}
                    SizeToContent := true
                    Widget := overlay:
                        Slots := array{
                            overlay_slot:
                                Widget := texture_block{
                                    DefaultImage := UI.RecticleBG
                                    DefaultDesiredSize := vector2{X := 85.0, Y := 85.0}
                                }
                                HorizontalAlignment := horizontal_alignment.Center
                                VerticalAlignment := vertical_alignment.Center
                            overlay_slot:
                                Widget := texture_block{
                                    DefaultImage := UI.RecticleFG
                                    DefaultDesiredSize := vector2{X := 128.0, Y := 128.0}
                                }
                                HorizontalAlignment := horizontal_alignment.Center
                                VerticalAlignment := vertical_alignment.Center
                        }
        # Print("PlayersStatusesUI got {GetPlayerList().Length} lines of players")
        # for (i : MainClass.GetPlayspace().GetPlayers(), PlUI := GetPlayerUI[i]):
        PlUI.AddWidget(Canvas)
        # Print("Added PlayersStatusesUI to a player")
        spawn{Update(?UpdateNameplates := true)}

    # AddTheWidget(Player:player):void=
        # if (PlUI := GetPlayerUI[Player]):
        #     PlUI.AddWidget(Canvas)
        #     Print("Added PlayersStatusesUI to the new player")

    GetPlayerList():[]stack_box_slot=
        # pres := MainClass.RolesInfo.Achievements.GetAchievement(Player, PossibleAchievements.Prestige)
        var PlayerList : []stack_box_slot = array{}
        # if (MainClass.PlayerListVersion = Versions.V1):
        #     set PlayerList += array{
        #         stack_box_slot{
        #             Distribution := false
        #             VerticalAlignment := vertical_alignment.Fill
        #             HorizontalAlignment := horizontal_alignment.Fill
        #             Widget := stack_box{
        #                 Orientation := orientation.Horizontal
        #                 Slots := array{
        #                     stack_box_slot{
        #                         Widget := text_block{
        #                             DefaultText := MessagesBank{}.YourPrestige
        #                             DefaultTextColor := color{R := 1.0, G := 1.0, B := 1.0}
        #                         }
        #                         Distribution := option{1.0}
        #                         HorizontalAlignment := horizontal_alignment.Right
        #                         VerticalAlignment := vertical_alignment.Center
        #                     }
        #                     stack_box_slot{
        #                         Widget := text_block{
        #                             DefaultText := MessagesBank{}.ToMsg(pres)
        #                             DefaultTextColor := GetPrestigeColor(pres)
        #                         }
        #                         Distribution := option{0.5}
        #                         HorizontalAlignment := horizontal_alignment.Left
        #                         VerticalAlignment := vertical_alignment.Center
        #                     }
        #                 }
        #             }
        #         }
        #     }
        var Surs : []player = array{}
        for (i : MainClass.RolesInfo.Survivors):
            set Surs += array{i}
        for (i := 0..Surs.Length, pl := Surs[i], pl <> MainClass.RolesInfo.Monster?):
            # case (MainClass.PlayerListVersion):
            #     Versions.V1 => set PlayerList += array{AddPlayer(i + 1, pl)}
            #     Versions.V2 => set PlayerList += array{AddPlayerV2(i + 1, pl)}
            set PlayerList += array{AddPlayerV2(i + 1, pl)}
        return PlayerList

    AddPlayer(PlayerId:int, Pl:player):stack_box_slot=
        return stack_box_slot:
            Distribution := option{1.0}
            HorizontalAlignment := horizontal_alignment.Fill
            VerticalAlignment := vertical_alignment.Fill
            Widget := stack_box:
                Orientation := orientation.Horizontal
                Slots := array:
                    stack_box_slot:
                        Widget := block:
                            case (PlayerId):
                                1 => StatusPl1
                                2 => StatusPl2
                                3 => StatusPl3
                                4 => StatusPl4
                                _ =>
                                    # Print("Player id is wrong!")
                                    texture_block:
                                        DefaultImage := UI.T_Gray
                                        DefaultDesiredSize := vector2{X := 50.0, Y := 50.0}
                        Distribution := false
                        Padding := margin{Top := 0.0, Left := 10.0, Right := 0.0, Bottom := 0.0}
                        HorizontalAlignment := horizontal_alignment.Left
                        VerticalAlignment := vertical_alignment.Center
                    stack_box_slot:
                        Widget := block:
                            case (PlayerId):
                                1 => PresPl1
                                2 => PresPl2
                                3 => PresPl3
                                4 => PresPl4
                                _ =>
                                    # Print("Player id is wrong!")
                                    text_block:
                                        DefaultText := MessagesBank{}.Error
                                        DefaultTextColor := Red
                        Distribution := option{0.2}
                        Padding := margin{Top := 0.0, Left := 5.0, Right := 0.0, Bottom := 0.0}
                        HorizontalAlignment := horizontal_alignment.Center
                        VerticalAlignment := vertical_alignment.Center
                    stack_box_slot:
                        Widget := block:
                            case (PlayerId):
                                1 => NamePl1
                                2 => NamePl2
                                3 => NamePl3
                                4 => NamePl4
                                _ =>
                                    # Print("Player id is wrong!")
                                    text_block:
                                        DefaultText := MessagesBank{}.ToMsg(Pl)
                                        DefaultTextColor := White
                        Distribution := option{0.9}
                        Padding := margin{Right := 10.0}
                        HorizontalAlignment := horizontal_alignment.Left
                        VerticalAlignment := vertical_alignment.Center

    AddPlayerV2(PlayerId:int, Pl:player):stack_box_slot=
        # Print(MessagesBank{}.ToMsg(Pl))
        return stack_box_slot:
            Widget := overlay:
                Slots := array:
                    overlay_slot:
                        Widget := block:
                            case (PlayerId):
                                1 => BackgroundPl1
                                2 => BackgroundPl2
                                3 => BackgroundPl3
                                4 => BackgroundPl4
                                5 => BackgroundPl5
                                _ =>
                                    # Print("Player id is wrong!")
                                    texture_block:
                                        DefaultImage := UI.PlayerLineBG
                                        DefaultDesiredSize := vector2{X := 500.0, Y := 50.0}
                                        DefaultTint := color{R := 0.04, G := 0.04, B := 0.04}
                        Padding := margin{Top := 0.0, Left := 34.0, Right := 0.0, Bottom := 0.0}
                        HorizontalAlignment := horizontal_alignment.Right
                        VerticalAlignment := vertical_alignment.Center
                    overlay_slot:
                        Widget := block:
                            case (PlayerId):
                                1 => AbilityPl1
                                2 => AbilityPl2
                                3 => AbilityPl3
                                4 => AbilityPl4
                                5 => AbilityPl5
                                _ =>
                                    # Print("Player id is wrong!")
                                    texture_block:
                                        DefaultImage := Icons.IconNone
                                        DefaultDesiredSize := vector2{X := 40.0, Y := 40.0}
                        Padding := margin{Top := 0.0, Left := 62.5, Right := 0.0, Bottom := 0.0}
                        HorizontalAlignment := horizontal_alignment.Left
                        VerticalAlignment := vertical_alignment.Center
                    overlay_slot:
                        Widget := block:
                            case (PlayerId):
                                1 => NamePl1
                                2 => NamePl2
                                3 => NamePl3
                                4 => NamePl4
                                5 => NamePl5
                                _ =>
                                    # Print("Player id is wrong!")
                                    text_block:
                                        DefaultText := MessagesBank{}.ToMsg(Pl)
                                        DefaultTextColor := White
                                        DefaultTextOpacity := 0.5
                        HorizontalAlignment := horizontal_alignment.Left
                        VerticalAlignment := vertical_alignment.Center
                        Padding := margin{Top := 0.0, Left := 110.0, Right := 100.0, Bottom := 0.0}
                    overlay_slot:
                        Widget := block:
                            case (PlayerId):
                                1 => PresPl1
                                2 => PresPl2
                                3 => PresPl3
                                4 => PresPl4
                                5 => PresPl5
                                _ =>
                                    # Print("Player id is wrong!")
                                    text_block:
                                        DefaultText := MessagesBank{}.Error
                                        DefaultTextColor := White
                        HorizontalAlignment := horizontal_alignment.Center
                        VerticalAlignment := vertical_alignment.Center
                        Padding := margin{Top := 0.0, Left := 210.0, Right := 0.0, Bottom := 0.0}
                    overlay_slot:
                        HorizontalAlignment := horizontal_alignment.Left
                        VerticalAlignment := vertical_alignment.Center
                        Widget := overlay:
                            Slots := array:
                                overlay_slot:
                                    HorizontalAlignment := horizontal_alignment.Center
                                    VerticalAlignment := vertical_alignment.Center
                                    Widget := block:
                                        case (PlayerId):
                                            1 => StatusPl1
                                            2 => StatusPl2
                                            3 => StatusPl3
                                            4 => StatusPl4
                                            5 => StatusPl5
                                            _ =>
                                                # Print("Player id is wrong!")
                                                texture_block:
                                                    DefaultImage := UI.T_Gray
                                                    DefaultDesiredSize := vector2{X := 60.0, Y := 60.0}
                                overlay_slot:
                                    HorizontalAlignment := horizontal_alignment.Center
                                    VerticalAlignment := vertical_alignment.Center
                                    Widget := block:
                                        case (PlayerId):
                                            1 => TimerPl1
                                            2 => TimerPl2
                                            3 => TimerPl3
                                            4 => TimerPl4
                                            5 => TimerPl5
                                            _ =>
                                                # Print("Player id is wrong!")
                                                text_block:
                                                    DefaultText := MessagesBank{}.Error
                                                    DefaultTextColor := White

    UpdatePlayers()<suspends>:void= #! DEPRECATED
        loop:
            for (i := 0..MainClass.RolesInfo.Survivors.Length, pl := MainClass.RolesInfo.Survivors[i], MainClass.PlayerListVersion = Versions.V2):
                var ToSet : texture = MainClass.DeadMaterial
                var BadgeColor : color = color{R := 0.04, G := 0.04, B := 0.04}
                case (MainClass.DefineState(pl)):
                    PlayerStates.Alive =>
                        set ToSet = MainClass.AliveMaterial
                        set BadgeColor = color{R := 0.0, G := 0.2, B := 0.0}
                    PlayerStates.Knocked =>
                        set ToSet = MainClass.KnockedMaterial
                        set BadgeColor = color{R := 0.2, G := 0.2, B := 0.0}
                    PlayerStates.InPortal =>
                        set ToSet = MainClass.InPortalMaterial
                        set BadgeColor = color{R := 0.2, G := 0.0, B := 0.0}
                    PlayerStates.Dead =>
                        if (pl = Pl1?):
                            NamePl1.SetTextOpacity(0.5)
                            AbilityPl1.SetTint(color{R := 0.3, G := 0.3, B := 0.3})
                        else if (pl = Pl2?):
                            NamePl2.SetTextOpacity(0.5)
                            AbilityPl2.SetTint(color{R := 0.3, G := 0.3, B := 0.3})
                        else if (pl = Pl3?):
                            NamePl3.SetTextOpacity(0.5)
                            AbilityPl3.SetTint(color{R := 0.3, G := 0.3, B := 0.3})
                        else if (pl = Pl4?):
                            NamePl4.SetTextOpacity(0.5)
                            AbilityPl4.SetTint(color{R := 0.3, G := 0.3, B := 0.3})
                        else if (pl = Pl5?):
                            NamePl5.SetTextOpacity(0.5)
                            AbilityPl5.SetTint(color{R := 0.3, G := 0.3, B := 0.3})
                # NewStat := MainClass.RolesInfo.Achievements.GetPlayerNameplate(pl)
                # set NewStat.StatusColor = BadgeColor
                if (pl = Pl1?):
                    StatusPl1.SetImage(ToSet)
                    BackgroundPl1.SetTint(BadgeColor)
                else if (pl = Pl2?):
                    StatusPl2.SetImage(ToSet)
                    BackgroundPl2.SetTint(BadgeColor)
                else if (pl = Pl3?):
                    StatusPl3.SetImage(ToSet)
                    BackgroundPl3.SetTint(BadgeColor)
                else if (pl = Pl4?):
                    StatusPl4.SetImage(ToSet)
                    BackgroundPl4.SetTint(BadgeColor)
                else if (pl = Pl5?):
                    StatusPl5.SetImage(ToSet)
                    BackgroundPl5.SetTint(BadgeColor)

                if (LeftUntil := MainClass.RolesInfo.PlayersParams[agent[pl]].LosingLivesInBasement.Value, LeftUntil > -1):
                    if (pl = Pl1?):
                        TimerPl1.SetText(MessagesBank{}.ToMsg(LeftUntil))
                    else if (pl = Pl2?):
                        TimerPl2.SetText(MessagesBank{}.ToMsg(LeftUntil))
                    else if (pl = Pl3?):
                        TimerPl3.SetText(MessagesBank{}.ToMsg(LeftUntil))
                    else if (pl = Pl4?):
                        TimerPl4.SetText(MessagesBank{}.ToMsg(LeftUntil))
                    else if (pl = Pl5?):
                        TimerPl5.SetText(MessagesBank{}.ToMsg(LeftUntil))
                else:
                    if (pl = Pl1?):
                        TimerPl1.SetText(MessagesBank{}.ToMsg(""))
                    else if (pl = Pl2?):
                        TimerPl2.SetText(MessagesBank{}.ToMsg(""))
                    else if (pl = Pl3?):
                        TimerPl3.SetText(MessagesBank{}.ToMsg(""))
                    else if (pl = Pl4?):
                        TimerPl4.SetText(MessagesBank{}.ToMsg(""))
                    else if (pl = Pl5?):
                        TimerPl5.SetText(MessagesBank{}.ToMsg(""))

            # if (MainClass.RolesInfo.GameEnded = true):
            #     Sleep(0.1)
            #     Remove()
            #     return
            Sleep(MainClass.TickRate)

    Update(?UpdateNameplates : logic = false)<suspends>:void=
        if (UpdateNameplates?):
            UpdatePlayersNameplates()
        for (i := 0..MainClass.RolesInfo.Survivors.Length, pl := MainClass.RolesInfo.Survivors[i], MainClass.PlayerListVersion = Versions.V2):
            var ToSet : texture = MainClass.DeadMaterial
            var BadgeColor : color = color{R := 0.04, G := 0.04, B := 0.04}
            case (MainClass.DefineState(pl)):
                PlayerStates.Alive =>
                    set ToSet = MainClass.AliveMaterial
                    set BadgeColor = color{R := 0.0, G := 0.2, B := 0.0}
                PlayerStates.Knocked =>
                    set ToSet = MainClass.KnockedMaterial
                    set BadgeColor = color{R := 0.2, G := 0.2, B := 0.0}
                PlayerStates.InPortal =>
                    set ToSet = MainClass.InPortalMaterial
                    set BadgeColor = color{R := 0.2, G := 0.0, B := 0.0}
                PlayerStates.Dead =>
                    if (pl = Pl1?):
                        NamePl1.SetTextOpacity(0.5)
                        AbilityPl1.SetTint(color{R := 0.3, G := 0.3, B := 0.3})
                    else if (pl = Pl2?):
                        NamePl2.SetTextOpacity(0.5)
                        AbilityPl2.SetTint(color{R := 0.3, G := 0.3, B := 0.3})
                    else if (pl = Pl3?):
                        NamePl3.SetTextOpacity(0.5)
                        AbilityPl3.SetTint(color{R := 0.3, G := 0.3, B := 0.3})
                    else if (pl = Pl4?):
                        NamePl4.SetTextOpacity(0.5)
                        AbilityPl4.SetTint(color{R := 0.3, G := 0.3, B := 0.3})
                    else if (pl = Pl5?):
                        NamePl5.SetTextOpacity(0.5)
                        AbilityPl5.SetTint(color{R := 0.3, G := 0.3, B := 0.3})
            if (pl = Pl1?):
                StatusPl1.SetImage(ToSet)
                # BackgroundPl1.SetTint(BadgeColor)
                NameplateTools.UpdateStatusColor(BadgeColor, BackgroundPl1.GetImage())
            else if (pl = Pl2?):
                StatusPl2.SetImage(ToSet)
                # BackgroundPl2.SetTint(BadgeColor)
                NameplateTools.UpdateStatusColor(BadgeColor, BackgroundPl2.GetImage())
            else if (pl = Pl3?):
                StatusPl3.SetImage(ToSet)
                # BackgroundPl3.SetTint(BadgeColor)
                NameplateTools.UpdateStatusColor(BadgeColor, BackgroundPl3.GetImage())
            else if (pl = Pl4?):
                StatusPl4.SetImage(ToSet)
                # BackgroundPl4.SetTint(BadgeColor)
                NameplateTools.UpdateStatusColor(BadgeColor, BackgroundPl4.GetImage())
            else if (pl = Pl5?):
                StatusPl5.SetImage(ToSet)
                # BackgroundPl5.SetTint(BadgeColor) 
                NameplateTools.UpdateStatusColor(BadgeColor, BackgroundPl5.GetImage())
            
    UpdateTimers():void=
        for (i := 0..MainClass.RolesInfo.Survivors.Length, pl := MainClass.RolesInfo.Survivors[i], MainClass.PlayerListVersion = Versions.V2):
            if (LeftUntil := MainClass.RolesInfo.PlayersParams[agent[pl]].LosingLivesInBasement.Value, LeftUntil > -1, MainClass.DefineState(pl) = PlayerStates.InPortal):
                if (pl = Pl1?):
                    TimerPl1.SetText(MessagesBank{}.ToMsg(LeftUntil))
                else if (pl = Pl2?):
                    TimerPl2.SetText(MessagesBank{}.ToMsg(LeftUntil))
                else if (pl = Pl3?):
                    TimerPl3.SetText(MessagesBank{}.ToMsg(LeftUntil))
                else if (pl = Pl4?):
                    TimerPl4.SetText(MessagesBank{}.ToMsg(LeftUntil))
                else if (pl = Pl5?):
                    TimerPl5.SetText(MessagesBank{}.ToMsg(LeftUntil))
            else:
                if (pl = Pl1?):
                    TimerPl1.SetText(MessagesBank{}.ToMsg(""))
                else if (pl = Pl2?):
                    TimerPl2.SetText(MessagesBank{}.ToMsg(""))
                else if (pl = Pl3?):
                    TimerPl3.SetText(MessagesBank{}.ToMsg(""))
                else if (pl = Pl4?):
                    TimerPl4.SetText(MessagesBank{}.ToMsg(""))
                else if (pl = Pl5?):
                    TimerPl5.SetText(MessagesBank{}.ToMsg(""))

    UpdatePlayersNameplates<private>()<suspends>:void=
        for (i := 0..MainClass.RolesInfo.Survivors.Length, pl := MainClass.RolesInfo.Survivors[i], MainClass.PlayerListVersion = Versions.V2):
            NewNameplate := MainClass.RolesInfo.Achievements.GetPlayerNameplate(pl).Material
            Sleep(0.0)
            loop:
                if (pl = Pl1?):
                    # CLR := NameplateTools.GetNameplateColor(BackgroundPl1.GetImage())
                    BackgroundPl1.SetImage(NewNameplate)
                    res := BackgroundPl1.GetImage()
                    if (NameplateTools.CheckIfNameplateWorks[res]):
                        break
                    # NameplateTools.UpdateStatusColor(CLR, BackgroundPl1.GetImage())
                else if (pl = Pl2?):
                    # CLR := NameplateTools.GetNameplateColor(BackgroundPl2.GetImage())
                    BackgroundPl2.SetImage(NewNameplate)
                    res := BackgroundPl2.GetImage()
                    if (NameplateTools.CheckIfNameplateWorks[res]):
                        break
                    # NameplateTools.UpdateStatusColor(CLR, BackgroundPl2.GetImage())
                else if (pl = Pl3?):
                    # CLR := NameplateTools.GetNameplateColor(BackgroundPl3.GetImage())
                    BackgroundPl3.SetImage(NewNameplate)
                    res := BackgroundPl3.GetImage()
                    if (NameplateTools.CheckIfNameplateWorks[res]):
                        break
                    # NameplateTools.UpdateStatusColor(CLR, BackgroundPl3.GetImage())
                else if (pl = Pl4?):
                    # CLR := NameplateTools.GetNameplateColor(BackgroundPl4.GetImage())
                    BackgroundPl4.SetImage(NewNameplate)
                    res := BackgroundPl4.GetImage()
                    if (NameplateTools.CheckIfNameplateWorks[res]):
                        break
                    # NameplateTools.UpdateStatusColor(CLR, BackgroundPl4.GetImage())
                else if (pl = Pl5?):
                    # CLR := NameplateTools.GetNameplateColor(BackgroundPl5.GetImage())
                    BackgroundPl5.SetImage(NewNameplate)
                    res := BackgroundPl5.GetImage()
                    if (NameplateTools.CheckIfNameplateWorks[res]):
                        break
                    # NameplateTools.UpdateStatusColor(CLR, BackgroundPl5.GetImage())

    # Remove() : void =
    #     PlUI.RemoveWidget(Canvas)
    #     # Print("Game ended")


HealthUI := class(creative_device):

    @editable
    DEBUG_ShowUpdateInfo : logic = false
    @editable
    DEBUG_FillPlayersBox : logic = false
    @editable
    DEBUG_Show : logic = true

    @editable
    TickRate : float = 0.5
    
    @editable
    VersionToUse : Versions = Versions.V1

    @editable
    PlayerListVersion : Versions = Versions.V1

    @editable
    RolesInfo : RolesHandler = RolesHandler{}
    @editable
    LivesInfo : LifeHandler = LifeHandler{}

    # @editable
    AliveMaterial : texture = UI.T_Green
    # @editable
    KnockedMaterial : texture = UI.T_Yellow
    # @editable
    InPortalMaterial : texture = UI.T_Red
    # @editable
    DeadMaterial : texture = UI.T_Gray
    

    var MaybeMyUIPerPlayer : [player]canvas = map{}
    

    OnBegin<override>()<suspends>:void=
        if (DEBUG_Show = false):
            return
        loop:
            if (RolesInfo.Start = true and RolesInfo.Survivors.Length > 0):
                break
            Sleep(0.1)
        # Sleep(5.0)
        # case (VersionToUse):
        #     Versions.V1 =>
        #         # Print("HealthUI: Using V1")
        #         if (RolesInfo.Survivors.Length >= 2):
        #             for (pl : RolesInfo.Survivors, PlUi := GetPlayerUI[pl]):
        #                 NewUI := CreateMyUI(pl)
        #                 PlUi.AddWidget(NewUI)
        #                 if (set MaybeMyUIPerPlayer[pl] = NewUI) {}
        #         if (pl := player[RolesInfo.Monster?], PlUi := GetPlayerUI[pl]):
        #             NewUI := CreateMyUI(pl)
        #             PlUi.AddWidget(NewUI)
        #             if (set MaybeMyUIPerPlayer[pl] = NewUI) {}
        #         spawn{UpdateHealthUI()}
        #     Versions.V2 =>
                # Print("HealthUI: Using V2")
                # for (pl : RolesInfo.Survivors, PlUi := GetPlayerUI[pl]):
                    # InfoHud{
                    #     Player := pl
                    #     PlUI := PlUi
                    #     MainClass := Self
                    # }.Start()
                    # if:
                    #     set PlayersHealthHuds[GetSession()] += array:
                    #         InfoHud:
                    #             Player := pl
                    #             PlUI := PlUi
                    #             MainClass := Self
                    #     NewUI := PlayersHealthHuds[GetSession()][]
                # if (pl := player[RolesInfo.Monster?], PlUi := GetPlayerUI[pl]):
                #     InfoHud{
                #         # Player := pl
                #         MainClass := Self
                #         # PlUI := PlUi
                #         # IsMonster := true
                #     }.Start()
        # if (set PublicHealthHud[GetSession()] += array{PlayersStatusesUI{MainClass := Self}}, PublicHud := PublicHealthHud[GetSession()]):
        #     PublicHud.Start()
        #     Print("Created Public Health UI")
        if (set PublicHealthHud[GetSession()] = array{}) {}
        if (set PublicNameplateUI[GetSession()] = map{}) {}
        GetPlayspace().PlayerAddedEvent().Subscribe(AddUIToPlayer)
        for (i : GetPlayspace().GetPlayers(), PlUI := GetPlayerUI[i]):
            if (set PublicHealthHud[GetSession()] += array{PlayersStatusesUI{PlUI := PlUI, MainClass := Self}}, PublicHud := PublicHealthHud[GetSession()][PublicHealthHud[GetSession()].Length-1]):
                PublicHud.Start()

    AddUIToPlayer(Player : player):void=
        if (PlUI := GetPlayerUI[Player], set PublicHealthHud[GetSession()] += array{PlayersStatusesUI{PlUI := PlUI, MainClass := Self}}, PublicHud := PublicHealthHud[GetSession()][PublicHealthHud[GetSession()].Length-1]):
            PublicHud.Start()

    DefineState(Agent : player)<transacts> : PlayerStates =
        if (not Agent.IsActive[] or not Agent.GetFortCharacter[].IsActive[] or LivesInfo.Lives[Agent] <= 0):
            return PlayerStates.Dead
        # for (pl -> state : LivesInfo.IsInBasement, pl = Agent, state = true):
        #     return PlayerStates.InPortal
        if (LivesInfo.IsInBasement[Agent] = true):
            return PlayerStates.InPortal
        if (not Agent.GetFortCharacter[].IsVulnerable[] and Agent.GetFortCharacter[].IsDownButNotOut[]):
            return PlayerStates.Knocked
        return PlayerStates.Alive

    GetPlayerList(Agent : agent, ?empty:logic=false):[]stack_box_slot=
        var Stats : string = ""
        Survivors := RolesInfo.Survivors
        var PlayerList : []stack_box_slot = array{}
        if (DEBUG_FillPlayersBox = true):
            set PlayerList = array{
                stack_box_slot{
                    Widget := stack_box{
                        Orientation := orientation.Horizontal
                        Slots := array{
                            stack_box_slot{
                                Widget := texture_block{
                                    DefaultImage := AliveMaterial
                                    DefaultDesiredSize := vector2{X := 100.0, Y := 100.0}
                                }
                                Distribution := option{0.3}
                                Padding := margin{Top := 0.0, Left := 20.0, Right := 0.0, Bottom := 0.0}
                                HorizontalAlignment := horizontal_alignment.Left
                                VerticalAlignment := vertical_alignment.Center
                            },
                            stack_box_slot{
                                Widget := text_block{
                                    DefaultText := MessagesBank{}.ToMsg("Player 1Player 1")
                                    DefaultTextColor := color{R := 1.0, G := 1.0, B := 1.0}
                                }
                                Distribution := option{0.9}
                                HorizontalAlignment := horizontal_alignment.Left
                                VerticalAlignment := vertical_alignment.Center
                            }
                        }
                    }
                    Distribution := option{1.0}
                },
                stack_box_slot{
                    Widget := stack_box{
                        Orientation := orientation.Horizontal
                        Slots := array{
                            stack_box_slot{
                                Widget := texture_block{
                                    DefaultImage := KnockedMaterial
                                    DefaultDesiredSize := vector2{X := 100.0, Y := 100.0}
                                }
                                Distribution := option{0.3}
                                Padding := margin{Top := 0.0, Left := 20.0, Right := 0.0, Bottom := 0.0}
                                HorizontalAlignment := horizontal_alignment.Left
                                VerticalAlignment := vertical_alignment.Center
                            },
                            stack_box_slot{
                                Widget := text_block{
                                    DefaultText := MessagesBank{}.ToMsg("Player 2Player 2")
                                    DefaultTextColor := color{R := 1.0, G := 1.0, B := 1.0}
                                }
                                Distribution := option{0.9}
                                HorizontalAlignment := horizontal_alignment.Left
                                VerticalAlignment := vertical_alignment.Center
                            }
                        }
                    }
                    Distribution := option{1.0}
                },
                stack_box_slot{
                    Widget := stack_box{
                        Orientation := orientation.Horizontal
                        Slots := array{
                            stack_box_slot{
                                Widget := texture_block{
                                    DefaultImage := InPortalMaterial
                                    DefaultDesiredSize := vector2{X := 100.0, Y := 100.0}
                                }
                                Distribution := option{0.3}
                                Padding := margin{Top := 0.0, Left := 20.0, Right := 0.0, Bottom := 0.0}
                                HorizontalAlignment := horizontal_alignment.Left
                                VerticalAlignment := vertical_alignment.Center
                            },
                            stack_box_slot{
                                Widget := text_block{
                                    DefaultText := MessagesBank{}.ToMsg("Player 3Player 3")
                                    DefaultTextColor := color{R := 1.0, G := 1.0, B := 1.0}
                                }
                                Distribution := option{0.9}
                                HorizontalAlignment := horizontal_alignment.Left
                                VerticalAlignment := vertical_alignment.Center
                            }
                        }
                    }
                    Distribution := option{1.0}
                },
                stack_box_slot{
                    Widget := stack_box{
                        Orientation := orientation.Horizontal
                        Slots := array{
                            stack_box_slot{
                                Widget := texture_block{
                                    DefaultImage := DeadMaterial
                                    DefaultDesiredSize := vector2{X := 100.0, Y := 100.0}
                                }
                                Distribution := option{0.3}
                                Padding := margin{Top := 0.0, Left := 20.0, Right := 0.0, Bottom := 0.0}
                                HorizontalAlignment := horizontal_alignment.Left
                                VerticalAlignment := vertical_alignment.Center
                            },
                            stack_box_slot{
                                Widget := text_block{
                                    DefaultText := MessagesBank{}.ToMsg("Player 4Player 4")
                                    DefaultTextColor := color{R := 0.5, G := 0.5, B := 0.5}
                                }
                                Distribution := option{0.9}
                                HorizontalAlignment := horizontal_alignment.Left
                                VerticalAlignment := vertical_alignment.Center
                            }
                        }
                    }
                    Distribution := option{1.0}
                }
            }
            return PlayerList
        for (pl : Survivors, pl <> player[Agent]):
            if (empty = false):
                var txtr : texture = DeadMaterial
                var name_color : color = color{R := 1.0, G := 1.0, B := 1.0}
                case (DefineState(pl)):
                    PlayerStates.Alive =>
                        set txtr = AliveMaterial
                    PlayerStates.Knocked =>
                        set txtr = KnockedMaterial
                    PlayerStates.InPortal =>
                        set txtr = InPortalMaterial
                    PlayerStates.Dead =>
                        set txtr = DeadMaterial
                        set name_color = color{R := 0.5, G := 0.5, B := 0.5}
                set PlayerList += array{
                    stack_box_slot{
                        Widget := stack_box{
                            Orientation := orientation.Horizontal
                            Slots := array{
                                stack_box_slot{
                                    Widget := texture_block{
                                        DefaultImage := txtr
                                        DefaultDesiredSize := vector2{X := 100.0, Y := 100.0}
                                    }
                                    Distribution := option{0.3}
                                    Padding := margin{Top := 0.0, Left := 20.0, Right := 0.0, Bottom := 0.0}
                                    HorizontalAlignment := horizontal_alignment.Left
                                    VerticalAlignment := vertical_alignment.Center
                                },
                                stack_box_slot{
                                    Widget := text_block{
                                        DefaultText := MessagesBank{}.ToMsg(pl)
                                        DefaultTextColor := name_color
                                    }
                                    Distribution := option{0.9}
                                    HorizontalAlignment := horizontal_alignment.Left
                                    VerticalAlignment := vertical_alignment.Center
                                }
                            }
                        }
                        Distribution := option{1.0}
                        HorizontalAlignment := horizontal_alignment.Fill
                        VerticalAlignment := vertical_alignment.Fill
                    }
                }
            else:
                set PlayerList += array{
                    stack_box_slot{
                        Widget := stack_box{
                            Orientation := orientation.Horizontal
                            Slots := array{
                                stack_box_slot{
                                    Widget := text_block{
                                        DefaultText := MessagesBank{}.ToMsg(pl)
                                    }
                                }
                            }
                        }
                        Distribution := option{1.0}
                    }
                }
        return PlayerList

    CreateMyUI(Agent : agent) : canvas =
        AllHealth : canvas = canvas:
            Slots := array:
                canvas_slot:
                    Anchors := anchors{Minimum := vector2{X := 0.725, Y := 0.04}, Maximum := vector2{X := 0.975, Y := 0.18}}
                    Offsets := margin{Top := 0.0, Left := 0.0, Right := 0.0, Bottom := -150.0*(GetPlayerList(Agent, ?empty:=true).Length - 1)}
                    Alignment := vector2{X := 1.0, Y := 0.0}
                    SizeToContent := false
                    Widget := overlay{
                        Slots := array{
                            overlay_slot{
                                Widget := color_block{
                                    DefaultColor := color{R := 0.0, G := 0.0, B := 0.0}
                                    DefaultOpacity := 0.5
                                }
                                HorizontalAlignment := horizontal_alignment.Fill
                                VerticalAlignment := vertical_alignment.Fill
                            }
                            overlay_slot{
                                Widget := stack_box{
                                    Orientation := orientation.Vertical
                                    Slots := GetPlayerList(Agent)
                                }
                                HorizontalAlignment := horizontal_alignment.Fill
                                VerticalAlignment := vertical_alignment.Fill
                            }
                        }
                    }

        return AllHealth

    UpdateHealthUI()<suspends>:void= #! DEPRECATED
        loop:
            for (pl : RolesInfo.Survivors, PlUi := GetPlayerUI[pl], MyUI := MaybeMyUIPerPlayer[pl]):
                NewUI := CreateMyUI(pl)
                PlUi.RemoveWidget(MyUI)
                PlUi.AddWidget(NewUI)
                if (set MaybeMyUIPerPlayer[pl] = NewUI) {}
                if (DEBUG_ShowUpdateInfo = true):
                    case (DefineState(pl)):
                        PlayerStates.Alive =>
                            # Print(StringToMessage("Alive", pl))
                        PlayerStates.Knocked =>
                            # Print(StringToMessage("Knocked", pl))
                        PlayerStates.InPortal =>
                            # Print(StringToMessage("InPortal", pl))
                        PlayerStates.Dead =>
                            # Print(StringToMessage("Dead", pl))
                    # Print(StringToMessage("updated", pl))
            if (Pl := player[RolesInfo.Monster?], PlUi := GetPlayerUI[Pl], MyUI := MaybeMyUIPerPlayer[Pl]):
                NewUI := CreateMyUI(Pl)
                PlUi.RemoveWidget(MyUI)
                PlUi.AddWidget(NewUI)
                if (set MaybeMyUIPerPlayer[Pl] = NewUI) {}
                # Print(StringToMessage("updated", Pl))
            Sleep(TickRate)