using { /Fortnite.com }
using { /Fortnite.com/Characters }
using { /Fortnite.com/Devices }
using { /Fortnite.com/Game }
using { /UnrealEngine.com/Temporary/Diagnostics }
using { /Verse.org/Simulation }
using { /Verse.org/SceneGraph }
using { /Verse.org/Colors }

Basement := class<concrete>:
    @editable
    Volume : volume_device = volume_device{}
    @editable
    Teleport : teleporter_device = teleporter_device{}
    @editable
    SaveButton : button_device = button_device{}

    var BasementTeleport : teleporter_device = teleporter_device{}
    var MainClass : LifeHandler = LifeHandler{}

    var IsEnabled : logic = true

    var PickupStatus : [agent]logic = map{}
    var IsInZone : [agent]logic = map{}
    var IsKnocked : [agent]logic = map{}

    var AllPlayers : []agent = array{}

    Start()<suspends>:void=
        Volume.AgentEntersEvent.Subscribe(Caught)
        Volume.AgentExitsEvent.Subscribe(Uncaught)
        # SaveButton.InteractedWithEvent.Subscribe(Rescue)
        MainClass.KnockDetect.AgentDownedEvent.Subscribe(Downed)
        MainClass.KnockDetect.AgentRevivedEvent.Subscribe(Upped)
        MainClass.KnockDetect.AgentPickedUpEvent.Subscribe(Picked)
        MainClass.KnockDetect.AgentDroppedEvent.Subscribe(Dropped)
        var UsualActors : []creative_object_interface = array{}
        var StrongActors : []creative_object_interface = array{}
        for (i : MainClass.FindCreativeObjectsWithTag(PortalTask{}), btn := button_device[i]):
            if:
                btn.GetTransform().Translation.X = SaveButton.GetTransform().Translation.X
                btn.GetTransform().Translation.Y = SaveButton.GetTransform().Translation.Y
                btn.GetTransform().Translation.Z = SaveButton.GetTransform().Translation.Z
            then:
                btn.InteractedWithEvent.Subscribe(Rescue)
                # Print("Lifes: Found the save button")
            else:
                # Print("Lifes: Save button not found")
        for (i : AllPlayers):
            if (set PickupStatus[i] = false) {}
            if (set IsInZone[i] = false) {}
            if (set IsKnocked[i] = false) {}

    Rescue(Agent:agent):void=
        MainClass.Achievments.Award(Agent, 4)
        MainClass.Achievments.SaveAchievement(Agent, PossibleAchievements.PlayersSaved, 1)
        MainClass.PlayerSaved(Teleport)

    Caught(Agent:agent):void=
        if (set IsInZone[Agent] = true) {}
        if (IsEnabled = true and PickupStatus[Agent] = false):
            MainClass.PlayerEnteredBasement(Agent)
        else:
            # Print("Players is picked up or the basement is disabled or the player is not knocked")

    Downed(Agent:agent):void=
        if (set IsKnocked[Agent] = true) {}
        # Print("Player knocked")

    Upped(Agent:agent):void=
        if (set IsKnocked[Agent] = false) {}
        # Print("Player upped")

    Uncaught(Agent:agent):void=
        if (set IsInZone[Agent] = false) {}
        # Print("Player removed from the zone")

    Blocked():void=
        if (IsEnabled = true):
            set IsEnabled = false
        else:
            set IsEnabled = true

    Picked(Agent:agent):void=
        if (set PickupStatus[Agent] = true) {}
        # Print("Player picked up")

    Dropped(Agent:agent):void=
        if (set PickupStatus[Agent] = false) {}
        if (IsEnabled = true and IsInZone[Agent] = true):
            MainClass.PlayerEnteredBasement(Agent)
        else:
            # Print("Player not on the basement or the basement is disabled or the player is not knocked")


LifeHandler := class(creative_device):

    @editable
    DEBUG_AlwaysFullHp : logic = false

    @editable
    DEBUG_ContinueAfterFullWipeout : logic = false

    @editable
    DEBUG_ShowHud : logic = true

    @editable
    Feedback_S_MonsterLeft : analytics_device = analytics_device{}
    @editable
    Feedback_M_AllSurvivorsEliminated : analytics_device = analytics_device{}
    @editable
    Feedback_M_AllSurvivorsLeft : analytics_device = analytics_device{}
    @editable
    Feedback_M_LastSurvivorLeft : analytics_device = analytics_device{}

    @editable
    KnockDetect : down_but_not_out_device = down_but_not_out_device{}

    @editable
    BasementTeleport : teleporter_device = teleporter_device{}

    @editable
    var LivesHUD : []hud_message_device = array{}
    
    @editable
    Basements : []Basement = array{}

    @editable
    PortalEnterSFX : audio_player_device = audio_player_device{}

    @editable
    ParkourVolume : volume_device = volume_device{}

    # @editable
    # SaveButtons : []button_device = array{}

    # @editable
    # SaveTeleports : []teleporter_device = array{}

    @editable
    RolesInfo : RolesHandler = RolesHandler{}

    @editable
    AllRevives : []hud_message_device = array{}

    @editable
    RevAnim : cinematic_sequence_device = cinematic_sequence_device{}

    @editable
    InBasementNotification : hud_message_device = hud_message_device{}

    @editable
    var OneHpReminder : cinematic_sequence_device = cinematic_sequence_device{}

    @editable
    Tasks : TasksHandler = TasksHandler{}

    @editable
    Achievments : AchievmentsHandler = AchievmentsHandler{}

    var Lives : [agent]int = map{}

    var Dead : []agent = array{}

    var Knocked : []agent = array{}

    var InQueueToRevive : []agent = array{}

    var Revives : [agent]hud_message_device = map{}

    var IsInBasement : [agent]logic = map{}

    var PlayersInBasement : []agent = array{}

    var SaveButtonsProgress : []int = array{}

    var AllPlayersDead : logic = false

    var NotifiedAboutRevives : logic = false

    var TotalJoined : int = 0

    var Timers : [agent]int = map{}

    OnBegin<override>()<suspends>:void=
        case (RolesInfo.DEBUG_InfoHudVersion):
            Versions.V1 =>
            Versions.V2 =>
                for (i := 0..LivesHUD.Length):
                    if (set LivesHUD[i] = hud_message_device{}) {}
                set OneHpReminder = cinematic_sequence_device{}
        loop:
            if (RolesInfo.Start = true and WaitingForPlayersStatus[GetSession()] = false):
                break
            Sleep(0.1)
        Sleep(5.0)
        AllPlayers := GetPlayspace().GetPlayers()
        set TotalJoined = AllPlayers.Length
        var MonsterFound : logic = false
        for (i := 0..AllPlayers.Length, pl := AllPlayers[i]):
            if (chr := pl.GetFortCharacter[]):
                chr.EliminatedEvent().Subscribe(PlayerDead)
                chr.DamagedEvent().Subscribe(PlayerDamaged)
            GetPlayspace().PlayerAddedEvent().Subscribe(PlayerJoined)
            GetPlayspace().PlayerRemovedEvent().Subscribe(PlayerLeft)
            if (pl <> RolesInfo.Monster? or RolesInfo.DEBUG_AllSurvivors = true):
                if (PlChar := pl.GetFortCharacter[]):
                    PlChar.SetVulnerability(true)
                if ((RolesInfo.DEBUG_AllSurvivors = true and AllPlayers.Length = 1 and DEBUG_AlwaysFullHp = false) or (RolesInfo.DEBUG_AllSurvivors = false and RolesInfo.Survivors.Length < 2 and DEBUG_AlwaysFullHp = false)):
                    if (set Lives[pl] = 1) {}
                    # Print("Lifes: One hp reminder")
                else:
                    if (set Lives[pl] = 3) {}
                    # Print(StrToMsg(?MessageAfter:=": 3 lives", ?Agent := pl))
                if (hud := LivesHUD[Lives[pl] - 1], DEBUG_ShowHud = true):
                    if ((RolesInfo.DEBUG_AllSurvivors = true and AllPlayers.Length = 1 and DEBUG_AlwaysFullHp = false) or (RolesInfo.DEBUG_AllSurvivors = false and AllPlayers.Length <= 2 and DEBUG_AlwaysFullHp = false)):
                        # hud.SetText(Msg("YOU ARE SOLO!"))
                        hud.SetText(MessagesBank{}.Solo)
                        hud.Show(pl)
                        OneHpReminder.Play(pl)
                    else:
                        # hud.SetText(Msg(""))
                        hud.SetText(MessagesBank{}.Empty)
                        hud.Show(pl)
                    # Print("Lifes: Showing the hud")
                if (set IsInBasement[pl] = false) {}
                if (MonsterFound = true):
                    if (rev := AllRevives[i - 1], set Revives[pl] = rev) {}
                else:
                    if (rev := AllRevives[i], set Revives[pl] = rev) {}
            else:
                set MonsterFound = true
                # Print("Lifes: Found the monster")
        for (base : Basements):
            set base.MainClass = Self
            set base.BasementTeleport = BasementTeleport
            set base.AllPlayers = AllPlayers
            if (not Tasks.SceneGraphActive?):
                base.Start()
        if (Tasks.SceneGraphActive?):
            if (Root := GetSimulationEntity[]):
                for (Entity : Root.FindDescendantEntitiesWithComponent(SG_PortalTask), TaskInfo := Entity.GetComponent[SG_PortalTask]):
                    set TaskInfo.LifesManager = option{Self}
        KnockDetect.AgentDownedEvent.Subscribe(PlayerKilled)
        ParkourVolume.AgentEntersEvent.Subscribe(PlayerCompletedParkour)
        # KnockDetect.AgentRevivedEvent.Subscribe(PlayerRevived) # TODO: Check if it breaks the game
        Tasks.TransmittersObjective.CompleteEvent.Subscribe(SetAllToZero)
        Tasks.PumpkinsObjective.CompleteEvent.Subscribe(SetAllToZero)
        Tasks.PortalsObjective.CompleteEvent.Subscribe(SetAllToZero)
        # spawn{CheckForEnd()}

    PlayerKilled(Player : agent):void=
        # if (Lives[Player] <= 0):
        #     Print("Player already dead, skipping autorevive...")
        #     return
        if (pl := RolesInfo.Monster?):
            Achievments.SaveAchievement(pl, PossibleAchievements.PlayersKnocked, 1)
        RolesInfo.StrongFlashlightToggle.Unregister(Player)
        # Check if all players are dead / knocked
        if (Player.GetFortCharacter[].SetVulnerability(false)) {}
        for (PublicHud : PublicHealthHud[GetSession()]):
            spawn{PublicHud.Update()}
        var Count : int = 0
        for (
            pl : RolesInfo.Survivors,
            not pl.IsActive[] or
            (
                (
                    not pl.GetFortCharacter[].IsActive or
                    not pl.GetFortCharacter[].IsVulnerable[] or
                    Lives[pl] = 0 or
                    pl = Player
                ) and
                RolesInfo.PlayersParams[pl].OneTimeInstantRevive.Value <= 0 and
                RolesInfo.PlayersParams[pl].GhostSave.Value <= 0
            )
        ):
            set Count += 1
        if (Count = RolesInfo.Survivors.Length and DEBUG_ContinueAfterFullWipeout = false):
            if (pl := RolesInfo.Monster?):
                Achievments.SaveAchievement(pl, PossibleAchievements.PlayersEliminated, 1)
            Winning{
                MainClass := RolesInfo
                # Reason := "The monster has eliminated all of the survivors"
                Reason := MessagesBank{}.AllSurvivorsEliminated
                Feedback := Feedback_M_AllSurvivorsEliminated
            }.MonsterWin(Player)
            return
        else:
            # Print("Game not over: {Count}/{RolesInfo.Survivors.Length} players are dead")
        spawn{PlayerKilled2(Player)}

    PlayerKilled2<private>(Player : agent)<suspends>:void=
        # Give 60 seconds for the player to be sent to prison, revive otherway
        if (PlChar := Player.GetFortCharacter[], RevScreen := Revives[Player], Pl := RolesInfo.Monster?):
            if (Tasks.PortalsObjective.GetValue() <> Tasks.PortalsObjective.GetTarget() and Lives[Player] > 0):
                # InBasementNotification.SetText(Msg("Bring the player to an open portal to capture them!"))
                InBasementNotification.SetText(MessagesBank{}.BringToPortal)
                InBasementNotification.Hide(Pl)
                InBasementNotification.Show(Pl)
            RolesInfo.SurIntro.Hide(Player)
            # RevScreen.SetText(Msg("Reviving in 60..."))
            var LostLifeAlready : logic = false
            if (Tasks.PortalsObjective.GetValue() = Tasks.PortalsObjective.GetTarget(), set Lives[Player] -= 1, LivesNum := Lives[Player]):
                set LostLifeAlready = true
                if (PrivateHud := StatsUiCon[GetSession()][Player]):
                    PrivateHud.UpdateHealth(LivesNum)
                if (Lives[Player] <= 0):
                    # Print("Player has no lives left, eliminating...")
                    Achievments.SaveAchievement(Pl, PossibleAchievements.PlayersEliminated, 1)
                    RevAnim.Stop(Player)
                    RevScreen.Hide(Player)
                    # if (GetPlayspace().GetTeamCollection().AddToTeam[Player, GetPlayspace().GetTeamCollection().GetTeams()[2]]) {}
                    PlChar.SetVulnerability(true)
                    Sleep(0.01)
                    PlChar.Damage(100000.0)
                    Achievments.Award(Pl, 3)
                    for (PublicHud : PublicHealthHud[GetSession()]):
                        spawn{PublicHud.Update()}
                    # if (not PlChar.IsActive[]):
                    #     if (GetPlayspace().GetTeamCollection().AddToTeam[Player, GetPlayspace().GetTeamCollection().GetTeams()[0]]) {}
            if (
                # Tasks.TransmittersObjective.GetValue() = Tasks.TransmittersObjective.GetTarget() and
                # Tasks.PumpkinsObjective.GetValue() = Tasks.PumpkinsObjective.GetTarget() and
                Tasks.PortalsObjective.GetValue() = Tasks.PortalsObjective.GetTarget()
            ):
                loop:
                    if (player[Player].IsActive[] and not Player.GetFortCharacter[].IsDownButNotOut[]):
                        # Revive
                        RevAnim.Stop(Player)
                        RevScreen.Hide(Player)
                        KnockDetect.Revive(Player)
                        PlChar.SetVulnerability(true)
                        RolesInfo.StrongFlashlightToggle.Register(Player)
                        for (PublicHud : PublicHealthHud[GetSession()]):
                            spawn{PublicHud.Update()}
                    if (not player[Player].IsActive[] or not Player.GetFortCharacter[].IsDownButNotOut[]):
                        return
                    Sleep(0.1)
            RevAnim.Stop(Player)
            RevAnim.Play(Player)
            if (DEBUG_ShowHud = true and Lives[Player] > 0):
                RevScreen.Show(Player)
            Achievments.Award(Pl, 1)
            RolesInfo.NoHpHud.ChangeClass(Player)
            for (i := 1..60, RolesInfo.PlayersParams[player[Player]].OneTimeInstantRevive.Value <= 0 and RolesInfo.PlayersParams[player[Player]].GhostSave.Value <= 0):
                Sleep(1.0)
                if (AllPlayersDead = true):
                    return
                if (IsInBasement[Player] = true):
                    if (RolesInfo.PlayersParams[RolesInfo.Monster?].OneTimeTwoHealthDamage.Value > 0):
                        if (set Lives[Player] -= 1) {}
                        if (set RolesInfo.PlayersParams[RolesInfo.Monster?].OneTimeTwoHealthDamage.Value -= 1) {}
                        if (RolesInfo.PlayersParams[RolesInfo.Monster?].OneTimeTwoHealthDamage.Value <= 0, set RolesInfo.PlayersParams[RolesInfo.Monster?].OneTimeTwoHealthDamage.GrayedOut = true) {}
                        if (Stats := StatsUiCon[GetSession()][RolesInfo.Monster?]):
                            Stats.FastUpdate()
                    if (LostLifeAlready = false, set Lives[Player] -= 1) {}
                    if (Stats := StatsUiCon[GetSession()][Player], LivesNum := Lives[Player]):
                        Stats.UpdateHealth(LivesNum)
                        # Print("Updated health value on StatsHud")
                    RevScreen.Hide(Player)
                    RevAnim.Stop(Player)
                    for (hud : LivesHUD):
                        hud.Hide(Player)
                    if (hud := LivesHUD[Lives[Player] - 1], DEBUG_ShowHud = true):
                        hud.Show(Player)
                    if (Lives[Player] <= 0):
                        Achievments.SaveAchievement(Pl, PossibleAchievements.PlayersEliminated, 1)
                        RevAnim.Stop(Player)
                        RevScreen.Hide(Player)
                        PlChar.SetVulnerability(true)
                        Sleep(0.01)
                        PlChar.Damage(100000.0)
                        Achievments.Award(Pl, 3)
                        # Print(StrToMsg(?MessageBefore:="Player ", ?Agent:=Player, ?MessageAfter:=" is dead"))
                        InBasementNotification.SetText(MessagesBank{}.Connect(MessagesBank{}.Empty, Player, MessagesBank{}.CaughtByMonster))
                        for (pl : RolesInfo.Survivors, pl <> Player, pl <> RolesInfo.Monster?, DEBUG_ShowHud = true):
                            for (anypl : PlayersInBasement, pl <> anypl):
                                InBasementNotification.Hide(pl)
                                InBasementNotification.Show(pl)
                        for (PublicHud : PublicHealthHud[GetSession()]):
                            spawn{PublicHud.Update()}
                        return
                    else if (Left := Lives[Player]):
                        Achievments.SaveAchievement(Pl, PossibleAchievements.PlayersInPortal, 1)
                        # Print(StrToMsg(?MessageBefore:="Player ", ?Agent:=Player, ?MessageAfter:=" has {Left} lives left"))
                        InBasementNotification.SetText(MessagesBank{}.Connect(MessagesBank{}.Empty, Player, MessagesBank{}.CaughtSave))
                        for (pl : RolesInfo.Survivors, pl <> Player, pl <> RolesInfo.Monster?):
                            for (anypl : PlayersInBasement, pl <> anypl):
                                InBasementNotification.Hide(pl)
                                InBasementNotification.Show(pl)
                        # InBasementNotification.SetText(Msg("You\'ve been caught! Wait for a teammate to save you!"))
                        InBasementNotification.SetText(MessagesBank{}.Caught)
                        if (DEBUG_ShowHud = true):
                            InBasementNotification.Hide(Player)
                            InBasementNotification.Show(Player)
                    BasementTeleport.Teleport(Player)
                    KnockDetect.Revive(Player)
                    if (PlayersInBasement.Length = 1):
                        for (base : Basements):
                            base.SaveButton.SetInteractionText(MessagesBank{}.Connect(MessagesBank{}.Save, Player, MessagesBank{}.Empty))
                    return
                else if (RolesInfo.GameEnded = true):
                    RevScreen.Hide(Player)
                    return
                else:
                    if (InQueueToRevive.Find[Player] or not Player.GetFortCharacter[].IsDownButNotOut[]):
                        RevAnim.Stop(Player)
                        RevScreen.Hide(Player)
                        KnockDetect.Revive(Player)
                        PlChar.SetVulnerability(true)
                        RolesInfo.StrongFlashlightToggle.Register(Player)
                        for (PublicHud : PublicHealthHud[GetSession()]):
                            spawn{PublicHud.Update()}
                        return
                    # RevScreen.SetText(Msg("Reviving in {60 - i}..."))
                    RevScreen.Hide(Player)
                    if (DEBUG_ShowHud = true):
                        RevScreen.Show(Player)
            if (RolesInfo.PlayersParams[player[Player]].OneTimeInstantRevive.Value > 0, Mon := RolesInfo.Monster?):
                # Notify about losing life
                # if (RolesInfo.DEBUG_InfoHudVersion = Versions.V1):
                InBasementNotification.SetText(MessagesBank{}.LostInstantRevive)
                InBasementNotification.Hide(Player)
                InBasementNotification.Show(Player)
                InBasementNotification.SetText(MessagesBank{}.InstantRevive)
                InBasementNotification.Hide(Mon)
                InBasementNotification.Show(Mon)
                if (set RolesInfo.PlayersParams[player[Player]].OneTimeInstantRevive.Value -= 1) {}
                if (RolesInfo.PlayersParams[player[Player]].OneTimeInstantRevive.Value <= 0, set RolesInfo.PlayersParams[player[Player]].OneTimeInstantRevive.GrayedOut = true) {}
                if (Stats := StatsUiCon[GetSession()][Player]):
                    Stats.FastUpdate()
            else if (RolesInfo.PlayersParams[player[Player]].GhostSave.Value > 0, Mon := RolesInfo.Monster?):
                # Notify about ghost save
                InBasementNotification.SetText(MessagesBank{}.GhostSave)
                InBasementNotification.Hide(Player)
                InBasementNotification.Show(Player)
                InBasementNotification.Hide(Mon)
                InBasementNotification.Show(Mon)
                if (set RolesInfo.PlayersParams[player[Player]].GhostSave.Value -= 1) {}
            RevAnim.Stop(Player)
            RevScreen.Hide(Player)
            KnockDetect.Revive(Player)
            PlChar.SetVulnerability(true)
            RolesInfo.StrongFlashlightToggle.Register(Player)
            for (PublicHud : PublicHealthHud[GetSession()]):
                spawn{PublicHud.Update()}
            # Print("Finished reviving player")

    PlayerRevived(Player : agent):void=
        Print(MessagesBank{}.Connect(MessagesBank{}.ToMsg("Player revived: "), Player, MessagesBank{}.Empty))
        # if (IsInBasement[Player] = false, chr := Player.GetFortCharacter[], chr.IsActive[]):
        #     chr.SetVulnerability(true)
        #     # Print("Player is now vulnerable")
        # else:
        #     # Print("Player is not in the basement or character is not active, skipping vulnerability")
        # for (PublicHud : PublicHealthHud[GetSession()]):
        #     spawn{PublicHud.Update()}
        spawn:
            PlayerRevivedSus(Player)

    PlayerRevivedSus(Player : agent)<suspends>:void=
        set InQueueToRevive += array{Player}
        Sleep(2.0)
        # var temp : []agent = array{}
        # for (pl : InQueueToRevive):
        #     if (pl <> Player):
        #         set temp += array{pl}
        # set InQueueToRevive = temp
        set InQueueToRevive = InQueueToRevive.RemoveAllElements(Player)
    
    PlayerEnteredBasement(Player : agent):void=
        if (not Player.GetFortCharacter[].IsVulnerable[], not IsInBasement[Player]?):
            PortalEnterSFX.Play()
            if (Pl := RolesInfo.Monster?):
                Achievments.Award(Pl, 2)
            if (set IsInBasement[Player] = true) {}
            set PlayersInBasement += array{Player}
            BasementTeleport.Teleport(Player)
            for (PublicHud : PublicHealthHud[GetSession()]):
                spawn{PublicHud.Update()}
            if (Lives[Player] > 1):
                Tasks.PlayerSentToBasement()
                spawn{LosingLives(Player)}
        
    LosingLives(Player : agent)<suspends>:void=
        # var Timer : int = 60
        loop:
            if:
                set Timers[Player] = 60
                set RolesInfo.PlayersParams[player[Player]].LosingLivesInBasement.Value = 60
            # for (i := 1..60):
            loop:
                if (Timers[Player] = 0):
                    break
                for (PublicHud : PublicHealthHud[GetSession()]):
                    PublicHud.UpdateTimers()
                if (PrivateHud := StatsUiCon[GetSession()][player[Player]]):
                    PrivateHud.FastUpdate()
                Sleep(1.0)
                if:
                    set Timers[Player] -= 1
                    temp := Timers[Player]
                    set RolesInfo.PlayersParams[player[Player]].LosingLivesInBasement.Value = temp
                # Check if the player is still in the basement
                if (IsInBasement[Player] = false):
                    # Print("Player left the basement")
                    break
            if (IsInBasement[Player] = false):
                # Print("Player left the basement")
                break
            if (set Lives[Player] -= 1) {}
            if (Stats := StatsUiCon[GetSession()][Player], LivesNum := Lives[Player]):
                Stats.UpdateHealth(LivesNum)
            if (Lives[Player] <= 0):
                Tasks.PlayerRescued()
                var temp : []agent = array{}
                for (pl : PlayersInBasement):
                    if (pl <> Player):
                        set temp += array{pl}
                set PlayersInBasement = temp
                for (i : Basements):
                    if (Pl := PlayersInBasement[0]):
                        i.SaveButton.SetInteractionText(MessagesBank{}.Connect(MessagesBank{}.Save, Pl, MessagesBank{}.Empty))
                    else:
                        i.SaveButton.Disable()
                # if (Pl := RolesInfo.Monster?):
                #     Achievments.SaveAchievement(Pl, PossibleAchievements.PlayersEliminated, 1)
                #     Achievments.Award(Pl, 3)
                if (chr := Player.GetFortCharacter[]):
                    chr.SetVulnerability(true)
                    Sleep(0.01)
                    # chr.Damage(3.0)
                    # if (Monika := RolesInfo.Monster?):
                    #     chr.Damage(damage_args{
                    #         Instigator := option{Monika.GetInstigator()}
                    #         Amount := 3.0
                    #     })
                    # loop:
                    #     if (not chr.IsActive[] and not chr.IsDownButNotOut[]):
                    #         Print("Player died in the basement, stopping the loop")
                    #         break
                    #     chr.SetVulnerability(true)
                    #     chr.Damage(3.0)
                    #     Sleep(0.1)
                    TeamsInfo := GetPlayspace().GetTeamCollection()
                    if (TeamsInfo.AddToTeam[Player, TeamsInfo.GetTeams()[2]]) {}
                    loop:
                        chr.Damage(100000.0)
                        if (not chr.IsActive[] and not chr.IsDownButNotOut[]):
                            # Print("Player died in the basement, stopping the loop")
                            break
                        Sleep(0.1)
                    if (TeamsInfo.AddToTeam[Player, TeamsInfo.GetTeams()[0]]) {}
                break
        if (set RolesInfo.PlayersParams[player[Player]].LosingLivesInBasement.Value = -1) {}
        for (PublicHud : PublicHealthHud[GetSession()]):
            spawn{PublicHud.Update()}
            PublicHud.UpdateTimers()
        if (PrivateHud := StatsUiCon[GetSession()][Player]):
            PrivateHud.FastUpdate()

    PlayerCompletedParkour(Agent : agent):void=
        if (set Timers[Agent] += 10):
            Print("Added 10 seconds to the timer", ?Color := NamedColors.Green)

    PlayerSaved(Teleport : teleporter_device):void=
        Tasks.PlayerRescued()
        if (Player := PlayersInBasement[0]):
            # Turn on the flashlight
            RolesInfo.StrongFlashlightToggle.Register(Player)
            if (set IsInBasement[Player] = false) {}
            #! Find the teleport of the specific basement
            Teleport.Teleport(Player)
            if (PlChar := Player.GetFortCharacter[]):
                PlChar.SetVulnerability(true)
            else:
                # Print("Player not found to make vulnerable")
            var temp : []agent = array{}
            for (pl : PlayersInBasement):
                if (pl <> Player):
                    set temp += array{pl}
            set PlayersInBasement = temp
            for (i : Basements):
                if (Pl := PlayersInBasement[0]):
                    i.SaveButton.SetInteractionText(MessagesBank{}.Connect(MessagesBank{}.Save, Pl, MessagesBank{}.Empty))
                else:
                    i.SaveButton.Disable()
            for (PublicHud : PublicHealthHud[GetSession()]):
                spawn{PublicHud.Update()}

    PlayerDamaged(info:damage_result):void=
        # Print("Player damaged...")
        spawn{PlayerDamagedSus(info)}

    PlayerDamagedSus(info:damage_result)<suspends>:void=
        Sleep(0.01)
        if (Target := fort_character[info.Target]):
            BeforeReturning := Target.GetHealth()
            # Print("{BeforeReturning} health before healing")
            # Target.SetVulnerability(false)
            Sleep(5.0)
            if (Target.GetHealth() = BeforeReturning):
                Target.SetHealth(100.0)
                # Print("Player healed")
                Sleep(2.0)
                if (Target.GetHealth() = 100.0):
                    if (Pl := fort_character[info.Target].GetAgent[]):
                        RolesInfo.NoHpHud.ChangeClass(Pl)
            else:
                # Print("Player got more damage, not healing")
            # Target.SetVulnerability(true)
        else:
            # Print("Player not found to protect")
    
    PlayerDead(info : elimination_result):void=
        for (PublicHud : PublicHealthHud[GetSession()]):
            spawn{PublicHud.Update()}
        if (AllPlayersDead = true):
            return
        var AllLeft : logic = true
        for (i : RolesInfo.Survivors, i.IsActive[]):
            set AllLeft = false
        if (AllLeft = true, rndpl := RolesInfo.Survivors[0]):
            set AllPlayersDead = true
            # RolesInfo.MonsterWin("All of the survivors have left the game")
            Winning{
                MainClass := RolesInfo
                # Reason := "All of the survivors have left the game"
                Reason := MessagesBank{}.AllSurvivorsLeft
                Feedback := Feedback_M_AllSurvivorsLeft
            }.MonsterWin(rndpl)
            return
        var AllDead : logic = true
        for (i : RolesInfo.Survivors, i.GetFortCharacter[].IsActive[]):
            set AllDead = false
        if (AllDead = true, rndpl := RolesInfo.Survivors[0]):
            set AllPlayersDead = true
            Winning{
                MainClass := RolesInfo
                # Reason := "The monster has eliminated all of the survivors"
                Reason := MessagesBank{}.AllSurvivorsEliminated
                Feedback := Feedback_M_AllSurvivorsEliminated
            }.MonsterWin(rndpl)
            return

    PlayerJoined(Player : player):void=
        # if (RolesInfo.GameEnded = true):
        #     case (RolesInfo.Winner):
        #         "Survivors" =>
        #             RolesInfo.SurvivorsWin()
        #         "Monster" =>
        #             RolesInfo.MonsterWin()
        #         "Canceled" =>
        #             RolesInfo.Canceled()
        #         _ =>
        #     return
        spawn{PlayerJoinedSus(Player)}
        
        
    PlayerJoinedSus(Player:player)<suspends>:void=
        for (PublicHud : PublicHealthHud[GetSession()]):
            spawn{PublicHud.Update()}
        if (PlChar := Player.GetFortCharacter[], PlChar.IsActive[]):
            PlChar.Damage(100.0)
            PlChar.SetVulnerability(true)
            Sleep(0.01)
            PlChar.Damage(5000.0)
        # if (TotalJoined < 5):
        #     set TotalJoined += 1
        #     AllTeams := GetPlayspace().GetTeamCollection().GetTeams()
        #     set RolesInfo.Survivors += array{Player}
        #     if (chr := Player.GetFortCharacter[]):
        #         chr.EliminatedEvent().Subscribe(PlayerDead)
        #         chr.DamagedEvent().Subscribe(PlayerDamaged)
        #     if (GetPlayspace().GetTeamCollection().AddToTeam[Player, AllTeams[0]]) {}
        #     if (set Lives[Player] = 3) {}
        # Print(StrToMsg(?MessageAfter:=": 3 lives", ?Agent := Player))
        #     if (hud := LivesHUD[Lives[Player] - 1]):
        #         hud.SetText(Msg(""))
        #         hud.Show(Player)
        #     if (set IsInBasement[Player] = false) {}
        #     if (rev := AllRevives[TotalJoined - 1], set Revives[Player] = rev) {}
        # else:
        # Print("Lifes: Player limit reached")
        #     # Eliminate the player
        #     if (chr := Player.GetFortCharacter[]):
        #         chr.Damage(100.0)
        #         Sleep(0.1)
        #         chr.SetVulnerability(false)
        #         chr.Damage(100.0)
        #     else:
        #         # Print("Lifes: Player not found to eliminate")

    PlayerLeft(Player : player):void=
        if (Player = RolesInfo.Monster?):
            RolesInfo.EndingVFX.Enable()
            RolesInfo.EndingVFX.Restart()
            Winning{
                MainClass := RolesInfo
                # Reason := "The monster has left the game"
                Reason := MessagesBank{}.MonsterLeft
                Feedback := Feedback_S_MonsterLeft
            }.SurvivorsWin(Player)
        for (PublicHud : PublicHealthHud[GetSession()]):
            spawn{PublicHud.Update()}
        if (AllPlayersDead = true):
            return
        var temp : []agent = array{}
        for (i : PlayersInBasement):
            if (i <> Player):
                set temp += array{i}
            else:
                if:
                    set IsInBasement[i] = false
                    set Lives[i] = 0
                Tasks.PlayerRescued()
        set PlayersInBasement = temp
        for (base : Basements):
            if (Pl := PlayersInBasement[0]):
                base.SaveButton.SetInteractionText(MessagesBank{}.Connect(MessagesBank{}.Save, Pl, MessagesBank{}.Empty))
            else:
                base.SaveButton.Disable()
        var AmountAlive : int = 0
        var AmountInGame : int = 0
        for (i : RolesInfo.Survivors, i.GetFortCharacter[].IsActive[]):
            set AmountAlive += 1
        for (i : RolesInfo.Survivors, i.IsActive[]):
            set AmountInGame += 1
        if (AmountAlive = 0 and AmountInGame > 0):
            set AllPlayersDead = true
            Winning{
                MainClass := RolesInfo
                # Reason := "The last survivor has left the game"
                Reason := MessagesBank{}.LastSurvivorLeft
                Feedback := Feedback_M_LastSurvivorLeft
            }.MonsterWin(Player)
        else if (AmountAlive = 0 and AmountInGame = 0):
            set AllPlayersDead = true
            Winning{
                MainClass := RolesInfo
                # Reason := "All of the survivors have left the game"
                Reason := MessagesBank{}.AllSurvivorsLeft
                Feedback := Feedback_M_AllSurvivorsLeft
            }.MonsterWin(Player)
        # else if ((AmountInGame = 1 or AmountAlive = 1) and DEBUG_AlwaysFullHp = false):
        #     # The last player's health is set to 1
        #     for (i : GetPlayspace().GetPlayers(), i <> RolesInfo.Monster? and Lives[i] > 0):
        #         if (set Lives[i] = 1) {}
        #         # Print(StrToMsg(?MessageAfter:=": 1 life", ?Agent := i))
        #         for (hud : LivesHUD):
        #             hud.Hide(i)
        #         if (hud := LivesHUD[Lives[i] - 1]):
        #             hud.SetText(Msg("YOU ARE SOLO!"))
        #             hud.Show(i)
        #             OneHpReminder.Play(i)

    SetAllToZero(Agent: agent):void=
        if:
            Tasks.TransmittersObjective.GetValue() = Tasks.TransmittersObjective.GetTarget() and
            Tasks.PumpkinsObjective.GetValue() = Tasks.PumpkinsObjective.GetTarget() and
            Tasks.PortalsObjective.GetValue() = Tasks.PortalsObjective.GetTarget()
        then:
            if (NotifiedAboutRevives = false):
                InBasementNotification.SetText(MessagesBank{}.PortalsClosed)
            for (pl : RolesInfo.Survivors, pl <> RolesInfo.Monster?, DEBUG_ShowHud = true):
                InBasementNotification.Hide(pl)
                InBasementNotification.Show(pl)
            for (i : GetPlayspace().GetPlayers(), i <> RolesInfo.Monster?):
                if (set Lives[i] = 1) {}
                # Print(StrToMsg(?MessageAfter:=": 1 life", ?Agent := i))
                for (hud : LivesHUD):
                    hud.Hide(i)
                if (hud := LivesHUD[Lives[i] - 1], DEBUG_ShowHud = true):
                    # hud.SetText(Msg("FINAL OBJECTIVE"))
                    hud.SetText(MessagesBank{}.FinalObjective)
                    hud.Show(i)
                    OneHpReminder.Play(i)
                for (PublicHud : PublicHealthHud[GetSession()]):
                    spawn{PublicHud.Update()}
                for (PlHud : StatsUiCon[GetSession()]):
                    PlHud.UpdateHealth(1)
        else if:
            Tasks.PortalsObjective.GetValue() = Tasks.PortalsObjective.GetTarget() and
            NotifiedAboutRevives = false
        then:
            # InBasementNotification.SetText(Msg("All small portals closed - Revives disabled!"))
            InBasementNotification.SetText(MessagesBank{}.PortalsClosed)
            var numAlive : int = 0
            for (pl : RolesInfo.Survivors, pl.GetFortCharacter[].IsActive[]):
                set numAlive += 1
            for (pl : RolesInfo.Survivors, pl <> RolesInfo.Monster?, DEBUG_ShowHud = true and numAlive > 1):
                InBasementNotification.Hide(pl)
                InBasementNotification.Show(pl)
            set NotifiedAboutRevives = true
        else:
            # Print("Lifes: Not all objectives are completed")

    CheckForEnd()<suspends>:void=
        loop:
            if (RolesInfo.GameEnded = true):
                break
            Sleep(0.5)
        for (pl : RolesInfo.Survivors, pl.IsActive[]):
            # Hide all huds
            for (hud : LivesHUD):
                hud.Hide(pl)