using { /Fortnite.com/Devices }
using { /Verse.org/Simulation }
using { /UnrealEngine.com/Temporary/Diagnostics }
using { /Fortnite.com/Characters }


note_obj := class<concrete>:
    @editable
    Button : button_device = button_device{}
    @editable
    Text : NotesChoice = NotesChoice.None

    @editable
    CustomModal : ?popup_dialog_device = false
    @editable
    CustomSound : ?audio_player_device = false

ShowNote := class:
    Title : message
    Text : message
    Modal : popup_dialog_device
    Audio : audio_player_device

    block:
        Modal.DismissedEvent.Subscribe(ModalDismissed)
    
    var PeopleToRemove : [agent]logic = map{}

    Show(Agent : agent):void=
        if (set PeopleToRemove[Agent] = false) {}
        Modal.SetTitleText(Title)
        Modal.SetDescriptionText(Text)
        Modal.Show(Agent)
        Audio.Play(Agent)
        spawn{CheckIfDead(Agent)}

    CheckIfDead(Agent : agent)<suspends>:void=
        if (chr := Agent.GetFortCharacter[]):
            loop:
                if (chr.IsDownButNotOut[] or not chr.IsActive[]):
                    Modal.Hide(Agent)
                    return
                if (PeopleToRemove[Agent] = true):
                    return
                Sleep(0.01)

    ModalDismissed(Agent : agent):void=
        if (set PeopleToRemove[Agent] = true) {}

NotesHandler := class(creative_device):

    @editable
    PageModal : popup_dialog_device = popup_dialog_device{}
    @editable
    PickUpSound : audio_player_device = audio_player_device{}

    @editable
    Notes : []note_obj = array{}
    
    OnBegin<override>()<suspends>:void=
        for (i : Notes):
            i.Button.InteractedWithEvent.Subscribe(ShowNote{
                Title := block:
                    case (i.Text):
                        NotesChoice.None =>
                            MessagesBank{}.Empty
                        NotesChoice.AlchemyRoom =>
                            MessagesBank{}.AlchemyNoteTitle
                        NotesChoice.LibraryRoom =>
                            MessagesBank{}.LibraryNoteTitle
                        NotesChoice.WarehouseRoom =>
                            MessagesBank{}.WarehouseNoteTitle
                        NotesChoice.FinaleRoom =>
                            MessagesBank{}.FinaleNoteTitle
                        NotesChoice.Her =>
                            MessagesBank{}.ToMsg(MessagesBank{}.HerNoteTitle)
                Text := block:
                    case (i.Text):
                        NotesChoice.None =>
                            MessagesBank{}.Empty
                        NotesChoice.AlchemyRoom =>
                            MessagesBank{}.AlchemyNoteText
                        NotesChoice.LibraryRoom =>
                            MessagesBank{}.LibraryNoteText
                        NotesChoice.WarehouseRoom =>
                            MessagesBank{}.ToMsg(MessagesBank{}.WarehouseNoteText)
                        NotesChoice.FinaleRoom =>
                            MessagesBank{}.FinaleNoteText
                        NotesChoice.Her =>
                            MessagesBank{}.ToMsg(MessagesBank{}.HerNoteText)
                Modal := block:
                    if (modal := i.CustomModal?):
                        modal
                    else:
                        PageModal
                Audio := block:
                    if (sound := i.CustomSound?):
                        sound
                    else:
                        PickUpSound
            }.Show)