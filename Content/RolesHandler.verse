using { /Fortnite.com/Devices }
using { /Fortnite.com/Characters }
using { /Fortnite.com/Teams }
using { /Fortnite.com/Game }
using { /Fortnite.com/UI }
using { /Fortnite.com/FortPlayerUtilities }
using { /Verse.org/Simulation }
using { /Verse.org/Random }
using { /Verse.org/Verse }
using { /Verse.org/Assets }
using { /Verse.org/Colors }
using { /UnrealEngine.com/Temporary/Diagnostics }
using { /UnrealEngine.com/Temporary/SpatialMath }
using { /UnrealEngine.com/Temporary/UI }
using { /Verse.org/Simulation/Tags }

var StatsUiCon : weak_map(session, [agent]PlayerStatsUI) = map{}

var WaitingForPlayersStatus : weak_map(session, logic) = map{}

Flashlight := class(tag){}

StrongFlashlight := class(tag){}

SoundsToDisableAtEnd := class(tag){}

MessagesBank := class:
    Empty<localizes> : message = ""

    # TasksHandler
    TransmittersPlacedInfo<localizes> : message = "Transmitters placed"
    PumpkinsPlacedInfo<localizes> : message = "Pumpkins placed"
    PortalsClosedInfo<localizes> : message = "Portals closed"
    FinalObjectiveInfo<localizes> : message = "Final portal"

    # LifeHandler
    Solo<localizes> : message = "YOU ARE SOLO!"
    FinalObjective<localizes> : message = "FINAL OBJECTIVE"
    BringToPortal<localizes> : message = "Bring the player to an open portal to capture them!"
    Caught<localizes> : message = "You\'ve been caught! You will lose 1 life if you don\'t get saved in the next 60 seconds!\nComplete the parkour to add 10 seconds to the timer."
    LostInstantRevive<localizes> : message = "You\'ve lost your instant revive!"
    InstantRevive<localizes> : message = "The player has used an instant revive!"
    GhostSave<localizes> : message = "What...?"
    PortalsClosed<localizes> : message = "All small portals closed - You will lose 1 life after being knocked!"
    CaughtByMonster<localizes> : message = " is caught by the monster!"
    CaughtSave<localizes> : message = " is caught by the monster! Save them from any open portal or remove planks from a closed one!"
    Save<localizes> : message = "Rescue "

    # TasksHandler
    SetUpTransmitter<localizes> : message = "Set up the transmitter"
    PlacePumpkin<localizes> : message = "Place the pumpkin"
    PlacePlanks<localizes> : message = "Place planks"
    RemovePlanks<localizes> : message = "Remove the planks"
    TaskCompleted<localizes> : message = "Task completed"
    TaskFailed<localizes> : message = "Task failed. Try again in 10 seconds.\nThe monster knows your location!"
    PlayerFailedTask<localizes> : message = "A player has failed a task!"

    # RolesHandler
    LosingLives<localizes> : message = "Losing lives"
    Lives : string = "Lives"
    Coins<localizes> : message = "Coins"
    FlashlightCharge<localizes> : message = "Flashlight Charge"
    HiddenTravel : string = "Hidden Travel"
    Scan<localizes> : message = "Scan"

    WaitingForPlayers<localizes> : message = "Waiting for players"
    YouWillBeReleasedIn<localizes> : message = "You will be released in "
    MonsterReleased<localizes> : message = "The monster has been released!"
    GettingIntoVent<localizes> : message = "Getting into the vent"
    TravelDisabled<localizes> : message = "Hidden travel has been used!"
    TravelEnabled<localizes> : message = "Hidden travel is now available!"

    # AbilitiesHandler
    Back<localizes> : message = "Back"

    YouveBeenScanned<localizes> : message = "You\'ve been scanned! Run!"
    CantTeleport<localizes> : message = "You can\'t teleport now!"
    CantTeleportMonster<localizes> : message = "You can\'t teleport the monster now!"
    YouveBeenTeleported<localizes> : message = "You\'ve been teleported!"
    ScanAvailable<localizes> : message = "Scan is now available!"
    GateControlAvailable<localizes> : message = "Gate control is now available!"
    CantPlaceTask<localizes> : message = "You can\'t place the task here!"
    FakeTaskFound<localizes> : message = "A player has triggered a fake task!"
    FakeTaskTriggered<localizes> : message = "You\'ve triggered a fake task!"
    TwoJumpsLeft<localizes> : message = "You have 2 jumps left!"
    OneJumpLeft<localizes> : message = "You have 1 jump left!"
    NoJumpsLeft<localizes> : message = "You have no jumps left!"

    ShopUiHeader<localizes> : message = "Choose your ability"
    DailyReward<localizes> : message = "Daily Reward"
    Nameplate<localizes> : message = "Nameplate"
    RoleStart<localizes> : message = "You are "
    RoleMonster<localizes> : message = "THE MONSTER"
    RoleSurvivor<localizes> : message = "A SURVIVOR"

    NotReady<localizes> : message = "Not Ready"
    Disconnected<localizes> : message = "Disconnected"

    AbilityNotAvailable<localizes> : message = "N/A"

    # AchievementsHandler
    AwaitingActivation<localizes> : message = "Awaiting Activation..."
    ActivatedSuccessfully<localizes> : message = "Activated Successfully!"
    ActivationFailed<localizes> : message = "Activation Failed!"
    TooManyRequests<localizes> : message = "Error 429: Too Many Requests"

    Close<localizes> : message = "Close"
    Activate<localizes> : message = "Activate"
    Clear<localizes> : message = "Clear"

    # HealthUI
    YourPrestige<localizes> : message = "Your prestige: "

    #* Rooms Names
    Kitchen<localizes> : message = "Kitchen"
    Vault<localizes> : message = "Vault"
    Jungle<localizes> : message = "Jungle"
    Library<localizes> : message = "Library"
    DiningRoom<localizes> : message = "Dining Room"
    SodaStorage<localizes> : message = "Soda Storage"
    Church<localizes> : message = "Church"
    AlchemistsRoom<localizes> : message = "Alchemist\'s Room"
    Chaos<localizes> : message = "Chaos"
    Warehouse<localizes> : message = "Warehouse"
    SurvivorsSpawn<localizes> : message = "Survivors\' Spawn"

    SurvivorsItem1Name<localizes> : message = "Improved Flashlight Battery"
    SurvivorsItem2Name<localizes> : message = "Gates Remote Control"
    SurvivorsItem3Name<localizes> : message = "Vent Access (Once per Round)"
    SurvivorsItem4Name<localizes> : message = "Instant Revive (Once per Round)"
    SurvivorsItem5Name<localizes> : message = "Emergency Teleport (Once per Round)"
    SurvivorsItem6Name<localizes> : message = "Monster Teleport (Once per Round)"

    SurvivorsItem1NameCut : string = "Improved Battery"
    SurvivorsItem2NameCut : string = "Remote Control"
    SurvivorsItem3NameCut : string = "Vent Access"
    SurvivorsItem4NameCut : string = "Instant Revive"
    SurvivorsItem5NameCut : string = "Teleport"
    SurvivorsItem6NameCut : string = "Monster Teleport"

    MonsterItem1Name<localizes> : message = "1.5x Damage"
    MonsterItem2Name<localizes> : message = "3 Jumps"
    MonsterItem3Name<localizes> : message = "Gates Control"
    MonsterItem4Name<localizes> : message = "3 Fake Tasks"
    MonsterItem5Name<localizes> : message = "Teleport (Once per Round)"
    MonsterItem6Name<localizes> : message = "2 Lives Damage (Once per Round)"

    MonsterItem1NameCut : string = "1.5x Damage"
    MonsterItem2NameCut : string = "Jumps"
    MonsterItem3NameCut : string = "Gates Control"
    MonsterItem4NameCut : string = "Fake Tasks"
    MonsterItem5NameCut : string = "Teleport"
    MonsterItem6NameCut : string = "2 Lives Damage"

    Prestige<localizes> : message = "Prestige"
    BuyItem<localizes> : message = "Buy"
    BuyNotAvailable<localizes> : message = "Buy (N/A)"
    NotEnough<localizes> : message = "Not Enough"
    NotEnoughNotAvailable<localizes> : message = "Not Enough (N/A)"
    EquipItem<localizes> : message = "Equip"
    ItemEquipped<localizes> : message = "Equipped"
    Ready<localizes> : message = "Ready"
    Unready<localizes> : message = "Cancel"

    Error<localizes> : message = "Error"
    
    # Reasons
    NotEnoughPlayers<localizes> : message = "Not enough players"
    TimeHasRunOut<localizes> : message = "Time has run out"
    MonsterDefeated<localizes> : message = "The monster has been defeated"
    AllPortalsClosed<localizes> : message = "All portals have been closed"
    AllSurvivorsEliminated<localizes> : message = "The monster has eliminated all of the survivors"
    AllSurvivorsLeft<localizes> : message = "All survivors have left the game"
    LastSurvivorLeft<localizes> : message = "The last survivor has left the game"
    MonsterLeft<localizes> : message = "The monster has left the game"
    MonsterAFK<localizes> : message = "The monster is AFK"

    NextRound<localizes> : message = "The next round will start in a moment"

    # NotesHandler
    AlchemyNoteTitle<localizes> : message = "Experiment #1604 - 04/05"
    AlchemyNoteText<localizes> : message = "Another experiment on the request of the king. I have spent over seven years trying to create at least one potion that would do the thing. I have failed every time. Each time a new potion fails, the king increases the amount of gold he gives me. I doubt that there is a lot left, it looks like he is willing to sacrifice everything for it. In any way, here is the new test."
    
    LibraryNoteTitle<localizes> : message = "SALVATION - ■■/■■"
    LibraryNoteText<localizes> : message = "Daniel left the project after the Experiment 1604. Because of that, the king has decided to go on with another plan. He will conquer the thing the people are feared of. He will be that fear, in the name of his people, and his family."

    WarehouseNoteTitle<localizes> : message = "Transportation result - 04/06"
    WarehouseNoteText : string = "■■■■■ - 10x\n■■■ - 10x\n■■■■■■ - 10x\n■■■■ - 10x\n■■■■ - 10x\n■■■■■■■■■■■■■■■■■■■■■■■■■■■■ - 1x"

    FinaleNoteTitle<localizes> : message = "The final experiment - 05/22"
    FinaleNoteText<localizes> : message = "The experiment will begin shortly. 2 people of the staff are standing near the glass shut button in case of an emergency. The room is ready, the people are ready.\nThe king is ready.\n\nShortly after starting the process, the visibility in the room got significantly lower.\nWe cannot see anything, the readers have turned off by themself. According to the regulations, the system shutdown is considered an emergency. The staff is currently working on the reserve power supply. The king is currently"

    HerNoteTitle : string = "Whispering Walls - ■■/■■"
    HerNoteText : string = "Empty room, a silent space,\nWhere echoes linger, lost embrace.\nWalls whisper tales of laughter\'s sound,\nLonely heart, where joy was found."

    Connect<localizes>(Msg1 : message, str : string, Msg2 : message) : message = "{Msg1}{str}{Msg2}"
    Connect<localizes>(Msg1 : message, Agent : agent, Msg2 : message) : message = "{Msg1}{Agent}{Msg2}"
    ToMsg<localizes>(num : int) : message = "{num}"
    ToMsg<localizes>(Agent : agent) : message = "{Agent}"
    # ToMsg<localizes>(Player : player) : message = "{Player}"
    ToMsg<localizes>(Str : string) : message = "{Str}"
    ToMsg<localizes>(Agent : agent, Str : string) : message = "{Agent}{Str}"

    StartingIn<localizes>(time:int) : message = "Starting in {time}s"
    QuestProgress<localizes>(Cur:int, Needed:int) : message = "{Cur}/{Needed}"
    Daytime<localizes>(H:string, M:string, S:string) : message = "{H}:{M}:{S}"
    Daytime<localizes>(M:string, S:string) : message = "{M}:{S}"
    Daytime<localizes>(S:string) : message = "{S}"

NotesChoice := enum:
    None
    AlchemyRoom
    LibraryRoom
    WarehouseRoom
    FinaleRoom
    Her

MonsterAnimation := enum:
    Default
    Crawl

DamageReaction := class:
    Agent : agent
    MainClass : RolesHandler

    Connect(dmg : damage_result):void=
        MainClass.damaged(Agent)

Winning := class:
    MainClass : RolesHandler
    Reason : message
    Feedback : analytics_device

    SurvivorsWin(Agent : agent):void=
        if (MainClass.GameEnded = true):
            return
        set MainClass.GameEnded = true
        MainClass.Achievements.Award(MainClass.Survivors, 0)
        MainClass.Achievements.SaveAchievement(MainClass.Survivors, PossibleAchievements.Wins, 1)
        if (Monika := MainClass.Monster?):
            Feedback.Submit(Monika)
        MainClass.SurvivorsWin(?Reason := option{Reason})
        spawn:
            EndRound(Agent)

    MonsterWin(Agent : agent):void=
        if (MainClass.GameEnded = true):
            return
        set MainClass.GameEnded = true
        if (Monika := MainClass.Monster?):
            MainClass.Achievements.Award(Monika, 0)
            MainClass.Achievements.SaveAchievement(Monika, PossibleAchievements.Wins, 1)
        if (Monika := MainClass.Monster?):
            Feedback.Submit(Monika)
        else if (Pl := MainClass.GetPlayspace().GetPlayers()[0]):
            Feedback.Submit(Pl)
        MainClass.MonsterWin(?Reason := option{Reason})
        spawn:
            EndRound(Agent)

    EndRound(Agent : agent)<suspends>:void=
        for (i : MainClass.Tasks_Interactions):
            i.Disable()
        for (i : MainClass.GetPlayspace().GetPlayers()):
            MainClass.SurIntro.Hide(i)
            MainClass.MonsterIntro.Hide(i)
        MainClass.EndingMusic.Disable()
        MainClass.EndScreenMusic.Play()
        for (pl : MainClass.Survivors, plchar := pl.GetFortCharacter[], pl.IsActive[]):
            MainClass.Achievements.SaveAchievement(Agent, PossibleAchievements.GamesPlayed, 1)
            plchar.PutInStasis(MainClass.StasisArgs)
            MainClass.P1_Flashlight.Hide(pl)
            MainClass.P2_Flashlight.Hide(pl)
            MainClass.P3_Flashlight.Hide(pl)
            MainClass.P4_Flashlight.Hide(pl)
            MainClass.P5_Flashlight.Hide(pl)
        if (Monika := MainClass.Monster?, MonikaChar := Monika.GetFortCharacter[]):
            MonikaChar.PutInStasis(MainClass.StasisArgs)
        for (pl : MainClass.GetPlayspace().GetPlayers()):
            for (snd : MainClass.NearSounds):
                snd.Unregister(pl)
            MainClass.Unchurch(pl)
        Sleep(8.5)
        # for (i : MainClass.FindCreativeObjectsWithTag(SoundsToDisableAtEnd{}), aud := audio_player_device[i]):
        #     aud.Disable()
        # for (i : MainClass.FindCreativeObjectsWithTag(SoundsToDisableAtEnd{}), rad := radio_device[i]):
        #     rad.UnregisterAll()
        for (pl : MainClass.GetPlayspace().GetPlayers(), pl.IsActive[], StatInfo := StatsUiCon[GetSession()][pl]):
            StatInfo.Remove()
        for (pl : MainClass.GetPlayspace().GetPlayers()):
            MainClass.NotEnoughPlayers.Activate(pl)

    Canceled(Agent : agent):void=
        if (MainClass.GameEnded = true):
            return
        set MainClass.GameEnded = true
        MainClass.Canceled(?Reason := option{Reason})

Vent := class<concrete>:

    @editable
    IsEnabled : logic = true

    @editable
    EnterZone : volume_device = volume_device{}

    @editable
    Location : teleporter_device = teleporter_device{}

    @editable
    TimeToClimb : float = 3.0

    var ClimbingMessage : hud_message_device = hud_message_device{}

    var WaitingForEntering : logic = false

    var MainClass : RolesHandler = RolesHandler{}

    Start():void=
        EnterZone.AgentEntersEvent.Subscribe(EnterVent)
        EnterZone.AgentExitsEvent.Subscribe(ExitVent)

    EnterVent(Agent : agent):void=
        # Check if the agent is in team 1
        if (MainClass.GetPlayspace().GetTeamCollection().GetTeam[Agent] = MainClass.GetPlayspace().GetTeamCollection().GetTeams()[0] and MainClass.PlayersParams[player[Agent]].OneTimeVentAccess.Value <= 0):
            return
        if (MainClass.GetPlayspace().GetTeamCollection().GetTeam[Agent] = MainClass.GetPlayspace().GetTeamCollection().GetTeams()[0]):
            # Print("Agent is in team 1")
        if (MainClass.PlayersParams[player[Agent]].OneTimeVentAccess.Value <= 0):
            # Print("Agent has no access")
        set WaitingForEntering = true
        spawn{EnterVentSus(Agent)}

    EnterVentSus(Agent : agent)<suspends>:void=
        # ClimbingMessage.SetText(Msg("Getting into the vent."))
        # ClimbingMessage.SetText(MessagesBank{}.GettingIntoVent)
        ClimbingMessage.SetText(MessagesBank{}.Connect(MessagesBank{}.GettingIntoVent, ".", MessagesBank{}.Empty))
        ClimbingMessage.Show(Agent)
        for (i := 1..3):
            for (g := 1..Int[TimeToClimb / 0.3]):
                if (WaitingForEntering = false):
                    ClimbingMessage.Hide(Agent)
                    return
                Sleep(0.1)
            # Check if two thirds of the time have passed
            if (i = 2):
                # ClimbingMessage.SetText(Msg("Getting into the vent..."))
                # ClimbingMessage.SetText(MessagesBank{}.GettingIntoVent3)
                ClimbingMessage.SetText(MessagesBank{}.Connect(MessagesBank{}.GettingIntoVent, "...", MessagesBank{}.Empty))
                ClimbingMessage.Hide(Agent)
                ClimbingMessage.Show(Agent)
            else if (i = 1):
                # ClimbingMessage.SetText(Msg("Getting into the vent.."))
                ClimbingMessage.SetText(MessagesBank{}.Connect(MessagesBank{}.GettingIntoVent, "..", MessagesBank{}.Empty))
                ClimbingMessage.Hide(Agent)
                ClimbingMessage.Show(Agent)
                
        if (WaitingForEntering = true):
            ClimbingMessage.Hide(Agent)
            Location.Teleport(Agent)
            if:
                MainClass.GetPlayspace().GetTeamCollection().GetTeam[Agent] = MainClass.GetPlayspace().GetTeamCollection().GetTeams()[0]
                set MainClass.PlayersParams[player[Agent]].OneTimeVentAccess.Value -= 1
                StatUi := StatsUiCon[GetSession()][Agent]
            then:
                for (i : MainClass.VentsMarkers):
                    i.RemoveFromShowList(Agent)
                if (MainClass.PlayersParams[player[Agent]].OneTimeVentAccess.Value <= 0, set MainClass.PlayersParams[player[Agent]].OneTimeVentAccess.GrayedOut = true) {}
                StatUi.FastUpdate()

    ExitVent(Agent : agent):void=
        set WaitingForEntering = false


# MakeCountdownTimer<constructor><public>(MaxTime : float, InPlayer : agent) := countdown_timer:
#     RemainingTime := MaxTime
#     MaybePlayerUI := option{GetPlayerUI[player[InPlayer]]} 


countdown_timer := class:
    MainClass : RolesHandler
    WinCondition : Winning
    Pl : player
    NotEnoughTime : audio_player_device = audio_player_device{}
    <# This block runs for each instance of the countdown_timer class.
    We can setup the canvas once here. #>
    block:
        set Canvas = canvas:
            Slots := array:
                canvas_slot:
                    Anchors := anchors:
                        Minimum := vector2{X := 0.5, Y := 0.125}
                        Maximum := vector2{X := 0.5, Y := 0.125}
                    Alignment := vector2{X := 0.5, Y := 0.5 }
                    Offsets := margin{Top := 0.0, Left := 0.0, Bottom := 0.0, Right := 0.0}
                    SizeToContent := true
                    Widget := RemainingTimeTextBlock
                # canvas_slot:
                #     Anchors := anchors:
                #         Minimum := vector2{X := 0.4, Y := 0.3}
                #         Maximum := vector2{X := 0.6, Y := 0.3}
                #     Alignment := vector2{X := 0.0, Y := 1.0}
                #     Offsets := margin{Top := 0.0, Right := 0.0}
                #     SizeToContent := true
                #     Widget := AddedTimeTextBlock
 
    # CountdownEndedEvent<public> : subscribable()
 
    StartCountdown<public>() : void =
        if (PlayerUI := MaybePlayerUI?):
            PlayerUI.AddWidget(Canvas)
 
            # Update the UI when we start the timer to see the initial RemainingTime on screen
            UpdateUI()
            spawn:
                RunCountdown()
 
    # AddRemainingTime<public>(Time : float) : void =
    #     set RemainingTime += Time
    #     # Immediately update UI for better player feedback when time is added.
    #     UpdateUI()
 
    #     # Fire a simple callout to show the time being added.
    #     spawn:
    #         AddedTimeCallout(Time)
 
    MaybePlayerUI<internal> : ?player_ui = false
    var RemainingTime<internal> : float = 0.0
    RemainingTimeTextBlock<private> : text_block = text_block{} 
    AddedTimeTextBlock<private> : text_block = text_block{}
    var Canvas<private> : canvas = canvas{}
    var TotalTime<private> : float = 0.0
    # The timer "precision": how often, in seconds, it ticks.
    TimerTickPeriod<private> : float = 1.0
 
    RunCountdown<private>()<suspends> : void =
        # We loop with the TimerTickPeriod.
        # The UI is also updated each time.
        loop:
            RemainingTimeTextBlock.SetTextColor(color{R := 204.0 / 255.0, G := 204.0 / 255.0, B := 204.0 / 255.0})
            Sleep(TimerTickPeriod)
            set TotalTime += TimerTickPeriod
            if (MainClass.GameEnded = true):
                # Remove the timer from the UI
                Canvas.RemoveWidget(RemainingTimeTextBlock)
                break
            set RemainingTime -= TimerTickPeriod
            UpdateUI()
 
            # Timer End
            if (
                300.0 >= RemainingTime > 299.0 or
                240.0 >= RemainingTime > 239.0 or
                180.0 >= RemainingTime > 179.0 or
                120.0 >= RemainingTime > 119.0 or
                60.0 >= RemainingTime > 59.0 or
                30.0 >= RemainingTime > 29.0
            ):
                NotEnoughTime.Play(Pl)
            if (MainClass.EndingVFXStart = true):
                Canvas.RemoveWidget(RemainingTimeTextBlock)
                if (Ui := MaybePlayerUI?):
                    Ui.RemoveWidget(Canvas)
                break
            if (RemainingTime <= 0.0):
                
                Canvas.RemoveWidget(RemainingTimeTextBlock)
 
                if (Ui := MaybePlayerUI?):
                    Ui.RemoveWidget(Canvas)
 
                WinCondition.MonsterWin(Pl)
                break
 
    # AddedTimeCallout<private>(Time : float)<suspends> : void =
    #     if:
    #         PlayerUI := MaybePlayerUI?
    #         IntTime := Int[Time]
    #     then:
    #         AddedTimeTextBlock.SetVisibility(widget_visibility.Visible)
    #         AddedTimeTextBlock.SetText(AddedTimeText(IntTime))
    #         Sleep(2.0)
    #         AddedTimeTextBlock.SetVisibility(widget_visibility.Hidden)
    
    UpdateUI<private>() : void =
        if (var IntTime : int = Int[RemainingTime]):
            # Convert to MM:SS format
            var Minutes : int = 0
            var Seconds : int = 0
            var res : string = ""
            if (IntTime < 60):
                set res = "00:{IntTime}"
            else if (IntTime < 600):
                if:
                    temp := IntTime / 60
                    set Minutes = Floor(temp)
                    set Seconds = IntTime - (Minutes * 60)
                    if (Seconds < 10 or IntTime < 10):
                        set res = "0{Minutes}:0{Seconds}"
                    else:
                        set res = "0{Minutes}:{Seconds}"
            else:
                if:
                    temp := IntTime / 60
                    set Minutes = Floor(temp)
                    set Seconds = IntTime - (Minutes * 60)
                    if (Seconds < 10 or IntTime < 10):
                        set res = "{Minutes}:0{Seconds}"
                    else:
                        set res = "{Minutes}:{Seconds}"
            # # Print(res)
            RemainingTimeTextBlock.SetText(MessagesBank{}.ToMsg(res))

player_param_info := class:
    var GrayedOut : logic = false
    var Value : int = -1

player_params := class:
    var PublicIcon : texture = Icons.IconNone
    var ImprovedFlashlightBattery : player_param_info = player_param_info{} # V
    var GatesRemoteControl : player_param_info = player_param_info{} # V
    var OneTimeVentAccess : player_param_info = player_param_info{} # V
    var OneTimeInstantRevive : player_param_info = player_param_info{} # V
    var OneTimeEmergencyTeleport : player_param_info = player_param_info{} # V
    var OneTimeMonsterTeleport : player_param_info = player_param_info{} # V
    var OneAndHalfDamage : player_param_info = player_param_info{} # V
    var MonsterJumps : player_param_info = player_param_info{} # V
    var GatesControl : player_param_info = player_param_info{} # V
    var FakeTasks : player_param_info = player_param_info{} # V
    var TeleportToSurvivor : player_param_info = player_param_info{} # V
    var OneTimeTwoHealthDamage : player_param_info = player_param_info{} # V

    var GhostSave : player_param_info = player_param_info{} # X

    var MonsterScream : player_param_info = player_param_info{} # X

    var LosingLivesInBasement : player_param_info = player_param_info{} # X

TimerEnum := enum:
    Off,
    Default,
    Minute1,
    Minutes2,
    Minutes3,
    Minutes5,
    Minutes6,
    Minutes10

FunEnum := enum:
    Random
    Num1
    Num404
    Num729674

PlayerStatsUI := class:
    Player : player
    PlUI : player_ui

    # var PlayerStatList : []string = array{} # Stops the game from removing some stats from the list when they are used
    var Lives<private> : int = 3
    var Canvas<private> : canvas = canvas{}

    # var PlayerUI : ?player_ui = false

    var Heart1<private> : texture_block = texture_block:
        DefaultImage := Icons.IconHeart
        # DefaultTint := color{R := 1.0, G := 1.0, B := 1.0}
        DefaultDesiredSize := vector2{X := 30.0, Y := 30.0}

    var Heart2<private> : texture_block = texture_block:
        DefaultImage := Icons.IconHeart
        # DefaultTint := color{R := 1.0, G := 1.0, B := 1.0}
        DefaultDesiredSize := vector2{X := 30.0, Y := 30.0}

    var Heart3<private> : texture_block = texture_block:
        DefaultImage := Icons.IconHeart
        # DefaultTint := color{R := 1.0, G := 1.0, B := 1.0}
        DefaultDesiredSize := vector2{X := 30.0, Y := 30.0}

    var HealthStat : stack_box_slot = stack_box_slot{Widget := overlay{}}

    MainClass : RolesHandler

    Start() : void=
        if (TempPlayerUI := GetPlayerUI[Player]):
            set HealthStat = AddHealthStat()
            if (MainClass.Survivors.Length = 1):
                UpdateHealth(1)
            else:
                FastUpdate()
            # Print("PlayerUI found")
            # TempPlayerUI.AddWidget(Canvas)
            # Print("Added widget")
            # Print("All stats length: {GetAllStats().Length}")
            # spawn:
            #     Update()

    # Update()<suspends> : void=
    #     var TheStart : logic = true
    #     loop:
    #         if (TempPlayerUI := PlayerUI?):
    #             AllStatsList := GetAllStats()
    #             tempCanvas := canvas:
    #                 Slots := array:
    #                     canvas_slot:
    #                         Anchors := anchors{Minimum := vector2{X := 0.85, Y := 0.96}, Maximum := vector2{X := 0.85, Y := 0.96}}
    #                         Offsets := margin{Left := 0.0, Top := 0.0, Right := 480.0, Bottom := AllStatsList.Length * 60.0}
    #                         # Offsets := margin{Top := 0.0, Left := 0.0, Right := 0.0, Bottom := 0.0}
    #                         Alignment := vector2{X := 0.5, Y := 1.0}
    #                         SizeToContent := false
    #                         Widget := stack_box{
    #                             Orientation := orientation.Vertical
    #                             Slots := AllStatsList
    #                         }
    #             if (tempCanvas <> Canvas or TheStart = true):
    #                 if (TheStart = true):
    #                     set TheStart = false
    #                 else:
    #                     TempPlayerUI.RemoveWidget(Canvas)
    #                 set Canvas = tempCanvas
    #                 TempPlayerUI.AddWidget(Canvas)
    #         # else:
    #         #     Print("PlayerUI not found")
    #         Sleep(0.01)
    #         if (MainClass.GameEnded = true or not Player.IsActive[] or not Player.GetFortCharacter[].IsActive[]):
    #             if (TempPlayerUI := PlayerUI?):
    #                 TempPlayerUI.RemoveWidget(Canvas)
    #             break
    
    FastUpdate():void=
        if (not Player.IsActive[]):
            return
        PlUI.RemoveWidget(Canvas)
        AllStatsList := GetAllStats()
        set Canvas = canvas:
            Slots := array:
                canvas_slot:
                    Anchors := anchors{Minimum := vector2{X := 0.85, Y := 0.96}, Maximum := vector2{X := 0.85, Y := 0.96}}
                    Offsets := margin{Left := 0.0, Top := 0.0, Right := 480.0, Bottom := AllStatsList.Length * 60.0}
                    # Offsets := margin{Top := 0.0, Left := 0.0, Right := 0.0, Bottom := 0.0}
                    Alignment := vector2{X := 0.5, Y := 1.0}
                    SizeToContent := false
                    Widget := stack_box{
                        Orientation := orientation.Vertical
                        Slots := AllStatsList
                    }
        PlUI.AddWidget(Canvas)
        
    AddHealthStat() : stack_box_slot=
        return stack_box_slot:
            HorizontalAlignment := horizontal_alignment.Fill
            VerticalAlignment := vertical_alignment.Fill
            Widget := overlay:
                Slots := array:
                    overlay_slot:
                        HorizontalAlignment := horizontal_alignment.Fill
                        VerticalAlignment := vertical_alignment.Center
                        Widget := texture_block:
                            DefaultImage := UI.StatBg
                            DefaultTint := color{R := 0.4, G := 0.0, B := 0.0}
                            DefaultDesiredSize := vector2{X := 480.0, Y := 60.0}
                    overlay_slot:
                        HorizontalAlignment := horizontal_alignment.Center
                        VerticalAlignment := vertical_alignment.Center
                        Padding := margin{Left := 73.0, Right := 95.0}
                        Widget := text_block:
                            DefaultText := MessagesBank{}.ToMsg(MessagesBank{}.Lives)
                            DefaultTextColor := color{R := 1.0, G := 1.0, B := 1.0}
                            DefaultOverflowPolicy := text_overflow_policy.Ellipsis
                            DefaultJustification := text_justification.Left
                    overlay_slot:
                        HorizontalAlignment := horizontal_alignment.Center
                        VerticalAlignment := vertical_alignment.Center
                        Padding := margin{Left := 188.0}
                        Widget := stack_box:
                            Orientation := orientation.Horizontal
                            Slots := array:
                                stack_box_slot:
                                    Widget := Heart1
                                stack_box_slot:
                                    Widget := Heart2
                                stack_box_slot:
                                    Widget := Heart3
                    overlay_slot:
                        HorizontalAlignment := horizontal_alignment.Center
                        VerticalAlignment := vertical_alignment.Center
                        Padding := margin{Right := 210.0}
                        Widget := texture_block:
                            DefaultImage := Icons.IconHeart
                            DefaultDesiredSize := vector2{X := 40.0, Y := 40.0}

    AddNewStat(Name : message, Value : message, Icon : texture, Color : color, ?IsGray : logic = false) : stack_box_slot=
        return stack_box_slot:
            HorizontalAlignment := horizontal_alignment.Fill
            VerticalAlignment := vertical_alignment.Fill
            Widget := overlay:
                Slots := array:
                    overlay_slot:
                        HorizontalAlignment := horizontal_alignment.Fill
                        VerticalAlignment := vertical_alignment.Center
                        Widget := texture_block:
                            DefaultImage := UI.StatBg
                            DefaultTint := block:
                                if (IsGray = false):
                                    Color
                                else:
                                    color{R := 0.05, G := 0.05, B := 0.05}
                            DefaultDesiredSize := vector2{X := 480.0, Y := 60.0}
                    overlay_slot:
                        HorizontalAlignment := horizontal_alignment.Center
                        VerticalAlignment := vertical_alignment.Center
                        Padding := margin{Left := 73.0, Right := 95.0}
                        Widget := text_block:
                            DefaultText := Name
                            DefaultTextColor := block:
                                if (Color.R + Color.G + Color.B > 1.5 and IsGray = false):
                                    color{R := 0.0, G := 0.0, B := 0.0}
                                else:
                                    color{R := 1.0, G := 1.0, B := 1.0}
                            DefaultTextOpacity := block:
                                if (IsGray = false):
                                    1.0
                                else:
                                    0.3
                            DefaultOverflowPolicy := text_overflow_policy.Ellipsis
                            DefaultJustification := text_justification.Left
                    overlay_slot:
                        Padding := margin{Left := 188.0}
                        HorizontalAlignment := horizontal_alignment.Center
                        VerticalAlignment := vertical_alignment.Center
                        Widget := text_block:
                            DefaultText := Value
                            DefaultTextColor := block:
                                if (Color.R + Color.G + Color.B > 1.5 and IsGray = false):
                                    color{R := 0.0, G := 0.0, B := 0.0}
                                else:
                                    color{R := 1.0, G := 1.0, B := 1.0}
                            DefaultTextOpacity := block:
                                if (IsGray = false):
                                    1.0
                                else:
                                    0.3
                    overlay_slot:
                        HorizontalAlignment := horizontal_alignment.Center
                        VerticalAlignment := vertical_alignment.Center
                        Padding := margin{Right := 210.0}
                        Widget := texture_block:
                            DefaultImage := Icon
                            DefaultTint := block:
                                if (IsGray = false):
                                    color{R := 1.0, G := 1.0, B := 1.0}
                                else:
                                    color{R := 0.25, G := 0.25, B := 0.25}
                            DefaultDesiredSize := vector2{X := 40.0, Y := 40.0}


    GetAllStats() : []stack_box_slot =
        var Stats : []stack_box_slot = array{}
        if:
            PlInfo := MainClass.PlayersParams[Player]
        then:
            if (PlInfo.LosingLivesInBasement.Value >= 0):
                set Stats += array:
                    AddNewStat(MessagesBank{}.LosingLives, MessagesBank{}.ToMsg(PlInfo.LosingLivesInBasement.Value), Icons.IconTimeout, color{R := 0.4, G := 0.0, B := 0.0}, ?IsGray := false)
            if (PlInfo.ImprovedFlashlightBattery.Value >= 0):
                set Stats += array:
                    AddNewStat(MessagesBank{}.ToMsg(MessagesBank{}.SurvivorsItem1NameCut), MessagesBank{}.ToMsg("On"), Icons.IconEnergy, color{R := 0.4, G := 0.0, B := 0.3}, ?IsGray := false)
            if (PlInfo.GatesRemoteControl.Value >= 0):
                Val := block:
                    if (PlInfo.GatesRemoteControl.Value = 0):
                        MessagesBank{}.ToMsg("On")
                    else:
                        MessagesBank{}.ToMsg(PlInfo.GatesRemoteControl.Value)
                set Stats += array:
                    AddNewStat(MessagesBank{}.ToMsg(MessagesBank{}.SurvivorsItem2NameCut), Val, Icons.IconLock, color{R := 0.4, G := 0.0, B := 0.3}, ?IsGray := PlInfo.GatesRemoteControl.GrayedOut)
            if (PlInfo.OneTimeVentAccess.Value >= 0):
                Val := block:
                    if (PlInfo.OneTimeVentAccess.Value = 1):
                        MessagesBank{}.ToMsg("On")
                    else:
                        MessagesBank{}.ToMsg("Off")
                set Stats += array:
                    AddNewStat(MessagesBank{}.ToMsg(MessagesBank{}.SurvivorsItem3NameCut), Val, Icons.IconBracket, color{R := 0.4, G := 0.0, B := 0.3}, ?IsGray := PlInfo.OneTimeVentAccess.GrayedOut)
            if (PlInfo.OneTimeInstantRevive.Value >= 0):
                Val := block:
                    if (PlInfo.OneTimeInstantRevive.Value = 1):
                        MessagesBank{}.ToMsg("On")
                    else:
                        MessagesBank{}.ToMsg("Off")
                set Stats += array:
                    AddNewStat(MessagesBank{}.ToMsg(MessagesBank{}.SurvivorsItem4NameCut), Val, Icons.IconReboot, color{R := 0.4, G := 0.0, B := 0.3}, ?IsGray := PlInfo.OneTimeInstantRevive.GrayedOut)
            if (PlInfo.OneTimeEmergencyTeleport.Value >= 0):
                Val := block:
                    if (PlInfo.OneTimeEmergencyTeleport.Value = 1):
                        MessagesBank{}.ToMsg("On")
                    else:
                        MessagesBank{}.ToMsg("Off")
                set Stats += array:
                    AddNewStat(MessagesBank{}.ToMsg(MessagesBank{}.SurvivorsItem5NameCut), Val, Icons.IconRift, color{R := 0.4, G := 0.0, B := 0.3}, ?IsGray := PlInfo.OneTimeEmergencyTeleport.GrayedOut)
            if (PlInfo.OneTimeMonsterTeleport.Value >= 0):
                Val := block:
                    if (PlInfo.OneTimeMonsterTeleport.Value = 1):
                        MessagesBank{}.ToMsg("On")
                    else:
                        MessagesBank{}.ToMsg("Off")
                set Stats += array:
                    AddNewStat(MessagesBank{}.ToMsg(MessagesBank{}.SurvivorsItem6NameCut), Val, Icons.IconPlace, color{R := 0.4, G := 0.0, B := 0.3}, ?IsGray := PlInfo.OneTimeMonsterTeleport.GrayedOut)
            if (PlInfo.OneAndHalfDamage.Value >= 0): # ✅
                set Stats += array:
                    AddNewStat(MessagesBank{}.ToMsg(MessagesBank{}.MonsterItem1NameCut), MessagesBank{}.ToMsg("On"), Icons.IconAddition, color{R := 0.4, G := 0.0, B := 0.3}, ?IsGray := false)
            if (PlInfo.MonsterJumps.Value >= 0):
                set Stats += array:
                    AddNewStat(MessagesBank{}.ToMsg(MessagesBank{}.MonsterItem2NameCut), MessagesBank{}.ToMsg(PlInfo.MonsterJumps.Value), Icons.IconUp, color{R := 0.4, G := 0.0, B := 0.3}, ?IsGray := PlInfo.MonsterJumps.GrayedOut)
            if (PlInfo.GatesControl.Value >= 0):
                set Stats += array:
                    AddNewStat(MessagesBank{}.ToMsg(MessagesBank{}.MonsterItem3NameCut), MessagesBank{}.ToMsg("On"), Icons.IconDoor, color{R := 0.4, G := 0.0, B := 0.3}, ?IsGray := false)
            if (PlInfo.FakeTasks.Value >= 0):
                set Stats += array:
                    AddNewStat(MessagesBank{}.ToMsg(MessagesBank{}.MonsterItem4NameCut), MessagesBank{}.ToMsg(PlInfo.FakeTasks.Value), Icons.IconBroken, color{R := 0.4, G := 0.0, B := 0.3}, ?IsGray := PlInfo.FakeTasks.GrayedOut)
            if (PlInfo.TeleportToSurvivor.Value >= 0):
                Val := block:
                    if (PlInfo.TeleportToSurvivor.Value = 0):
                        MessagesBank{}.ToMsg("On")
                    else:
                        MessagesBank{}.ToMsg(PlInfo.TeleportToSurvivor.Value)
                set Stats += array:
                    AddNewStat(MessagesBank{}.ToMsg(MessagesBank{}.MonsterItem5NameCut), Val, Icons.IconBehind, color{R := 0.4, G := 0.0, B := 0.3}, ?IsGray := PlInfo.TeleportToSurvivor.GrayedOut)
            if (PlInfo.OneTimeTwoHealthDamage.Value >= 0):
                Val := block:
                    if (PlInfo.OneTimeTwoHealthDamage.Value = 1):
                        MessagesBank{}.ToMsg("On")
                    else:
                        MessagesBank{}.ToMsg("Off")
                set Stats += array:
                    AddNewStat(MessagesBank{}.ToMsg(MessagesBank{}.MonsterItem6NameCut), Val, Icons.IconDouble, color{R := 0.4, G := 0.0, B := 0.3}, ?IsGray := PlInfo.OneTimeTwoHealthDamage.GrayedOut)
            if (PlInfo.MonsterScream.Value >= 0):
                var IsGray : logic = false
                Val := block:
                    if (PlInfo.MonsterScream.Value = 0):
                        MessagesBank{}.ToMsg("On")
                    else:
                        set IsGray = true
                        MessagesBank{}.ToMsg(PlInfo.MonsterScream.Value)
                set Stats += array:
                    AddNewStat(MessagesBank{}.Scan, Val, Icons.IconEye, color{R := 0.25, G := 0.0, B := 1.0}, ?IsGray := IsGray)
        if (Player <> MainClass.Monster?):
            HtVal := block:
                if (MainClass.HiddenTravelInfo = 0):
                    MessagesBank{}.ToMsg("On")
                else:
                    MessagesBank{}.ToMsg(MainClass.HiddenTravelInfo)
            set Stats += array:
                AddNewStat(MessagesBank{}.ToMsg(MessagesBank{}.HiddenTravel), HtVal, Icons.IconTravel, color{R := 0.0, G := 0.02, B := 0.5}, ?IsGray := MainClass.TravelDisabled)
            if (Percent := Round[MainClass.FlashCharges[Player] * 100], PublicHealthHud[GetSession()][0].MainClass.DefineState(Player) <> PlayerStates.Dead):
                var IsGray : logic = false
                if (MainClass.PlayersParams[Player].ImprovedFlashlightBattery.GrayedOut = true):
                    set IsGray = true
                set Stats += array:
                    AddNewStat(MessagesBank{}.FlashlightCharge, MessagesBank{}.ToMsg("{Percent}%"), Icons.IconElectric, color{R := 0.05, G := 0.3, B := 0.4}, ?IsGray := IsGray)
            # set Stats += array{HealthStat}
            if (PublicHealthHud[GetSession()][0].MainClass.DefineState(Player) <> PlayerStates.Dead):
                set Stats += array:
                    stack_box_slot:
                        HorizontalAlignment := horizontal_alignment.Fill
                        VerticalAlignment := vertical_alignment.Fill
                        Widget := overlay:
                            Slots := array:
                                overlay_slot:
                                    HorizontalAlignment := horizontal_alignment.Fill
                                    VerticalAlignment := vertical_alignment.Center
                                    Widget := texture_block:
                                        DefaultImage := UI.StatBg
                                        DefaultTint := color{R := 0.4, G := 0.0, B := 0.0}
                                        DefaultDesiredSize := vector2{X := 480.0, Y := 60.0}
                                overlay_slot:
                                    HorizontalAlignment := horizontal_alignment.Center
                                    VerticalAlignment := vertical_alignment.Center
                                    Padding := margin{Left := 73.0, Right := 95.0}
                                    Widget := text_block:
                                        DefaultText := MessagesBank{}.ToMsg(MessagesBank{}.Lives)
                                        DefaultTextColor := color{R := 1.0, G := 1.0, B := 1.0}
                                        DefaultOverflowPolicy := text_overflow_policy.Ellipsis
                                        DefaultJustification := text_justification.Left
                                overlay_slot:
                                    HorizontalAlignment := horizontal_alignment.Center
                                    VerticalAlignment := vertical_alignment.Center
                                    Padding := margin{Left := 188.0}
                                    Widget := stack_box:
                                        Orientation := orientation.Horizontal
                                        Slots := array:
                                            stack_box_slot:
                                                Widget := texture_block:
                                                    DefaultImage := Icons.IconHeart
                                                    DefaultTint := block:
                                                        if (Lives >= 1):
                                                            color{R := 1.0, G := 1.0, B := 1.0}
                                                        else:
                                                            color{R := 0.0, G := 0.0, B := 0.0}
                                                    DefaultDesiredSize := vector2{X := 30.0, Y := 30.0}
                                            stack_box_slot:
                                                Widget := texture_block:
                                                    DefaultImage := Icons.IconHeart
                                                    DefaultTint := block:
                                                        if (Lives >= 2):
                                                            color{R := 1.0, G := 1.0, B := 1.0}
                                                        else:
                                                            color{R := 0.0, G := 0.0, B := 0.0}
                                                    DefaultDesiredSize := vector2{X := 30.0, Y := 30.0}
                                            stack_box_slot:
                                                Widget := texture_block:
                                                    DefaultImage := Icons.IconHeart
                                                    DefaultTint := block:
                                                        color{R := 1.0, G := 1.0, B := 1.0}
                                                        if (Lives >= 3):
                                                            color{R := 1.0, G := 1.0, B := 1.0}
                                                        else:
                                                            color{R := 0.0, G := 0.0, B := 0.0}
                                                    DefaultDesiredSize := vector2{X := 30.0, Y := 30.0}
                                overlay_slot:
                                    HorizontalAlignment := horizontal_alignment.Center
                                    VerticalAlignment := vertical_alignment.Center
                                    Padding := margin{Right := 210.0}
                                    Widget := texture_block:
                                        DefaultImage := Icons.IconHeart
                                        DefaultDesiredSize := vector2{X := 40.0, Y := 40.0}
        if (Player.IsActive[], Profile := PlayersStats[Player]):
            set Stats += array:
        #         AddNewStat(MessagesBank{}.Coins, MessagesBank{}.ToMsg(Profile.Coins), Icons.IconCoin, color{R := 1.0, G := 0.5, B := 0.0}) # ✅ /// ❓
                AddNewStat(MessagesBank{}.Prestige, MessagesBank{}.ToMsg(Profile.Prestige), Icons.IconCrown, color{R := 1.0, G := 0.45, B := 0.0}) # ✅ /// ❓
        return Stats

    UpdateHealth(HP : int) : void=
        # if (HP < 3):
        #     Heart3.SetTint(color{R := 0.1, G := 0.1, B := 0.1})
        # if (HP < 2):
        #     Heart2.SetTint(color{R := 0.1, G := 0.1, B := 0.1})
        # if (HP < 1):
        #     Heart1.SetTint(color{R := 0.1, G := 0.1, B := 0.1})
        set Lives = HP
        FastUpdate()

    Remove():void=
        PlUI.RemoveWidget(Canvas)

roundnum := enum:
    Round1,
    Other,
    LastRound

RolesHandler := class(creative_device):

    # @editable
    # DEBUG : logic = false

    # @editable
    # DEBUG_EnableSolo : logic = false

    @editable
    DEBUG_SelectTeam : logic = false

    @editable
    DEBUG_AllSurvivors : logic = false

    @editable
    DEBUG_ShowTimer : logic = true

    @editable
    DEBUG_ShowFlashlights : logic = true

    @editable
    DEBUG_ShowNotifications : logic = true

    @editable
    DEBUG_Timer : TimerEnum = TimerEnum.Default

    @editable
    DEBUG_FUN : FunEnum = FunEnum.Random

    @editable
    DEBUG_MonsterDamageVersion : Versions = Versions.V1

    @editable
    DEBUG_MonsterMovementVersion : Versions = Versions.V1

    @editable
    DEBUG_InfoHudVersion : Versions = Versions.V1

    @editable
    DEBUG_MonsterAnimation : MonsterAnimation = MonsterAnimation.Default

    @editable
    Feedback_S_TimeOut : analytics_device = analytics_device{}
    @editable
    Feedback_S_MonsterDefeated : analytics_device = analytics_device{}
    @editable
    Feedback_S_AllPortalsClosed : analytics_device = analytics_device{}
    @editable
    Feedback_S_MonsterAFK : analytics_device = analytics_device{}

    @editable
    Feedback_1P : analytics_device = analytics_device{}
    @editable
    Feedback_2P : analytics_device = analytics_device{}
    @editable
    Feedback_3P : analytics_device = analytics_device{}
    @editable
    Feedback_4P : analytics_device = analytics_device{}
    @editable
    Feedback_5P : analytics_device = analytics_device{}
    @editable
    Feedback_ErrP : analytics_device = analytics_device{}

    @editable
    Tasks_Interactions : []skilled_interaction_device = array{}

    @editable
    MonsterStart<private> : teleporter_device = teleporter_device{}

    @editable
    MonsterObj<private> : animated_mesh_device = animated_mesh_device{}

    @editable
    MonsterLocationUpdateTime : float = 0.1
    @editable
    FlashlightUpdateTime : float = 0.1

    @editable
    MonsterReleaseTime : int = 15

    @editable
    WaitingForPlayersHud : hud_message_device = hud_message_device{}
    @editable
    WaitingForPlayersSeqCut : cinematic_sequence_device = cinematic_sequence_device{}
    @editable
    WaitingForPlayersSeqs : []cinematic_sequence_device = array{}
    @editable
    WaitingForPlayersHudController : hud_controller_device = hud_controller_device{}
    @editable
    WaitingForPlayersNextRound : end_game_device = end_game_device{}
    @editable
    WaitingForPlayersSpawner : trigger_device = trigger_device{}
    @editable
    StartSpawns : []player_spawner_device = array{}
    # @editable
    # NewPlayersKill : damage_volume_device = damage_volume_device{}

    @editable
    SurvivorsCamera<private> : gameplay_camera_first_person_device = gameplay_camera_first_person_device{}
    @editable
    MonsterCamera<private> : gameplay_camera_orbit_device = gameplay_camera_orbit_device{}
    @editable
    MonsterDmg : damage_volume_device = damage_volume_device{}
    @editable
    MonsterDmgTrigger : volume_device = volume_device{}
    @editable
    MonsterDmgPerception : perception_trigger_device = perception_trigger_device{}
    @editable
    MonsterAtk : input_trigger_device = input_trigger_device{}
    @editable
    DmgRaycastStart : creative_prop = creative_prop{}
    @editable
    DmgRaycastEnd : creative_prop = creative_prop{}

    @editable
    HideMonster : cinematic_sequence_device = cinematic_sequence_device{}

    @editable
    StrongFlashlightToggle : input_trigger_device = input_trigger_device{}

    @editable
    P1_Flashlight : hud_message_device = hud_message_device{}
    @editable
    P2_Flashlight : hud_message_device = hud_message_device{}
    @editable
    P3_Flashlight : hud_message_device = hud_message_device{}
    @editable
    P4_Flashlight : hud_message_device = hud_message_device{}
    @editable
    P5_Flashlight : hud_message_device = hud_message_device{}
    @editable
    P6_Flashlight : hud_message_device = hud_message_device{}

    @editable
    Discharged : audio_player_device = audio_player_device{}
    @editable
    Charged : audio_player_device = audio_player_device{}
    @editable
    Toggled : audio_player_device = audio_player_device{}
    @editable
    Timer : audio_player_device = audio_player_device{}
    @editable
    NotEnoughTime : audio_player_device = audio_player_device{}
    @editable
    DamageDealtAudio : audio_player_device = audio_player_device{}
    @editable
    Traceback : audio_player_device = audio_player_device{}
    @editable
    Her : radio_device = radio_device{}
    @editable
    HerNote : button_device = button_device{}
    @editable
    TracebackTrigger : volume_device = volume_device{}
    @editable
    MusicBox : radio_device = radio_device{}
    @editable
    NearSounds : []audio_player_device = array{}

    @editable
    ScarySounds : []audio_player_device = array{}
    @editable
    AttackAud : audio_player_device = audio_player_device{}

    @editable
    VentSounds : audio_player_device = audio_player_device{}
    @editable
    # Vent : volume_device = volume_device{}
    AllVents : []Vent = array{}
    @editable
    VentsMarkers : []beacon_device = array{}
    @editable
    ClimbingMessage : hud_message_device = hud_message_device{}

    @editable
    ChurchTrigger : volume_device = volume_device{}
    @editable
    ChruchSFX : radio_device = radio_device{}
    @editable
    ChurchAudios : audio_player_device = audio_player_device{}

    @editable
    AttackCon : input_trigger_device = input_trigger_device{}

    @editable
    Hidden : trigger_device = trigger_device{}
    @editable
    Shown : trigger_device = trigger_device{}
    @editable
    EnableTravel : trigger_device = trigger_device{}
    @editable
    DisableTravel : trigger_device = trigger_device{}

    @editable
    Lvl1Dmg : hud_message_device = hud_message_device{}
    @editable
    Lvl2Dmg : hud_message_device = hud_message_device{}
    @editable
    Lvl3Dmg : hud_message_device = hud_message_device{}

    @editable
    NoHpHud : class_and_team_selector_device = class_and_team_selector_device{}
    @editable
    HpHud : class_and_team_selector_device = class_and_team_selector_device{}

    @editable
    Notification : hud_message_device = hud_message_device{}

    @editable
    HpMarker : player_marker_device = player_marker_device{}

    @editable
    SurIntro : hud_message_device = hud_message_device{}
    @editable
    MonsterIntro : hud_message_device = hud_message_device{}

    @editable
    DarkScreenStart : hud_message_device = hud_message_device{}
    @editable
    DarkScreenEnd : hud_message_device = hud_message_device{}
    @editable
    DarkScreenAnim : cinematic_sequence_device = cinematic_sequence_device{}
    @editable
    EndingMusic : audio_player_device = audio_player_device{}
    @editable
    EndingVFX : vfx_spawner_device = vfx_spawner_device{}
    @editable
    EndScreenMusic : radio_device = radio_device{}
    @editable
    SurWin : popup_dialog_device = popup_dialog_device{}
    @editable
    MonWin : popup_dialog_device = popup_dialog_device{}
    @editable
    Cancel : popup_dialog_device = popup_dialog_device{}

    @editable
    NotEnoughPlayers : end_game_device = end_game_device{}

    @editable
    DEBUG_TeamSelect : popup_dialog_device = popup_dialog_device{}

    @editable
    Achievements : AchievmentsHandler = AchievmentsHandler{}

    @editable
    FirstRoundSignal : round_settings_device = round_settings_device{}
    @editable
    AnyRoundSignal : round_settings_device = round_settings_device{}
    @editable
    LastRoundSignal : round_settings_device = round_settings_device{}

    @editable
    FUN_Show : billboard_device = billboard_device{}

    # P1FlashlightProgressbar : UI.P1_Flashlight_Mat_Inst_material = UI.P1_Flashlight_Mat_Inst_material{}
    # P2FlashlightProgressbar : UI.P2_Flashlight_Mat_Inst_material = UI.P2_Flashlight_Mat_Inst_material{}
    # P3FlashlightProgressbar : UI.P3_Flashlight_Mat_Inst_material = UI.P3_Flashlight_Mat_Inst_material{}
    # P4FlashlightProgressbar : UI.P4_Flashlight_Mat_Inst_material = UI.P4_Flashlight_Mat_Inst_material{}
    # P5FlashlightProgressbar : UI.P5_Flashlight_Mat_Inst_material = UI.P5_Flashlight_Mat_Inst_material{}
    # P6FlashlightProgressbar : UI.P6_Flashlight_Mat_Inst_material = UI.P6_Flashlight_Mat_Inst_material{}
    
    var Monster<public> : ?player = false

    var PlayersToFlashes<private> : [player][]customizable_light_device = map{}

    var PlayersToMats : [player]int = map{}

    var FlashCharges : [player]float = map{}
    
    var FlashEnabled<private> : [player]logic = map{}

    var Charging<private> : [player]logic = map{}

    var PlayerIsInVent : logic = false
    var SurvivorsInVent : []player = array{}

    var Survivors : []player = array{}

    var GameEnded : logic = false
    var Winner : string = ""

    StasisArgs : stasis_args = stasis_args{AllowTurning := true, AllowEmotes := false, AllowFalling := true}

    var PublicCountdownTimer : ?countdown_timer = false

    var MonsterAFK : logic = true

    var Start : logic = false

    var EndingVFXStart : logic = false

    # var CurrentRound:weak_map(session, int) = map{}
    # var IsLast : logic = false

    var PlayersParams : [player]player_params = map{}

    var TravelDisabled : logic = false
    var HiddenTravelInfo : int = 0

    var CurrentRound : ?roundnum = false
    var CurrentSeq : cinematic_sequence_device = cinematic_sequence_device{}

    WaitForPlayers(StartedWithEnough:logic, ?SameCycle:logic=false)<suspends> : []player=
        var ToNextSwitch : int = 10
        var CurSeqInd : int = -1
        set CurrentSeq = block:
            var NewInd : int = GetRandomInt(0, WaitingForPlayersSeqs.Length - 1)
            loop:
                if (NewInd = CurSeqInd):
                    set NewInd = GetRandomInt(0, WaitingForPlayersSeqs.Length - 1)
                else:
                    break
            set CurSeqInd = NewInd
            if (Seq := WaitingForPlayersSeqs[CurSeqInd]):
                Seq
            else:
                cinematic_sequence_device{}
        var AllSeing : []player = GetPlayspace().GetPlayers()
        WaitingForPlayersHud.SetText(MessagesBank{}.Connect(MessagesBank{}.WaitingForPlayers, ".", MessagesBank{}.Empty))
        for (i : AllSeing, chr := i.GetFortCharacter[], SameCycle = false):
            Sleep(0.0)
            chr.PutInStasis(stasis_args{AllowTurning := false, AllowEmotes := false, AllowFalling := true})
            # PlUi.AddWidget(UI.WB_WaitingForPlayers{})
            # CurrentSeq.Play(i)
            WaitingForPlayersHud.Show()
        CurrentSeq.Play()
        var DotsCount : int = 0
        loop:
            if (GetPlayspace().GetPlayers().Length >= 2):
                break
            Sleep(1.0)
            set DotsCount += 1
            set ToNextSwitch -= 1
            case (DotsCount):
                1 => WaitingForPlayersHud.SetText(MessagesBank{}.Connect(MessagesBank{}.WaitingForPlayers, ".", MessagesBank{}.Empty))
                2 => WaitingForPlayersHud.SetText(MessagesBank{}.Connect(MessagesBank{}.WaitingForPlayers, "..", MessagesBank{}.Empty))
                3 =>
                    WaitingForPlayersHud.SetText(MessagesBank{}.Connect(MessagesBank{}.WaitingForPlayers, "...", MessagesBank{}.Empty))
                    set DotsCount = 0
                _ =>
            for (i : GetPlayspace().GetPlayers()):
                WaitingForPlayersHud.Hide()
                WaitingForPlayersHud.Show()
            # if (ToNextSwitch = 1):
            #     Print("WaitForPlayers: Showing cut")
            #     WaitingForPlayersSeqCut.Play()
            if (ToNextSwitch = 0):
                set ToNextSwitch = 10
                # for (i : GetPlayspace().GetPlayers()):
                #     CurrentSeq.Stop(i)
                CurrentSeq.Stop()
                set CurrentSeq = block:
                    var NewInd : int = GetRandomInt(0, WaitingForPlayersSeqs.Length - 1)
                    loop:
                        if (NewInd = CurSeqInd):
                            set NewInd = GetRandomInt(0, WaitingForPlayersSeqs.Length - 1)
                        else:
                            break
                    set CurSeqInd = NewInd
                    if (Seq := WaitingForPlayersSeqs[CurSeqInd]):
                        Seq
                    else:
                        cinematic_sequence_device{}
                # for (i : GetPlayspace().GetPlayers()):
                #     CurrentSeq.Play(i)
            CurrentSeq.Play()
        for (time := 1..30):
            if (GetPlayspace().GetPlayers().Length >= 5 or StartedWithEnough = true):
                if (30 - time <= 9):
                    Print("Time: {time}")
                    Sleep(1.0)
                    set DotsCount += 1
                # for (i : GetPlayspace().GetPlayers(), chr := i.GetFortCharacter[]):
                #     CurrentSeq.Stop(i)
                #     WaitingForPlayersHud.Hide()
                #     chr.ReleaseFromStasis()
                # return GetPlayspace().GetPlayers()
            else:
                Sleep(1.0)
                set DotsCount += 1
                set ToNextSwitch -= 1
            if (GetPlayspace().GetPlayers().Length >= 5):
                WaitingForPlayersHud.SetText(MessagesBank{}.StartingIn(30 - time))
            else:
                case (DotsCount):
                    1 => WaitingForPlayersHud.SetText(MessagesBank{}.Connect(MessagesBank{}.WaitingForPlayers, ".\n", MessagesBank{}.StartingIn(30 - time)))
                    2 => WaitingForPlayersHud.SetText(MessagesBank{}.Connect(MessagesBank{}.WaitingForPlayers, "..\n", MessagesBank{}.StartingIn(30 - time)))
                    3 =>
                        WaitingForPlayersHud.SetText(MessagesBank{}.Connect(MessagesBank{}.WaitingForPlayers, "...\n", MessagesBank{}.StartingIn(30 - time)))
                        set DotsCount = 0
                    _ =>
            for (i : GetPlayspace().GetPlayers(), chr := i.GetFortCharacter[]):
                if (not AllSeing.Find[i]):
                    spawn:
                        NewPlayerDuringWaitForPlayer(i)
                    set AllSeing += array{i}
                chr.PutInStasis(stasis_args{AllowTurning := false, AllowEmotes := false, AllowFalling := false})
                WaitingForPlayersHud.Hide()
                WaitingForPlayersHud.Show()
            # if (ToNextSwitch = 1):
            #     Print("WaitForPlayers: Showing cut")
            #     WaitingForPlayersSeqCut.Play()
            if (ToNextSwitch = 0):
                set ToNextSwitch = 10
                # for (i : GetPlayspace().GetPlayers()):
                #     CurrentSeq.Stop(i)
                CurrentSeq.Stop()
                set CurrentSeq = block:
                    var NewInd : int = GetRandomInt(0, WaitingForPlayersSeqs.Length - 1)
                    loop:
                        if (NewInd = CurSeqInd):
                            set NewInd = GetRandomInt(0, WaitingForPlayersSeqs.Length - 1)
                        else:
                            break
                    set CurSeqInd = NewInd
                    if (Seq := WaitingForPlayersSeqs[CurSeqInd]):
                        Seq
                    else:
                        cinematic_sequence_device{}
                # for (i : GetPlayspace().GetPlayers()):
                #     CurrentSeq.Play(i)
                CurrentSeq.Play()
        if (GetPlayspace().GetPlayers().Length > 1):
            CurrentSeq.Stop()
            for (i : GetPlayspace().GetPlayers(), chr := i.GetFortCharacter[]):
                # CurrentSeq.Stop(i)
                WaitingForPlayersHud.Hide()
                chr.ReleaseFromStasis()
            return GetPlayspace().GetPlayers()
        else if (GetPlayspace().GetPlayers().Length = 0):
            return array{}
        else:
            return WaitForPlayers(false, ?SameCycle := true)

    NewPlayerDuringWaitForPlayer(Agent : agent)<suspends>:void=
        # Print("New player added during waiting")
        WaitingForPlayersSpawner.Trigger(Agent)
        for (i := 0..10):
            Sleep(0.1)
            loop:
                if (Agent.GetFortCharacter[].IsActive[]):
                    break
                Sleep(0.1)
            Sleep(1.0)
            # CurrentSeq.Play(Agent)

    OnFirstRound():void=
        # Print("First round triggered")
        set CurrentRound = option{roundnum.Round1}

    OnAnyRound():void=
        # Print("Any round triggered")
        if (not CurrentRound?):
            set CurrentRound = option{roundnum.Other}

    OnLastRound():void=
        # Print("Last round triggered")
        # set IsLast = true
        set CurrentRound = option{roundnum.LastRound}

    OnBegin<override>()<suspends>:void=
        FirstRoundSignal.RoundBeginEvent.Subscribe(OnFirstRound)
        AnyRoundSignal.RoundBeginEvent.Subscribe(OnAnyRound)
        LastRoundSignal.RoundBeginEvent.Subscribe(OnLastRound)
        GetPlayspace().PlayerAddedEvent().Subscribe(PlayerAddedAfterWaiting)
        if (set StatsUiCon[GetSession()] = map{}) {}
        var FUN : int = 0
        case (DEBUG_FUN):
            FunEnum.Random =>
                set FUN = GetRandomInt(0, 404)
            FunEnum.Num1 =>
                set FUN = 1
            FunEnum.Num404 =>
                set FUN = 404
            FunEnum.Num729674 =>
                set FUN = 0
        if (FUN = 0):
            set FUN = 729674
            MusicBox.Play()
            Her.Play()
            HerNote.Enable()
        FUN_Show.SetText(MessagesBank{}.ToMsg(FUN))
        # Print("FUN: {FUN}")
        case (FUN):
            404 =>
                TracebackTrigger.AgentEntersEvent.Subscribe(PlayTraceback)
            729674 =>
                TracebackTrigger.AgentEntersEvent.Subscribe(PlayHer)
            _ =>
        # TracebackTrigger.AgentEntersEvent.Subscribe(MakePlayerInvincible)
        # TracebackTrigger.AgentExitsEvent.Subscribe(MakePlayerVulnerable)
        # if:
        #     not CurrentRound[GetSession()]
        #     set CurrentRound[GetSession()] = 1
        # Print("Current round: {CurrentRound[GetSession()]}")
        # else if:
        #     set CurrentRound[GetSession()] += 1
        # Print("Current round: {CurrentRound[GetSession()]}")
        Cancel.RespondingButtonEvent.Subscribe(SendBackToLobby)
        # ChruchSFX.Play()
        ChurchTrigger.AgentEntersEvent.Subscribe(Church)
        ChurchTrigger.AgentExitsEvent.Subscribe(Unchurch)
        var allplayers : []player = GetPlayspace().GetPlayers()
        loop:
            if (CurrentRound?):
                break
            Sleep(0.01)
        if ((allplayers.Length < 5 and CurrentRound? = roundnum.Round1) or allplayers.Length = 1):
            # Print("No players found")
            if (pl := allplayers[0], set WaitingForPlayersStatus[GetSession()] = true):
                # NotEnoughPlayers.Activate(pl)
                # Canceled("Not enough players")
                # Winning{MainClass := Self, Reason := "Not enough players"}.Canceled(pl)
                WaitingForPlayersHudController.Enable()
                if (allplayers.Length >= 2):
                    set allplayers = WaitForPlayers(true)
                else:
                    set allplayers = WaitForPlayers(false)
                if (LastPlayer := allplayers[0]):
                    WaitingForPlayersNextRound.Activate(pl)
                    return
                if (allplayers.Length < 2):
                    Winning{
                        MainClass := Self
                        Reason := MessagesBank{}.NotEnoughPlayers
                        Feedback := analytics_device{}
                    }.Canceled(pl)
                    return
                WaitingForPlayersHudController.Disable()
        if (pl := allplayers[0]):
            case (allplayers.Length):
                1 => Feedback_1P.Submit(pl)
                2 => Feedback_2P.Submit(pl)
                3 => Feedback_3P.Submit(pl)
                4 => Feedback_4P.Submit(pl)
                5 => Feedback_5P.Submit(pl)
                _ => Feedback_ErrP.Submit(pl)
        for (i : StartSpawns):
            i.Disable()
        if (set WaitingForPlayersStatus[GetSession()] = false) {}
        SpeculatedMon := Achievements.GetRandomMonster()
        if (DEBUG_SelectTeam = true):
            for (pl : allplayers):
                DEBUG_TeamSelect.Show(pl)
        # else if (DEBUG_AllSurvivors = false, WillBeMonster := allplayers[GetRandomInt(0, allplayers.Length - 1)]):
        else if (DEBUG_AllSurvivors = false, WillBeMonster := SpeculatedMon?):
            # MonsterIntro.Show(WillBeMonster)
            spawn:
                MonsterSpawned(WillBeMonster)
            Sleep(0.01)
            var temp : []player = array{}
            for (pl : allplayers):
                if (pl <> WillBeMonster):
                    set temp += array{pl}
            set allplayers = temp
        else if (DEBUG_AllSurvivors = false):
            # Cancel the game
            for (pl : allplayers):
                Winning{
                    MainClass := Self
                    Reason := MessagesBank{}.NotEnoughPlayers
                    Feedback := analytics_device{}
                }.Canceled(pl)
                return
        AllTeams := GetPlayspace().GetTeamCollection().GetTeams()
        # UsualActors := GetCreativeObjectsWithTag(Flashlight{})
        # StrongActors := GetCreativeObjectsWithTag(StrongFlashlight{})
        var UsualActors : []creative_object_interface = array{}
        var StrongActors : []creative_object_interface = array{}
        for (fl : FindCreativeObjectsWithTag(Flashlight{})):
            set UsualActors += array{fl}
        for (sfl : FindCreativeObjectsWithTag(StrongFlashlight{})):
            set StrongActors += array{sfl}
        # # Print("Usual Flashes found: {UsualActors.Length}")
        # # Print("Strong Flashes found: {StrongActors.Length}")
        set Survivors = allplayers
        for (pl : allplayers):
            if (GetPlayspace().GetTeamCollection().AddToTeam[pl, AllTeams[2]]) {}
        if (GetPlayspace().GetTeamCollection().AddToTeam[Monster?, AllTeams[2]]) {}
        WaitingForPlayersHudController.Enable()
        loop:
            if (Start = true):
                break
            Sleep(0.1)
        WaitingForPlayersHudController.Disable()
        for (i : GetPlayspace().GetPlayers()):
            DarkScreenStart.Show(i)
        Sleep(0.5)
        for (i := 0..allplayers.Length, LightDevice := customizable_light_device[UsualActors[i]], BetterLightDevice := customizable_light_device[StrongActors[i]], pl := allplayers[i], plchar := pl.GetFortCharacter[]):
            plchar.PutInStasis(StasisArgs)
            if (GetPlayspace().GetTeamCollection().AddToTeam[pl, AllTeams[0]]) {}
            SurvivorsCamera.AddToAll()
            if (DEBUG_AllSurvivors = true):
                plchar.SetVulnerability(true)
            if:
                set PlayersToFlashes[pl] = array{LightDevice, BetterLightDevice}
                set FlashCharges[pl] = 1.0
                if (PlayersParams[pl].ImprovedFlashlightBattery.Value >= 0):
                    set FlashCharges[pl] = 1.5
                set PlayersToMats[pl] = i
            then:
                plchar.DamagedEvent().Subscribe(DamageReaction{Agent := pl, MainClass := Self}.Connect)
                BetterLightDevice.TurnOff()
                case (DEBUG_InfoHudVersion):
                    Versions.V1 =>
                        if (DEBUG_ShowFlashlights = true):
                            case (i):
                                0 =>
                                    P1_Flashlight.Show(pl)
                                1 =>
                                    P2_Flashlight.Show(pl)
                                2 =>
                                    P3_Flashlight.Show(pl)
                                3 =>
                                    P4_Flashlight.Show(pl)
                                4 =>
                                    P5_Flashlight.Show(pl)
                                5 =>
                                    P6_Flashlight.Show(pl)
                                _ =>
                    Versions.V2 =>
                        if (Monika := player[Monster?], PlUI := GetPlayerUI[Monika], not StatsUiCon[GetSession()][Monika], set StatsUiCon[GetSession()][Monika] = PlayerStatsUI{Player := Monika, PlUI := PlUI, MainClass := Self}, NewHud := StatsUiCon[GetSession()][Monika]):
                            NewHud.Start()
                        if (PlUI := GetPlayerUI[pl], set StatsUiCon[GetSession()][pl] = PlayerStatsUI{Player := pl, PlUI := PlUI, MainClass := Self}, NewHud := StatsUiCon[GetSession()][pl]):
                            NewHud.Start()
                # Print(StrToMsg(?Agent:=pl, ?MessageAfter:=" has been added to the flashlight system"))
            else:
                # Print(StrToMsg(?Agent:=pl, ?MessageAfter:=" has not been added to the flashlight system"))
        # DarkScreenStart.Show()
        # set Monster = option{allplayers[GetRandomInt(0, allplayers.Length - 1)]}
        # MonsterStart.SpawnedEvent.Subscribe(MonsterSpawned)
        # DEBUG_TeamSelect.RespondingButtonEvent.Subscribe(TeamSelect)
        StrongFlashlightToggle.PressedEvent.Subscribe(ToggleFlashlight)
        case (DEBUG_MonsterDamageVersion):
            Versions.V1 =>
                MonsterAtk.PressedEvent.Subscribe(MonsterAttackingV1)
            Versions.V2 =>
                MonsterDmg.Disable()
                MonsterAtk.PressedEvent.Subscribe(MonsterAttackingV2)
        SurWin.RespondingButtonEvent.Subscribe(SendBackToLobby)
        MonWin.RespondingButtonEvent.Subscribe(SendBackToLobby)
        Hidden.TriggeredEvent.Subscribe(TurnOffFlashlight)
        Shown.TriggeredEvent.Subscribe(TurnOnFlashlight)
        for (i := 0..AllVents.Length, vent := AllVents[i], marker := VentsMarkers[i], AllVents[i].IsEnabled = true):
            # marker.MoveTo(vent.Location.GetTransform().Translation, vent.Location.GetTransform().Rotation, 0.01)
            marker.MoveTo(vent.EnterZone.GetTransform().Translation, vent.EnterZone.GetTransform().Rotation, 0.01)
            marker.Enable()
            if (DEBUG_ShowNotifications = true):
                set vent.ClimbingMessage = ClimbingMessage
            set vent.MainClass = Self
            vent.Start()
        # Vent.AgentEntersEvent.Subscribe(MonsterInVent)
        # Vent.AgentExitsEvent.Subscribe(MonsterOutVent)
        for (i : Survivors):
            SurIntro.Show(i)
        loop:
            for (i : Survivors, i.IsActive[]):
                DarkScreenAnim.Play(i)
                break
            break
        spawn{WaitForMonsterRelease()}
        spawn{PlayerLightControl()}
        case (DEBUG_InfoHudVersion):
            Versions.V1 =>
                spawn{FlashlightStatus()}
            Versions.V2 =>
                # for (i : Survivors):
                #     StatsHud{Player := i, MainClass := Self}.Start()
        spawn{ScarePlayers()}
        spawn{ChurchSus()}
        Sleep(4.5)
        for (i : Survivors):
            StrongFlashlightToggle.Register(i)
        for (i : GetPlayspace().GetPlayers(), i <> Monster?):
            DarkScreenStart.Hide(i)
        Sleep(0.5)
        # Remove everyone from stasis
        for (pl : GetPlayspace().GetPlayers(), plchar := pl.GetFortCharacter[]):
            if (plchar.IsActive[]):
                plchar.ReleaseFromStasis()
        for (pl : allplayers, DEBUG_ShowTimer = true):
            if (PlayerUI := GetPlayerUI[pl]):
                var BestRemainingTime : float = 0.0
                case (DEBUG_Timer):
                    TimerEnum.Off =>
                        set BestRemainingTime = -1.0
                    TimerEnum.Default =>
                        var ToCount : int = allplayers.Length - 1
                        if (DEBUG_AllSurvivors = true):
                            set ToCount += 1
                        case (ToCount):
                            1 =>
                                set BestRemainingTime = 1800.0
                            2 =>
                                set BestRemainingTime = 1200.0
                            3 =>
                                set BestRemainingTime = 900.0
                            4 =>
                                set BestRemainingTime = 600.0
                            _ =>
                                set BestRemainingTime = 1800.0
                    TimerEnum.Minute1 =>
                        set BestRemainingTime = 60.0
                    TimerEnum.Minutes2 =>
                        set BestRemainingTime = 120.0
                    TimerEnum.Minutes3 =>
                        set BestRemainingTime = 180.0
                    TimerEnum.Minutes5 =>
                        set BestRemainingTime = 300.0
                    TimerEnum.Minutes6 =>
                        set BestRemainingTime = 360.0
                    TimerEnum.Minutes10 =>
                        set BestRemainingTime = 600.0
                if (BestRemainingTime > 0.0):
                    cnd := countdown_timer{
                        MainClass := Self,
                        MaybePlayerUI := option{PlayerUI},
                        RemainingTime := BestRemainingTime,
                        Pl := pl,
                        # WinCondition := Winning{MainClass := Self, Reason := "Time has run out"},
                        WinCondition := Winning{
                            MainClass := Self
                            Reason := MessagesBank{}.TimeHasRunOut
                            Feedback := Feedback_S_TimeOut
                        },
                        NotEnoughTime := NotEnoughTime
                    }
                    cnd.StartCountdown()
                    set PublicCountdownTimer = option{cnd}
        if (pl := player[Monster?]):
            if (PlayerUI := GetPlayerUI[pl]):
                var BestRemainingTime : float = 0.0
                case (DEBUG_Timer):
                    TimerEnum.Off =>
                        set BestRemainingTime = -1.0
                    TimerEnum.Default =>
                        var ToCount : int = allplayers.Length - 1
                        if (DEBUG_AllSurvivors = true):
                            set ToCount += 1
                        case (ToCount):
                            1 =>
                                set BestRemainingTime = 1800.0
                            2 =>
                                set BestRemainingTime = 1200.0
                            3 =>
                                set BestRemainingTime = 900.0
                            4 =>
                                set BestRemainingTime = 600.0
                            _ =>
                                set BestRemainingTime = 1800.0
                    TimerEnum.Minute1 =>
                        set BestRemainingTime = 60.0
                    TimerEnum.Minutes2 =>
                        set BestRemainingTime = 120.0
                    TimerEnum.Minutes3 =>
                        set BestRemainingTime = 180.0
                    TimerEnum.Minutes5 =>
                        set BestRemainingTime = 300.0
                    TimerEnum.Minutes6 =>
                        set BestRemainingTime = 360.0
                    TimerEnum.Minutes10 =>
                        set BestRemainingTime = 600.0
                if (BestRemainingTime > 0.0):
                    cnd := countdown_timer{
                        MainClass := Self,
                        MaybePlayerUI := option{PlayerUI},
                        RemainingTime := BestRemainingTime,
                        Pl := pl,
                        # WinCondition := Winning{MainClass := Self, Reason := "Time has run out"},
                        WinCondition := Winning{
                            MainClass := Self
                            Reason := MessagesBank{}.TimeHasRunOut
                            Feedback := Feedback_S_TimeOut
                        },
                        NotEnoughTime := NotEnoughTime
                    }
                    cnd.StartCountdown()
                    set PublicCountdownTimer = option{cnd}

    PlayTraceback(Agent : agent):void=
        Traceback.Play(Agent)
        if (TracebackTrigger.TeleportTo[vector3{X := 0.0, Y := 0.0, Z := 0.0}, MakeRotation(vector3{X := 0.0, Y := 0.0, Z := 0.0}, 0.0)]) {}

    PlayHer(Agent : agent):void=
        Her.Stop()
        if (TracebackTrigger.TeleportTo[vector3{X := 0.0, Y := 0.0, Z := 0.0}, MakeRotation(vector3{X := 0.0, Y := 0.0, Z := 0.0}, 0.0)]) {}

    MakePlayerInvincible(Agent : agent):void=
        if (chr := Agent.GetFortCharacter[]):
            chr.SetVulnerability(false)

    MakePlayerVulnerable(Agent : agent):void=
        if (chr := Agent.GetFortCharacter[]):
            chr.SetVulnerability(true)

    # TeamSelect(Agent: agent, Btn: int):void=
    #     case (Btn):
    #         0 =>
    #             set Monster = option{Agent}
    #             spawn:
    #                 MonsterSpawned(Agent)
    #         1 =>
    #         _ =>

    WaitForMonsterRelease()<suspends>:void=
        if (Mon := Monster?):
            if (Mon.GetFortCharacter[].TeleportTo[MonsterAtk.GetTransform().Translation, MonsterAtk.GetTransform().Rotation]) {}
            for (i := 0..MonsterReleaseTime):
                # Print(MessagesBank{}.Connect(MessagesBank{}.YouWillBeReleasedIn, "{MonsterReleaseTime + 1 - i}...", MessagesBank{}.Empty))
                Sleep(1.0)
            MonsterIntro.Show(Mon)
            spawn{WaitForMonster(Mon)}
            DarkScreenAnim.Play(Mon)
            Notification.SetText(MessagesBank{}.MonsterReleased)
            for (i : GetPlayspace().GetPlayers(), i <> Mon):
                # Notify survivors about monster release
                Notification.Hide(i)
                Notification.Show(i)
            Sleep(4.5)
            DarkScreenStart.Hide(Mon)

    WaitForMove(Agent : agent)<suspends>:void=
        # Sleep(0.5)
        # var done : []player = array{}
        # var poss : [player]vector3 = map{}
        # for (i : GetPlayspace().GetPlayers(), plchar := i.GetFortCharacter[], set poss[i] = plchar.GetTransform().Translation) {}
        # loop:
        #     if (done.Length = Survivors.Length + 1):
        #         break
        #     for (i -> pos : poss):
        #         var is_ignored : logic = false
        #         for (d : done):
        #             if (d = i):
        #                 set is_ignored = true
        #         if (is_ignored = false):
        #             if ((i.GetFortCharacter[].GetTransform().Translation.X <> pos.X or
        #                 i.GetFortCharacter[].GetTransform().Translation.Y <> pos.Y or
        #                 not i.GetFortCharacter[].IsActive[]) and player[Agent] = i):
        #                     set done += array{i}
        #                     SurIntro.Hide(i)
        #                     MonsterIntro.Hide(i)
        #                     # Print("Player has moved")
        #     Sleep(0.1)
        var pos : vector3 = vector3{}
        # if (Agent = Monster?):
        #     Print("Waiting for monster to spawn")
        #     Sleep(MonsterReleaseTime * 1.0)
        #     Print("Started checking monster\'s position")
        if (chr := Agent.GetFortCharacter[], set pos = chr.GetTransform().Translation) {}
        loop:
            if ((Agent.GetFortCharacter[].GetTransform().Translation.X <> pos.X or
            Agent.GetFortCharacter[].GetTransform().Translation.Y <> pos.Y or
            not Agent.GetFortCharacter[].IsActive[])):
                SurIntro.Hide(Agent)
                MonsterIntro.Hide(Agent)
                break
            Sleep(0.01)

    PlayerAddedAfterWaiting(Agent : player):void=
        # Print("Player added during gameplay")
        if (not WaitingForPlayersStatus[GetSession()] or WaitingForPlayersStatus[GetSession()] = true):
            if (Team1 := GetPlayspace().GetTeamCollection().GetTeams()[0]):
                if (GetPlayspace().GetTeamCollection().AddToTeam[Agent, Team1]) {}
            # Spawn the player at any spawn point
            if (SpawnPoint := StartSpawns[GetRandomInt(0, StartSpawns.Length - 1)], chr := Agent.GetFortCharacter[].TeleportTo[SpawnPoint.GetTransform().Translation, SpawnPoint.GetTransform().Rotation]) {}
        else:
            spawn{PlayerAddedAfterWaitingSus(Agent)}

    PlayerAddedAfterWaitingSus(Agent : player)<suspends>:void=
        # Eliminate the player
        # Add to survivors team
        var IsDamaged : logic = false
        if (chr := Agent.GetFortCharacter[]):
            loop:
                if (chr.IsActive[]):
                    # Print("Damaging")
                    chr.SetVulnerability(true)
                    # chr.SetHealth(0.0)
                    chr.Damage(100.0)
                if ((IsDamaged = true and not chr.IsActive[]) or not Agent.IsActive[]):
                    # Print("New player is dead")
                    break
                Sleep(0.1)

    PlayerAdded(Agent : player):void=
        if (PlayerUI := GetPlayerUI[Agent]):
            if (cnd := PublicCountdownTimer?):
                countdown_timer{
                    MainClass := Self,
                    MaybePlayerUI := option{PlayerUI},
                    RemainingTime := cnd.RemainingTime,
                    Pl := Agent,
                    WinCondition := cnd.WinCondition,
                    NotEnoughTime := cnd.NotEnoughTime
                }
                cnd.StartCountdown()
                set PublicCountdownTimer = option{cnd}
            else:
                # Print("No countdown timer found")
            if (DEBUG_SelectTeam = true):
                DEBUG_TeamSelect.Show(Agent)
            if (pl := player[Agent]):
                if (GetPlayspace().GetTeamCollection().AddToTeam[pl, GetPlayspace().GetTeamCollection().GetTeams()[0]]) {}
                if (plchar := pl.GetFortCharacter[]):
                    # plchar.PutInStasis(stasis_args{AllowTurning := true, AllowEmotes := true, AllowFalling := true})
                    var UsualActors : []creative_object_interface = array{}
                    var StrongActors : []creative_object_interface = array{}
                    for (fl : FindCreativeObjectsWithTag(Flashlight{})):
                        set UsualActors += array{fl}
                    for (sfl : FindCreativeObjectsWithTag(StrongFlashlight{})):
                        set StrongActors += array{sfl}
                    if:
                        N_fl := customizable_light_device[UsualActors[PlayersToFlashes.Length]]
                        S_fl := customizable_light_device[StrongActors[PlayersToFlashes.Length]]
                        set PlayersToFlashes[pl] = array{N_fl, S_fl}
                        set FlashCharges[pl] = 1.0
                        set PlayersToMats[pl] = PlayersToFlashes.Length - 1
                    then:
                        plchar.DamagedEvent().Subscribe(DamageReaction{Agent := pl, MainClass := Self}.Connect)
                        S_fl.TurnOff()
                        if (DEBUG_InfoHudVersion = Versions.V1):
                            case (PlayersToFlashes.Length - 1):
                                0 =>
                                    P1_Flashlight.Show(pl)
                                1 =>
                                    P2_Flashlight.Show(pl)
                                2 =>
                                    P3_Flashlight.Show(pl)
                                3 =>
                                    P4_Flashlight.Show(pl)
                                4 =>
                                    P5_Flashlight.Show(pl)
                                5 =>
                                    P6_Flashlight.Show(pl)
                                _ =>
                        StrongFlashlightToggle.Register(Agent)
                        # Print(StrToMsg(?Agent:=pl, ?MessageAfter:=" has been added to the flashlight system"))
                    else:
                        # Print(StrToMsg(?Agent:=pl, ?MessageAfter:=" has not been added to the flashlight system"))
                else:
                    # Print(StrToMsg(?Agent:=pl, ?MessageAfter:=" has no character"))
            else:
                # Print("No player found")
    
    MonsterSpawned<private>(Player : player)<suspends>:void=
        AllTeams := GetPlayspace().GetTeamCollection().GetTeams()
        if:
            pl := Player.GetFortCharacter[]
            pl.SetVulnerability(false)
            GetPlayspace().GetTeamCollection().AddToTeam[Player, AllTeams[1]]
        then:
            pl.Hide()
            MonsterCamera.AddTo(Player)
        HideMonster.Play(Player)
        set Monster = option{Player}
        MusicBox.Register(Player)
        Her.Register(Player)
        loop:
            if (Start = true):
                break
            Sleep(0.1)
        Sleep(MonsterReleaseTime * 1.0 - 0.1)
        AttackAud.Register(Player)
        MonsterStart.Teleport(Player)
            # loop:
            #     if (Start = true):
            #         break
            #     Sleep(0.1)
        spawn{MonsterControl()}
    
    MonsterControl<private>()<suspends>:void=
        # Get the movement from the monster, teleport the mesh, as if the player controls it, and then teleport the player to the default position
        var OldLocation : vector3 = MonsterObj.GetTransform().Translation
        loop:
            if (EndingVFXStart = true):
                break
            if:
                pl := Monster?.GetFortCharacter[]
                # Extract the yaw angle in degrees
                YawDegrees := pl.GetViewRotation().GetYawPitchRollDegrees()[0] - 90.0
        
                # Convert yaw angle to radians
                YawRadians := YawDegrees * (PiFloat / 180.0)
        
                # Create the rotation using the yaw angle in radians
                MonsterRot := MakeRotation(vector3{X := 0.0, Y := 0.0, Z := 1.0}, YawRadians)
            then:
                # Get a new translation that's 40 points below the player
                posforobj := vector3{X := pl.GetTransform().Translation.X, Y := pl.GetTransform().Translation.Y, Z := pl.GetTransform().Translation.Z - 80.0}
        
                # Calculate the forward vector based on the monster's rotation
                ForwardVector := vector3{
                    X := Cos(YawRadians),
                    Y := Sin(YawRadians),
                    Z := 0.0
                }
        
                # Calculate the right vector (perpendicular to the forward vector)
                RightVector := vector3{
                    X := -ForwardVector.Y,
                    Y := ForwardVector.X,
                    Z := 0.0
                }
        
                # # Calculate the movement vector
                CurrentLocation := MonsterObj.GetTransform().Translation
                # MovementVector := CurrentLocation - OldLocation
        
                # # Calculate the dot product for forward/backward movement
                # DP_Forward := DotProduct(ForwardVector, MovementVector)
        
                # # Calculate the dot product for left/right movement
                # DP_Right := DotProduct(RightVector, MovementVector)
        
                # # Check if the monster moved forward, backward, left, or right
                # if (DP_Forward > 0.0):
                # Print("Monster moved forward")
                #     MonsterObj.Play()
                # else if (DP_Forward < 0.0):
                # Print("Monster moved backward")
                #     MonsterObj.PlayReverse()
                # else if (DP_Right > 0.0):
                # Print("Monster moved right")
                #     MonsterObj.Play()
                # else if (DP_Right < 0.0):
                # Print("Monster moved left")
                #     MonsterObj.Play()
                # else:
                # Print("Monster did not move")
                #     MonsterObj.Pause()
                if (CurrentLocation.X <> OldLocation.X or CurrentLocation.Y <> OldLocation.Y):
                    MonsterObj.Play()
                else:
                    MonsterObj.Pause()
        
                # Update OldLocation for the next iteration
                set OldLocation = CurrentLocation
        
                # Check if the monster is near the player (2560 units radius)
                for (sur : GetPlayspace().GetPlayers(), sur.IsActive[], surchar := sur.GetFortCharacter[]):
                    if (Distance(MonsterObj.GetTransform().Translation, surchar.GetTransform().Translation) <= 2560.0 and sur <> Monster?):
                        for (snd : NearSounds):
                            snd.Register(sur)
                            snd.Play()
                    else if (Distance(MonsterObj.GetTransform().Translation, surchar.GetTransform().Translation) <= 4096.0, snd := NearSounds[0]):
                        snd.Register(sur)
                        snd.Play()
                        var first_sound : logic = true
                        for (sndrm : NearSounds):
                            if (first_sound = true):
                                set first_sound = false
                            else:
                                sndrm.Unregister(sur)
                    else:
                        for (snd : NearSounds):
                            snd.Unregister(sur)
                case (DEBUG_MonsterMovementVersion):
                    Versions.V1 =>
                        # Move the monster object to the new position
                        MonsterObj.MoveTo(posforobj, MonsterRot, MonsterLocationUpdateTime)
                    Versions.V2 =>
                        if (MonsterObj.TeleportTo[posforobj, MonsterRot]):
                            Sleep(0.0)
                        else:
                            break

                # # Print("Moving to {posforobj}\n{MonsterRot}")
                # var act : string = "Monster\' actual rotation:"
                # for (i : MonsterObj.GetTransform().Rotation.GetYawPitchRollDegrees()):
                #     set act += " {i}"
                # # Print(act)
            else if (player[Monster?].IsActive[]):
                if (not player[Monster?].GetFortCharacter[].IsActive[], ag := Monster?):
                    # Winning{MainClass := Self, Reason := "Monster has been defeated"}.SurvivorsWin(ag)
                    if (EndingVFX.TeleportTo[MonsterObj.GetTransform().Translation, MonsterObj.GetTransform().Rotation]):
                        set EndingVFXStart = true
                    EndingVFX.Enable()
                    EndingVFX.Restart()
                    Winning{
                        MainClass := Self
                        Reason := MessagesBank{}.MonsterDefeated
                        Feedback := Feedback_S_MonsterDefeated
                    }.SurvivorsWin(ag)
                    break
                Sleep(0.1)
            else:
                break
                    
            # Sleep(0.001)

    PlayerLightControl()<suspends>:void=
        loop:
            for (pl -> flahes : PlayersToFlashes, UsualLight := flahes[0], ImprovedLight := flahes[1]):
                if (pl.IsActive[], Player := pl.GetFortCharacter[], Player.IsActive[]):
                    PlayerPosition := Player.GetTransform().Translation
                    PlayerRotation := Player.GetViewRotation()
                    if (Yaw := PlayerRotation.GetYawPitchRollDegrees()[0]):
                
                        # Convert yaw angle to radians
                        YawRadians := Yaw * (PiFloat / 180.0)
                        
                        # Calculate forward vector
                        ForwardVector := vector3{X := Cos(YawRadians), Y := Sin(YawRadians), Z := 0.0} * 20.0
                        
                        # Calculate new position 10 units forward
                        NewPosition := PlayerPosition + ForwardVector
                        spawn{UsualLight.MoveTo(NewPosition, PlayerRotation, FlashlightUpdateTime)}
                        spawn{ImprovedLight.MoveTo(NewPosition, PlayerRotation, FlashlightUpdateTime)}
                else:
                    UsualLight.TurnOff()
                    ImprovedLight.TurnOff()
            Sleep(FlashlightUpdateTime)

    ToggleFlashlight<private>(Player : agent):void=
        if (FlashEnabled[player[Player]] = true or FlashCharges[player[Player]] <= 0.0, set FlashEnabled[player[Player]] = false):
            return
        if (set FlashEnabled[player[Player]] = true) {}
        spawn{StartFlashlight(Player)}

    StartFlashlight<private>(Player : agent)<suspends>:void=
        Toggled.Play(Player)
        var Notified : logic = false
        if (FlashlightData := PlayersToFlashes[Player][0], StrongFlashlightData := PlayersToFlashes[Player][1]):
            FlashlightData.TurnOff()
            StrongFlashlightData.TurnOn()
        # Start the flashlight with 100% in 15 seconds
        if (PlHud := StatsUiCon[GetSession()][Player]):
            loop:
                Sleep(0.2)
                if (set FlashCharges[player[Player]] -= 0.01):
                    if (info := FlashCharges[Player]):
                        # # Print(StrToMsg(?Agent:=Player, ?MessageAfter:=": {info}"))
                PlHud.FastUpdate()
                if (FlashCharges[Player] <= 0.0 or FlashEnabled[Player] = false or PublicHealthHud[GetSession()][0].MainClass.DefineState(player[Player]) = PlayerStates.Knocked, set FlashEnabled[player[Player]] = false):
                    break
                else if (FlashCharges[Player] <= 0.5):
                    if (Notified = false):
                        Timer.Play(Player)
                        set Notified = true
            Timer.Stop(Player)
            if (FlashlightData := PlayersToFlashes[Player][0], StrongFlashlightData := PlayersToFlashes[Player][1]):
                Toggled.Play(Player)
                FlashlightData.TurnOn()
                StrongFlashlightData.TurnOff()
            if (FlashCharges[Player] <= 0.0):
                Discharged.Play(Player)
                StrongFlashlightToggle.Unregister(Player)
                if (set PlayersParams[player[Player]].ImprovedFlashlightBattery.GrayedOut = true) {}
            PlHud.FastUpdate()
        if (pl := player[Player]):
            RestoreFlashlight(pl)

    RestoreFlashlight(Player : player)<suspends>:void=
        # Restore the flashlight with 10% in 3 seconds
        if (Charging[Player] = true):
            return
        if:
            set Charging[Player] = true
        if:
            PlHud := StatsUiCon[GetSession()][Player]
        then:
            loop:
                Sleep(0.3)
                if (FlashCharges[Player] >= 0.5):
                    if (set PlayersParams[Player].ImprovedFlashlightBattery.GrayedOut = false) {}
                    StrongFlashlightToggle.Register(Player)
                if (set FlashCharges[Player] += 0.01):
                    if (info := FlashCharges[Player]):
                        # # Print(StrToMsg(?Agent:=Player, ?MessageAfter:=": {info}"))
                PlHud.FastUpdate()
                if (PlayersParams[Player].ImprovedFlashlightBattery.Value >= 0):
                    if (FlashCharges[Player] >= 1.5 or FlashEnabled[Player] = true):
                        if (FlashCharges[Player] > 1.5, set FlashCharges[Player] = 1.5) {}
                        if (FlashCharges[Player] = 1.5):
                            Charged.Play(Player)
                        break
                else:
                    if (FlashCharges[Player] >= 1.0 or FlashEnabled[Player] = true):
                        if (FlashCharges[Player] > 1.0, set FlashCharges[Player] = 1.0) {}
                        if (FlashCharges[Player] = 1.0):
                            Charged.Play(Player)
                        break
        if:
            set Charging[Player] = false

    TurnOffFlashlight(Agent : ?agent):void=
        if (Player := player[Agent?], FlashlightData := PlayersToFlashes[Player][0], StrongFlashlightData := PlayersToFlashes[Player][1], set FlashEnabled[Player] = false):
            FlashlightData.TurnOff()
            StrongFlashlightData.TurnOff()
            FlashlightData.Disable()
            StrongFlashlightData.Disable()
            StrongFlashlightToggle.Unregister(Player)
        if (TravelDisabled = false):
            spawn{TravelWait()}
    
    TurnOnFlashlight(Agent : ?agent):void=
        if (Player := player[Agent?], FlashlightData := PlayersToFlashes[Player][0], StrongFlashlightData := PlayersToFlashes[Player][1]):
            FlashlightData.Enable()
            FlashlightData.TurnOn()
            StrongFlashlightData.Enable()
            StrongFlashlightData.TurnOff()
            StrongFlashlightToggle.Register(Player)

    TravelWait()<suspends>:void=
        Sleep(0.5)
        DisableTravel.Trigger()
        set TravelDisabled = true
        Notification.SetText(MessagesBank{}.TravelDisabled)
        for (i : Survivors, DEBUG_InfoHudVersion = Versions.V1):
            Notification.Hide(i)
            Notification.Show(i)
        # Sleep(60.0)
        set HiddenTravelInfo = 60
        for (i : Survivors, PlHud := StatsUiCon[GetSession()][i]):
            PlHud.FastUpdate()
        for (time := 1..60):
            Sleep(1.0)
            set HiddenTravelInfo -= 1
            for (i : Survivors, PlHud := StatsUiCon[GetSession()][i]):
                PlHud.FastUpdate()
        EnableTravel.Trigger()
        set TravelDisabled = false
        for (i : Survivors, PlHud := StatsUiCon[GetSession()][i]):
            PlHud.FastUpdate()
        Notification.SetText(MessagesBank{}.TravelEnabled)
        for (i : Survivors, DEBUG_InfoHudVersion = Versions.V1):
            Notification.Hide(i)
            Notification.Show(i)

    FlashlightStatus()<suspends> : void =
        loop:
            if (GameEnded = true):
                for (Player : Survivors):
                    P1_Flashlight.Hide(Player)
                    P2_Flashlight.Hide(Player)
                    P3_Flashlight.Hide(Player)
                    P4_Flashlight.Hide(Player)
                    P5_Flashlight.Hide(Player)
                    P6_Flashlight.Hide(Player)
                break
            for (Player : Survivors):
                if (Percent := FlashCharges[Player], plsmat := PlayersToMats[Player], ToShow := Round[Percent * 100]):
                    case (plsmat):
                        0 =>
                            # set P1FlashlightProgressbar.Percentage = Percent
                            P1_Flashlight.SetText(MessagesBank{}.ToMsg("{ToShow}%"))
                            P1_Flashlight.Hide(Player)
                            P1_Flashlight.Show(Player)
                        1 =>
                            # set P2FlashlightProgressbar.Percentage = Percent
                            P2_Flashlight.SetText(MessagesBank{}.ToMsg("{ToShow}%"))
                            P2_Flashlight.Hide(Player)
                            P2_Flashlight.Show(Player)
                        2 =>
                            # set P3FlashlightProgressbar.Percentage = Percent
                            P3_Flashlight.SetText(MessagesBank{}.ToMsg("{ToShow}%"))
                            P3_Flashlight.Hide(Player)
                            P3_Flashlight.Show(Player)
                        3 =>
                            # set P4FlashlightProgressbar.Percentage = Percent
                            P4_Flashlight.SetText(MessagesBank{}.ToMsg("{ToShow}%"))
                            P4_Flashlight.Hide(Player)
                            P4_Flashlight.Show(Player)
                        4 =>
                            # set P5FlashlightProgressbar.Percentage = Percent
                            P5_Flashlight.SetText(MessagesBank{}.ToMsg("{ToShow}%"))
                            P5_Flashlight.Hide(Player)
                            P5_Flashlight.Show(Player)
                        5 =>
                            # set P6FlashlightProgressbar.Percentage = Percent
                            P6_Flashlight.SetText(MessagesBank{}.ToMsg("{ToShow}%"))
                            P6_Flashlight.Hide(Player)
                            P6_Flashlight.Show(Player)
                        _ =>
            Sleep(0.1)


    ScarePlayers()<suspends>:void=
        loop:
            ToWait := GetRandomFloat(60.0, 180.0)
            # Print("Waiting for {ToWait} seconds to play a scary sound")
            Sleep(ToWait)
            if (snd := ScarySounds[GetRandomInt(0, ScarySounds.Length - 1)]):
                # Print("Playing scary sound")
                snd.Play()
    
    damaged(Agent : agent):void=
        if (Agent.GetFortCharacter[].IsVulnerable[]):
            DamageDealtAudio.Play(Agent)
            MonsterDmg.Disable()
            HpHud.ChangeClass(Agent)
            HpMarker.Attach(Agent)
        else:
            NoHpHud.ChangeClass(Agent)
            HpMarker.Detach(Agent)
        if (not Agent.GetFortCharacter[].IsActive[]):
            return
        # if (PlayersParams[Monster?].OneAndHalfDamage.Value > 0, frt_chr := Agent.GetFortCharacter[]):
        #     # if (frt_chr.GetHealth() = 50.0):
        #     #     frt_chr.SetHealth(0.0)
        #     # else:
        #     #     frt_chr.SetHealth(frt_chr.GetHealth() - 50.0)
        #     frt_chr.Damage(damage_args{
        #         Source := option{Monster?.GetFortCharacter[]}
        #         Amount := 0.5
        #     })

    MonsterInVent(Agent : agent):void=
        # Check if it's a survivor
        for (sur : Survivors, sur = Agent):
            set SurvivorsInVent += array{sur}
            spawn{MonsterInVentSus(Agent)}
            return
        set PlayerIsInVent = true
        spawn{MonsterInVentSus(Agent)}
    
    MonsterOutVent(Agent : agent):void=
        # Check if it's a survivor
        var temp : []player = array{}
        var plfound : logic = false
        for (sur : SurvivorsInVent):
            if (sur <> Agent):
                set temp += array{sur}
            else:
                set plfound = true
        set SurvivorsInVent = temp
        if (plfound = false):
            set PlayerIsInVent = false

    MonsterInVentSus(Agent : agent)<suspends>:void=
        VentSounds.Play()
        VentSounds.Register(Agent)
        loop:
            if (PlayerIsInVent = true, plpos := Agent.GetFortCharacter[].GetTransform(), VentSounds.TeleportTo[plpos]) {}
            else:
                VentSounds.Stop()
                break
            Sleep(0.1)

    Church(Agent : agent):void=
        # ChruchSFX.Register(Agent)
        ChurchAudios.Register(Agent)

    Unchurch(Agent : agent):void=
        # ChruchSFX.Unregister(Agent)
        ChurchAudios.Unregister(Agent)

    ChurchSus()<suspends>:void=
        loop:
            Sleep(GetRandomFloat(60.0, 3600.0))
            ChurchAudios.Play()

    MonsterAttackingV1(Agent : agent):void=
        spawn{MonsterAttackingV1Sus(Agent)}
    
    MonsterAttackingV1Sus(Agent : agent)<suspends>:void=
        AttackAud.Play()
        var AllHealth : [agent]float = map{}
        for (pl : GetPlayspace().GetPlayers()):
            if (pl.GetFortCharacter[].IsActive[], chr := pl.GetFortCharacter[], set AllHealth[pl] = chr.GetHealth()) {}
        MonsterAtk.Disable()
        MonsterDmg.Enable()
        Sleep(0.5)
        MonsterDmg.Disable()
        for (pl : GetPlayspace().GetPlayers()):
            if (pl.GetFortCharacter[].IsActive[]):
                if (AllHealth[pl] <> pl.GetFortCharacter[].GetHealth(), ToShow := "{AllHealth[pl]} -> {pl.GetFortCharacter[].GetHealth()}"):
                    # Print(StrToMsg(?MessageBefore:="Player ", ?Agent:=pl, ?MessageAfter:=" has been damaged: {ToShow}"))
        Sleep(0.5)
        MonsterAtk.Enable()

    MonsterAttackingV2(Agent : agent):void=
        spawn{MonsterAttackingV2Sus(Agent)}

    MonsterAttackingV2Sus(Agent : agent)<suspends>:void=
        MonsterAtk.Disable()
        AttackAud.Play()
        if (MonikaChr := Monster?.GetFortCharacter[]):
            var DmgInfo : damage_args = damage_args:
                Source := option{MonikaChr}
                Amount := block:
                    if (PlayersParams[Monster?].OneAndHalfDamage.Value >= 0):
                        1.5
                    else:
                        1.0
            RotationToPlayer := MonikaChr.GetViewRotation()
            if (MonsterDmgPerception.TeleportTo[MonikaChr.GetViewLocation(), RotationToPlayer]):
                if (MonikaChr.GetViewLocation() <> MonsterDmgPerception.GetTransform().Translation):
                    # Print("Ray not teleported")
                    MonsterDmgPerception.MoveTo(MonikaChr.GetViewLocation(), RotationToPlayer, 0.01)
                    # Sleep(0.05)
                    if (MonikaChr.GetViewLocation() <> MonsterDmgPerception.GetTransform().Translation):
                        # Print("Ray not moved at all")
                        return
                else:
                    # Print("Ray teleported")
                    Sleep(0.05)
                var ClosestAgent : ?agent = false
                var ClosestDistance : float = 0.0
                for (i : MonsterDmgPerception.GetPerceivedAgents(), MonsterDmgTrigger.GetAgentsInVolume().Find[i], chr := i.GetFortCharacter[], chr.IsVulnerable[]):
                    var Dist : float = Distance(chr.GetTransform().Translation, MonsterDmgPerception.GetTransform().Translation)
                    if ((ClosestDistance = 0.0 or Dist < ClosestDistance) and Dist < 512.0):
                        set ClosestDistance = Dist
                        set ClosestAgent = option{i}
                if (i := ClosestAgent?, chr := i.GetFortCharacter[]):
                    # Print(MessagesBank{}.ToMsg(i, "Hit registered"), ?Color := NamedColors.Green)
                    chr.Damage(DmgInfo)
                # else:
                #     Print("Hit not registered: {Dist}", ?Color := NamedColors.Red)
            # for (ag : MonsterDmgTrigger.GetAgentsInVolume(), ag <> Monster?, i := player[ag], chr := i.GetFortCharacter[], i.IsActive[], chr.IsActive[], not chr.IsDownButNotOut[]):
            #     var DirectionToPlayer : vector3 = chr.GetTransform().Translation - MonikaChr.GetTransform().Translation
            #     var RotationToPlayer : rotation = MakeRotation(DirectionToPlayer, 0.0)
            #     if (DmgRaycastStart.TeleportTo[MonikaChr.GetViewLocation(), RotationToPlayer]):
            #         if (MonikaChr.GetViewLocation() <> DmgRaycastStart.GetTransform().Translation):
            #             Print("Ray not teleported")
            #             DmgRaycastStart.MoveTo(MonikaChr.GetViewLocation(), RotationToPlayer, 0.01)
            #             Sleep(0.05)
            #             if (MonikaChr.GetViewLocation() <> DmgRaycastStart.GetTransform().Translation):
            #                 Print("Ray not moved at all")
            #         Dist := Distance(chr.GetTransform().Translation, DmgRaycastEnd.GetTransform().Translation)
            #         if (
            #             Dist <= 30.0,
            #             chr.IsVulnerable[]
            #         ):
            #             Print("Hit registered: {Dist}", ?Color := NamedColors.Green)
            #             chr.Damage(DmgInfo)
            #         else:
            #             Print("Hit not registered: {Dist}", ?Color := NamedColors.Red)
        Sleep(1.0)
        MonsterAtk.Enable()

    SurvivorsWin(?Reason : ?message = false):void=
        <# Reasons for winning:
        • The monster has left the game
        • All of the portals have been closed
        #>
        # Print("Survivors win")
        set Winner = "Survivors"
        if (var Rsn : message = Reason?):
            if (CurrentRound? <> roundnum.LastRound):
                # set Rsn += "\n\nThe next round will start in a moment"
                set Rsn = MessagesBank{}.Connect(Rsn, "\n\n", MessagesBank{}.NextRound)
            SurWin.SetDescriptionText(Rsn)
        SurWin.Show()
        DarkScreenEnd.Show()
        EndingMusic.Play()
        # for (i : GetPlayspace().GetPlayers()):
        #     DarkScreenAnim.PlayReverse(i)
        loop:
            for (i : Survivors, i.IsActive[]):
                DarkScreenAnim.PlayReverse(i)
                break
            break
        if (i := Monster?):
            DarkScreenAnim.PlayReverse(i)

    MonsterWin(?Reason : ?message = false):void=
        <# Reasons for winning:
        • The monster has eliminated all of the survivors
        • All of the survivors have left the game
        • Time has run out
        #>
        # Print("Monster wins")
        set Winner = "Monster"
        if (var Rsn : message = Reason?):
            if (CurrentRound? <> roundnum.LastRound):
                # set Rsn += "\n\nThe next round will start in a moment"
                set Rsn = MessagesBank{}.Connect(Rsn, "\n\n", MessagesBank{}.NextRound)
            MonWin.SetDescriptionText(Rsn)
        MonWin.Show()
        DarkScreenEnd.Show()
        EndingMusic.Play()
        # for (i : GetPlayspace().GetPlayers()):
        #     DarkScreenAnim.PlayReverse(i)
        loop:
            for (i : Survivors, i.IsActive[]):
                DarkScreenAnim.PlayReverse(i)
                break
            break
        if (i := Monster?):
            DarkScreenAnim.PlayReverse(i)

    Canceled(?Reason : ?message = false):void=
        # Print("Game canceled")
        set Winner = "Canceled"
        if (Rsn := Reason?):
            Cancel.SetDescriptionText(Rsn)
        Cancel.Show()
        DarkScreenEnd.Show()
        EndingMusic.Play()
        # for (i : GetPlayspace().GetPlayers()):
        #     DarkScreenAnim.PlayReverse(i)
        loop:
            for (i : Survivors, i.IsActive[]):
                DarkScreenAnim.PlayReverse(i)
                break
            break
        if (i := Monster?):
            DarkScreenAnim.PlayReverse(i)

    Timeout()<suspends>:void=
        Sleep(60.0)
        for (pl : GetPlayspace().GetPlayers()):
            if (pl.IsActive[]):
                pl.SendToLobby()

    SendBackToLobby(Agent : agent, num : int):void=
        spawn{SendBackToLobbySus(Agent)}

    SendBackToLobbySus(Agent : agent)<suspends>:void=
        loop:
            if (player[Agent].SendToLobby()) {}
            else:
                # Print("Couldn\'t get player to send to lobby")
            if (not player[Agent].IsActive[]):
                # Print("Player sent to lobby")
                return
            else:
                # Print("Couldn\'t send player to lobby")
            Sleep(0.1)

    FinVFXSus(Agent:agent)<suspends>:void=
        if (EndingVFX.TeleportTo[MonsterObj.GetTransform().Translation, MonsterObj.GetTransform().Rotation]):
        set EndingVFXStart = true
        Sleep(0.02)
        AttackCon.Disable()
        # Freeze monster and lift it up
        freeze_info := stasis_args{
            AllowTurning := false
            AllowFalling := false
            AllowEmotes := false
        }
        MonsterObj.Pause()
        if (Pl := Monster?, chr := Pl.GetFortCharacter[]):
            # chr.PutInStasis(freeze_info)
            var new_pos : vector3 = chr.GetTransform().Translation
            set new_pos.Z += 100.0
            EndingVFX.Enable()
            EndingVFX.Restart()
            spawn:
                EndingVFX.MoveTo(new_pos, MonsterObj.GetTransform().Rotation, 5.0)
            MonsterObj.MoveTo(new_pos, MonsterObj.GetTransform().Rotation, 5.0)
        Winning{
            MainClass := Self
            Reason := MessagesBank{}.AllPortalsClosed
            Feedback := Feedback_S_AllPortalsClosed
        }.SurvivorsWin(Agent)

    WaitForMonster(Mon : agent)<suspends>:void=
        if (chr := Mon.GetFortCharacter[]):
            startloc := chr.GetTransform().Translation
            # Sleep(120.0)
            for (i := 0..900):
                if (chr.GetTransform().Translation <> startloc):
                    # Print("Monster has moved")
                    set MonsterAFK = false
                    return
                Sleep(0.1)
            if (EndingVFX.TeleportTo[MonsterObj.GetTransform().Translation, MonsterObj.GetTransform().Rotation]):
                set EndingVFXStart = true
            EndingVFX.Enable()
            EndingVFX.Restart()
            Winning{
                MainClass := Self
                Reason := MessagesBank{}.MonsterAFK
                Feedback := Feedback_S_MonsterAFK
            }.SurvivorsWin(Mon)
            # if:
            #     player[Mon].SendToLobby()
            # Print("Monster AFK")