using { /Verse.org }
using { /Verse.org/Native }
using { /Verse.org/SceneGraph }
using { /Verse.org/Simulation }
using { /Fortnite.com/Devices }
using { /UnrealEngine.com/Temporary/SpatialMath }
using { /Verse.org/Colors }
using { /Fortnite.com/Characters }

task_type := enum:
    Transmitter
    Pumpkin
    Portal

task_state := enum:
    Complete
    Failed
    InProgress
    NotStarted

SG_NotPlaced<public> := class<final_super>(component) {}
SG_Placed<public> := class<final_super>(component) {}

# A Verse-authored component that can be added to entities
SG_GeneralTask<public> := class<final_super>(component):

    @editable
    TaskType : task_type = task_type.Transmitter

    var IsFake : logic = false

    var Start : logic = false

    var Marker : beacon_device = beacon_device{}
    var FailMarker : beacon_device = beacon_device{}
    var Objective : tracker_device = tracker_device{}

    var ProcessInteraction : skilled_interaction_device = skilled_interaction_device{}

    var TaskManager : TasksHandler = TasksHandler{}
    var AbilitiesManager : ?AbilitiesHandler = false

    var TaskState : task_state = task_state.NotStarted

    # OnBeginSimulation<override>():void =
    #     (super:)OnBeginSimulation()
    #     Print("OnBeginSimulation")

    OnSimulate<override>()<suspends>:void =
        (super:)OnSimulate()
        loop:
            if (Start?):
                break
            Sleep(1.0)
        # Print("All variables are set, starting simulation")
        var MarkerPos : transform = FromTransform(Entity.GetGlobalTransform())
        if (TaskType = task_type.Portal):
            # set MarkerPos.Translation.Z += 0.0
        else:
            set MarkerPos.Translation.Z += 20.0
        if:
            Marker.TeleportTo[MarkerPos]
            FailMarker.TeleportTo[MarkerPos]
        then:
            Marker.Enable()
        if (IsFake?):
            Print("This is a fake task, removing the completed entity")
            for (i : Entity.GetEntities()):
                if (i.GetComponent[SG_Placed]):
                    i.RemoveFromParent()
        ShowReverted()

    StartingTask(Agent : agent) : void =
        if (TaskState = task_state.InProgress):
            return
        if (IsFake?):
            Print("This is a fake task, not starting it")
            spawn{TaskFaked(Agent)}
            HideAll()
            return
        else if:
            TaskState <> task_state.Failed
            TaskManager.PlayersToDown[Agent] = false
            ProcessInteraction.BeginInteraction[Agent]
            set TaskManager.PlayersToNewTasks[Agent] = option{Self}
            # Button := Entity.GetComponent[SG_InteractionWithTask]
        then:
            set TaskState = task_state.InProgress
            # Button.Disable()

    TaskFaked(Agent:agent)<suspends>:void=
        Print("Fake task activated")
        Marker.Disable()
        FailMarker.Enable()
        if (Ab := AbilitiesManager?):
            Ab.FakeV2Triggered(Agent)
        Sleep(5.0)
        FailMarker.Disable()
        set TaskState = task_state.Complete
        if (Par := Entity.GetParent[]):
            Par.RemoveFromParent()
        else:
            Entity.RemoveFromParent()

    TaskFailed(Agent : agent) : void =
        Print("Player failed a task")
        if (set TaskManager.PlayersToNewTasks[Agent] = false) {}
        set TaskState = task_state.Failed
        FailMarker.Enable()
        spawn{TaskFailedkSus()}
    
    TaskFailedkSus()<suspends>:void=
        Sleep(5.0)
        FailMarker.Disable()
        Sleep(5.0)
        set TaskState = task_state.NotStarted
        if (Button := Entity.GetComponent[SG_InteractionWithTask]):
            Button.Enable()

    TaskCompleted(Agent : agent):void=
        Marker.Disable()
        Objective.Increment(Agent)
        if (TaskManager.PlayersToDown[Agent] = true):
            return
        ShowCompleted()
        set TaskState = task_state.Complete
        if (Objective.GetValue() = Objective.GetTarget()):
            TaskManager.CheckWin(Agent)

    ShowCompleted():void=
        for (i : Entity.GetEntities()):
            if (i.GetComponent[SG_Placed]):
                # Print("Showing completed")
                for (j : i.FindDescendantEntities(entity), NewAsset := j.GetComponent[mesh_component]):
                    NewAsset.Enable()
                    set NewAsset.Visible = true
                    set NewAsset.Collidable = true
            else if (i.GetComponent[SG_NotPlaced]):
                # Print("Removing not completed")
                for (j : i.FindDescendantEntities(entity), NewAsset := j.GetComponent[mesh_component]):
                    NewAsset.Disable()
                    set NewAsset.Visible = false

    ShowReverted():void=
        for (i : Entity.GetEntities()):
            if (i.GetComponent[SG_Placed]):
                # Print("Removing completed")
                for (j : i.FindDescendantEntities(entity), NewAsset := j.GetComponent[mesh_component]):
                    NewAsset.Disable()
                    set NewAsset.Visible = false
                    set NewAsset.Collidable = false
            else if (i.GetComponent[SG_NotPlaced]):
                # Print("Showing not completed")
                for (j : i.FindDescendantEntities(entity), NewAsset := j.GetComponent[mesh_component]):
                    NewAsset.Enable()
                    set NewAsset.Visible = true

    HideAll():void=
        for (i : Entity.GetEntities()):
            if (i.GetComponent[SG_Placed]):
                # Print("Hiding completed")
                for (j : i.FindDescendantEntities(entity), NewAsset := j.GetComponent[mesh_component]):
                    NewAsset.Disable()
                    set NewAsset.Visible = false
                    set NewAsset.Collidable = false
            else if (i.GetComponent[SG_NotPlaced]):
                # Print("Hiding not completed")
                for (j : i.FindDescendantEntities(entity), NewAsset := j.GetComponent[mesh_component]):
                    NewAsset.Disable()
                    set NewAsset.Visible = false

    TaskIgnored():void=
        set TaskState = task_state.Complete
        Entity.RemoveFromParent()

    IsComplete()<decides><transacts>:void=
        TaskState = task_state.Complete


SG_PortalTask<public> := class<final_super>(SG_GeneralTask):

    TaskType<override> : task_type = task_type.Portal

    @editable
    var MonsterMarker : beacon_device = beacon_device{}
    @editable
    CatchVolume : volume_device = volume_device{}
    @editable
    SaveButton : button_device = button_device{}
    @editable
    SaveTeleport : teleporter_device = teleporter_device{}

    var PortalButton : button_device = button_device{}

    var IsClosed : logic = false
    var WaitingToRescue : logic = false
    var ForeverClosed : logic = false

    var InZone<private> : []agent = array{}
    var PickedUp<private> : []agent = array{}

    var LifesManager : ?LifeHandler = false

    OnSimulate<override>()<suspends>:void =
        if (TaskType <> task_type.Portal):
            Print("Portal task type mismatch, crashing simulation", ?Color := NamedColors.Red)
            return
        (super:)OnSimulate()
        if:
            MonsterMarker.TeleportTo[FromTransform(Entity.GetGlobalTransform())]
            PortalButton.TeleportTo[FromTransform(Entity.GetGlobalTransform())]
        MonsterMarker.Enable()
        PortalButton.SetInteractionTime(0.0)
        PortalButton.SetInteractionText(MessagesBank{}.PlacePlanks)
        CatchVolume.AgentEntersEvent.Subscribe(PlayerEnteredZone)
        CatchVolume.AgentExitsEvent.Subscribe(PlayerLeftZone)
        PortalButton.InteractedWithEvent.Subscribe(StartingTask)
        SaveButton.InteractedWithEvent.Subscribe(Rescue)
        SaveButton.Disable()
        PortalButton.Enable()
        loop:
            if (LM := LifesManager?):
                LM.KnockDetect.AgentPickedUpEvent.Subscribe(Picked)
                LM.KnockDetect.AgentDroppedEvent.Subscribe(Dropped)
                break
            Sleep(1.0)

    PlayerEnteredZone(Agent : agent):void=
        set InZone += array{Agent}
        if (not IsClosed? and not PickedUp.Find[Agent], LM := LifesManager?):
            LM.PlayerEnteredBasement(Agent)

    PlayerLeftZone(Agent : agent):void=
        set InZone = InZone.RemoveAllElements(Agent)

    Rescue(Agent : agent):void=
        if (LM := LifesManager?):
            LM.Achievments.Award(Agent, 4)
            LM.Achievments.SaveAchievement(Agent, PossibleAchievements.PlayersSaved, 1)
            LM.PlayerSaved(SaveTeleport)

    Picked(Agent:agent):void=
        set PickedUp += array{Agent}
        # Print("Player picked up")

    Dropped(Agent:agent):void=
        set PickedUp = PickedUp.RemoveAllElements(Agent)
        if (not IsClosed? and InZone.Find[Agent], LM := LifesManager?):
            LM.PlayerEnteredBasement(Agent)

    StartingTask<override>(Agent:agent):void=
        if (not IsClosed? and WaitingToRescue?):
            return
        PortalButton.Disable()
        (super:)StartingTask(Agent)

    TaskCompleted<override>(Agent : agent):void=
        if (IsClosed?):
            set IsClosed = false
            Objective.Decrement(Agent)
            MonsterMarker.Enable()
            PortalButton.SetInteractionText(MessagesBank{}.PlacePlanks)
            set TaskState = task_state.NotStarted
            for (i : Entity.FindDescendantEntitiesWithComponent(SG_NotPlaced), SoundPlayer := i.GetComponent[sound_component]):
                SoundPlayer.Play()
            if (Button := Entity.GetComponent[SG_InteractionWithTask]):
                if (not WaitingToRescue?):
                    ShowReverted()
                    # Button.Enable()
                    PortalButton.Enable()
                else:
                    PortalButton.Disable()
                    SaveButton.Enable()
                    HideAll()
        else:
            set IsClosed = true
            Objective.Increment(Agent)
            MonsterMarker.Disable()
            PortalButton.SetInteractionText(MessagesBank{}.RemovePlanks)
            ShowCompleted()
            set TaskState = task_state.Complete
            # if (Button := Entity.GetComponent[SG_InteractionWithTask]):
            #     Button.Enable()
            PortalButton.Enable()
            for (i : Entity.FindDescendantEntitiesWithComponent(SG_Placed), SoundPlayer := i.GetComponent[sound_component]):
                SoundPlayer.Play()
        if (Objective.GetValue() = Objective.GetTarget()):
            TaskManager.CheckWin(Agent)

    FirstPlayerCaught():void=
        if (not IsClosed?):
            # if (Button := Entity.GetComponent[SG_InteractionWithTask]):
            #     Button.Disable()
            PortalButton.Disable()
            SaveButton.Enable()
            HideAll()
        set WaitingToRescue = true

    AllPlayersSaved():void=
        set WaitingToRescue = false
        if (not IsClosed?):
            SaveButton.Disable()
            # if (Button := Entity.GetComponent[SG_InteractionWithTask]):
            #     Button.Enable()
            PortalButton.Enable()
            ShowReverted()

    AllTasksDone():void=
        set ForeverClosed = true
        Marker.Disable()
        SaveButton.Disable()
        # if (Button := Entity.GetComponent[SG_InteractionWithTask]):
        #     Button.Disable()
        PortalButton.Disable()

    TaskFailed<override>(Agent : agent):void=
        PortalButton.Disable()
        (super:)TaskFailed(Agent)

    TaskFailedkSus<override>()<suspends>:void=
        MonsterMarker.Disable()
        spawn{(super:)TaskFailedkSus()}
        Sleep(10.0)
        PortalButton.Enable()
        if (not IsClosed?):
            MonsterMarker.Enable()


SG_InteractionWithTask<public> := class<final_super>(interactable_component):

    Empty<localizes><private> : message = ""
    Error<localizes><private> : message = "Error"
    SetUpTransmitter<localizes><private> : message = "Set up the transmitter"
    PlacePumpkin<localizes><private> : message = "Place the pumpkin"
    PlacePlanks<localizes><private> : message = "Place planks"
    RemovePlanks<localizes><private> : message = "Remove the planks"
    TaskCompleted<localizes><private> : message = "Task completed"
    TaskFailed<localizes><private> : message = "Task failed. Try again in 10 seconds.\nThe monster knows your location!"
    PlayerFailedTask<localizes><private> : message = "A player has failed a task!"

    BeingCompleted<localizes><private> : message = "This task is being completed by another player"
    BeingRestored<localizes><private> : message = "Task on cooldown"

    var TaskInfo : ?SG_GeneralTask = false

    var TaskManager : TasksHandler = TasksHandler{}

    # Runs when the component should start simulating in a running game.
    OnBeginSimulation<override>():void =
        # Run OnBeginSimulation from the parent class before
        # running this component's OnBeginSimulation logic
        (super:)OnBeginSimulation()

    OnSimulate<override>()<suspends>:void =
        if (TempPortalInfo := Entity.GetComponent[SG_PortalTask]):
            # Print("PortalInfo found, setting TaskType to Portal")
            loop:
                if (TempPortalInfo.Start?):
                    # Set the TaskType based on the TaskInfo component
                    break
                Sleep(1.0)
            # Print("PortalInfo found, setting TaskType")
            # set TaskType = task_type.Portal
            set TaskInfo = option{TempPortalInfo}
            set TaskManager = TempPortalInfo.TaskManager
            Enable()
            SucceededEvent.Subscribe(TempPortalInfo.StartingTask)
        else if (TempTaskInfo := Entity.GetComponent[SG_GeneralTask]):
            loop:
                if (TempTaskInfo.Start?):
                    # Set the TaskType based on the TaskInfo component
                    break
                Sleep(1.0)
            # Print("TaskInfo found, setting TaskType")
            # set TaskType = TaskInfo.TaskType
            set TaskInfo = option{TempTaskInfo}
            set TaskManager = TempTaskInfo.TaskManager
            Enable()
            SucceededEvent.Subscribe(TempTaskInfo.StartingTask)
        else:
            Print("No TaskInfo found, exiting simulation", ?Color := NamedColors.Red)
            # Disable()
            return
        if (set TaskInfo?.TaskState = task_state.NotStarted) {}

    CanInteract<override>(Agent:agent)<decides><reads> : void=
        if (Agent = TaskManager.Roles.Monster?):
            false?
        else if (TaskInfo?.TaskType = task_type.Portal):
            false?
            if ((not SG_PortalTask[TaskInfo?].IsClosed? and SG_PortalTask[TaskInfo?].WaitingToRescue?) or SG_PortalTask[TaskInfo?].ForeverClosed?):
                false?
            # else:
            #     true?
        else if (TaskInfo?.TaskState = task_state.Complete):
            false?
        else:
            true?

    IsEnabled<override>()<transacts><decides>:void=
        # if (TaskInfo?.TaskType = task_type.Portal, (not SG_PortalTask[TaskInfo?].IsClosed? and SG_PortalTask[TaskInfo?].WaitingToRescue?) or SG_PortalTask[TaskInfo?].ForeverClosed?):
        #     false
        if (TaskInfo?.TaskType = task_type.Portal):
            false
        TaskInfo?.TaskState <> task_state.Complete and TaskInfo?.TaskState <> task_state.Failed

    InteractMessage<override>(Agent:agent)<decides><reads> : message =
        if (Agent = TaskManager.Roles.Monster?):
            false
            return Empty
        else if (TaskInfo?.TaskType = task_type.Portal):
            false
            return Empty
        else if (TaskInfo?.TaskType = task_type.Portal and ((not SG_PortalTask[TaskInfo?].IsClosed? and SG_PortalTask[TaskInfo?].WaitingToRescue?) or SG_PortalTask[TaskInfo?].ForeverClosed?)):
            false
            return Empty
        else:
            case (TaskInfo?.TaskState):
                task_state.Complete =>
                    if (TaskInfo?.TaskType = task_type.Portal):
                        # Check if all tasks are done
                        if (PortalInfo := SG_PortalTask[TaskInfo?].ForeverClosed?):
                            return TaskCompleted
                        else:
                            return RemovePlanks
                    else:
                        false
                        return Empty
                task_state.Failed =>
                    return BeingRestored
                task_state.InProgress =>
                    if (TaskManager.PlayersToNewTasks[Agent]? = Self):
                        false
                        return Empty
                    else:
                        return BeingCompleted
                task_state.NotStarted =>
                    case (TaskInfo?.TaskType):
                        task_type.Transmitter =>
                            return SetUpTransmitter
                        task_type.Pumpkin =>
                            return PlacePumpkin
                        task_type.Portal =>
                            return PlacePlanks