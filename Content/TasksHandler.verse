using { /Fortnite.com/Devices }
using { /Fortnite.com/UI }
using { /UnrealEngine.com/Temporary/Diagnostics }
using { /UnrealEngine.com/Temporary/SpatialMath }
using { /UnrealEngine.com/Temporary/UI }
using { /Verse.org/Assets }
using { /Verse.org/Colors }
using { /Verse.org/Random }
using { /Verse.org/Simulation }
using { /Verse.org/Simulation/Tags }
using { /Verse.org/SceneGraph }
using { /Verse.org/SpatialMath }


var PlayersTasksUis : weak_map(session, [player]TasksUI) = map{}

TasksUI := class:
    MainClass : TasksHandler
    Player : player
    PlUi : player_ui

    var Canvas : canvas = canvas{}

    Update():void=
        PlUi.RemoveWidget(Canvas)
        set Canvas = canvas:
            Slots := array:
                canvas_slot:
                    Anchors := anchors:
                        Minimum := vector2{X:=0.0,Y:=0.1}
                        Maximum := vector2{X:=0.0,Y:=0.1}
                    Offsets := margin:
                        Left := 20.0
                        Top := 0.0
                        Right := 0.0
                        Bottom := 0.0
                    SizeToContent := true
                    Widget := stack_box:
                        Orientation := orientation.Vertical
                        Slots := block:
                            array:
                                block:
                                    if (MainClass.TransmittersObjective.GetValue() <> MainClass.TransmittersObjective.GetTarget()):
                                        GetTask( # Transmitters
                                            MainClass.TasksIcons.TransmitterTaskIcon,
                                            MessagesBank{}.TransmittersPlacedInfo,
                                            MainClass.TransmittersObjective.GetValue(),
                                            MainClass.TransmittersObjective.GetTarget(),
                                            color{
                                                R:=0.3,
                                                G:=0.0,
                                                B:=0.65
                                            },
                                            color{
                                                R:=0.2,
                                                G:=0.0,
                                                B:=0.4
                                            }
                                        )
                                    else:
                                        stack_box_slot:
                                            Widget := color_block:
                                                DefaultDesiredSize := vector2{X:=0.0,Y:=0.0}
                                block:
                                    if (MainClass.PumpkinsObjective.GetValue() <> MainClass.PumpkinsObjective.GetTarget()):
                                        GetTask( # Pumpkins
                                            MainClass.TasksIcons.PumpkinTaskIcon,
                                            MessagesBank{}.PumpkinsPlacedInfo,
                                            MainClass.PumpkinsObjective.GetValue(),
                                            MainClass.PumpkinsObjective.GetTarget(),
                                            color{
                                                R:=0.65,
                                                G:=0.3,
                                                B:=0.0
                                            },
                                            color{
                                                R:=0.6,
                                                G:=0.2,
                                                B:=0.0
                                            }
                                        )
                                    else:
                                        stack_box_slot:
                                            Widget := color_block:
                                                DefaultDesiredSize := vector2{X:=0.0,Y:=0.0}
                                block:
                                    if (
                                        (MainClass.PumpkinsObjective.GetValue() <> MainClass.PumpkinsObjective.GetTarget() or MainClass.TransmittersObjective.GetValue() <> MainClass.TransmittersObjective.GetTarget()) or MainClass.PortalsObjective.GetValue() <> MainClass.PortalsObjective.GetTarget()
                                    ):
                                        IconForPortal := block:
                                            if (MainClass.Roles.Monster? = Player):
                                                MainClass.TasksIcons.MonsterPortalTaskIcon
                                            else:
                                                MainClass.TasksIcons.PortalTaskIcon
                                        GetTask( # Portals
                                            IconForPortal,
                                            MessagesBank{}.PortalsClosedInfo,
                                            MainClass.PortalsObjective.GetValue(),
                                            MainClass.PortalsObjective.GetTarget(),
                                            color{
                                                R:=0.0,
                                                G:=0.3,
                                                B:=0.65
                                            },
                                            color{
                                                R:=0.0,
                                                G:=0.2,
                                                B:=0.65
                                            }
                                        )
                                    else:
                                        stack_box_slot:
                                            Widget := color_block:
                                                DefaultDesiredSize := vector2{X:=0.0,Y:=0.0}
                                block:
                                    if (MainClass.PortalsObjective.GetValue() = MainClass.PortalsObjective.GetTarget() and MainClass.TransmittersObjective.GetValue() = MainClass.TransmittersObjective.GetTarget() and MainClass.PumpkinsObjective.GetValue() = MainClass.PumpkinsObjective.GetTarget()):
                                        GetTask( # Final
                                            MainClass.TasksIcons.FinalTaskIcon,
                                            MessagesBank{}.FinalObjectiveInfo,
                                            MainClass.FinalObjective.GetValue(),
                                            MainClass.FinalObjective.GetTarget(),
                                            color{
                                                R:=0.65,
                                                G:=0.0,
                                                B:=0.0
                                            },
                                            color{
                                                R:=0.5,
                                                G:=0.0,
                                                B:=0.0
                                            }
                                        )
                                    else:
                                        stack_box_slot:
                                            Widget := color_block:
                                                DefaultDesiredSize := vector2{X:=0.0,Y:=0.0}

        PlUi.AddWidget(Canvas)


    GetTask(Icon : texture, Text : message, Progress : int, Required : int, IconColor : color, BackplateColor : color):stack_box_slot=
        return stack_box_slot:
            HorizontalAlignment := horizontal_alignment.Fill
            Padding := margin:
                Bottom := 10.0
            Widget := overlay:
                Slots := array:
                    overlay_slot:
                        Padding := margin:
                            Left := 20.0
                        HorizontalAlignment := horizontal_alignment.Fill
                        VerticalAlignment := vertical_alignment.Center
                        Widget := texture_block:
                            DefaultImage := UI.PlayerLineBG
                            DefaultDesiredSize := vector2{X:=500.0,Y:=50.0}
                            DefaultTint := block:
                                if (MainClass.Roles.Monster? <> Player):
                                    BackplateColor
                                else:
                                    color{R:=0.2,G:=0.0,B:=0.4}
                    overlay_slot:
                        HorizontalAlignment := horizontal_alignment.Left
                        VerticalAlignment := vertical_alignment.Center
                        Widget := overlay:
                            Slots := array:
                                overlay_slot:
                                    HorizontalAlignment := horizontal_alignment.Center
                                    VerticalAlignment := vertical_alignment.Center
                                    Widget := texture_block:
                                        DefaultImage := UI.PlayerLineCircle
                                        DefaultDesiredSize := vector2{X:=60.0,Y:=60.0}
                                        DefaultTint := block:
                                            if (MainClass.Roles.Monster? <> Player):
                                                IconColor
                                            else:
                                                color{R:=0.3,G:=0.0,B:=0.65}
                                overlay_slot:
                                    HorizontalAlignment := horizontal_alignment.Center
                                    VerticalAlignment := vertical_alignment.Center
                                    Widget := texture_block:
                                        DefaultImage := Icon
                                        DefaultDesiredSize := vector2{X:=50.0,Y:=50.0}
                    overlay_slot:
                        Padding := margin:
                            Left := 70.0
                            Right := 125.0
                        HorizontalAlignment := horizontal_alignment.Left
                        VerticalAlignment := vertical_alignment.Center
                        Widget := text_block:
                            DefaultText := Text
                            DefaultTextColor := color{R:=1.0,G:=1.0,B:=1.0}
                    overlay_slot:
                        Padding := margin:
                            Right := 50.0
                        HorizontalAlignment := horizontal_alignment.Right
                        VerticalAlignment := vertical_alignment.Center
                        Widget := text_block:
                            DefaultText := MessagesBank{}.QuestProgress(Progress, Required)
                            DefaultTextColor := color{R:=1.0,G:=1.0,B:=1.0}
                        # Widget := block:
                        #     if (Progress <> Required):
                        #         text_block:
                        #             DefaultText := MessagesBank{}.QuestProgress(Progress, Required)
                        #             DefaultTextColor := color{R:=1.0,G:=1.0,B:=1.0}
                        #     else:
                        #         texture_block:
                        #             DefaultImage := Icons.IconCheckmark
                        #             DefaultDesiredSize := vector2{X:=40.0,Y:=40.0}
                        #             DefaultTint := color{R:=0.0,G:=0.6,B:=0.0}


TransmitterTaskConnection := class:
    WaitingTransmitter : creative_prop
    ReadyTransmitter : creative_prop_asset
    Marker : beacon_device
    FailMarker : beacon_device
    Button : button_device
    Obj : tracker_device
    MonObj : tracker_device

    MainClass : TasksHandler

    Start() : void =
        if:
            Button.TeleportTo[WaitingTransmitter.GetTransform()]
            Marker.TeleportTo[WaitingTransmitter.GetTransform()]
            FailMarker.TeleportTo[WaitingTransmitter.GetTransform()]
        then:
            # Button.SetInteractionText(Msg("Set up transmitter"))
            Button.SetInteractionText(MessagesBank{}.SetUpTransmitter)
            Button.Enable()
            Marker.Enable()
            Button.InteractedWithEvent.Subscribe(CompletingTask)

    CompletingTask(Agent : agent) : void =
        if:
            MainClass.PlayersToDown[Agent] = false
            MainClass.Transmitter_TaskInteraction.BeginInteraction[Agent]
            set MainClass.PlayersToTransmitterTasks[Agent] = option{Self}
        then:
            Button.Disable()

    TaskCompleted(Agent : agent) : void =
        # Print("Connection recieved the completion")
        if (set MainClass.PlayersToTransmitterTasks[Agent] = false) {}
        # Button.Disable()
        Marker.Disable()
        ToSetUp := WaitingTransmitter.GetTransform()
        WaitingTransmitter.Dispose()
        SpawnProp(ReadyTransmitter, ToSetUp)
        Obj.Increment(Agent)
        if (Obj.GetValue() = Obj.GetTarget()):
            for (AgentToRemove -> g : MainClass.PlayersToTransmitterTasks, MainClass.Transmitter_TaskInteraction.EndInteraction[AgentToRemove]) {}
            set MainClass.PlayersToTransmitterTasks = map{}
        if (pl := MainClass.Roles.Monster?):
            MonObj.Increment(pl)
        if (Obj.GetValue() = Obj.GetTarget()):
            MainClass.Achievements.Award(Agent, 6)
        else:
            MainClass.Achievements.Award(Agent, 5)
        MainClass.Achievements.SaveAchievement(Agent, PossibleAchievements.CompletedQuests, 1)
        if (Obj.GetValue() = Obj.GetTarget()):
            MainClass.CheckWin(Agent)

    FailTask(Agent : agent) : void =
        # Print("Transmitter connection recieved the fail")
        if (set MainClass.PlayersToTransmitterTasks[Agent] = false) {}
        FailMarker.Enable()
        spawn{FailTaskSus()}
    
    FailTaskSus()<suspends>:void=
        Sleep(5.0)
        FailMarker.Disable()
        Sleep(5.0)
        # Print("Transmitter task enabled back")
        Button.Enable()
        Button.Enable()
        Button.Enable()

PumpkinTaskConnection := class:
    WaitingPumpkin : creative_prop
    ReadyPumpkin : creative_prop_asset
    Marker : beacon_device
    FailMarker : beacon_device
    Button : button_device
    Obj : tracker_device
    MonObj : tracker_device

    MainClass : TasksHandler

    Start() : void =
        if:
            Button.TeleportTo[WaitingPumpkin.GetTransform()]
            Marker.TeleportTo[WaitingPumpkin.GetTransform()]
            FailMarker.TeleportTo[WaitingPumpkin.GetTransform()]
        then:
            # Button.SetInteractionText(Msg("Place pumpkin"))
            Button.SetInteractionText(MessagesBank{}.PlacePumpkin)
            Button.Enable()
            Marker.Enable()
            Button.InteractedWithEvent.Subscribe(CompletingTask)

    CompletingTask(Agent : agent) : void =
        if:
            MainClass.PlayersToDown[Agent] = false
            MainClass.Pumpkin_TaskInteraction.BeginInteraction[Agent]
            set MainClass.PlayersToPumpkinTasks[Agent] = option{Self}
        then:
            Button.Disable()

    TaskCompleted(Agent : agent) : void =
        # Print("Pumpkin connection recieved the completion")
        if (set MainClass.PlayersToPumpkinTasks[Agent] = false) {}
        # Button.Disable()
        Marker.Disable()
        ToSetUp := WaitingPumpkin.GetTransform()
        WaitingPumpkin.Dispose()
        SpawnProp(ReadyPumpkin, ToSetUp)
        Obj.Increment(Agent)
        if (Obj.GetValue() = Obj.GetTarget()):
            for (AgentToRemove -> g : MainClass.PlayersToPumpkinTasks, MainClass.Pumpkin_TaskInteraction.EndInteraction[AgentToRemove]) {}
            set MainClass.PlayersToPumpkinTasks = map{}
        if (pl := MainClass.Roles.Monster?):
            MonObj.Increment(pl)
        if (Obj.GetValue() = Obj.GetTarget()):
            MainClass.Achievements.Award(Agent, 6)
        else:
            MainClass.Achievements.Award(Agent, 5)
        MainClass.Achievements.SaveAchievement(Agent, PossibleAchievements.CompletedQuests, 1)
        if (Obj.GetValue() = Obj.GetTarget()):
            MainClass.CheckWin(Agent)

    FailTask(Agent : agent) : void =
        # Print("Connection recieved the fail")
        if (set MainClass.PlayersToPumpkinTasks[Agent] = false) {}
        FailMarker.Enable()
        spawn{FailTaskSus()}

    FailTaskSus()<suspends>:void=
        Sleep(5.0)
        FailMarker.Disable()
        Sleep(5.0)
        # Print("Pumpkin task enabled back")
        Button.Enable()

PortalTaskConnection := class:
    WaitingPortal : button_device
    PlanksReady : creative_prop_asset
    PlanksNotReady : creative_prop_asset
    Marker : beacon_device
    MonMarker : beacon_device
    FailMarker : beacon_device
    Button : button_device
    Obj : tracker_device
    MonObj : tracker_device

    PlacingSound : audio_player_device
    RemovingSound : audio_player_device

    MainClass : TasksHandler

    var IsBlocked : logic = false
    var WaitingToRescue : logic = false
    var PlanksProp : creative_prop = creative_prop{}
    var NonPlanksProp : creative_prop = creative_prop{}
    var ForeverClosed : logic = false
    var AchTriggered : logic = false

    Start() : void =
        if:
            var PortalPos : (/UnrealEngine.com/Temporary/SpatialMath:)transform = WaitingPortal.GetTransform()
            Button.TeleportTo[WaitingPortal.GetTransform()]
            PlacingSound.TeleportTo[WaitingPortal.GetTransform()]
            RemovingSound.TeleportTo[WaitingPortal.GetTransform()]
            var NewPortalPos : (/UnrealEngine.com/Temporary/SpatialMath:)transform = PortalPos
            set NewPortalPos.Translation.Z -= 176.0
            Marker.TeleportTo[NewPortalPos]
            MonMarker.TeleportTo[NewPortalPos]
            set NewPortalPos.Translation.Z += 100.0
            FailMarker.TeleportTo[NewPortalPos]
        then:
            # Button.SetInteractionText(Msg("Place planks"))
            Button.SetInteractionText(MessagesBank{}.PlacePlanks)
            Button.Enable()
            Marker.Enable()
            MonMarker.Enable()
        if (MainClass.PortalsVersion = Versions.V1):
            Button.InteractedWithEvent.Subscribe(CheckState)
        if (MainClass.PortalsVersion = Versions.V2):
            Button.SetInteractionTime(0.0)
            Button.InteractedWithEvent.Subscribe(CompletingTask)
        var ToSetUp : (/UnrealEngine.com/Temporary/SpatialMath:)vector3 = WaitingPortal.GetTransform().Translation
        set ToSetUp.Z -= 160.0
        Rotation := MakeRotation((/UnrealEngine.com/Temporary/SpatialMath:)vector3{X:=0.0,Y:=0.0,Z:=0.0}, 0.0)
        SpawnedPlanks := SpawnProp(PlanksNotReady, ToSetUp, Rotation)
        if (prop := SpawnedPlanks(0)?, set NonPlanksProp = prop) {}

    CheckState(Agent : agent):void=
        if (set MainClass.PlayersToPortalTasks[Agent] = false) {}
        Button.Enable()
        if (IsBlocked = true):
            TaskReverted(Agent)
        else:
            TaskCompleted(Agent)

    CompletingTask(Agent : agent) : void =
        if:
            MainClass.PlayersToDown[Agent] = false
            MainClass.Portal_TaskInteraction.BeginInteraction[Agent]
            set MainClass.PlayersToPortalTasks[Agent] = option{Self}
        then:
            Button.Disable()

    FailTask(Agent : agent) : void =
        if (set MainClass.PlayersToPortalTasks[Agent] = false) {}
        FailMarker.Enable()
        spawn{FailTaskSus()}

    FailTaskSus()<suspends>:void=
        Sleep(5.0)
        FailMarker.Disable()
        Sleep(5.0)
        # Print("Pumpkin task enabled back")
        Button.Enable()

    TaskCompleted(Agent : agent) : void =
        if (MainClass.PlayersToDown[Agent] = true):
            return
        if (not NonPlanksProp.IsValid[]):
            return
        PlacingSound.Play()
        MonMarker.Disable()
        Ending := NonPlanksProp.GetTransform()
        NonPlanksProp.Dispose()
        SpawnedPlanks := SpawnProp(PlanksReady, Ending)
        if (prop := SpawnedPlanks(0)?, set PlanksProp = prop) {}
        Obj.Increment(Agent)
        if (pl := MainClass.Roles.Monster?):
            MonObj.Increment(pl)
        # Button.SetInteractionText(Msg("Remove planks"))
        Button.SetInteractionText(MessagesBank{}.RemovePlanks)
        set IsBlocked = true
        if (Obj.GetValue() = Obj.GetTarget()):
            MainClass.CheckWin(Agent)
            if (AchTriggered = false):
                MainClass.Achievements.Award(Agent, 6)
        if (AchTriggered = false):
            set AchTriggered = true
            MainClass.Achievements.Award(Agent, 5)
            MainClass.Achievements.SaveAchievement(Agent, PossibleAchievements.CompletedQuests, 1)

    TaskReverted(Agent : agent) : void =
        if (MainClass.PlayersToDown[Agent] = true):
            return
        if (not PlanksProp.IsValid[]):
            return
        RemovingSound.Play()
        # Marker.Enable()
        MonMarker.Enable()
        Ending := PlanksProp.GetTransform()
        PlanksProp.Dispose()
        Obj.Decrement(Agent)
        if (Mon := MainClass.Roles.Monster?):
            MonObj.Decrement(Mon)
        # MonObj.Decrement(Agent)
        # Button.SetInteractionText(Msg("Place planks"))
        Button.SetInteractionText(MessagesBank{}.PlacePlanks)
        set IsBlocked = false
        if (WaitingToRescue = true):
            Button.Disable()
            # Enable the rescue button
            WaitingPortal.Enable()
        else:
            SpawnedPlanks := SpawnProp(PlanksNotReady, Ending)
            if (prop := SpawnedPlanks(0)?, set NonPlanksProp = prop) {}

    PlayerSentToBasement() : void =
        # Print("Player sent to basement")
        if (IsBlocked = false):
            Button.Disable()
            # Enable the rescue button
            WaitingPortal.Enable()
            NonPlanksProp.Dispose()
        set WaitingToRescue = true

    PlayerRescued() : void =
        # Print("Player rescued")
        set WaitingToRescue = false
        Button.Enable()
        if (IsBlocked = false):
            var ToSetUp : (/UnrealEngine.com/Temporary/SpatialMath:)vector3 = WaitingPortal.GetTransform().Translation
            set ToSetUp.Z -= 144.0
            Rotation := MakeRotation((/UnrealEngine.com/Temporary/SpatialMath:)vector3{X:=0.0,Y:=0.0,Z:=0.0}, 0.0)
            SpawnedPlanks := SpawnProp(PlanksNotReady, ToSetUp, Rotation)
            if (prop := SpawnedPlanks(0)?, set NonPlanksProp = prop) {}

    Finale() : void =
        set ForeverClosed = true
        Marker.Disable()
        Button.Disable()
        MonMarker.Disable()
        WaitingPortal.Disable()

TransmitterTask := class(tag){}
PlacedTransmitter := class(tag){}
PumpkinTask := class(tag){}
PortalTask := class(tag){}

icons := struct<concrete>:
    @editable
    TransmitterTaskIcon : texture = Icons.IconNone
    @editable
    PumpkinTaskIcon : texture = Icons.IconNone
    @editable
    PortalTaskIcon : texture = Icons.IconNone
    @editable
    MonsterPortalTaskIcon : texture = Icons.IconNone
    @editable
    FinalTaskIcon : texture = Icons.IconNone

TasksHandler := class(creative_device):

    # @editable
    # WaitingTransmitters : []creative_prop = array{}

    @editable
    DEBUG_OneOfEach : logic = false

    @editable
    DEBUG_ShowHUD : logic = true

    @editable
    SceneGraphActive : logic = false

    @editable
    PortalsVersion : Versions = Versions.V1

    @editable
    DownDetector : down_but_not_out_device = down_but_not_out_device{}

    @editable
    Transmitter_TaskInteraction : skilled_interaction_device = skilled_interaction_device{}
    @editable
    Pumpkin_TaskInteraction : skilled_interaction_device = skilled_interaction_device{}
    @editable
    Portal_TaskInteraction : skilled_interaction_device = skilled_interaction_device{}

    @editable
    Notification : hud_message_device = hud_message_device{}

    @editable
    ReadyTransmitter : creative_prop_asset := DefaultCreativePropAsset

    @editable
    TasksIcons : icons = icons{}

    @editable
    var TransmitterButtons : []button_device = array{}

    @editable
    var TransmitterMarkers : []beacon_device = array{}

    @editable
    TransmittersObjective: tracker_device = tracker_device{}

    @editable
    MonTransmittersObjective: tracker_device = tracker_device{}

    @editable
    ReadyPumpkin : creative_prop_asset := DefaultCreativePropAsset

    @editable
    PumpkinButtons : []button_device = array{}

    @editable
    PumpkinMarkers : []beacon_device = array{}

    @editable
    PumpkinsObjective: tracker_device = tracker_device{}

    @editable
    MonPumpkinsObjective: tracker_device = tracker_device{}

    @editable
    PortalsButtons : []button_device = array{}

    @editable
    PortalsMarkers : []beacon_device = array{}

    @editable
    PortalsMarkersForMonster : []beacon_device = array{}

    @editable
    PortalsPlacingSounds : []audio_player_device = array{}

    @editable
    PortalsRemovingSounds : []audio_player_device = array{}

    @editable
    PortalsObjective: tracker_device = tracker_device{}

    @editable
    MonPortalsObjective: tracker_device = tracker_device{}

    @editable
    PlanksForPortal : creative_prop_asset = DefaultCreativePropAsset

    @editable
    NonExistingPlanksForPortal : creative_prop_asset = DefaultCreativePropAsset

    @editable
    FinalObjective : tracker_device = tracker_device{}

    @editable
    MonFinalObjective : tracker_device = tracker_device{}

    @editable
    FinalMusic : radio_device = radio_device{}

    @editable
    FinalMarker : beacon_device = beacon_device{}

    @editable
    FinalMapMarker : map_indicator_device = map_indicator_device{}

    @editable
    EndingZone : capture_area_device = capture_area_device{}

    @editable
    EndingBarrier : barrier_device = barrier_device{}

    @editable
    EndingPlanks : []creative_prop = array{}

    @editable
    EndingLevitating : []creative_prop = array{}

    @editable
    FailedTaskMarkers : []beacon_device = array{}

    @editable
    FinalTeam : class_and_team_selector_device = class_and_team_selector_device{}

    @editable
    Roles : RolesHandler = RolesHandler{}

    # @editable
    # Gates : GatesHandler = GatesHandler{}

    @editable
    Achievements : AchievmentsHandler = AchievmentsHandler{}

    # @editable
    # LivesController : LifeHandler = LifeHandler{}

    var AllEndingPlanks : [](/UnrealEngine.com/Temporary/SpatialMath:)transform = array{}
    var AllPortals : []PortalTaskConnection = array{}

    var AllNewPortals : []SG_PortalTask = array{}
    var InBasement : int = 0

    var PlayersToTransmitterTasks : [agent]?TransmitterTaskConnection = map{}
    var PlayersToPumpkinTasks : [agent]?PumpkinTaskConnection = map{}
    var PlayersToPortalTasks : [agent]?PortalTaskConnection = map{}

    var PlayersToNewTasks : [agent]?SG_GeneralTask = map{}

    var PlayersToDown : [agent]logic = map{}

    var PlayersPickedUpByMonster : []agent = array{}
    
    OnBegin<override>()<suspends>:void=
        # if:
        #     var LocationToMoveTo : vector3 = EndingBarrier.GetTransform().Translation
        #     set LocationToMoveTo.Z -= 10000.0
        #     EndingBarrier.TeleportTo[LocationToMoveTo, EndingBarrier.GetTransform().Rotation]
        FinalMapMarker.Disable()
        DownDetector.AgentDownedEvent.Subscribe(RemoveInteraction)
        DownDetector.AgentRevivedEvent.Subscribe(Revived)
        DownDetector.AgentPickedUpEvent.Subscribe(PickedUp)
        DownDetector.AgentDroppedEvent.Subscribe(Dropped)
        GetPlayspace().PlayerRemovedEvent().Subscribe(RemoveInteraction)
        Transmitter_TaskInteraction.InteractionSucceededEvent.Subscribe(CompleteTask)
        Transmitter_TaskInteraction.InteractionFailedEvent.Subscribe(FailTask)
        Transmitter_TaskInteraction.InteractionCanceledEvent.Subscribe(FailTask)
        Pumpkin_TaskInteraction.InteractionSucceededEvent.Subscribe(CompleteTask)
        Pumpkin_TaskInteraction.InteractionFailedEvent.Subscribe(FailTask)
        Pumpkin_TaskInteraction.InteractionCanceledEvent.Subscribe(FailTask)
        Portal_TaskInteraction.InteractionSucceededEvent.Subscribe(CompleteTask)
        Portal_TaskInteraction.InteractionFailedEvent.Subscribe(FailTask)
        Portal_TaskInteraction.InteractionCanceledEvent.Subscribe(FailTask)
        var ERR : logic = false
        loop:
            if (WaitingForPlayersStatus[GetSession()] = false):
                break
            Sleep(0.1)
        # var WaitingTransmitters : []creative_object_interface = GetCreativeObjectsWithTag(TransmitterTask{})
        # var WaitingPumpkins : []creative_object_interface = GetCreativeObjectsWithTag(PumpkinTask{})
        # var WaitingPortals : []creative_object_interface = GetCreativeObjectsWithTag(PortalTask{})
        var WaitingTransmitters : []creative_object_interface = array{}
        var WaitingPumpkins : []creative_object_interface = array{}
        var WaitingPortals : []creative_object_interface = array{}
        for (i : FindCreativeObjectsWithTag(TransmitterTask{})):
            set WaitingTransmitters += array{i}
        for (i : FindCreativeObjectsWithTag(PumpkinTask{})):
            set WaitingPumpkins += array{i}
        for (i : FindCreativeObjectsWithTag(PortalTask{})):
            set WaitingPortals += array{i}
        for (i : GetPlayspace().GetPlayers()):
            TransmittersObjective.Reset(i)
            MonTransmittersObjective.Reset(i)
            PumpkinsObjective.Reset(i)
            MonPumpkinsObjective.Reset(i)
            PortalsObjective.Reset(i)
            MonPortalsObjective.Reset(i)
            FinalObjective.Reset(i)
            MonFinalObjective.Reset(i)
        TransmittersObjective.SetValue(0)
        MonTransmittersObjective.SetValue(0)
        PumpkinsObjective.SetValue(0)
        MonPumpkinsObjective.SetValue(0)
        PortalsObjective.SetValue(0)
        MonPortalsObjective.SetValue(0)
        if (not SceneGraphActive?):
            if (TransmittersObjective.GetTarget() + 1 > TransmitterButtons.Length):
                # Print("Not enough buttons for transmitters")
                set ERR = true
            if (TransmittersObjective.GetTarget() + 1 > TransmitterMarkers.Length):
                # Print("Not enough markers for transmitters")
                set ERR = true
            if (PumpkinsObjective.GetTarget() + 1 > PumpkinButtons.Length):
                # Print("Not enough buttons for pumpkins")
                set ERR = true
            if (PumpkinsObjective.GetTarget() + 1 > PumpkinMarkers.Length):
                # Print("Not enough markers for pumpkins")
                set ERR = true
            if (PortalsObjective.GetTarget() > PortalsButtons.Length):
                # Print("Not enough buttons for portals")
                set ERR = true
            if (PortalsObjective.GetTarget() > PortalsMarkers.Length):
                # Print("Not enough markers for portals")
                set ERR = true
            if (PortalsObjective.GetTarget() > PortalsMarkersForMonster.Length):
                # Print("Not enough monster markers for portals")
                set ERR = true
            if (FailedTaskMarkers.Length < TransmittersObjective.GetTarget() + PumpkinsObjective.GetTarget() + PortalsObjective.GetTarget()):
                # Print("Not enough failed task markers")
                set ERR = true
            if (ERR = true):
                return
            for (i := 0..20):
                set WaitingTransmitters = Shuffle(WaitingTransmitters)
                set WaitingPumpkins = Shuffle(WaitingPumpkins)
            # Print("Total transmitters: {WaitingTransmitters.Length}")
            # Print("Total pumpkins: {WaitingPumpkins.Length}")
            # Print("Total portals: {WaitingPortals.Length}")
            # var NotAddedTransmitters : []creative_object_interface = WaitingTransmitters
            # var NotAddedPumpkins : []creative_object_interface = WaitingPumpkins
            for (i := 0..TransmittersObjective.GetTarget() + 1, transmitter := creative_prop[WaitingTransmitters[0]], Button := TransmitterButtons[i], Marker := TransmitterMarkers[i], FailMarker := FailedTaskMarkers[i]):
                TransmitterTaskConnection{
                    WaitingTransmitter := transmitter
                    ReadyTransmitter := ReadyTransmitter
                    Marker := Marker
                    FailMarker := FailMarker
                    Button := Button
                    Obj := TransmittersObjective
                    MonObj := MonTransmittersObjective
                    MainClass := Self
                }.Start()
                var temp : []creative_object_interface = array{}
                for (j := 0..WaitingTransmitters.Length):
                    if (j <> 0, set temp += array{WaitingTransmitters[j]}) {}
                set WaitingTransmitters = temp
            # Print("Transmitters removed: {WaitingTransmitters.Length}")

            # var temp : []creative_object_interface = array{}
            # for (j := (TransmittersObjective.GetTarget() - 1)..NotAddedTransmitters.Length, set temp += array{NotAddedTransmitters[j]}) {}
            # set NotAddedTransmitters = temp

            for (i := 0..PumpkinsObjective.GetTarget() + 1, pumpkin := creative_prop[WaitingPumpkins[0]], Button := PumpkinButtons[i], Marker := PumpkinMarkers[i], FailMarker := FailedTaskMarkers[i + TransmittersObjective.GetTarget()]):
                PumpkinTaskConnection{
                    WaitingPumpkin := pumpkin
                    ReadyPumpkin := ReadyPumpkin
                    Marker := Marker
                    FailMarker := FailMarker
                    Button := Button
                    Obj := PumpkinsObjective
                    MonObj := MonPumpkinsObjective
                    MainClass := Self
                }.Start()
                var temp : []creative_object_interface = array{}
                for (j := 0..WaitingPumpkins.Length):
                    if (j <> 0, set temp += array{WaitingPumpkins[j]}) {}
                set WaitingPumpkins = temp
            # Print("Pumpkins removed: {WaitingPumpkins.Length}")

            for (i := 0..WaitingPortals.Length, portal := button_device[WaitingPortals[i]], Button := PortalsButtons[i], Marker := PortalsMarkers[i], MonsterMarker := PortalsMarkersForMonster[i], Placing := PortalsPlacingSounds[i], Removing := PortalsRemovingSounds[i], FailMarker := FailedTaskMarkers[i + TransmittersObjective.GetTarget() + PumpkinsObjective.GetTarget()]):
                set AllPortals += array{PortalTaskConnection{
                    WaitingPortal := portal
                    PlanksReady := PlanksForPortal
                    PlanksNotReady := NonExistingPlanksForPortal
                    Marker := Marker
                    MonMarker := MonsterMarker
                    FailMarker := FailMarker
                    Button := Button
                    Obj := PortalsObjective
                    MonObj := MonPortalsObjective
                    PlacingSound := Placing
                    RemovingSound := Removing
                    MainClass := Self
                }}
                if (thisport := AllPortals[AllPortals.Length - 1]):
                    thisport.Start()
                # Print("Added portal {i}")

                PortalsObjective.SetTarget(WaitingPortals.Length)
                MonPortalsObjective.SetTarget(WaitingPortals.Length)
                if (DEBUG_OneOfEach = true):
                    TransmittersObjective.SetTarget(1)
                    MonTransmittersObjective.SetTarget(1)
                    PumpkinsObjective.SetTarget(1)
                    MonPumpkinsObjective.SetTarget(1)
                    PortalsObjective.SetTarget(1)
                    MonPortalsObjective.SetTarget(1)

            # set temp = array{}
            # for (j := (PumpkinsObjective.GetTarget() - 1)..NotAddedPumpkins.Length, set temp += array{NotAddedPumpkins[j]}) {}
            # set NotAddedPumpkins = temp

        for (i : EndingPlanks):
            set AllEndingPlanks += array{i.GetTransform()}
            i.Dispose()

        if (SceneGraphActive?):
            var LatestTransmitter : int = 0
            var LatestPumpkin : int = 0
            var LatestPortal : int = 0
            if (Root := GetSimulationEntity[]):
                var RandomTransmitters : []SG_GeneralTask = array{}
                var RandomPumpkins : []SG_GeneralTask = array{}
                for (Entity : Root.FindDescendantEntitiesWithComponent(SG_GeneralTask), Info := Entity.GetComponent[SG_GeneralTask], not SG_PortalTask[Info]):
                    case (Info.TaskType):
                        task_type.Transmitter =>
                            set RandomTransmitters += array{Info}
                        task_type.Pumpkin =>
                            set RandomPumpkins += array{Info}
                        task_type.Portal =>
                            Print("Found portal with wrong component: {Entity.GetGlobalTransform()}", ?Color := NamedColors.Red)
                set RandomPumpkins = Shuffle(RandomPumpkins)
                for (Info : Shuffle(RandomTransmitters)):
                    if (LatestTransmitter <= 12):
                        if (Marker := TransmitterMarkers[LatestTransmitter]):
                            set Info.Marker = Marker
                        else:
                            Print("Not enough markers for transmitters!", ?Color := NamedColors.Red)
                        set Info.Objective = TransmittersObjective
                        set Info.ProcessInteraction = Transmitter_TaskInteraction
                        if (FailMarker := FailedTaskMarkers[LatestTransmitter]):
                            set Info.FailMarker = FailMarker
                        else:
                            Print("Not enough failed task markers!", ?Color := NamedColors.Red)
                        set Info.TaskManager = Self
                        set Info.Start = true
                        set LatestTransmitter += 1
                    else:
                        Info.Entity.RemoveFromParent()
                for (Info : Shuffle(RandomPumpkins)):
                    if (LatestPumpkin <= 12):
                        if (Marker := PumpkinMarkers[LatestPumpkin]):
                            set Info.Marker = Marker
                        else:
                            Print("Not enough markers for pumpkins!", ?Color := NamedColors.Red)
                        set Info.Objective = PumpkinsObjective
                        set Info.ProcessInteraction = Pumpkin_TaskInteraction
                        if (FailMarker := FailedTaskMarkers[LatestTransmitter + LatestPumpkin]):
                            set Info.FailMarker = FailMarker
                        else:
                            Print("Not enough failed task markers!", ?Color := NamedColors.Red)
                        set Info.TaskManager = Self
                        set Info.Start = true
                        set LatestPumpkin += 1
                    else:
                        Info.Entity.RemoveFromParent()
                for (Entity : Root.FindDescendantEntitiesWithComponent(SG_PortalTask), Info := Entity.GetComponent[SG_PortalTask]):
                    if (Marker := PortalsMarkers[LatestPortal]):
                        set Info.Marker = Marker
                    else:
                        Print("Not enough markers for portals!", ?Color := NamedColors.Red)
                    if (Button := PortalsButtons[LatestPortal]):
                        set Info.PortalButton = Button
                    else:
                        Print("Not enough buttons for portals!", ?Color := NamedColors.Red)
                    if (MonsterMarker := PortalsMarkersForMonster[LatestPortal]):
                        set Info.MonsterMarker = MonsterMarker
                    else:
                        Print("Not enough monster markers for portals!", ?Color := NamedColors.Red)
                    set Info.Objective = PortalsObjective
                    set Info.ProcessInteraction = Portal_TaskInteraction
                    if (FailMarker := FailedTaskMarkers[LatestTransmitter + LatestPumpkin + LatestPortal]):
                        set Info.FailMarker = FailMarker
                    else:
                        Print("Not enough failed task markers!", ?Color := NamedColors.Red)
                    set Info.TaskManager = Self
                    set Info.Start = true
                    set LatestPortal += 1
                    set AllNewPortals += array{Info}
            for (i : WaitingTransmitters, transmitter := creative_prop[i]):
                transmitter.Dispose()
            for (i : WaitingPumpkins, pumpkin := creative_prop[i]):
                pumpkin.Dispose()
            for (i : WaitingPortals, portal := button_device[i]):
                portal.Disable()
            if (DEBUG_OneOfEach?):
                TransmittersObjective.SetTarget(1)
                PumpkinsObjective.SetTarget(1)
                PortalsObjective.SetTarget(1)
            else:
                TransmittersObjective.SetTarget(LatestTransmitter - 1)
                PumpkinsObjective.SetTarget(LatestPumpkin - 1)
                PortalsObjective.SetTarget(LatestPortal)

        # if (FinalMarker.TeleportTo[EndingZone.GetTransform()]) {}
        # FinalObjective.CompleteEvent.Subscribe(Winning{MainClass := Roles, Reason := "All of the portals have been closed"}.SurvivorsWin)
        FinalObjective.CompleteEvent.Subscribe(FinVFX)
        EndingZone.ControlChangeEvent.Subscribe(PlacePlanks)
        loop:
            if (Roles.Start = true):
                break
            Sleep(1.0)
        if (DEBUG_ShowHUD = false):
            # Disable all beacons
            for (i : TransmitterMarkers):
                i.Disable()
            for (i : PumpkinMarkers):
                i.Disable()
            for (i : PortalsMarkers):
                i.Disable()
            for (i : PortalsMarkersForMonster):
                i.Disable()
        Sleep(5.0)
        TransmittersObjective.AssignToAll()
        MonTransmittersObjective.AssignToAll()
        PumpkinsObjective.AssignToAll()
        MonPumpkinsObjective.AssignToAll()
        PortalsObjective.AssignToAll()
        MonPortalsObjective.AssignToAll()
        TransmittersObjective.SetValue(0)
        MonTransmittersObjective.SetValue(0)
        PumpkinsObjective.SetValue(0)
        MonPumpkinsObjective.SetValue(0)
        PortalsObjective.SetValue(0)
        MonPortalsObjective.SetValue(0)
        if (set PlayersTasksUis[GetSession()] = map{}) {}
        GetPlayspace().PlayerAddedEvent().Subscribe(AddTasksUI)
        for (i : GetPlayspace().GetPlayers()):
            if (PlUi := GetPlayerUI[i], set PlayersTasksUis[GetSession()][i] = TasksUI{
                MainClass := Self
                Player := i
                PlUi := PlUi
            }):
                if (PlHud := PlayersTasksUis[GetSession()][i]):
                    PlHud.Update()
        for (i : Roles.Survivors, set PlayersToDown[i] = false) {}

    AddTasksUI(Player : player) : void =
        if (PlUi := GetPlayerUI[Player], set PlayersTasksUis[GetSession()][Player] = TasksUI{
            MainClass := Self
            Player := Player
            PlUi := PlUi
        }):
            if (PlHud := PlayersTasksUis[GetSession()][Player]):
                PlHud.Update()
        # Print("Added tasks UI for {Player.GetName()}")

    PlayerSentToBasement() : void =
        set InBasement += 1
        # Print("Tasks - PlayerSentToBasement: InBasement = {InBasement}")
        if (not SceneGraphActive?):
            for (portal : AllPortals, InBasement = 1):
                portal.PlayerSentToBasement()
        else:
            for (portal : AllNewPortals, InBasement = 1):
                portal.FirstPlayerCaught()

    PlayerRescued() : void =
        set InBasement -= 1
        # Print("Tasks - PlayerRescued: InBasement = {InBasement}")
        if (not SceneGraphActive?):
            for (portal : AllPortals, InBasement = 0):
                portal.PlayerRescued()
        else:
            for (portal : AllNewPortals, InBasement = 0):
                portal.AllPlayersSaved()

    CompleteTask(Agent : agent) : void =
        if (Task := PlayersToTransmitterTasks[Agent]?):
            Task.TaskCompleted(Agent)
            # Print("Transmitter task completed")
        else if (Task := PlayersToPumpkinTasks[Agent]?):
            Task.TaskCompleted(Agent)
            # Print("Pumpkin task completed")
        else if (Task := PlayersToPortalTasks[Agent]?):
            Task.CheckState(Agent)
            for (i : GetPlayspace().GetPlayers()):
                if (PlHud := PlayersTasksUis[GetSession()][i]):
                    PlHud.Update()
            # Print("Portal task completed")
            return
        else if (Task := PlayersToNewTasks[Agent]?):
            Task.TaskCompleted(Agent)
            # Print("New task completed")
            if (set PlayersToNewTasks[Agent] = false) {}
        else:
            Print("No task found", ?Color := NamedColors.Red)
        # Notification.SetText(MessagesBank{}.TaskCompleted)
        # Notification.Hide(Agent)
        # Notification.Show(Agent)
        for (i : GetPlayspace().GetPlayers()):
            if (PlHud := PlayersTasksUis[GetSession()][i]):
                PlHud.Update()

    FailTask(Agent : agent) : void =
        if (Task := PlayersToTransmitterTasks[Agent]?):
            Task.FailTask(Agent)
            # Print("Transmitter task failed")
        else if (Task := PlayersToPumpkinTasks[Agent]?):
            Task.FailTask(Agent)
            # Print("Pumpkin task failed")
        else if (Task := PlayersToPortalTasks[Agent]?):
            Task.FailTask(Agent)
            # Print("Portal task failed")
        else if (Task := PlayersToNewTasks[Agent]?):
            Task.TaskFailed(Agent)
            # Print("New task failed")
            if (set PlayersToNewTasks[Agent] = false) {}
        else:
            Print("No task found", ?Color := NamedColors.Red)
        Notification.SetText(MessagesBank{}.TaskFailed)
        Notification.Hide(Agent)
        Notification.Show(Agent)
        if (Monika := Roles.Monster?):
            Notification.SetText(MessagesBank{}.PlayerFailedTask)
            Notification.Hide(Monika)
            Notification.Show(Monika)
        
    RemoveInteraction(Agent : agent) : void =
        if (CoPa := PlayersCodePanels[GetSession()][Agent]):
            CoPa.Remove()
        if (NP := PublicNameplateUI[GetSession()][Agent]):
            NP.Remove()
        if (Transmitter_TaskInteraction.EndInteraction[Agent]) {}
        if (Pumpkin_TaskInteraction.EndInteraction[Agent]) {}
        if (Portal_TaskInteraction.EndInteraction[Agent]) {}
        if (Task := PlayersToTransmitterTasks[Agent]?):
            Task.FailTask(Agent)
        else if (Task := PlayersToPumpkinTasks[Agent]?):
            Task.FailTask(Agent)
        else if (Task := PlayersToPortalTasks[Agent]?):
            Task.FailTask(Agent)
        else if (Task := PlayersToNewTasks[Agent]?):
            Task.TaskFailed(Agent)
        else:
            Print("No task found", ?Color := NamedColors.Red)
        if (set PlayersToDown[Agent] = true) {}

    Revived(Agent : agent) : void =
        if (set PlayersToDown[Agent] = false) {}

    PickedUp(Agent : agent) : void=
        set PlayersPickedUpByMonster += array{Agent}

    Dropped(Agent : agent) : void=
        set PlayersPickedUpByMonster = PlayersPickedUpByMonster.RemoveAllElements(Agent)
        
    CheckWin(Agent : agent) : void =
        if:
            TransmittersObjective.GetValue() = TransmittersObjective.GetTarget()
        then:
            # WaitingTransmitters : []creative_object_interface = GetCreativeObjectsWithTag(TransmitterTask{})
            var WaitingTransmitters : []creative_object_interface = array{}
            for (i : FindCreativeObjectsWithTag(TransmitterTask{})):
                set WaitingTransmitters += array{i}
            # Dispose of all transmitters holograms
            for (i : TransmitterMarkers):
                i.Disable()
            for (i : TransmitterButtons):
                i.Disable()
            # TODO: Add scene graph support - Check
            for (i : WaitingTransmitters, transmitter := creative_prop[i]):
                transmitter.Dispose()
            if (Root := GetSimulationEntity[]):
                for (Entity : Root.FindDescendantEntitiesWithComponent(SG_GeneralTask), Info := Entity.GetComponent[SG_GeneralTask]):
                    if (Info.TaskType = task_type.Transmitter, not Info.IsComplete[]):
                        # Entity.RemoveFromParent()
                        Info.TaskIgnored()
            # Print("Transmitters completed")
        if:
            PumpkinsObjective.GetValue() = PumpkinsObjective.GetTarget()
        then:
            # WaitingPumpkins : []creative_object_interface = GetCreativeObjectsWithTag(PumpkinTask{})
            var WaitingPumpkins : []creative_object_interface = array{}
            for (i : FindCreativeObjectsWithTag(PumpkinTask{})):
                set WaitingPumpkins += array{i}
            # Dispose of all pumpkins holograms
            for (i : PumpkinMarkers):
                i.Disable()
            for (i : PumpkinButtons):
                i.Disable()
            # TODO: Add scene graph support - Check
            for (i : WaitingPumpkins, pumpkin := creative_prop[i]):
                pumpkin.Dispose()
            if (Root := GetSimulationEntity[]):
                for (Entity : Root.FindDescendantEntitiesWithComponent(SG_GeneralTask), Info := Entity.GetComponent[SG_GeneralTask]):
                    if (Info.TaskType = task_type.Pumpkin, not Info.IsComplete[]):
                        # Entity.RemoveFromParent()
                        Info.TaskIgnored()
            # Print("Pumpkins completed")
        if (TransmittersObjective.GetValue() = TransmittersObjective.GetTarget() and PumpkinsObjective.GetValue() = PumpkinsObjective.GetTarget() and PortalsObjective.GetValue() = PortalsObjective.GetTarget()):
            # Print("Final Objective!")
            # for (pl : GetPlayspace().GetPlayers()):
            PortalsObjective.RemoveFromAll()
            MonPortalsObjective.RemoveFromAll()
            for (i : AllPortals):
                i.Finale()
            for (i : AllNewPortals):
                i.AllTasksDone()
            for (i : Roles.Survivors):
                # if (GetPlayspace().GetTeamCollection().AddToTeam[i, GetPlayspace().GetTeamCollection().GetTeams()[3]]):
                #     Print("Switched player to final team")
                FinalTeam.ChangeTeam(i)
            FinalObjective.SetValue(0)
            MonFinalObjective.SetValue(0)
            FinalObjective.AssignToAll()
            MonFinalObjective.AssignToAll()
            FinalObjective.SetValue(0)
            MonFinalObjective.SetValue(0)
            FinalMapMarker.Enable()
            FinalMusic.Play()
            FinalMarker.Enable()
            EndingZone.Enable()
            if:
                var LocationToMoveTo : (/UnrealEngine.com/Temporary/SpatialMath:)vector3 = EndingBarrier.GetTransform().Translation
                set LocationToMoveTo.Z += 10000.0
                EndingBarrier.TeleportTo[LocationToMoveTo, EndingBarrier.GetTransform().Rotation]
            EndingBarrier.Enable()
            if (Root := GetSimulationEntity[]):
                for (Entity : Root.FindDescendantEntitiesWithComponent(SG_Gate), Info := Entity.GetComponent[SG_Gate]):
                    Info.Finale(Agent)
            # Gates.Finale(Agent)
            for (i : Roles.Survivors):
                spawn{ShowFinalePulse(i)}
        else:
            # Print("Not all tasks are completed")

    ShowFinalePulse(Agent:agent)<suspends> : void =
        for (i := 1..3):
            # FinalMapMarker.ActivateObjectivePulse(Agent)
            Sleep(0.5)

    PlacePlanks(Agent : agent) : void =
        FinalObjective.Increment(Agent)
        for (i : PlayersTasksUis[GetSession()]):
            i.Update()
        for (i : EndingLevitating):
            i.Dispose()
        for (i : AllEndingPlanks):
            SpawnProp(PlanksForPortal, i)
        # Print("Planks placed")

    FinVFX(Agent : agent) : void =
        # Print("Final objective completed")
        # EndingBarrier.Disable()
        if:
            var LocationToMoveTo : (/UnrealEngine.com/Temporary/SpatialMath:)vector3 = EndingBarrier.GetTransform().Translation
            set LocationToMoveTo.Z -= 10000.0
            EndingBarrier.TeleportTo[LocationToMoveTo, EndingBarrier.GetTransform().Rotation]
        spawn{Roles.FinVFXSus(Agent)}