using { /Fortnite.com/Devices }
using { /Verse.org/Simulation }
using { /UnrealEngine.com/Temporary/Diagnostics }


VentHandler := class(creative_device):

    @editable
    AllVents : []volume_device = array{}

    @editable
    VentingSound : audio_player_device = audio_player_device{}

    @editable
    Floor1Map : map_controller_device = map_controller_device{}

    @editable
    OtherFloors : []map_controller_device = array{}

    # var MonsterAddedInZones : int = 0 #! DEPRECATED
    # var SurvivorsAddedInZones : []agent = array{} #! DEPRECATED
    # var SurvivorAddedIntoZones : [agent]int = map{} #! DEPRECATED

    var PlayersInZones : [agent]int = map{}
    
    OnBegin<override>()<suspends>:void=
        for (i : GetPlayspace().GetPlayers(), set PlayersInZones[agent[i]] = 0) {}
        for (vent : AllVents):
            vent.AgentEntersEvent.Subscribe(InNewZone)
            vent.AgentExitsEvent.Subscribe(OutNewZone)

    InNewZone(Agent : agent):void=
        # teamcoll := GetPlayspace().GetTeamCollection()
        # if (teamcoll.IsOnTeam[Agent, teamcoll.GetTeams()[0]]):
        #     if (set SurvivorAddedIntoZones[Agent] += 1) {}
        #     for (surv : SurvivorsAddedInZones, surv = Agent) {return}
        #     set SurvivorsAddedInZones += array{Agent}
        #     VentingSound.Register(Agent)
        #     if (SurvivorsAddedInZones.Length = 1 and MonsterAddedInZones = 0):
        #         VentingSound.Play()
        # else:
        #     set MonsterAddedInZones += 1
        # Print("Agent entered a new vent zone - Currently in {MonsterAddedInZones} zones")
        #     if (MonsterAddedInZones = 1):
        #         VentingSound.Register(Agent)
        #     if (SurvivorsAddedInZones.Length = 0 and MonsterAddedInZones = 1):
        #         VentingSound.Play()
        if (set PlayersInZones[Agent] += 1, info := PlayersInZones[Agent]):
            # Print("Agent entered a new vent zone - Currently in {info} zones")
            for (i : OtherFloors):
                i.Deactivate(Agent)
            Floor1Map.Activate(Agent)
        if (PlayersInZones[Agent] = 1):
            VentingSound.Register(Agent)
            VentingSound.Play()

    OutNewZone(Agent : agent):void=
        # teamcoll := GetPlayspace().GetTeamCollection()
        # if (teamcoll.IsOnTeam[Agent, teamcoll.GetTeams()[0]]):
        #     if (set SurvivorAddedIntoZones[Agent] -= 1) {}
        #     if (SurvivorAddedIntoZones[Agent] = 0):
        #         # set SurvivorsAddedInZones -= array{Agent}
        #         var temp : []agent = array{}
        #         for (surv : SurvivorsAddedInZones):
        #             if (surv <> Agent):
        #                 set temp += array{surv}
        #         set SurvivorsAddedInZones = temp
        #         VentingSound.Unregister(Agent)
        #         if (SurvivorsAddedInZones.Length = 0 and MonsterAddedInZones = 0):
        #             VentingSound.Stop()
        # else:
        #     set MonsterAddedInZones -= 1
        # Print("Agent exited a vent zone - Currently in {MonsterAddedInZones} zones")
        #     if (MonsterAddedInZones = 0 and SurvivorsAddedInZones.Length = 0):
        #         VentingSound.Unregister(Agent)
        #         VentingSound.Stop()
        if (set PlayersInZones[Agent] -= 1, info := PlayersInZones[Agent]):
            # Print("Agent exited a vent zone - Currently in {info} zones")
        if (PlayersInZones[Agent] = 0):
            VentingSound.Unregister(Agent)

    PlayerInVent(Agent:agent)<transacts><decides>:void=
        PlayersInZones[Agent] > 0